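# Activatable confirmation screen in Pilgrim Quickplay that asks whether to restart the song from the song results.
# Confirm_Restart is the focus target and asks the owning player controller's BP_PilgrimPlayerControllerComponent, after a validity check, to send a solo replay request for the previous song.
# BackButton, bound to the generic back action, calls CloseScene, which removes the RestartSong scene through DynamicUIManager.
# The confirm path does not close the scene itself.
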WidgetBlueprint /PilgrimQuickplay/Widgets/Results/SongResults/WBP_PilgrimQuickplay_RestartSong
graphs [
 BP_GetDesiredFocusTarget {nodes [
   Get Desired Focus Target()
   > Return Node!(execute, ReturnValue:Widget Object Reference=Get Confirm_Restart() -> (Confirm_Restart:WBP UIKit Button Regular Object Reference))
  ]}
 CloseScene {nodes [
   CloseScene() {is=title K2Node_FunctionEntry}
   > RemoveScene!(
    execute
    self:Dynamic UIManager Object Reference=Get DynamicUIManager() -> (ReturnValue:Dynamic UIManager Object Reference) {is=title K2Node_GetSubsystem}
    Scene:Dynamic UIScene Object Reference=Get RestartSongScene() -> (RestartSongScene:Dynamic UIScene Object Reference)
    Player:"Player Controller Object Reference (by ref)"=GetFortPlayerControllerFromActor!(Actor:Actor Object Reference=GetOwningPlayer()) -> (ReturnValue:Fort Player Controller Object Reference)
   ) -> (then)
  ]}
 EventGraph {nodes [
   "On Clicked (BackButton)"!() -> (OutputDelegate, then, Button:Common Button Base Object Reference) {is=title K2Node_ComponentBoundEvent}
   > CloseScene() {is=title K2Node_CallFunction}
   "On Clicked (Confirm_Restart)"!() -> (OutputDelegate, then, Button:Common Button Base Object Reference) {is=title K2Node_ComponentBoundEvent}
   n1 = GetComponentByClass!(
    self:Actor Object Reference=GetOwningPlayer()
    ComponentClass:Actor Component Class Reference=/PilgrimCore/Player/Components/BP_PilgrimPlayerControllerComponent.BP_PilgrimPlayerControllerComponent_C
   ) -> (ReturnValue:BP Pilgrim Player Controller Component Object Reference)
   > Is Valid(InputObject=@n1)
   > Solo_SendRequestToReplayPreviousSong(self:Pilgrim Player Controller Component Object Reference=@n1)
  ]}
]
variables [RestartSongScene {type=DynamicUIScene, default=/PilgrimQuickplay/Widgets/Scenes/UIScene_PilgrimQuickplay_RestartSong}]
widgets {
 info {inheritedWidgetCount=0, parentClass=&CommonActivatableWidget, rootWidgetClass=&GridPanel, widgetCount=4}
 widgets [
  GridPanel_15 {widgetClassPath=&GridPanel, widget=GridPanel_15, properties {columnFill [1], rowFill [1]}}
  Confirm_Restart {
   widgetClassPath=&WBP_UIKit_Button_Regular
   bIsVariable=true
   parent=GridPanel_15
   widget=Confirm_Restart
   properties {
    bHideInputActionWithKeyboard=true
    buttonSize=Small
    navigation=&self:WidgetTree.Confirm_Restart.WidgetNavigation_0
    text=CONFIRM RESTART
   }
   slot {class=GridSlot, properties {horizontalAlignment=HAlign_Center, verticalAlignment=VAlign_Center}}
  }
  UEFN_TextBlock_C_194 {
   widgetClassPath=&UEFN_TextBlock
   parent=GridPanel_15
   widget=UEFN_TextBlock_C_194
   properties {text=RESTART SONG?}
   slot {class=GridSlot, properties {horizontalAlignment=HAlign_Center, verticalAlignment=VAlign_Center, padding=<ltrb 0 0 0 150>}}
  }
  BackButton {
   widgetClassPath=/FortUILibrary/UIKit/Buttons/WBP_UIKit_Button_Quiet
   bIsVariable=true
   parent=GridPanel_15
   widget=BackButton
   properties {text=Back, triggeringInputAction {dataTable=&InputActionDataTable, rowName=GenericBack}}
   slot {class=GridSlot, properties {horizontalAlignment=HAlign_Right, verticalAlignment=VAlign_Bottom, padding=<ltrb 20 20 70 20>}}
  }
 ]
}
