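# Sound cue for the band bonus on the Pilgrim Quickplay results screen, judging by its name and folder.
# A single wave player feeds the Results_BandBonus sound wave straight to the root, with no randomization, mixing or modulation.
# The cue plays at reduced volume through the Fort_SFX_InGame sound class and sets a very large max distance.
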SoundCue /PilgrimQuickplay/Audio/Results/Results_BandBonus_Cue
root=WavePlayer_0
nodes [{id=WavePlayer_0, properties {soundWaveAssetPtr=&/Results_BandBonus}}]
properties {maxDistance=2097152, soundClassObject=&Fort_SFX_InGame, volumeMultiplier=0.5}
