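# Sound cue for the intro crowd audio used in Pilgrim Quickplay results, as its name and folder indicate.
# A single wave player root plays the Results_IntroCrowd wave directly, with no mixing, randomization or modulation nodes.
# The cue routes through the Sparks_GreenRm sound class from the SparksMaps GreenRoom audio folder.
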SoundCue /PilgrimQuickplay/Audio/Results/Results_IntroCrowd_Cue
root=WavePlayer_0
nodes [{id=WavePlayer_0, properties {soundWaveAssetPtr=&/Results_IntroCrowd}}]
properties {maxDistance=2097152, soundClassObject=/SparksMaps/Audio/GreenRoom/Sparks_GreenRm, volumeMultiplier=1}
