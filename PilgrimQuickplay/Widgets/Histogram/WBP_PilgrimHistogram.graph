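# Dev-only Pilgrim Quickplay widget that charts a player's input timing offsets as a histogram, shown only when either FMCalibration debug CVar is on.
# PopulateFromPerformanceData clears old bars, finds the peak in-window count, then spawns bar widgets, using canvas anchors for position and height.
# Bars shift from green toward red away from center, and a cyan marker shows average offset.
# PreConstruct feeds sample data for designer preview.
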WidgetBlueprint /PilgrimQuickplay/Widgets/Histogram/WBP_PilgrimHistogram
graphs [
 EventGraph {nodes [
   n1 = Event PreConstruct()
   > Branch(Condition=@n1.IsDesignTime)
   n2 = Make Pilgrim Histogram Sample!(OffsetMs:int=0, NumSamples:int=100) -> (PilgrimHistogramSample:Pilgrim Histogram Sample Structure) {is=title K2Node_MakeStruct}
   n3 = Make Pilgrim Histogram Sample(OffsetMs=-25, NumSamples=50)
   n4 = Make Pilgrim Histogram Sample(OffsetMs=25, NumSamples=80)
   > PopulateFromPerformanceData!(
    execute
    self
    PilgrimQuickplayPlayerPerformanceData:"Pilgrim Quickplay Player Performance Data Structure (by ref)"=Make Pilgrim Quickplay Player Performance Data!(
     Accuracy:float=0
     AverageOffset:float=25
     StandardDeviation:float=0
     FullCombo:bool=False
     NotesHit:int=0
     NotesPassed:int=0
     NotesMissed:int=0
     TotalNotes:int=0
     LongestStreak:int=0
     TimeInOverdriveMs:float=0
     HistogramSamples:Array of Pilgrim Histogram Sample Structures=Make Array!("[0]":Pilgrim Histogram Sample Structure=@n2, "[1]":Pilgrim Histogram Sample Structure=@n3, "[2]":Pilgrim Histogram Sample Structure=@n4) -> (Array:Array of Pilgrim Histogram Sample Structures)
     AccuracyTierCounts:Array of Integers
     SectionsCompletedByTier:Array of Integers
    ) -> (PilgrimQuickplayPlayerPerformanceData:Pilgrim Quickplay Player Performance Data Structure) {is=title K2Node_MakeStruct}
   ) -> (then)
  ]}
 PopulateFromPerformanceData {nodes [
   PopulateFromPerformanceData!() -> (then, PilgrimQuickplayPlayerPerformanceData:"Pilgrim Quickplay Player Performance Data Structure (by ref)") {is=title K2Node_FunctionEntry}
   n1 = OR Boolean(
    A="Equal (Float)"(A=GetNumericCVar!(CVar:String=FMCalibration.ShowInputHistogram) -> (OutValue:float, ReturnValue:bool), B=1)
    B="Equal (Float)"(A=GetNumericCVar(CVar=FMCalibration.IsCalibrationDebugEnabled), B=1)
   )
   > SetVisibility(InVisibility=Select(Option 0=Collapsed, Option 1=SelfHitTestInvisible, Index=@n1))
   > Branch(Condition=@n1)
   > n2 = Sequence!(execute) -> (then_0, then_1, then_2:Exec, then_3:Exec, then_4:Exec)
   Branch(execute=@n2.then_2, Condition="integer <= integer"(A=Get MaxSampleCount() -> (MaxSampleCount:int=0)))
   > Return Node()
   n3 = Get PilgrimQuickplayPlayerPerformanceData() -> (PilgrimQuickplayPlayerPerformanceData:"Pilgrim Quickplay Player Performance Data Structure (by ref)")
   n4 = Set CurrentMsOffset(
    execute=@n2.then_4
    CurrentMsOffset:int=Truncate(A=Break Pilgrim Quickplay Player Performance Data!(PilgrimQuickplayPlayerPerformanceData:"Pilgrim Quickplay Player Performance Data Structure (by ref)"=@n3) -> (AverageOffset:float=~, TotalNotes:int, AccuracyTierCounts:Array of Integers, SectionsCompletedByTier:Array of Integers) {is=title K2Node_BreakStruct})
   ) -> (Output_Get:int=0)
   > n5 = Branch(Condition="integer <= integer"(A=@n4, B=Get kSlopWindowSizeMs() -> (kSlopWindowSizeMs:int=0)))
   n6 = Get kSlopWindowSizeMs() -> (kSlopWindowSizeMs=0)
   n7 = Get CurrentMsOffset() -> (CurrentMsOffset:int=0)
   n8 = Get kSlopWindowSizeMs() -> (kSlopWindowSizeMs=0)
   > Set CurrentBarAnchorX(CurrentBarAnchorX:double=float / float(A=float + float(A:int=@n6, B:int=@n7), B:int=int * int(A=@n8, B=2))) -> (Output_Get:double)
   > SetVisibility(self=Get AvgImage() -> (AvgImage:Image Object Reference), InVisibility=HitTestInvisible)
   n9 = MakeVector2D(X=Get CurrentBarAnchorX() -> (CurrentBarAnchorX:double), Y=1)
   > SetAnchors!(
    execute
    self:Canvas Panel Slot Object Reference=SlotAsCanvasSlot!(Widget:Widget Object Reference=Get AvgImage()) -> (ReturnValue:Canvas Panel Slot Object Reference)
    InAnchors:Anchors Structure=Make Anchors!(Minimum:Vector 2D Structure=@n9, Maximum:Vector 2D Structure=@n9) -> (Anchors:Anchors Structure) {is=title K2Node_MakeStruct}
   ) -> (then)
   SetVisibility(execute=@n5.else, self=Get AvgImage(), InVisibility=Collapsed)
   n10 = For Each Loop(
    Exec=@n2.then_3
    Array:Array of Pilgrim Histogram Sample Structures=Break Pilgrim Quickplay Player Performance Data!(PilgrimQuickplayPlayerPerformanceData=Get PilgrimQuickplayPlayerPerformanceData()) -> (TotalNotes, HistogramSamples:Array of Pilgrim Histogram Sample Structures, AccuracyTierCounts, SectionsCompletedByTier).HistogramSamples
   ) -> (Array Element:Pilgrim Histogram Sample Structure)
   n11 = Break Pilgrim Histogram Sample!(PilgrimHistogramSample:"Pilgrim Histogram Sample Structure (by ref)"=@n10) -> (OffsetMs:int, NumSamples:int) {is=title K2Node_BreakStruct}
   > Set CurrentMsOffset(CurrentMsOffset=@n11) -> (Output_Get=0)
   > Set CurrentSampleCount(CurrentSampleCount:int=@n11.NumSamples) -> (Output_Get:int=0)
   > Branch(Condition="integer > integer"(A=Get CurrentSampleCount() -> (CurrentSampleCount:int=0)))
   n12 = Get kSlopWindowSizeMs() -> (kSlopWindowSizeMs=0)
   n13 = Get CurrentMsOffset() -> (CurrentMsOffset=0)
   n14 = Get kSlopWindowSizeMs() -> (kSlopWindowSizeMs=0)
   > Set CurrentBarAnchorX(CurrentBarAnchorX=float / float(A=float + float(A=@n14, B=@n13), B=int * int(A=@n12, B=2)))
   > Set CurrentBarHeight(CurrentBarHeight:double=float / float(A:int=Get CurrentSampleCount() -> (CurrentSampleCount=0), B=Get MaxSampleCount() -> (MaxSampleCount=0))) -> (Output_Get:double)
   > n15 = Create Widget!(
    execute:Exec
    Class:User Widget Class Reference=&/WBP_PilgrimHistogramBar.WBP_PilgrimHistogramBar_C
    OwningPlayer:Player Controller Object Reference=GetOwningPlayer()
   ) -> (then:Exec, ReturnValue:WBP Pilgrim Histogram Bar Object Reference)
   n16 = Get CurrentBarAnchorX()
   n17 = float * float(A="Absolute (Float)"(A=float - float(A=@n16, B=0.5)), B=4) {is=title K2Node_PromotableOperator}
   n18 = float - float(A=0.9, B=float * float(A=@n17, B=0.5))
   > SetBrushTintColor!(
    execute
    self:Image Object Reference=Get BarImage!(self:WBP Pilgrim Histogram Bar Object Reference=@n15) -> (BarImage:Image Object Reference)
    TintColor:Slate Color Structure=Make Slate Color(SpecifiedColor=Make Linear Color!(R:float=@n17, G:float=@n18, B:float=0, A:float=1) -> (LinearColor:Linear Color Structure) {is=title K2Node_MakeStruct})
   ) -> (then)
   > Add(
    TargetArray:Array of WBP Pilgrim Histogram Bar Object References=Get AddedBars() -> (AddedBars:Array of WBP Pilgrim Histogram Bar Object References)
    NewItem:"WBP Pilgrim Histogram Bar Object Reference (by ref)"=@n15
   ) {is=title K2Node_CallArrayFunction}
   > n19 = AddChildToCanvas!(
    execute
    self:Canvas Panel Object Reference=Get SamplesCanvas() -> (SamplesCanvas:Canvas Panel Object Reference)
    Content:Widget Object Reference=@n15
   ) -> (then, ReturnValue:Canvas Panel Slot Object Reference)
   > SetSize!(execute, self:Canvas Panel Slot Object Reference=@n19, InSize:Vector 2D Structure=MakeVector2D(X=Get BarWidth() -> (BarWidth:double), Y=0)) -> (then)
   > SetAlignment!(execute, self:Canvas Panel Slot Object Reference=@n19, InAlignment:Vector 2D Structure=(X=0.5,Y=0)) -> (then)
   > Set bAutoSize!(execute, bAutoSize:bool=false, self:Canvas Panel Slot Object Reference=@n19) -> (then, Output_Get:bool)
   n20 = Get CurrentBarAnchorX()
   n21 = float - float(A=1, B=Get CurrentBarHeight() -> (CurrentBarHeight:double))
   n22 = Get CurrentBarAnchorX()
   > SetAnchors(self=@n19, InAnchors=Make Anchors(Minimum=MakeVector2D(X=@n22, Y=@n21), Maximum=MakeVector2D(X=@n20, Y=1)))
   n23 = For Each Loop(Exec=@n2.then_0, Array:Array of WBP Pilgrim Histogram Bar Object References=Get AddedBars()) -> (Array Element:WBP Pilgrim Histogram Bar Object Reference)
   Clear(execute=@n23.Completed, TargetArray:Array of WBP Pilgrim Histogram Bar Object References=Get AddedBars())
   RemoveFromParent(execute=@n23.LoopBody, self=@n23)
   n24 = For Each Loop(
    Exec=@n2.then_1
    Array:Array of Pilgrim Histogram Sample Structures=Break Pilgrim Quickplay Player Performance Data(PilgrimQuickplayPlayerPerformanceData=Get PilgrimQuickplayPlayerPerformanceData()).HistogramSamples
   ) -> (Array Element:Pilgrim Histogram Sample Structure)
   n25 = Break Pilgrim Histogram Sample(PilgrimHistogramSample=@n24)
   > Branch(Condition="integer <= integer"(A="Absolute (Integer)"!(A:int=@n25) -> (ReturnValue:int=0), B=Get kSlopWindowSizeMs() -> (kSlopWindowSizeMs=0)))
   > Branch(Condition="integer > integer"(A=@n25.NumSamples, B=Get MaxSampleCount() -> (MaxSampleCount=0)))
   > Set MaxSampleCount(MaxSampleCount:int=@n25.NumSamples) -> (Output_Get:int=0)
  ]}
]
variables [
 AddedBars {type="array<WBP_PilgrimHistogramBar_C>", default []}
 BarWidth {type=number, default=6}
 kSlopWindowSizeMs {type=integer, default=150}
]
widgets {
 info {inheritedWidgetCount=0, parentClass=&UserWidget, rootWidgetClass=&Overlay, widgetCount=12}
 widgets [
  Overlay_2 {widgetClassPath=&Overlay, widget=Overlay_2}
  SlightlyTransparentBkg {
   widgetClassPath=&Image
   bIsVariable=true
   parent=Overlay_2
   widget=SlightlyTransparentBkg
   properties {brush {
     bIsDynamicallyLoaded=false
     drawAs=Image
     imageType=NoImage
     mirroring=NoMirror
     resourceName=None
     resourceObject=null
     tiling=NoTile
     imageSize=<32 32>
     margin=<ltrb 0 0 0 0>
     outlineSettings {
      bUseBrushTransparency=false
      roundingType=HalfHeightRadius
      width=0
      color {colorUseRule=UseColor_Specified, specifiedColor=<rgba 0 0 0 0>}
      cornerRadii=<0 0 0 0>
     }
     tintColor {colorUseRule=UseColor_Specified, specifiedColor=<rgba 0 0 0 0.2>}
     uVRegion {bIsValid=false, max=<0 0>, min=<0 0>}
    }}
   slot {class=OverlaySlot, properties {horizontalAlignment=HAlign_Fill, verticalAlignment=VAlign_Fill}}
  }
  VerticalBox_18 {
   widgetClassPath=&VerticalBox
   parent=Overlay_2
   widget=VerticalBox_18
   slot {class=OverlaySlot, properties {horizontalAlignment=HAlign_Fill, verticalAlignment=VAlign_Fill}}
  }
  CommonTextBlock_38 {
   widgetClassPath=&CommonTextBlock
   parent=VerticalBox_18
   widget=CommonTextBlock_38
   properties {
    style=/FortUILibrary/Styles/Typography/Foundation/TextStyle-Base-XS.TextStyle-Base-XS_C
    text="Input Histogram (Dev Only)"
    font {
     bForceMonospaced=false
     bMaterialIsStencil=false
     fontMaterial=null
     fontObject=&NotoSans
     letterSpacing=0
     monospacedWidth=1
     size=16
     skewAmount=0
     typefaceFontName=NotoSans-Regular
     outlineSettings {
      bApplyOutlineToDropShadows=false
      bMiteredCorners=false
      bSeparateFillAlpha=false
      outlineMaterial=null
      outlineSize=0
      outlineColor=<rgba 0 0 0 1>
     }
    }
    shadowOffset=<0 0>
   }
   slot {class=VerticalBoxSlot, properties {horizontalAlignment=HAlign_Center}}
  }
  Overlay_1 {
   widgetClassPath=&Overlay
   parent=VerticalBox_18
   widget=Overlay_1
   slot {class=VerticalBoxSlot, properties {size {sizeRule=Fill, value=1}}}
  }
  SamplesCanvas {
   widgetClassPath=&CanvasPanel
   bIsVariable=true
   parent=Overlay_1
   widget=SamplesCanvas
   slot {class=OverlaySlot, properties {horizontalAlignment=HAlign_Fill, verticalAlignment=VAlign_Fill}}
  }
  AvgImage {
   widgetClassPath=&Image
   bIsVariable=true
   parent=SamplesCanvas
   widget=AvgImage
   properties {
    brush {
     bIsDynamicallyLoaded=false
     drawAs=Image
     imageType=NoImage
     mirroring=NoMirror
     resourceName=None
     resourceObject=/Game/Athena/HUD/Art/cross_Shadow
     tiling=NoTile
     imageSize=<40 40>
     margin=<ltrb 0 0 0 0>
     outlineSettings {
      bUseBrushTransparency=false
      roundingType=HalfHeightRadius
      width=0
      color {colorUseRule=UseColor_Specified, specifiedColor=<rgba 0 0 0 0>}
      cornerRadii=<0 0 0 0>
     }
     tintColor {colorUseRule=UseColor_Specified, specifiedColor=<rgba 0 0.8606 1 1>}
     uVRegion {bIsValid=false, max=<0 0>, min=<0 0>}
    }
    renderTransform {angle=45, scale=<1 1>, shear=<0 0>, translation=<0 0>}
   }
   slot {
    class=CanvasPanelSlot
    properties {zOrder=2, layoutData {alignment=<0.5 0.5>, anchors {maximum=<0.5833 1>, minimum=<0.5833 1>}, offsets=<ltrb 0 0 40 40>}}
   }
  }
  SpacerForSomeHeight {widgetClassPath=&Spacer, parent=Overlay_1, widget=SpacerForSomeHeight, properties {size=<0 100>}, slot {class=OverlaySlot}}
  Overlay_0 {
   widgetClassPath=&Overlay
   parent=VerticalBox_18
   widget=Overlay_0
   slot {class=VerticalBoxSlot, properties {size {sizeRule=Automatic, value=0.1}}}
  }
  DividingLine {
   widgetClassPath=&Image
   bIsVariable=true
   parent=Overlay_0
   widget=DividingLine
   properties {brush {
     bIsDynamicallyLoaded=false
     drawAs=Image
     imageType=NoImage
     mirroring=NoMirror
     resourceName=None
     resourceObject=null
     tiling=NoTile
     imageSize=<100 2>
     margin=<ltrb 0 0 0 0>
     outlineSettings {
      bUseBrushTransparency=false
      roundingType=HalfHeightRadius
      width=0
      color {colorUseRule=UseColor_Specified, specifiedColor=<rgba 0 0 0 0>}
      cornerRadii=<0 0 0 0>
     }
     tintColor {colorUseRule=UseColor_Specified, specifiedColor=<rgba 1 1 1 1>}
     uVRegion {bIsValid=false, max=<0 0>, min=<0 0>}
    }}
   slot {class=OverlaySlot, properties {horizontalAlignment=HAlign_Fill}}
  }
  CenterTickMark {
   widgetClassPath=&Image
   bIsVariable=true
   parent=Overlay_0
   widget=CenterTickMark
   properties {
    brush {
     bIsDynamicallyLoaded=false
     drawAs=Image
     imageType=NoImage
     mirroring=NoMirror
     resourceName=None
     resourceObject=null
     tiling=NoTile
     imageSize=<5 25>
     margin=<ltrb 0 0 0 0>
     outlineSettings {
      bUseBrushTransparency=false
      roundingType=HalfHeightRadius
      width=0
      color {colorUseRule=UseColor_Specified, specifiedColor=<rgba 0 0 0 0>}
      cornerRadii=<0 0 0 0>
     }
     tintColor {colorUseRule=UseColor_Specified, specifiedColor=<rgba 1 1 1 1>}
     uVRegion {bIsValid=false, max=<0 0>, min=<0 0>}
    }
    renderTransform {angle=0, scale=<1 0.5>, shear=<0 0>, translation=<0 0>}
    renderTransformPivot=<0.5 0>
   }
   slot {class=OverlaySlot, properties {horizontalAlignment=HAlign_Center, verticalAlignment=VAlign_Fill}}
  }
  WBP_PilgrimQuickplay_DebugLocalCalibrationOffsets {
   widgetClassPath=/PilgrimQuickplay/Widgets/DebugTools/WBP_PilgrimQuickplay_DebugLocalCalibrationOffsets
   bIsVariable=true
   parent=VerticalBox_18
   widget=WBP_PilgrimQuickplay_DebugLocalCalibrationOffsets
   slot {class=VerticalBoxSlot, properties {horizontalAlignment=HAlign_Left, verticalAlignment=VAlign_Top}}
  }
 ]
}
