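# Quickplay song results panel that shows a "NEW HIGH SCORE!" banner with friend rank, global rank and their movement.
# SetPlayerState caches the player component and setlist index, then reveals immediately or on a matching NewHighScoreReceived router event.
# Global position switches to "TOP %" for large ranks with valid percentiles, with movement computed against the loaded previous best.
# The reveal plays banner and fade animations, plus a sound for the local player.
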WidgetBlueprint /PilgrimQuickplay/Widgets/Results/SongResults/WBP_PilgrimQuickly_NewHighScoreRankPanel
animations [FadeInScoreDetails {
  binding_count=1
  length_seconds=0.25
  object=&self:FadeInScoreDetails
  sequence {bindings [RankAndPercentileBox {
     id=#1
     possessed_class=HorizontalBox
     tracks [{
       type=MovieSceneFloatTrack
       display_name=Render Opacity
       sections [{type=MovieSceneFloatSection, channels [None {type=MovieSceneScriptingFloatChannel, keys [[0,0],[5,1]]}]}]
      }]
    }]}
 }]
graphs [
 EventGraph {nodes [
   n1 = SetPlayerState!() -> (OutputDelegate, then, PlayerState:Player State Object Reference, SetlistIndex:int, IsSolo:bool=true) {is=title K2Node_CustomEvent}
   > Set SetlistIndex(SetlistIndex:int=@n1.SetlistIndex) -> (Output_Get:int=0)
   > Set bIsSolo(bIsSolo:bool=@n1.IsSolo) -> (Output_Get:bool)
   > n2 = GetPilgrimQuickplayPlayerStateComponent!(execute, PlayerState:Player State Object Reference=@n1.PlayerState) -> (then, ReturnValue:Pilgrim Quickplay Player State Component Object Reference)
   > Is Valid(InputObject=@n2)
   > Set PilgrimQuickplayPlayerComponent(PilgrimQuickplayPlayerComponent:Pilgrim Quickplay Player State Component Object Reference=@n2) -> (Output_Get:Pilgrim Quickplay Player State Component Object Reference)
   > n3 = PilgrimWaitForEventRouter!(execute:Exec, ContextActor:Actor Object Reference=GetOwningPlayer()) -> (then:Exec, OnEventRouterAvailable:Exec, EventRouter:Gameplay Event Router Component Object Reference) {is=title K2Node_AsyncAction}
   n4 = Branch(execute=@n3.OnEventRouterAvailable, Condition=HasReceivedNewHighScore!() -> (Return Value:bool))
   SetVisibility(execute=[@n4.then, @n7.then], self=Get ScoreAndStars() -> (ScoreAndStars:Vertical Box Object Reference), InVisibility=HitTestInvisible)
   > SetGlobalRankOrPercentile()
   > SetFriendRank()
   > PlayHighScoreAnimation()
   > PlayAnimationForward(InAnimation=Get FadeInScoreDetails() -> (FadeInScoreDetails:Widget Animation Object Reference))
   > Branch(Condition=IsLocalController!(self:Controller Object Reference=GetPlayerController!(self:Player State Object Reference=@n1.PlayerState) -> (ReturnValue:Player Controller Object Reference)) -> (ReturnValue:bool))
   > PlaySound2D(Sound=Get HighScoreRevealSound() -> (HighScoreRevealSound:Sound Base Object Reference))
   n5 = StartListeningToEvent!(
    execute:Exec=@n4.else
    Target:Gameplay Event Router Component Object Reference=@n3
    EventType:Script Struct Object Reference=/Script/PilgrimQuickplayRuntime.PilgrimEvent_QuickplayScoring_NewHighScoreReceived
    Context:Object Reference
    EventBubblingRule:EEventBubblingRule Enum
   ) -> (
    then:Exec
    AsyncTaskProxy:Async Action Start Listening To Event Object Reference
    OnEventReceived:Exec
    Payload_ForPlayer:Fort Player Pawn Object Reference
    Payload_SetListIndex:int
   )
   > Add(
    TargetArray:Array of Async Action Start Listening To Event Object References=Get GameplayEventHandles() -> (GameplayEventHandles:Array of Async Action Start Listening To Event Object References)
    NewItem:"Async Action Start Listening To Event Object Reference (by ref)"=@n5
   ) {is=title K2Node_CallArrayFunction}
   n6 = Get SetlistIndex() -> (SetlistIndex:int=0)
   n7 = Branch(execute=@n5.OnEventReceived, Condition=AND Boolean(A="Equal (Integer)"(A=@n5.Payload_SetListIndex, B=@n6), B=HasReceivedNewHighScore()))
   Event PreConstruct()
   > SetHorizontalAlignment(
    self:Vertical Box Slot Object Reference=SlotAsVerticalBoxSlot(Widget=Get NewHighScoreBox() -> (NewHighScoreBox:Horizontal Box Object Reference))
    InHorizontalAlignment=Select(
     Option 0:EHorizontalAlignment Enum=HAlign_Fill
     Option 1:EHorizontalAlignment Enum=HAlign_Center
     Index=Get CenterHighScoreAlert() -> (CenterHighScoreAlert:bool)
    ) -> (ReturnValue:EHorizontalAlignment Enum)
   )
  ]}
 HasReceivedNewHighScore {nodes [
   HasReceivedNewHighScore() {is=title K2Node_FunctionEntry}
   > n1 = Get!(execute:Exec) -> (then:Exec, else:Exec, PilgrimQuickplayPlayerComponent:Pilgrim Quickplay Player State Component Object Reference)
   n2 = GetHighscoreResult!(
    self:Pilgrim Quickplay Player State Component Object Reference=@n1
    SetlistIndex:int=Get SetlistIndex() -> (SetlistIndex=0)
    bLoaded:bool=false
   ) -> (OutResult:Pilgrim Personal Best Result Rpl Structure, ReturnValue:bool)
   n3 = Break Pilgrim Leaderboard Entry Rpl!(PilgrimLeaderboardEntryRpl:"Pilgrim Leaderboard Entry Rpl Structure (by ref)"=Break Pilgrim Personal Best Result Rpl!(PilgrimPersonalBestResultRpl:"Pilgrim Personal Best Result Rpl Structure (by ref)"=@n2) -> (
     BestSolo:Pilgrim Leaderboard Entry Rpl Structure
     BestBand:Pilgrim Leaderboard Entry Rpl Structure
     BestSpotlight:Pilgrim Leaderboard Entry Rpl Structure
     BestAlltime:Pilgrim Leaderboard Entry Rpl Structure
    ) {is=title K2Node_BreakStruct}) -> (Rank:int, Percentile:float=~, ScoreData:Pilgrim Leaderboard Score Data Structure, DateTime:Date Time Structure, FriendRank:int) {is=title K2Node_BreakStruct}
   > Return Node!(execute, Return Value:bool=AND Boolean(A="integer > integer"(A=@n3), B=@n2.ReturnValue))
   Return Node(execute=@n1.else, Return Value=false)
   Select!(Option 0:Pilgrim Leaderboard Entry Rpl Structure, Option 1:Pilgrim Leaderboard Entry Rpl Structure, Index=Get bIsSolo() -> (bIsSolo:bool)) -> (
    ReturnValue_Rank:int
    ReturnValue_Percentile:float=~
    ReturnValue_ScoreData:Pilgrim Leaderboard Score Data Structure
    ReturnValue_DateTime:Date Time Structure
    ReturnValue_FriendRank:int
   )
  ]}
 PlayHighScoreAnimation {nodes [
   PlayHighScoreAnimation() {is=title K2Node_FunctionEntry}
   n1 = Get HighScoreNotif() -> (HighScoreNotif:WBP Pilgrim Results Reward Notif Object Reference)
   > SetVisibility(self=@n1, InVisibility=SelfHitTestInvisible)
   > PlayAnimation(
    self=@n1
    InAnimation=Get anim_Intro!(self:WBP Pilgrim Results Reward Notif Object Reference=@n1) -> (anim_Intro:Widget Animation Object Reference)
   )
  ]}
 SetFriendRank {nodes [
   SetFriendRank() {is=title K2Node_FunctionEntry}
   > n1 = Get()
   n2 = GetFortPlayerControllerFromActor!(Actor:Actor Object Reference=GetOwner(self=@n1)) -> (ReturnValue:Fort Player Controller Object Reference)
   > Is Valid(InputObject=@n2)
   n3 = GetComponentByClass!(self:Actor Object Reference=@n2, ComponentClass:Actor Component Class Reference=/Script/PilgrimCoreRuntime.PilgrimLeaderboardPlayerComponent) -> (ReturnValue:Pilgrim Leaderboard Player Component Object Reference)
   > Is Valid(InputObject=@n3)
   n4 = Break Pilgrim Personal Best Result Rpl(PilgrimPersonalBestResultRpl=GetHighscoreResult(
     self=Get PilgrimQuickplayPlayerComponent() -> (PilgrimQuickplayPlayerComponent:Pilgrim Quickplay Player State Component Object Reference)
     SetlistIndex=Get SetlistIndex() -> (SetlistIndex=0)
     bLoaded=false
    ))
   n5 = GetPostMatchFriendRankAndDelta!(
    self:Pilgrim Leaderboard Player Component Object Reference=@n3
    FinalScore:int=Break Pilgrim Leaderboard Score Data!(PilgrimLeaderboardScoreData:"Pilgrim Leaderboard Score Data Structure (by ref)"=Break Pilgrim Leaderboard Entry Rpl(PilgrimLeaderboardEntryRpl=@n4) -> (Percentile=~).ScoreData) -> (
     Score:int
     StarsEarned:int
     Accuracy:float=~
     bFullCombo:bool=~
     Difficulty:EPilgrimSongDifficulty Enum
     InstrumentsPlayed:Array of EPilgrimTrackType Enums
     SeasonNumber:int
    ) {is=title K2Node_BreakStruct}
   ) -> (OutRank:int=0, OutDelta:int=0)
   > "SetText (Text)"(self=Get Rank() -> (Rank:Common Text Object Reference), InText=Format Text!(Format="LOCTEXT(\"##{Rank}\")", Rank:int=@n5) -> (Result))
   > n6 = Branch(Condition="Not Equal (Integer)"(A=@n5.OutDelta))
   > "SetText (Text)"(self=Get RankMovement() -> (RankMovement:Common Text Object Reference), InText="To Text (Integer)"(Value=@n5.OutDelta))
   SetVisibility(execute=@n6.else, self=[Get RankMovement(), Get RankMovementIcon() -> (RankMovementIcon:Image Object Reference)], InVisibility=Collapsed)
   Select(Index=Get bIsSolo()) -> (ReturnValue_Percentile=~)
  ]}
 SetGlobalRankOrPercentile {nodes [
   SetGlobalRankOrPercentile() {is=title K2Node_FunctionEntry}
   > n1 = Get()
   n2 = GetHighscoreResult(self=@n1, SetlistIndex=Get SetlistIndex() -> (SetlistIndex=0), bLoaded=false)
   > n3 = Branch(Condition=@n2.ReturnValue)
   n4 = Break Pilgrim Leaderboard Entry Rpl(PilgrimLeaderboardEntryRpl=Break Pilgrim Personal Best Result Rpl(PilgrimPersonalBestResultRpl=@n2)) -> (Percentile=~)
   > n5 = Set bUsePercent(bUsePercent:bool=AND Boolean(A="integer >= integer"(A=@n4, B=10000), B="float > float"(A=@n4.Percentile))) -> (Output_Get:bool)
   n6 = float * float(A=@n4.Percentile, B=100) {is=title K2Node_PromotableOperator}
   > n7 = Set NewPositionValue(NewPositionValue:int=Select!(Option 0:int=@n4, Option 1:int=Round(A=@n6) {is=title K2Node_CallFunction}, Index=@n5) -> (ReturnValue:int)) -> (Output_Get:int=0)
   > "SetText (Text)"(
    self=Get Percentile() -> (Percentile:Common Text Object Reference)
    InText=Select(
     Option 0:Text=Format Text(Format="LOCTEXT(\"##{Rank}\")", Rank=@n7)
     Option 1:Text=Format Text!(Format="LOCTEXT(\"TOP {Percentile}%\")", Percentile:int=@n7) -> (Result)
     Index=Get bUsePercent() -> (bUsePercent:bool)
    ) -> (ReturnValue:Text)
   )
   n8 = GetHighscoreResult(self=Get PilgrimQuickplayPlayerComponent(), SetlistIndex=Get SetlistIndex() -> (SetlistIndex=0), bLoaded=true)
   > n9 = Branch(Condition=@n8.ReturnValue)
   n10 = Break Pilgrim Leaderboard Entry Rpl(PilgrimLeaderboardEntryRpl=Break Pilgrim Personal Best Result Rpl(PilgrimPersonalBestResultRpl=@n8)) -> (Percentile=~)
   > n11 = Branch(Condition="Not Equal (Integer)"(A=@n10))
   n12 = float * float(A=@n10.Percentile, B=100)
   n13 = int - int(
    A=Select(Option 0:int=@n10, Option 1:int=Round(A=@n12), Index=Get bUsePercent()) -> (ReturnValue:int)
    B=Get NewPositionValue() -> (NewPositionValue:int=0)
   )
   > n14 = Branch(Condition="Not Equal (Integer)"(A=@n13))
   > "SetText (Text)"(
    self=Get PercentileMovement() -> (PercentileMovement:Common Text Object Reference)
    InText=Select(
     Option 0:Text="To Text (Integer)"(Value=@n13)
     Option 1:Text=Format Text(Format="LOCTEXT(\"{Percentile}%\")", Percentile=@n13)
     Index=Get bUsePercent()
    ) -> (ReturnValue:Text)
   )
   SetVisibility(
    execute=[@n14.else, @n11.else, @n9.else, @n3.else, @n1.else]
    self=[Get PercentileMovement(), Get PercentileMovementIcon() -> (PercentileMovementIcon:Image Object Reference)]
    InVisibility=Collapsed
   )
   Select!(Option 0:Pilgrim Leaderboard Entry Rpl Structure, Option 1:Pilgrim Leaderboard Entry Rpl Structure, Index=Get bIsSolo()) -> (
    ReturnValue_Rank:int
    ReturnValue_Percentile:float=~
    ReturnValue_ScoreData:Pilgrim Leaderboard Score Data Structure
    ReturnValue_DateTime:Date Time Structure
    ReturnValue_FriendRank:int
   )
   Select!(Option 0, Option 1, Index=Get bIsSolo()) -> (ReturnValue:Pilgrim Leaderboard Entry Rpl Structure)
  ]}
]
variables [
 AllStars {type="array<Image>", default []}
 bIsSolo {type=boolean, default=true}
 CenterHighScoreAlert {type=boolean, default=true}
 HighScoreRevealSound {type=SoundBase, default=/PilgrimQuickplay/Audio/Results/MSS_FM_ResultsVictory_HiScoreBanner}
 LocalPlayerSongResultData {
  type=PilgrimQuickplaySongResultPlayerData
  default {
   instrumentType=TrackNone
   playerAccountId=INVALID
   playerDifficulty=None
   playerName=""
   playerScore=0
   playerStars=0
   playerState=null
   songTrackDifficulty=0
   team=false
   instrumentTags {gameplayTags []}
  }
 }
 PilgrimQuickplayPlayerComponent {type=PilgrimQuickplayPlayerStateComponent, default=null}
 SetlistIndex {type=integer, default=-1}
 ShouldShowRankMovement {type=boolean}
]
widgets {
 info {
  inheritedWidgetCount=0
  parentClass=/PilgrimQuickplay/Widgets/Results/WBP_Pilgrim_UI_Results_Base.WBP_Pilgrim_UI_Results_Base_C
  rootWidgetClass=&VerticalBox
  widgetCount=19
 }
 widgets [
  ScoreAndStars {widgetClassPath=&VerticalBox, bIsVariable=true, widget=ScoreAndStars, properties {visibility=Hidden}}
  NewHighScoreBox {
   widgetClassPath=&HorizontalBox
   bIsVariable=true
   parent=ScoreAndStars
   widget=NewHighScoreBox
   slot {class=VerticalBoxSlot, properties {size {sizeRule=Automatic, value=0.5}}}
  }
  ScaleBox_4 {
   widgetClassPath=&ScaleBox
   parent=NewHighScoreBox
   widget=ScaleBox_4
   properties {stretch=UserSpecified, userSpecifiedScale=0.6}
   slot {class=HorizontalBoxSlot, properties {verticalAlignment=VAlign_Center}}
  }
  HighScoreNotif {
   widgetClassPath=/PilgrimQuickplay/Widgets/Results/WBP_Pilgrim_Results_RewardNotif
   bIsVariable=true
   parent=ScaleBox_4
   widget=HighScoreNotif
   properties {notificationText="NEW HIGH SCORE!", visibility=Collapsed}
   slot {class=ScaleBoxSlot}
  }
  Spacer_74 {widgetClassPath=&Spacer, parent=ScoreAndStars, widget=Spacer_74, properties {size=<1 5>}, slot {class=VerticalBoxSlot}}
  RankAndPercentileBox {
   widgetClassPath=&HorizontalBox
   parent=ScoreAndStars
   widget=RankAndPercentileBox
   slot {class=VerticalBoxSlot, properties {horizontalAlignment=HAlign_Center, size {sizeRule=Automatic, value=0.5}}}
  }
  RankBox {widgetClassPath=&HorizontalBox, parent=RankAndPercentileBox, widget=RankBox, slot {class=HorizontalBoxSlot}}
  RankIcon {
   widgetClassPath=&Image
   bIsVariable=true
   parent=RankBox
   widget=RankIcon
   properties {
    visibility=SelfHitTestInvisible
    brush {
     bIsDynamicallyLoaded=false
     drawAs=Image
     imageType=NoImage
     mirroring=NoMirror
     resourceName=None
     resourceObject=/FortUILibrary/Textures/MaterialInstances/MI_UIKit_Icon_CurrentPlayerCount
     tiling=NoTile
     imageSize=<32 32>
     margin=<ltrb 0 0 0 0>
     outlineSettings {
      bUseBrushTransparency=false
      roundingType=HalfHeightRadius
      width=0
      color {colorUseRule=UseColor_Specified, specifiedColor=<rgba 0 0 0 0>}
      cornerRadii=<0 0 0 0>
     }
     tintColor {colorUseRule=UseColor_Specified, specifiedColor=<rgba 1 1 1 1>}
     uVRegion {bIsValid=false, max=<0 0>, min=<0 0>}
    }
   }
   slot {class=HorizontalBoxSlot, properties {horizontalAlignment=HAlign_Center, verticalAlignment=VAlign_Center}}
  }
  Rank {
   widgetClassPath=&CommonTextBlock
   bIsVariable=true
   parent=RankBox
   widget=Rank
   properties {
    text="##-1"
    font {
     bForceMonospaced=false
     bMaterialIsStencil=false
     fontMaterial=null
     fontObject=&HeadingNow
     letterSpacing=0
     monospacedWidth=1
     size=13.5
     skewAmount=0
     typefaceFontName=74 Regular
     outlineSettings {
      bApplyOutlineToDropShadows=false
      bMiteredCorners=false
      bSeparateFillAlpha=false
      outlineMaterial=null
      outlineSize=0
      outlineColor=<rgba 0 0 0 1>
     }
    }
    shadowOffset=<0 0>
   }
   slot {class=HorizontalBoxSlot, properties {horizontalAlignment=HAlign_Left, verticalAlignment=VAlign_Center, padding=<ltrb 5 0 10 0>}}
  }
  RankDeltaBox {widgetClassPath=&HorizontalBox, parent=RankAndPercentileBox, widget=RankDeltaBox, slot {class=HorizontalBoxSlot}}
  RankMovementIcon {
   widgetClassPath=&Image
   bIsVariable=true
   parent=RankDeltaBox
   widget=RankMovementIcon
   properties {
    brush {
     bIsDynamicallyLoaded=false
     drawAs=Image
     imageType=NoImage
     mirroring=NoMirror
     resourceName=None
     resourceObject=/Game/Athena/HUD/Bacchus/HUDLayoutTool/HLT_Icons/T-HLT-Triangle
     tiling=NoTile
     imageSize=<16 16>
     margin=<ltrb 0 0 0 0>
     outlineSettings {
      bUseBrushTransparency=false
      roundingType=HalfHeightRadius
      width=0
      color {colorUseRule=UseColor_Specified, specifiedColor=<rgba 0 0 0 0>}
      cornerRadii=<0 0 0 0>
     }
     tintColor {colorUseRule=UseColor_Specified, specifiedColor=<rgba 1 1 1 1>}
     uVRegion {bIsValid=false, max=<0 0>, min=<0 0>}
    }
    colorAndOpacity=<rgba 0.0343 1 0.227 1>
   }
   slot {class=HorizontalBoxSlot, properties {verticalAlignment=VAlign_Top, padding=<ltrb 0 4 0 0>}}
  }
  RankMovement {
   widgetClassPath=&CommonTextBlock
   bIsVariable=true
   parent=RankDeltaBox
   widget=RankMovement
   properties {
    text=X
    colorAndOpacity {colorUseRule=UseColor_Specified, specifiedColor=<rgba 0.0343 1 0.227 1>}
    font {
     bForceMonospaced=false
     bMaterialIsStencil=false
     fontMaterial=null
     fontObject=&HeadingNow
     letterSpacing=0
     monospacedWidth=1
     size=15
     skewAmount=0
     typefaceFontName=75 Medium
     outlineSettings {
      bApplyOutlineToDropShadows=false
      bMiteredCorners=false
      bSeparateFillAlpha=false
      outlineMaterial=null
      outlineSize=0
      outlineColor=<rgba 0 0 0 1>
     }
    }
    shadowOffset=<0 0>
   }
   slot {class=HorizontalBoxSlot, properties {horizontalAlignment=HAlign_Left, verticalAlignment=VAlign_Center, padding=<ltrb 5 0 10 0>}}
  }
  Spacer_43 {widgetClassPath=&Spacer, parent=RankAndPercentileBox, widget=Spacer_43, properties {size=<8 1>}, slot {class=HorizontalBoxSlot}}
  PercentileBox {widgetClassPath=&HorizontalBox, parent=RankAndPercentileBox, widget=PercentileBox, slot {class=HorizontalBoxSlot}}
  PercentileIcon {
   widgetClassPath=&Image
   bIsVariable=true
   parent=PercentileBox
   widget=PercentileIcon
   properties {
    visibility=SelfHitTestInvisible
    brush {
     bIsDynamicallyLoaded=false
     drawAs=Image
     imageType=NoImage
     mirroring=NoMirror
     resourceName=None
     resourceObject=/Game/UI/Friends_UI/Social/Globe_PlatformIcon_32x
     tiling=NoTile
     imageSize=<24 24>
     margin=<ltrb 0 0 0 0>
     outlineSettings {
      bUseBrushTransparency=false
      roundingType=HalfHeightRadius
      width=0
      color {colorUseRule=UseColor_Specified, specifiedColor=<rgba 0 0 0 0>}
      cornerRadii=<0 0 0 0>
     }
     tintColor {colorUseRule=UseColor_Specified, specifiedColor=<rgba 1 1 1 1>}
     uVRegion {bIsValid=false, max=<0 0>, min=<0 0>}
    }
   }
   slot {class=HorizontalBoxSlot, properties {horizontalAlignment=HAlign_Center, verticalAlignment=VAlign_Center}}
  }
  Percentile {
   widgetClassPath=&CommonTextBlock
   bIsVariable=true
   parent=PercentileBox
   widget=Percentile
   properties {
    text=TOP -1%
    font {
     bForceMonospaced=false
     bMaterialIsStencil=false
     fontMaterial=null
     fontObject=&HeadingNow
     letterSpacing=0
     monospacedWidth=1
     size=13.5
     skewAmount=0
     typefaceFontName=74 Regular
     outlineSettings {
      bApplyOutlineToDropShadows=false
      bMiteredCorners=false
      bSeparateFillAlpha=false
      outlineMaterial=null
      outlineSize=0
      outlineColor=<rgba 0 0 0 1>
     }
    }
    shadowOffset=<0 0>
   }
   slot {class=HorizontalBoxSlot, properties {horizontalAlignment=HAlign_Left, verticalAlignment=VAlign_Center, padding=<ltrb 5 0 10 0>}}
  }
  PercentileDeltaBox {widgetClassPath=&HorizontalBox, parent=RankAndPercentileBox, widget=PercentileDeltaBox, slot {class=HorizontalBoxSlot}}
  PercentileMovementIcon {
   widgetClassPath=&Image
   bIsVariable=true
   parent=PercentileDeltaBox
   widget=PercentileMovementIcon
   properties {
    brush {
     bIsDynamicallyLoaded=false
     drawAs=Image
     imageType=NoImage
     mirroring=NoMirror
     resourceName=None
     resourceObject=&T-HLT-Triangle
     tiling=NoTile
     imageSize=<16 16>
     margin=<ltrb 0 0 0 0>
     outlineSettings {
      bUseBrushTransparency=false
      roundingType=HalfHeightRadius
      width=0
      color {colorUseRule=UseColor_Specified, specifiedColor=<rgba 0 0 0 0>}
      cornerRadii=<0 0 0 0>
     }
     tintColor {colorUseRule=UseColor_Specified, specifiedColor=<rgba 1 1 1 1>}
     uVRegion {bIsValid=false, max=<0 0>, min=<0 0>}
    }
    colorAndOpacity=<rgba 0.0343 1 0.227 1>
   }
   slot {class=HorizontalBoxSlot, properties {verticalAlignment=VAlign_Top, padding=<ltrb 0 4 0 0>}}
  }
  PercentileMovement {
   widgetClassPath=&CommonTextBlock
   bIsVariable=true
   parent=PercentileDeltaBox
   widget=PercentileMovement
   properties {
    text=Y%
    colorAndOpacity {colorUseRule=UseColor_Specified, specifiedColor=<rgba 0.0343 1 0.227 1>}
    font {
     bForceMonospaced=false
     bMaterialIsStencil=false
     fontMaterial=null
     fontObject=&HeadingNow
     letterSpacing=0
     monospacedWidth=1
     size=15
     skewAmount=0
     typefaceFontName=75 Medium
     outlineSettings {
      bApplyOutlineToDropShadows=false
      bMiteredCorners=false
      bSeparateFillAlpha=false
      outlineMaterial=null
      outlineSize=0
      outlineColor=<rgba 0 0 0 1>
     }
    }
    shadowOffset=<0 0>
   }
   slot {class=HorizontalBoxSlot, properties {horizontalAlignment=HAlign_Left, verticalAlignment=VAlign_Center, padding=<ltrb 5 0 0 0>}}
  }
 ]
}
