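# Band score panel widget for the Pilgrim Quickplay song results screen, showing an animated count-up score over a glow image.
# ResultsScreenCountUp interpolates the CommonNumericTextBlock to CountUpFloatScore.
# For a score above zero, it binds the text block's outro event and plays a spawned tally sound, keeping the audio component reference.
# The outro event and widget Destruct both stop and destroy that component if it is valid.
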WidgetBlueprint /PilgrimQuickplay/Widgets/Results/SongResults/WBP_PilgrimQuickplay_BandScorePanel
graphs [EventGraph {nodes [
   ResultsScreenCountUp() {is=title K2Node_CustomEvent}
   n1 = Get Score() -> (Score:Common Numeric Text Block Object Reference)
   > InterpolateToValue!(
    execute
    self:Common Numeric Text Block Object Reference=@n1
    TargetValue:float=Get CountUpFloatScore() -> (CountUpFloatScore:double)
    MaximumInterpolationDuration:float=4
    MinimumChangeRate:float=1
    OutroOffset:float=0
   ) -> (then)
   > n2 = Branch(Condition="float > float"(A=Get CountUpFloatScore()))
   n3 = OnOutroEvent_Event!() -> (OutputDelegate, then, NumericTextBlock:Common Numeric Text Block Object Reference) {is=title K2Node_CustomEvent}
   Assign On Outro Event!(execute:Exec=@n2.then, self:Common Numeric Text Block Object Reference=@n1, Delegate:"Delegate (by ref)"=@n3) -> (then:Exec) {is=title K2Node_AssignDelegate}
   > n4 = SpawnSound2D!(
    execute
    Sound:Sound Base Object Reference=Get TallySound() -> (TallySound:Sound Base Object Reference)
    VolumeMultiplier:float=1
    PitchMultiplier:float=1
    StartTime:float=0
    ConcurrencySettings:Sound Concurrency Object Reference
    bPersistAcrossLevelTransition:bool=false
    bAutoDestroy:bool=false
   ) -> (then, ReturnValue:Audio Component Object Reference)
   > n5 = Set TallySoundReference(TallySoundReference:Audio Component Object Reference=@n4) -> (Output_Get:Audio Component Object Reference)
   > Play!(execute, self:Audio Component Object Reference=@n5, StartTime:float=0) -> (then)
   n6 = Event Destruct()
   n7 = Get TallySoundReference() -> (TallySoundReference:Audio Component Object Reference)
   Is Valid(exec=[@n3.then, @n6.then], InputObject=@n7)
   > Stop(self:Audio Component Object Reference=@n7)
   > Destroy Component(self:Actor Component Object Reference=@n7)
  ]}]
variables [
 CountUpFloatScore {type=number}
 TallySound {type=SoundBase, default=/PilgrimQuickplay/Audio/Results/MSS_FM_ResultsVictory_TallyBandNums}
 TallySoundReference {type=AudioComponent, default=null}
]
widgets {
 info {
  inheritedWidgetCount=0
  parentClass=/PilgrimQuickplay/Widgets/Results/WBP_Pilgrim_UI_Results_Base.WBP_Pilgrim_UI_Results_Base_C
  rootWidgetClass=&VerticalBox
  widgetCount=4
 }
 widgets [
  ScoreBox {widgetClassPath=&VerticalBox, widget=ScoreBox}
  Overlay_0 {
   widgetClassPath=&Overlay
   parent=ScoreBox
   widget=Overlay_0
   slot {class=VerticalBoxSlot, properties {horizontalAlignment=HAlign_Left, verticalAlignment=VAlign_Bottom}}
  }
  Glow {
   widgetClassPath=&Image
   bIsVariable=true
   parent=Overlay_0
   widget=Glow
   properties {
    visibility=SelfHitTestInvisible
    brush {
     bIsDynamicallyLoaded=false
     drawAs=Box
     imageType=NoImage
     mirroring=NoMirror
     resourceName=None
     resourceObject=/PilgrimCore/Widgets/Materials/MI_UI_Pilgrim_ScoreGlow
     tiling=NoTile
     imageSize=<40 40>
     margin=<ltrb 0 0 0 0>
     outlineSettings {
      bUseBrushTransparency=false
      roundingType=HalfHeightRadius
      width=0
      color {colorUseRule=UseColor_Specified, specifiedColor=<rgba 0 0 0 0>}
      cornerRadii=<0 0 0 0>
     }
     tintColor {colorUseRule=UseColor_Specified, specifiedColor=<rgba 1 1 1 1>}
     uVRegion {bIsValid=false, max=<0 0>, min=<0 0>}
    }
    renderTransform {angle=0, scale=<1 1>, shear=<-8 0>, translation=<0 0>}
   }
   slot {class=OverlaySlot, properties {horizontalAlignment=HAlign_Fill, verticalAlignment=VAlign_Fill}}
  }
  Score {
   widgetClassPath=/Script/CommonUI.CommonNumericTextBlock
   bIsVariable=true
   parent=Overlay_0
   widget=Score
   properties {
    easeOutInterpolationExponent=4
    performSizeInterpolation=false
    text=Text Block
    font {
     bForceMonospaced=false
     bMaterialIsStencil=false
     fontMaterial=null
     fontObject=&HeadingNow
     letterSpacing=0
     monospacedWidth=1
     size=30
     skewAmount=0
     typefaceFontName=76 Bold Italic
     outlineSettings {
      bApplyOutlineToDropShadows=false
      bMiteredCorners=false
      bSeparateFillAlpha=false
      outlineMaterial=&M-FontShadow
      outlineSize=1
      outlineColor=<rgba 0.8333 0.5 1 1>
     }
    }
    shadowOffset=<0 0>
   }
   slot {class=OverlaySlot, properties {horizontalAlignment=HAlign_Center, verticalAlignment=VAlign_Center}}
  }
 ]
}
