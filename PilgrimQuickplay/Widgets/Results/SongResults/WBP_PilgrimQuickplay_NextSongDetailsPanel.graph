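# Results widget for Pilgrim Quickplay that shows the upcoming song's album art, title, artist and intensity for the owning player's instrument.
# On construct it waits for playlist data and the event router, then refreshes on instrument or custom setlist changes.
# UpdateNextSong reads band state, picking the current song when isInGreenRoom, otherwise the next, and collapses without valid data.
# ShowOnlyIntensity switches to an intensity-only layout, applied here only in design-time preview.
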WidgetBlueprint /PilgrimQuickplay/Widgets/Results/SongResults/WBP_PilgrimQuickplay_NextSongDetailsPanel
graphs [
 EventGraph {nodes [
   UpdateNextSong() {is=title K2Node_CustomEvent}
   n1 = GetSparksMusicPlayspaceOfClass!(
    PlayspaceType:Sparks Music Playspace Class Reference=/Script/PilgrimQuickplayRuntime.PilgrimQuickplayPlayspace
    Actor:Actor Object Reference=Get _RepresentedPlayer() -> (_RepresentedPlayer:Player State Object Reference)
    bMatchChildTypes:bool=false
   ) -> (ReturnValue:Pilgrim Quickplay Playspace Object Reference)
   > n2 = Is Valid(InputObject=@n1)
   n3 = FortShippingLog!(
    execute=@n2.Is Not Valid
    WorldContextObject:Object Reference
    InString:String=No valid sparks music playspace in WBP_PilgrimQuickplay_NextSongDetailsPanel
    bLogAsWarning:bool=false
   ) -> (then)
   SetVisibility(
    execute=[@n6.then, @n3.then]
    self=Get VerticalBox_Contents() -> (VerticalBox_Contents:Vertical Box Object Reference)
    InVisibility=Collapsed
   )
   n4 = GetBandState!(ContextActor:Actor Object Reference=@n1) -> (ReturnValue:Pilgrim Quickplay Band State Component Object Reference)
   n5 = Is Valid(exec=@n2.Is Valid, InputObject=@n4)
   n6 = FortShippingLog(execute=@n5.Is Not Valid, InString=No valid band state in WBP_PilgrimQuickplay_NextSongDetailsPanel, bLogAsWarning=false)
   Set LocalBandState(execute=@n5.Is Valid, LocalBandState:Pilgrim Quickplay Band State Component Object Reference=@n4) -> (Output_Get:Pilgrim Quickplay Band State Component Object Reference)
   n7 = Get LocalBandState() -> (LocalBandState:Pilgrim Quickplay Band State Component Object Reference)
   n8 = Select(
    Option 0:Pilgrim Song Metadata Object Reference=GetNextSong!(self:Pilgrim Core Band State Component Object Reference=@n7) -> (ReturnValue:Pilgrim Song Metadata Object Reference)
    Option 1:Pilgrim Song Metadata Object Reference=GetCurrentSong!(self:Pilgrim Core Band State Component Object Reference=@n7) -> (ReturnValue:Pilgrim Song Metadata Object Reference)
    Index=Get isInGreenRoom() -> (isInGreenRoom:bool)
   ) -> (ReturnValue:Pilgrim Song Metadata Object Reference)
   n9 = IsValid(Object=@n8)
   > n10 = Branch(Condition=@n9)
   > n11 = Song Metadata!(execute, self, SongMetadata:Pilgrim Song Metadata Object Reference=@n8) -> (then)
   SetVisibility(
    execute=[@n11.then, @n12.then]
    self=Get VerticalBox_Contents()
    InVisibility=Select(Option 0:ESlateVisibility Enum=Collapsed, Option 1:ESlateVisibility Enum=SelfHitTestInvisible, Index=@n9) -> (ReturnValue:ESlateVisibility Enum)
   )
   n12 = FortShippingLog(execute=@n10.else, InString=Invalid song Metadata in WBP_PilgrimQuickplay_NextSongDetailsPanel, bLogAsWarning=false)
   Event Destruct()
   > Parent: Destruct()
   > n13 = Sequence()
   > n14 = For Each Loop(Array:Array of Async Action Start Listening To Event Object References=Get Async Tasks() -> (Async Tasks:Array of Async Action Start Listening To Event Object References)) -> (Array Element:Async Action Start Listening To Event Object Reference)
   Clear(execute=@n14.Completed, TargetArray:Array of Async Action Start Listening To Event Object References=Get Async Tasks())
   StopListeningToEvent(execute=@n14.LoopBody, self:Async Action Start Listening To Event Object Reference=@n14)
   n15 = For Each Loop(
    Exec=@n13.then_1
    Array:Array of Async Action Start Listening To Stateful Event Object References=Get Stateful Async Tasks() -> (Stateful Async Tasks:Array of Async Action Start Listening To Stateful Event Object References)
   ) -> (Array Element:Async Action Start Listening To Stateful Event Object Reference)
   Clear(execute=@n15.Completed, TargetArray:Array of Async Action Start Listening To Stateful Event Object References=Get Stateful Async Tasks())
   StopListeningToStatefulEvent(execute=@n15.LoopBody, self:Async Action Start Listening To Stateful Event Object Reference=@n15)
   Event Construct()
   > SetVisibility(self=Get VerticalBox_Contents(), InVisibility=Collapsed)
   n16 = GetOwningPlayer()
   > n17 = Is Valid(InputObject=@n16)
   FortShippingLog(execute=@n17.Is Not Valid, InString=No valid Owning Player Controller in WBP_PilgrimQuickplay_NextSongDetailsPanel, bLogAsWarning=false)
   SetRepresentedPlayer!(
    execute=@n17.Is Valid
    self
    RepresentedPlayer:Player State Object Reference=Get PlayerState!(self:Controller Object Reference=@n16) -> (PlayerState:Player State Object Reference)
   ) -> (then)
   > n18 = PlaylistDataReadyAsync!(execute:Exec) -> (
    then:Exec
    OnReady:Exec
    GameState:Fort Game State Athena Object Reference
    Playlist:Fort Playlist Object Reference
    PlaylistContextTags:"Gameplay Tag Container Structure (by ref)"
   ) {is=title K2Node_AsyncAction}
   n19 = PilgrimWaitForEventRouter!(execute:Exec=@n18.OnReady, ContextActor:Actor Object Reference=GetOwningPlayer()) -> (then:Exec, OnEventRouterAvailable:Exec, EventRouter:Gameplay Event Router Component Object Reference) {is=title K2Node_AsyncAction}
   n20 = Set Event Router(execute=@n19.OnEventRouterAvailable, Event Router:Gameplay Event Router Component Object Reference=@n19) -> (Output_Get:Gameplay Event Router Component Object Reference)
   > n21 = StartListeningToEvent!(
    execute:Exec
    Target:Gameplay Event Router Component Object Reference=@n20
    EventType:Script Struct Object Reference=/Script/PilgrimCoreRuntime.PilgrimEvent_PlayerState_InstrumentChanged
    Context:Object Reference=Get _RepresentedPlayer()
    EventBubblingRule:EEventBubblingRule Enum
   ) -> (
    then:Exec
    AsyncTaskProxy:Async Action Start Listening To Event Object Reference
    OnEventReceived:Exec
    Payload:Pilgrim Event Player State Instrument Changed Structure
   )
   > Add(
    TargetArray:Array of Async Action Start Listening To Event Object References=Get Async Tasks()
    NewItem:"Async Action Start Listening To Event Object Reference (by ref)"=@n21
   ) {is=title K2Node_CallArrayFunction}
   > n22 = StartListeningToEvent(
    Target=Get Event Router() -> (Event Router:Gameplay Event Router Component Object Reference)
    EventType=/Script/PilgrimCoreRuntime.PilgrimEvent_BandState_CustomSetlistChanged
   ) -> (Payload:Pilgrim Event Band State Custom Setlist Changed Structure)
   > Add(TargetArray=Get Async Tasks(), NewItem=@n22)
   > UpdateNextSong() {is=title K2Node_CallFunction}
   UpdateNextSong(execute=@n22.OnEventReceived)
   Set SelectedInstrument(
    execute=@n21.OnEventReceived
    SelectedInstrument:EPilgrimTrackType Enum=Break Pilgrim Event Player State Instrument Changed!(PilgrimEvent_PlayerState_InstrumentChanged:"Pilgrim Event Player State Instrument Changed Structure (by ref)"=@n21.Payload) -> (NewInstrumentChoice:EPilgrimTrackType Enum) {is=title K2Node_BreakStruct}
   ) -> (Output_Get:EPilgrimTrackType Enum=TrackGuitar)
   > UpdateNextSong()
   n23 = Event PreConstruct()
   > Branch(Condition=@n23.IsDesignTime)
   > ShowOnlyIntensity!(execute, self, Toggle:bool=Get Preview ShowOnlyIntensity() -> (Preview ShowOnlyIntensity:bool)) -> (then)
  ]}
 GetRepresentedPlayer {nodes [
   GetRepresentedPlayer() {is=title K2Node_FunctionEntry}
   > Return Node!(execute, _RepresentedPlayer:Player State Object Reference=Get _RepresentedPlayer())
  ]}
 SetIntensityForInstrument {nodes [
   n1 = SetIntensityForInstrument!() -> (
    then
    Widget:WBP Pilgrim Quickplay Song Intensity Entry Object Reference
    SongMetadata:Pilgrim Song Metadata Object Reference
    Instrument:EPilgrimTrackType Enum=TrackGuitar
    InstrumentIcon:ESparksInstrumentType Enum=Guitar
   ) {is=title K2Node_FunctionEntry}
   > Update Intensity for Instrument!(
    execute
    self:WBP Pilgrim Quickplay Song Intensity Entry Object Reference=@n1
    Intensity:int=GetTrackIntensity!(self:Pilgrim Song Metadata Object Reference=@n1.SongMetadata, TrackType:EPilgrimTrackType Enum=@n1.Instrument) -> (ReturnValue:int=0)
    TrackType:EPilgrimTrackType Enum=@n1.Instrument
    StartingInstrument:ESparksInstrumentType Enum=@n1.InstrumentIcon
    Is3D_deprecated:bool=false
    SongMetadata:Pilgrim Song Metadata Object Reference=@n1.SongMetadata
   ) -> (then)
  ]}
 SetRepresentedPlayer {nodes [
   n1 = SetRepresentedPlayer!() -> (then, RepresentedPlayer:Player State Object Reference) {is=title K2Node_FunctionEntry}
   > n2 = Set _RepresentedPlayer(_RepresentedPlayer:Player State Object Reference=@n1) -> (Output_Get:Player State Object Reference)
   > n3 = GetPilgrimCorePlayerStateComponent!(execute, PlayerState:Player State Object Reference=@n2) -> (then, ReturnValue:Pilgrim Core Player State Component Object Reference)
   > Set SelectedInstrument(SelectedInstrument=GetInstrumentChoice!(self:Pilgrim Core Player State Component Object Reference=@n3) -> (ReturnValue:EPilgrimTrackType Enum=TrackGuitar)) -> (Output_Get=TrackGuitar)
   > UpdateNextSong()
  ]}
 ShowOnlyIntensity {nodes [
   n1 = ShowOnlyIntensity!() -> (then, Toggle:bool=~) {is=title K2Node_FunctionEntry}
   > SetVisibility(
    self=[
     Get Text_UpcomingLabel() -> (Text_UpcomingLabel:Common Text Object Reference)
     Get CanvasPanel_Songtitle() -> (CanvasPanel_Songtitle:Canvas Panel Object Reference)
     Get CanvasPanel_ArtistName() -> (CanvasPanel_ArtistName:Canvas Panel Object Reference)
     Get Spacer_BetweenSongIdAndIntensity() -> (Spacer_BetweenSongIdAndIntensity:Spacer Object Reference)
     Get Overlay_Album() -> (Overlay_Album:Overlay Object Reference)
    ]
    InVisibility=Select(Option 0=SelfHitTestInvisible, Option 1=Collapsed, Index=@n1)
   )
   > SetVisibility(
    self=Get Text_Intensity_InLMP() -> (Text_Intensity_InLMP:Common Text Object Reference)
    InVisibility=Select(Option 0=Collapsed, Option 1=SelfHitTestInvisible, Index=@n1)
   )
  ]}
 Song Metadata {nodes [
   n1 = Song Metadata!() -> (then, SongMetadata:Pilgrim Song Metadata Object Reference) {is=title K2Node_FunctionEntry}
   > Is Valid(InputObject=@n1)
   > n2 = Sequence!(execute) -> (then_0, then_1, then_2:Exec, then_3:Exec)
   "SetText (Text)"(
    execute=@n2.then_1
    self=Get NextSongArtist_Label() -> (NextSongArtist_Label:Common Text Object Reference)
    InText=GetArtistText!(self:Sparks Song Data Object Reference=@n1) -> (ReturnValue:Text)
   )
   UpdateNextSongIntensities!(
    execute=@n2.then_3
    self
    SongMetadata:Pilgrim Song Metadata Object Reference=@n1
    Instrument:EPilgrimTrackType Enum=Get SelectedInstrument() -> (SelectedInstrument:EPilgrimTrackType Enum=TrackGuitar)
   ) -> (then)
   LazyLoadAlbumArt!(
    execute=@n2.then_2
    Song Metadata:Pilgrim Song Metadata Object Reference=@n1
    Lazy Image:Fort Lazy Image Object Reference=Get AlbumArt_Image() -> (AlbumArt_Image:Fort Lazy Image Object Reference)
   ) -> (then)
   "SetText (Text)"(
    execute=@n2.then_0
    self=Get NextSongTitle_Label() -> (NextSongTitle_Label:Common Text Object Reference)
    InText=GetTitleText!(self:Sparks Song Data Object Reference=@n1) -> (ReturnValue:Text)
   )
  ]}
 UpdateNextSongIntensities {nodes [
   n1 = UpdateNextSongIntensities!() -> (then, SongMetadata:Pilgrim Song Metadata Object Reference, Instrument:EPilgrimTrackType Enum=TrackGuitar) {is=title K2Node_FunctionEntry}
   > SetIntensityForInstrument!(
    execute
    self
    Widget:WBP Pilgrim Quickplay Song Intensity Entry Object Reference=Get Instrument_Intensity() -> (Instrument_Intensity:WBP Pilgrim Quickplay Song Intensity Entry Object Reference)
    SongMetadata:Pilgrim Song Metadata Object Reference=@n1
    Instrument:EPilgrimTrackType Enum=@n1.Instrument
    InstrumentIcon:ESparksInstrumentType Enum=GetStartingInstrumentType!(self:Pilgrim Song Metadata Object Reference=@n1, TrackType:EPilgrimTrackType Enum=@n1.Instrument) -> (ReturnValue:ESparksInstrumentType Enum=Guitar)
   ) -> (then) {is=title K2Node_CallFunction}
  ]}
]
variables [
 _RepresentedPlayer {type=PlayerState, default=null}
 Async Tasks {type="array<AsyncAction_StartListeningToEvent>", default []}
 Event Router {type=GameplayEventRouterComponent, default=null}
 isInGreenRoom {type=boolean}
 LocalBandState {type=PilgrimQuickplayBandStateComponent, default=null}
 Preview ShowOnlyIntensity {type=boolean}
 SelectedInstrument {type=EPilgrimTrackType, default=TrackGuitar}
 Stateful Async Tasks {type="array<AsyncAction_StartListeningToStatefulEvent>", default []}
]
widgets {
 info {
  inheritedWidgetCount=0
  parentClass=/PilgrimQuickplay/Widgets/Results/WBP_Pilgrim_UI_Results_Base.WBP_Pilgrim_UI_Results_Base_C
  rootWidgetClass=&VerticalBox
  widgetCount=24
 }
 widgets [
  VerticalBox_Contents {widgetClassPath=&VerticalBox, bIsVariable=true, widget=VerticalBox_Contents, properties {bOverride_Cursor=true}}
  Text_UpcomingLabel {
   widgetClassPath=&CommonTextBlock
   bIsVariable=true
   parent=VerticalBox_Contents
   widget=Text_UpcomingLabel
   properties {
    clipping=OnDemand
    flowDirectionPreference=Culture
    text=UPCOMING
    textTransformPolicy=ToUpper
    font {
     bForceMonospaced=false
     bMaterialIsStencil=false
     fontMaterial=null
     fontObject=&HeadingNow
     letterSpacing=0
     monospacedWidth=1
     size=24
     skewAmount=0
     typefaceFontName=76 Bold Italic
     outlineSettings {
      bApplyOutlineToDropShadows=false
      bMiteredCorners=false
      bSeparateFillAlpha=false
      outlineMaterial=&M-FontShadow
      outlineSize=2
      outlineColor=<rgba 0 0 0 1>
     }
    }
    shadowOffset=<0 0>
    strikeBrush {
     bIsDynamicallyLoaded=false
     drawAs=NoDrawType
     imageType=NoImage
     mirroring=NoMirror
     resourceName=None
     resourceObject=null
     tiling=NoTile
     imageSize=<32 32>
     margin=<ltrb 0 0 0 0>
     outlineSettings {
      bUseBrushTransparency=false
      roundingType=HalfHeightRadius
      width=0
      color {colorUseRule=UseColor_Specified, specifiedColor=<rgba 0 0 0 0>}
      cornerRadii=<0 0 0 0>
     }
     tintColor {colorUseRule=UseColor_Specified, specifiedColor=<rgba 1 1 1 1>}
     uVRegion {bIsValid=false, max=<0 0>, min=<0 0>}
    }
   }
   slot {class=VerticalBoxSlot, properties {horizontalAlignment=HAlign_Left, verticalAlignment=VAlign_Top}}
  }
  Overlay_0 {widgetClassPath=&Overlay, parent=VerticalBox_Contents, widget=Overlay_0, slot {class=VerticalBoxSlot}}
  WBP_Sparks_BackingPanel {
   widgetClassPath=/PilgrimCore/Widgets/SparksUILibrary/WBP_Sparks_BackingPanel
   bIsVariable=true
   parent=Overlay_0
   widget=WBP_Sparks_BackingPanel
   slot {class=OverlaySlot, properties {horizontalAlignment=HAlign_Fill, verticalAlignment=VAlign_Fill}}
  }
  HorizontalBox_2 {
   widgetClassPath=&HorizontalBox
   parent=Overlay_0
   widget=HorizontalBox_2
   slot {class=OverlaySlot, properties {horizontalAlignment=HAlign_Fill, verticalAlignment=VAlign_Fill, padding=<ltrb 8 8 16 8>}}
  }
  Overlay_Album {widgetClassPath=&Overlay, bIsVariable=true, parent=HorizontalBox_2, widget=Overlay_Album, slot {class=HorizontalBoxSlot}}
  Spacer_74 {widgetClassPath=&Spacer, parent=Overlay_Album, widget=Spacer_74, properties {size=<144 144>}, slot {class=OverlaySlot}}
  AlbumGlow {
   widgetClassPath=&Image
   parent=Overlay_Album
   widget=AlbumGlow
   properties {
    brush {
     bIsDynamicallyLoaded=false
     drawAs=Image
     imageType=NoImage
     mirroring=NoMirror
     resourceName=None
     resourceObject=/PilgrimCore/Widgets/Materials/M_UI_Pilgrim_AlbumArt_Glow
     tiling=NoTile
     imageSize=<32 32>
     margin=<ltrb 0 0 0 0>
     outlineSettings {
      bUseBrushTransparency=false
      roundingType=HalfHeightRadius
      width=0
      color {colorUseRule=UseColor_Specified, specifiedColor=<rgba 0 0 0 0>}
      cornerRadii=<0 0 0 0>
     }
     tintColor {colorUseRule=UseColor_Specified, specifiedColor=<rgba 1 1 1 1>}
     uVRegion {bIsValid=false, max=<0 0>, min=<0 0>}
    }
    renderTransform {angle=0, scale=<1.1 1.1>, shear=<0 0>, translation=<0 0>}
   }
   slot {class=OverlaySlot, properties {horizontalAlignment=HAlign_Fill, verticalAlignment=VAlign_Fill}}
  }
  AlbumArt_Image {
   widgetClassPath=&FortLazyImage
   bIsVariable=true
   parent=Overlay_Album
   widget=AlbumArt_Image
   properties {
    materialTextureParamName=T_AlbumArt
    brush {
     bIsDynamicallyLoaded=false
     drawAs=Image
     imageType=FullColor
     mirroring=NoMirror
     resourceName=None
     resourceObject=/SparksUICommon/Materials/MI_UI_Sparks_SetList_AlbumArt_Rounded
     tiling=NoTile
     imageSize=<128 128>
     margin=<ltrb 0 0 0 0>
     outlineSettings {
      bUseBrushTransparency=false
      roundingType=HalfHeightRadius
      width=0
      color {colorUseRule=UseColor_Specified, specifiedColor=<rgba 0 0 0 0>}
      cornerRadii=<0 0 0 0>
     }
     tintColor {colorUseRule=UseColor_Specified, specifiedColor=<rgba 1 1 1 1>}
     uVRegion {bIsValid=false, max=<0 0>, min=<0 0>}
    }
   }
   slot {class=OverlaySlot, properties {horizontalAlignment=HAlign_Center, verticalAlignment=VAlign_Center}}
  }
  SizeBox_2 {
   widgetClassPath=&SizeBox
   bIsVariable=true
   parent=HorizontalBox_2
   widget=SizeBox_2
   properties {bOverride_MaxDesiredHeight=true, maxDesiredHeight=144}
   slot {class=HorizontalBoxSlot, properties {padding=<ltrb 8 0 0 0>, size {sizeRule=Fill, value=1}}}
  }
  VerticalBox_3 {
   widgetClassPath=&VerticalBox
   parent=SizeBox_2
   widget=VerticalBox_3
   slot {class=SizeBoxSlot, properties {verticalAlignment=VAlign_Center}}
  }
  CanvasPanel_Songtitle {widgetClassPath=&CanvasPanel, bIsVariable=true, parent=VerticalBox_3, widget=CanvasPanel_Songtitle, slot {class=VerticalBoxSlot}}
  NextSongTitle_Label {
   widgetClassPath=&CommonTextBlock
   bIsVariable=true
   parent=CanvasPanel_Songtitle
   widget=NextSongTitle_Label
   properties {
    scrollStyle=/PilgrimQuickplay/Widgets/ScrollStyles/TextScrollStyle_PilgrimQuickplay_Setlist.TextScrollStyle_PilgrimQuickplay_Setlist_C
    text=Song Title Is Long
    textTransformPolicy=ToUpper
    font {
     bForceMonospaced=false
     bMaterialIsStencil=false
     fontMaterial=null
     fontObject=&HeadingNow
     letterSpacing=0
     monospacedWidth=1
     size=13.5
     skewAmount=0
     typefaceFontName=75 Medium
     outlineSettings {
      bApplyOutlineToDropShadows=false
      bMiteredCorners=false
      bSeparateFillAlpha=false
      outlineMaterial=null
      outlineSize=0
      outlineColor=<rgba 0 0 0 1>
     }
    }
    shadowOffset=<0 0>
    strikeBrush {
     bIsDynamicallyLoaded=false
     drawAs=NoDrawType
     imageType=NoImage
     mirroring=NoMirror
     resourceName=None
     resourceObject=null
     tiling=NoTile
     imageSize=<32 32>
     margin=<ltrb 0 0 0 0>
     outlineSettings {
      bUseBrushTransparency=false
      roundingType=HalfHeightRadius
      width=0
      color {colorUseRule=UseColor_Specified, specifiedColor=<rgba 0 0 0 0>}
      cornerRadii=<0 0 0 0>
     }
     tintColor {colorUseRule=UseColor_Specified, specifiedColor=<rgba 1 1 1 1>}
     uVRegion {bIsValid=false, max=<0 0>, min=<0 0>}
    }
   }
   slot {class=CanvasPanelSlot, properties {layoutData {alignment=<0 0>, anchors {maximum=<1 0>, minimum=<0 0>}, offsets=<ltrb 0 0 0 32.5>}}}
  }
  CanvasPanel_ArtistName {widgetClassPath=&CanvasPanel, bIsVariable=true, parent=VerticalBox_3, widget=CanvasPanel_ArtistName, slot {class=VerticalBoxSlot}}
  NextSongArtist_Label {
   widgetClassPath=&CommonTextBlock
   bIsVariable=true
   parent=CanvasPanel_ArtistName
   widget=NextSongArtist_Label
   properties {
    clipping=OnDemand
    scrollStyle=&TextScrollStyle_PilgrimQuickplay_Setlist_C
    text=Artist Name
    textTransformPolicy=ToUpper
    colorAndOpacity {colorUseRule=UseColor_Specified, specifiedColor=<rgba 1 1 1 0.8>}
    font {
     bForceMonospaced=false
     bMaterialIsStencil=false
     fontMaterial=null
     fontObject=&HeadingNow
     letterSpacing=0
     monospacedWidth=1
     size=12
     skewAmount=0
     typefaceFontName=75 Medium
     outlineSettings {
      bApplyOutlineToDropShadows=false
      bMiteredCorners=false
      bSeparateFillAlpha=false
      outlineMaterial=null
      outlineSize=0
      outlineColor=<rgba 0 0 0 1>
     }
    }
    shadowOffset=<0 0>
    strikeBrush {
     bIsDynamicallyLoaded=false
     drawAs=NoDrawType
     imageType=NoImage
     mirroring=NoMirror
     resourceName=None
     resourceObject=null
     tiling=NoTile
     imageSize=<32 32>
     margin=<ltrb 0 0 0 0>
     outlineSettings {
      bUseBrushTransparency=false
      roundingType=HalfHeightRadius
      width=0
      color {colorUseRule=UseColor_Specified, specifiedColor=<rgba 0 0 0 0>}
      cornerRadii=<0 0 0 0>
     }
     tintColor {colorUseRule=UseColor_Specified, specifiedColor=<rgba 1 1 1 1>}
     uVRegion {bIsValid=false, max=<0 0>, min=<0 0>}
    }
   }
   slot {class=CanvasPanelSlot, properties {layoutData {alignment=<0 0>, anchors {maximum=<1 0>, minimum=<0 0>}, offsets=<ltrb 0 0 0 30>}}}
  }
  Spacer_BetweenSongIdAndIntensity {
   widgetClassPath=&Spacer
   bIsVariable=true
   parent=VerticalBox_3
   widget=Spacer_BetweenSongIdAndIntensity
   properties {size=<1 10>}
   slot {class=VerticalBoxSlot}
  }
  ScaleBox_81 {
   widgetClassPath=&ScaleBox
   parent=VerticalBox_3
   widget=ScaleBox_81
   slot {class=VerticalBoxSlot, properties {verticalAlignment=VAlign_Center, size {sizeRule=Fill, value=1}}}
  }
  GridPanel_417 {
   widgetClassPath=&GridPanel
   parent=ScaleBox_81
   widget=GridPanel_417
   properties {columnFill [1], rowFill [1]}
   slot {class=ScaleBoxSlot, properties {horizontalAlignment=HAlign_Fill, verticalAlignment=VAlign_Top}}
  }
  HorizontalBox_0 {
   widgetClassPath=&HorizontalBox
   parent=GridPanel_417
   widget=HorizontalBox_0
   properties {bOverride_Cursor=true}
   slot {class=GridSlot}
  }
  Text_Intensity_InLMP {
   widgetClassPath=&CommonTextBlock
   bIsVariable=true
   parent=HorizontalBox_0
   widget=Text_Intensity_InLMP
   properties {
    clipping=OnDemand
    flowDirectionPreference=Culture
    text=Intensity
    font {
     bForceMonospaced=false
     bMaterialIsStencil=false
     fontMaterial=null
     fontObject=&HeadingNow
     letterSpacing=0
     monospacedWidth=1
     size=16.5
     skewAmount=0
     typefaceFontName=76 Bold Italic
     outlineSettings {
      bApplyOutlineToDropShadows=false
      bMiteredCorners=false
      bSeparateFillAlpha=false
      outlineMaterial=null
      outlineSize=0
      outlineColor=<rgba 0 0 0 1>
     }
    }
    shadowOffset=<0 0>
    strikeBrush {
     bIsDynamicallyLoaded=false
     drawAs=NoDrawType
     imageType=NoImage
     mirroring=NoMirror
     resourceName=None
     resourceObject=null
     tiling=NoTile
     imageSize=<32 32>
     margin=<ltrb 0 0 0 0>
     outlineSettings {
      bUseBrushTransparency=false
      roundingType=HalfHeightRadius
      width=0
      color {colorUseRule=UseColor_Specified, specifiedColor=<rgba 0 0 0 0>}
      cornerRadii=<0 0 0 0>
     }
     tintColor {colorUseRule=UseColor_Specified, specifiedColor=<rgba 1 1 1 1>}
     uVRegion {bIsValid=false, max=<0 0>, min=<0 0>}
    }
   }
   slot {class=HorizontalBoxSlot, properties {verticalAlignment=VAlign_Bottom, padding=<ltrb 0 0 10 0>}}
  }
  CommonVisibilitySwitcher_0 {
   widgetClassPath=&CommonVisibilitySwitcher
   parent=HorizontalBox_0
   widget=CommonVisibilitySwitcher_0
   slot {class=HorizontalBoxSlot}
  }
  Instrument_Intensity {
   widgetClassPath=/PilgrimQuickplay/Widgets/SetlistAndSongDisplay/WBP_PilgrimQuickplay_SongIntensityEntry
   bIsVariable=true
   parent=CommonVisibilitySwitcher_0
   widget=Instrument_Intensity
   slot {class=CommonVisibilitySwitcherSlot}
  }
  MissingPartIcon {
   widgetClassPath=&Image
   bIsVariable=true
   parent=CommonVisibilitySwitcher_0
   widget=MissingPartIcon
   properties {
    visibility=SelfHitTestInvisible
    brush {
     bIsDynamicallyLoaded=false
     drawAs=Image
     imageType=FullColor
     mirroring=NoMirror
     resourceName=None
     resourceObject=/PilgrimQuickplay/UI/MI_UI_MissingPartsAlertIcon
     tiling=NoTile
     imageSize=<23 23>
     margin=<ltrb 0 0 0 0>
     outlineSettings {
      bUseBrushTransparency=false
      roundingType=HalfHeightRadius
      width=0
      color {colorUseRule=UseColor_Specified, specifiedColor=<rgba 0 0 0 0>}
      cornerRadii=<0 0 0 0>
     }
     tintColor {colorUseRule=UseColor_Specified, specifiedColor=<rgba 0.9301 1 0.0103 1>}
     uVRegion {bIsValid=false, max=<0 0>, min=<0 0>}
    }
   }
   slot {class=CommonVisibilitySwitcherSlot, properties {horizontalAlignment=HAlign_Right, verticalAlignment=VAlign_Center}}
  }
  OptionalMeterEntry {
   widgetClassPath=&NamedSlot
   bIsVariable=true
   parent=GridPanel_417
   widget=OptionalMeterEntry
   properties {visibility=Collapsed}
   slot {class=GridSlot, properties {row=1}}
  }
 ]
}
