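# Post-match Battle Pass progress widget for Pilgrim Quickplay, showing level, XP to next level and match XP earned on a segmented bar.
# Parent PilgrimPostGameProgressWidget events refresh the text through UpdateText, which also keeps the star reward box collapsed in both Select options.
# StartProgressAnimation sets a starting fill from XP values, then transitions to CurrentProgress, gated by CanAnimateXP until destruct.
# Battle Pass ownership sets a flag, then broadcasts a delayed dispatcher.
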WidgetBlueprint /PilgrimQuickplay/Widgets/Results/SongResults/WBP_PilgrimQuickplay_BattlePass_Progress
graphs [
 EventGraph {nodes [
   "On Progress Transition Finished (ProgressBar)"() {is=title K2Node_ComponentBoundEvent}
   > Set CanAnimateXP(CanAnimateXP:bool=false) -> (Output_Get:bool)
   StartProgressAnimation() {is=title K2Node_CustomEvent}
   > Branch(Condition=Get CanAnimateXP() -> (CanAnimateXP:bool))
   n1 = Get XpToNextLevel() -> (XpToNextLevel:int=0)
   n2 = int - int(A=Get CurrentXP() -> (CurrentXP:int=0), B=GetTotalXP!(self:Pilgrim Post Game Progress Widget Object Reference) -> (ReturnValue:int=0))
   n3 = Get ProgressBar() -> (ProgressBar:WBP UIKit Progress Bar Object Reference)
   > SetProgress!(
    execute
    self:WBP UIKit Progress Bar Base Object Reference=@n3
    Progress:double=SafeDivide(A="To Float (Integer)"(InInt=@n2), B="To Float (Integer)"(InInt=@n1))
   ) -> (then)
   > SetProgressTransition!(execute, self:WBP UIKit Progress Bar Base Object Reference=@n3, Progress:double=Get CurrentProgress() -> (CurrentProgress:float)) -> (then)
   Event OnBattlePassOwned() {is=title K2Node_Event}
   > Set OwnsBattlePass(OwnsBattlePass:bool=true) -> (Output_Get:bool)
   > Delay(Duration=0.2)
   > Call On Bp Owned!(execute:Exec, self:WBP Pilgrim Quickplay Battle Pass Progress Object Reference) -> (then:Exec) {is=title K2Node_CallDelegate}
   Event OnLevelChanged() {is=title K2Node_Event}
   > UpdateText()
   Event OnProgressUpdated() {is=title K2Node_Event}
   > UpdateText()
   Event Construct()
   n4 = Get ProgressBar()
   > SetProgress(self=@n4, Progress=Get ProgressValue!(self:WBP UIKit Progress Bar Base Object Reference=@n4) -> (ProgressValue:double))
   Event Destruct()
   > UnbindAllFromAnimationFinished!(
    execute
    self:User Widget Object Reference
    Animation:Widget Animation Object Reference=Get Anim_Progress_Base!(self:WBP UIKit Progress Bar Base Object Reference=Get ProgressBar()) -> (Anim_Progress_Base:Widget Animation Object Reference)
   ) -> (then)
   > Set CanAnimateXP(CanAnimateXP=true)
   Event OnSuperCharged() {is=title K2Node_Event}
  ]}
 On Bp Owned {nodes [On Bp Owned() {is=title K2Node_FunctionEntry}]}
 UpdateText {nodes [
   UpdateText() {is=title K2Node_FunctionEntry}
   n1 = Get CurrentXP() -> (CurrentXP=0)
   n2 = Get XpToNextLevel() -> (XpToNextLevel=0)
   n3 = Get CurrentLevel() -> (CurrentLevel:int=0)
   > SetText(
    self=Get RichText_Level() -> (RichText_Level:Common Rich Text Block Object Reference)
    InText=Format Text!(
     Format="LOCTEXT(\"<text color=\\\"FFFFFFFF\\\">{XPamount} XP</> to LVL {level}\")"
     XPamount:int=int - int(A=@n2, B=@n1) {is=title K2Node_CallFunction}
     level:int=int + int(A=@n3, B=1) -> (ReturnValue=0) {is=title K2Node_CommutativeAssociativeBinaryOperator}
    ) -> (Result)
   )
   n4 = Get MaxBPLevelForRewards() -> (MaxBPLevelForRewards:int=0)
   > SetVisibility(
    self=Get VerticalBox_5stars() -> (VerticalBox_5stars:Vertical Box Object Reference)
    InVisibility=Select(Option 0=Collapsed, Option 1=Collapsed, Index="integer > integer"(A=@n3, B=@n4))
   )
   > SetText(
    self=Get RichText_CurrentLevel() -> (RichText_CurrentLevel:Common Rich Text Block Object Reference)
    InText=Format Text!(Format="LOCTEXT(\"LVL <text color=\\\"FFFFFFFF\\\">{level}</>\")", level=@n3) -> (Result)
   )
   n5 = GetMatchXP!(self:Pilgrim Post Game Progress Widget Object Reference) -> (ReturnValue:int=0)
   > SetVisibility(
    self=Get RichText_SongXP() -> (RichText_SongXP:Common Rich Text Block Object Reference)
    InVisibility=Select(Option 0=Collapsed, Option 1=SelfHitTestInvisible, Index="integer > integer"(A=@n5))
   )
   > SetText(self=Get RichText_SongXP(), InText=Format Text!(Format="LOCTEXT(\"EARNED <text color=\\\"FFFFFFFF\\\">{XP}</>\")", XP:int=@n5) -> (Result))
  ]}
]
variables [
 Bp Material {type=MaterialInterface, default=/Game/UI/Foundation/XPBoost/M-BattlePassBadge}
 CanAnimateXP {type=boolean, default=true}
 MaxBPLevelForRewards {type=integer, default=199}
 On Bp Owned {}
 OwnsBattlePass {type=boolean}
]
widgets {
 info {inheritedWidgetCount=2, parentClass=/Script/PilgrimCoreUI.PilgrimPostGameProgressWidget, rootWidgetClass=&Overlay, widgetCount=20}
 widgets [
  Overlay_92 {widgetClassPath=&Overlay, widget=Overlay_92}
  DesignTimeSizePreview {
   widgetClassPath=&Spacer
   parent=Overlay_92
   widget=DesignTimeSizePreview
   properties {visibility=Collapsed, size=<1080 1>}
   slot {class=OverlaySlot}
  }
  Image_126 {
   widgetClassPath=&Image
   bIsVariable=true
   parent=Overlay_92
   widget=Image_126
   properties {
    visibility=SelfHitTestInvisible
    brush {
     bIsDynamicallyLoaded=false
     drawAs=Image
     imageType=FullColor
     mirroring=NoMirror
     resourceName=None
     resourceObject=/PilgrimCore/Widgets/Materials/MI_UI_Pilgrim_Panel_EmptyPip
     tiling=NoTile
     imageSize=<32 32>
     margin=<ltrb 0 0 0 0>
     outlineSettings {
      bUseBrushTransparency=false
      roundingType=HalfHeightRadius
      width=0
      color {colorUseRule=UseColor_Specified, specifiedColor=<rgba 0 0 0 0>}
      cornerRadii=<0 0 0 0>
     }
     tintColor {colorUseRule=UseColor_Specified, specifiedColor=<rgba 1 1 1 1>}
     uVRegion {bIsValid=false, max=<0 0>, min=<0 0>}
    }
    renderTransform {angle=0, scale=<1 1>, shear=<-10 0>, translation=<0 0>}
   }
   slot {class=OverlaySlot, properties {horizontalAlignment=HAlign_Fill, verticalAlignment=VAlign_Fill}}
  }
  HorizontalBox_2 {
   widgetClassPath=&HorizontalBox
   parent=Overlay_92
   widget=HorizontalBox_2
   slot {class=OverlaySlot, properties {horizontalAlignment=HAlign_Fill, verticalAlignment=VAlign_Fill, padding=<ltrb 36 0 36 0>}}
  }
  Overlay_Icons {
   widgetClassPath=&Overlay
   parent=HorizontalBox_2
   widget=Overlay_Icons
   properties {visibility=Visible, renderTransform {angle=0, scale=<1 1>, shear=<0 0>, translation=<-20 0>}}
   slot {class=HorizontalBoxSlot, properties {horizontalAlignment=HAlign_Center, verticalAlignment=VAlign_Center, padding=<ltrb 20 0 0 0>}}
  }
  Image_Strobe {
   widgetClassPath=&Image
   bIsVariable=true
   parent=Overlay_Icons
   widget=Image_Strobe
   properties {
    visibility=SelfHitTestInvisible
    brush {
     bIsDynamicallyLoaded=false
     drawAs=Image
     imageType=NoImage
     mirroring=NoMirror
     resourceName=None
     resourceObject=/PilgrimCore/Widgets/Materials/MI_UI_Pilgrim_ScoreGlow
     tiling=NoTile
     imageSize=<100 100>
     margin=<ltrb 0 0 0 0>
     outlineSettings {
      bUseBrushTransparency=false
      roundingType=HalfHeightRadius
      width=0
      color {colorUseRule=UseColor_Specified, specifiedColor=<rgba 0 0 0 0>}
      cornerRadii=<0 0 0 0>
     }
     tintColor {colorUseRule=UseColor_Specified, specifiedColor=<rgba 1 1 1 1>}
     uVRegion {bIsValid=false, max=<0 0>, min=<0 0>}
    }
    renderTransform {angle=12, scale=<1.75 2>, shear=<0 0>, translation=<0 0>}
   }
   slot {class=OverlaySlot, properties {horizontalAlignment=HAlign_Center, verticalAlignment=VAlign_Center}}
  }
  Image_BattlePass {
   widgetClassPath=&Image
   bIsVariable=true
   parent=Overlay_Icons
   widget=Image_BattlePass
   properties {
    visibility=SelfHitTestInvisible
    brush {
     bIsDynamicallyLoaded=false
     drawAs=Image
     imageType=FullColor
     mirroring=NoMirror
     resourceName=None
     resourceObject=/Game/UI/Foundation/Textures/Icons/Items/UniversalXP/T_UI_UniversalXP_Icon_Standard_512
     tiling=NoTile
     imageSize=<100 100>
     margin=<ltrb 0 0 0 0>
     outlineSettings {
      bUseBrushTransparency=false
      roundingType=HalfHeightRadius
      width=0
      color {colorUseRule=UseColor_Specified, specifiedColor=<rgba 0 0 0 0>}
      cornerRadii=<0 0 0 0>
     }
     tintColor {colorUseRule=UseColor_Specified, specifiedColor=<rgba 1 1 1 1>}
     uVRegion {bIsValid=false, max=<0 0>, min=<0 0>}
    }
    renderTransform {angle=0, scale=<1.3 1.3>, shear=<0 0>, translation=<0 0>}
   }
   slot {class=OverlaySlot, properties {horizontalAlignment=HAlign_Center, verticalAlignment=VAlign_Center}}
  }
  Spacer_64 {
   widgetClassPath=&Spacer
   parent=HorizontalBox_2
   widget=Spacer_64
   properties {size=<20 1>}
   slot {class=HorizontalBoxSlot, properties {horizontalAlignment=HAlign_Left, verticalAlignment=VAlign_Top}}
  }
  GridPanel_BattlePassContainer {
   widgetClassPath=&GridPanel
   parent=HorizontalBox_2
   widget=GridPanel_BattlePassContainer
   properties {columnFill [1], rowFill [1]}
   slot {class=HorizontalBoxSlot, properties {padding=<ltrb 0 0 0 8>, size {sizeRule=Fill, value=1}}}
  }
  ProgressBar {
   widgetClassPath=/FortUILibrary/UIKit/ProgressBar/WBP_UIKit_ProgressBar
   bIsVariable=true
   parent=GridPanel_BattlePassContainer
   widget=ProgressBar
   properties {
    cornerRoundness=1
    durationTime=1.25
    height=10
    progressValue=0.2
    segmentsCount=10
    separatorThickness_PX=3
    useCircular=false
    useDeltaBar=true
    useDeltaDecay=true
    useSegments=true
    colorA=<rgba 0 0.855 0.9911 1>
    colorB=<rgba 0.5583 0.9734 1 1>
    colorC=<rgba 0.5583 0.9734 1 1>
    deltaColor=<rgba 1 0.7426 0.3375 1>
    deltaDecayColor=<rgba 1 0.7033 0 1>
    emptyColor=<rgba 0.004 0.0015 0.0452 0.8>
    renderTransform {angle=0, scale=<1 1>, shear=<-10 0>, translation=<0 0>}
   }
   slot {class=GridSlot, properties {verticalAlignment=VAlign_Center, padding=<ltrb 0 45 0 40>}}
  }
  HorizontalBox_0 {
   widgetClassPath=&HorizontalBox
   parent=GridPanel_BattlePassContainer
   widget=HorizontalBox_0
   slot {class=GridSlot, properties {verticalAlignment=VAlign_Top, padding=<ltrb 0 12 3 4>}}
  }
  RichText_CurrentLevel {
   widgetClassPath=&CommonRichTextBlock
   bIsVariable=true
   parent=HorizontalBox_0
   widget=RichText_CurrentLevel
   properties {
    bOverrideDefaultStyle=true
    mobileTextBlockScale=1.1
    text="LVL <text color=\"FFFFFFFF\">X</>"
    textTransformPolicy=ToUpper
    defaultTextStyleOverride {
     overflowPolicy=Clip
     transformPolicy=None
     colorAndOpacity {colorUseRule=UseColor_Specified, specifiedColor=<rgba 0 0.8714 1 1>}
     font {
      bForceMonospaced=false
      bMaterialIsStencil=false
      fontMaterial=null
      fontObject=&HeadingNow
      letterSpacing=0
      monospacedWidth=1
      size=18
      skewAmount=0
      typefaceFontName=76 Bold
      outlineSettings {
       bApplyOutlineToDropShadows=false
       bMiteredCorners=false
       bSeparateFillAlpha=false
       outlineMaterial=null
       outlineSize=0
       outlineColor=<rgba 0 0 0 1>
      }
     }
     highlightColor {colorUseRule=UseColor_Specified, specifiedColor=<rgba 0 0 0 1>}
     highlightShape {
      bIsDynamicallyLoaded=false
      drawAs=Image
      imageType=NoImage
      mirroring=NoMirror
      resourceName=None
      resourceObject=null
      tiling=NoTile
      imageSize=<32 32>
      margin=<ltrb 0 0 0 0>
      outlineSettings {
       bUseBrushTransparency=false
       roundingType=HalfHeightRadius
       width=0
       color {colorUseRule=UseColor_Specified, specifiedColor=<rgba 0 0 0 0>}
       cornerRadii=<0 0 0 0>
      }
      tintColor {colorUseRule=UseColor_Specified, specifiedColor=<rgba 1 1 1 1>}
      uVRegion {bIsValid=false, max=<0 0>, min=<0 0>}
     }
     selectedBackgroundColor {colorUseRule=UseColor_Foreground, specifiedColor=<rgba 1 0 1 1>}
     shadowColorAndOpacity=<rgba 0 0 0 1>
     shadowOffset=<0 0>
     strikeBrush {
      bIsDynamicallyLoaded=false
      drawAs=Image
      imageType=NoImage
      mirroring=NoMirror
      resourceName=None
      resourceObject=null
      tiling=NoTile
      imageSize=<32 32>
      margin=<ltrb 0 0 0 0>
      outlineSettings {
       bUseBrushTransparency=false
       roundingType=HalfHeightRadius
       width=0
       color {colorUseRule=UseColor_Specified, specifiedColor=<rgba 0 0 0 0>}
       cornerRadii=<0 0 0 0>
      }
      tintColor {colorUseRule=UseColor_Specified, specifiedColor=<rgba 1 1 1 1>}
      uVRegion {bIsValid=false, max=<0 0>, min=<0 0>}
     }
     underlineBrush {
      bIsDynamicallyLoaded=false
      drawAs=Image
      imageType=NoImage
      mirroring=NoMirror
      resourceName=None
      resourceObject=null
      tiling=NoTile
      imageSize=<32 32>
      margin=<ltrb 0 0 0 0>
      outlineSettings {
       bUseBrushTransparency=false
       roundingType=HalfHeightRadius
       width=0
       color {colorUseRule=UseColor_Specified, specifiedColor=<rgba 0 0 0 0>}
       cornerRadii=<0 0 0 0>
      }
      tintColor {colorUseRule=UseColor_Specified, specifiedColor=<rgba 1 1 1 1>}
      uVRegion {bIsValid=false, max=<0 0>, min=<0 0>}
     }
    }
   }
   slot {class=HorizontalBoxSlot, properties {padding=<ltrb 0 0 20 0>}}
  }
  RichText_Level {
   widgetClassPath=&CommonRichTextBlock
   bIsVariable=true
   parent=HorizontalBox_0
   widget=RichText_Level
   properties {
    bOverrideDefaultStyle=true
    justification=Right
    mobileTextBlockScale=1.1
    scrollStyle=&TextScrollStyle-Base_C
    text="<text color=\"FFFFFFFF\">10000 XP</> to LVL 3"
    defaultTextStyleOverride {
     overflowPolicy=Clip
     transformPolicy=None
     colorAndOpacity {colorUseRule=UseColor_Specified, specifiedColor=<rgba 0 0.8714 1 1>}
     font {
      bForceMonospaced=false
      bMaterialIsStencil=false
      fontMaterial=null
      fontObject=&HeadingNow
      letterSpacing=0
      monospacedWidth=1
      size=15.75
      skewAmount=0
      typefaceFontName=75 Medium
      outlineSettings {
       bApplyOutlineToDropShadows=false
       bMiteredCorners=false
       bSeparateFillAlpha=false
       outlineMaterial=null
       outlineSize=0
       outlineColor=<rgba 0 0 0 1>
      }
     }
     highlightColor {colorUseRule=UseColor_Specified, specifiedColor=<rgba 0 0 0 1>}
     highlightShape {
      bIsDynamicallyLoaded=false
      drawAs=Image
      imageType=NoImage
      mirroring=NoMirror
      resourceName=None
      resourceObject=null
      tiling=NoTile
      imageSize=<32 32>
      margin=<ltrb 0 0 0 0>
      outlineSettings {
       bUseBrushTransparency=false
       roundingType=HalfHeightRadius
       width=0
       color {colorUseRule=UseColor_Specified, specifiedColor=<rgba 0 0 0 0>}
       cornerRadii=<0 0 0 0>
      }
      tintColor {colorUseRule=UseColor_Specified, specifiedColor=<rgba 1 1 1 1>}
      uVRegion {bIsValid=false, max=<0 0>, min=<0 0>}
     }
     selectedBackgroundColor {colorUseRule=UseColor_Foreground, specifiedColor=<rgba 1 0 1 1>}
     shadowColorAndOpacity=<rgba 0 0 0 1>
     shadowOffset=<0 0>
     strikeBrush {
      bIsDynamicallyLoaded=false
      drawAs=Image
      imageType=NoImage
      mirroring=NoMirror
      resourceName=None
      resourceObject=null
      tiling=NoTile
      imageSize=<32 32>
      margin=<ltrb 0 0 0 0>
      outlineSettings {
       bUseBrushTransparency=false
       roundingType=HalfHeightRadius
       width=0
       color {colorUseRule=UseColor_Specified, specifiedColor=<rgba 0 0 0 0>}
       cornerRadii=<0 0 0 0>
      }
      tintColor {colorUseRule=UseColor_Specified, specifiedColor=<rgba 1 1 1 1>}
      uVRegion {bIsValid=false, max=<0 0>, min=<0 0>}
     }
     underlineBrush {
      bIsDynamicallyLoaded=false
      drawAs=Image
      imageType=NoImage
      mirroring=NoMirror
      resourceName=None
      resourceObject=null
      tiling=NoTile
      imageSize=<32 32>
      margin=<ltrb 0 0 0 0>
      outlineSettings {
       bUseBrushTransparency=false
       roundingType=HalfHeightRadius
       width=0
       color {colorUseRule=UseColor_Specified, specifiedColor=<rgba 0 0 0 0>}
       cornerRadii=<0 0 0 0>
      }
      tintColor {colorUseRule=UseColor_Specified, specifiedColor=<rgba 1 1 1 1>}
      uVRegion {bIsValid=false, max=<0 0>, min=<0 0>}
     }
    }
   }
   slot {class=HorizontalBoxSlot, properties {verticalAlignment=VAlign_Center, size {sizeRule=Fill, value=1}}}
  }
  RichText_SongXP {
   widgetClassPath=&CommonRichTextBlock
   bIsVariable=true
   parent=GridPanel_BattlePassContainer
   widget=RichText_SongXP
   properties {
    bOverrideDefaultStyle=true
    justification=Right
    mobileTextBlockScale=1.1
    text="EARNED <text color=\"FFFFFFFF\">10000 XP</>"
    defaultTextStyleOverride {
     overflowPolicy=Clip
     transformPolicy=None
     colorAndOpacity {colorUseRule=UseColor_Specified, specifiedColor=<rgba 0 0.8714 1 1>}
     font {
      bForceMonospaced=false
      bMaterialIsStencil=false
      fontMaterial=null
      fontObject=&HeadingNow
      letterSpacing=0
      monospacedWidth=1
      size=15.75
      skewAmount=0
      typefaceFontName=75 Medium
      outlineSettings {
       bApplyOutlineToDropShadows=false
       bMiteredCorners=false
       bSeparateFillAlpha=false
       outlineMaterial=null
       outlineSize=0
       outlineColor=<rgba 0 0 0 1>
      }
     }
     highlightColor {colorUseRule=UseColor_Specified, specifiedColor=<rgba 0 0 0 1>}
     highlightShape {
      bIsDynamicallyLoaded=false
      drawAs=Image
      imageType=NoImage
      mirroring=NoMirror
      resourceName=None
      resourceObject=null
      tiling=NoTile
      imageSize=<32 32>
      margin=<ltrb 0 0 0 0>
      outlineSettings {
       bUseBrushTransparency=false
       roundingType=HalfHeightRadius
       width=0
       color {colorUseRule=UseColor_Specified, specifiedColor=<rgba 0 0 0 0>}
       cornerRadii=<0 0 0 0>
      }
      tintColor {colorUseRule=UseColor_Specified, specifiedColor=<rgba 1 1 1 1>}
      uVRegion {bIsValid=false, max=<0 0>, min=<0 0>}
     }
     selectedBackgroundColor {colorUseRule=UseColor_Foreground, specifiedColor=<rgba 1 0 1 1>}
     shadowColorAndOpacity=<rgba 0 0 0 1>
     shadowOffset=<0 0>
     strikeBrush {
      bIsDynamicallyLoaded=false
      drawAs=Image
      imageType=NoImage
      mirroring=NoMirror
      resourceName=None
      resourceObject=null
      tiling=NoTile
      imageSize=<32 32>
      margin=<ltrb 0 0 0 0>
      outlineSettings {
       bUseBrushTransparency=false
       roundingType=HalfHeightRadius
       width=0
       color {colorUseRule=UseColor_Specified, specifiedColor=<rgba 0 0 0 0>}
       cornerRadii=<0 0 0 0>
      }
      tintColor {colorUseRule=UseColor_Specified, specifiedColor=<rgba 1 1 1 1>}
      uVRegion {bIsValid=false, max=<0 0>, min=<0 0>}
     }
     underlineBrush {
      bIsDynamicallyLoaded=false
      drawAs=Image
      imageType=NoImage
      mirroring=NoMirror
      resourceName=None
      resourceObject=null
      tiling=NoTile
      imageSize=<32 32>
      margin=<ltrb 0 0 0 0>
      outlineSettings {
       bUseBrushTransparency=false
       roundingType=HalfHeightRadius
       width=0
       color {colorUseRule=UseColor_Specified, specifiedColor=<rgba 0 0 0 0>}
       cornerRadii=<0 0 0 0>
      }
      tintColor {colorUseRule=UseColor_Specified, specifiedColor=<rgba 1 1 1 1>}
      uVRegion {bIsValid=false, max=<0 0>, min=<0 0>}
     }
    }
   }
   slot {class=GridSlot, properties {horizontalAlignment=HAlign_Left, verticalAlignment=VAlign_Bottom, padding=<ltrb 4 0 0 0>}}
  }
  VerticalBox_5stars {
   widgetClassPath=&VerticalBox
   bIsVariable=true
   parent=HorizontalBox_2
   widget=VerticalBox_5stars
   properties {visibility=Collapsed}
   slot {class=HorizontalBoxSlot, properties {horizontalAlignment=HAlign_Left, verticalAlignment=VAlign_Top, padding=<ltrb 8 4 8 0>}}
  }
  Image_392 {
   widgetClassPath=&Image
   bIsVariable=true
   parent=VerticalBox_5stars
   widget=Image_392
   properties {brush {
     bIsDynamicallyLoaded=false
     drawAs=Image
     imageType=NoImage
     mirroring=NoMirror
     resourceName=None
     resourceObject=/Game/Athena/UI/Frontend/Art/M_UI_BP_BattleStar
     tiling=NoTile
     imageSize=<20 20>
     margin=<ltrb 0 0 0 0>
     outlineSettings {
      bUseBrushTransparency=false
      roundingType=HalfHeightRadius
      width=0
      color {colorUseRule=UseColor_Specified, specifiedColor=<rgba 0 0 0 0>}
      cornerRadii=<0 0 0 0>
     }
     tintColor {colorUseRule=UseColor_Specified, specifiedColor=<rgba 1 1 1 1>}
     uVRegion {bIsValid=false, max=<0 0>, min=<0 0>}
    }}
   slot {class=VerticalBoxSlot, properties {horizontalAlignment=HAlign_Center}}
  }
  CommonTextBlock_209 {
   widgetClassPath=&CommonTextBlock
   parent=VerticalBox_5stars
   widget=CommonTextBlock_209
   properties {
    lineHeightPercentage=0.75
    mobileFontSizeMultiplier=1.4
    text=+5
    font {
     bForceMonospaced=false
     bMaterialIsStencil=false
     fontMaterial=null
     fontObject=&BurbankSmall
     letterSpacing=0
     monospacedWidth=1
     size=12
     skewAmount=0
     typefaceFontName=Black
     outlineSettings {
      bApplyOutlineToDropShadows=false
      bMiteredCorners=false
      bSeparateFillAlpha=false
      outlineMaterial=null
      outlineSize=0
      outlineColor=<rgba 0 0 0 1>
     }
    }
    shadowOffset=<0 0>
   }
   slot {class=VerticalBoxSlot, properties {horizontalAlignment=HAlign_Center}}
  }
  ScaleBox_101 {
   widgetClassPath=&ScaleBox
   parent=HorizontalBox_2
   widget=ScaleBox_101
   properties {visibility=Collapsed}
   slot {class=HorizontalBoxSlot}
  }
  LevelRewards {
   widgetClassPath=/FortniteCoreUI/HUD/MiniMap/AthenaChallengeRewards
   bInherited=true
   bIsVariable=true
   parent=ScaleBox_101
   widget=LevelRewards
   slot {class=ScaleBoxSlot, properties {horizontalAlignment=HAlign_Fill, verticalAlignment=VAlign_Fill}}
  }
  SocialAvatarIcon {widgetClassPath=/Script/FortniteUI.FortSocialAvatarIcon, bInherited=true, widget=null}
 ]
}
