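# Quest Progress panel for the Pilgrim Quickplay song results screen, showing a quest list configured to hide completed and unchanged quests.
# A header row with the label and a right-stick scroll icon sits above the embedded WBP_PilgrimQuestList.
# On initialization it plays Anim_On at double speed to fade and slide content in horizontally.
# OnKeyDownScroll applies a deadzone to axis input, then moves the list's ScrollBox by inverted input times ScrollSpeed, clamped to bounds.
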WidgetBlueprint /PilgrimQuickplay/Widgets/Results/SongResults/WBP_PilgrimQuickplay_ProgressQuests
animations [Anim_On {
  binding_count=2
  length_seconds=1.3
  object=&self:Anim_On
  sequence {bindings [
    WBP_PilgrimQuickplay_ProgressQuests_C_32 {
     id=#1
     possessed_class=WBP_PilgrimQuickplay_ProgressQuests_C
     tracks [
      {
       type=MovieSceneFloatTrack
       display_name=Render Opacity
       sections [{type=MovieSceneFloatSection, end_frame=26, start_frame=0, channels [None {type=MovieSceneScriptingFloatChannel, keys [[16,0],[26,1]]}]}]
      }
      {
       type=MovieScene2DTransformTrack
       sections [{
         type=MovieScene2DTransformSection
         end_frame=26
         start_frame=0
         channels [Translation.X {type=MovieSceneScriptingFloatChannel, keys [[16,-308.8451],[26,0]]}]
        }]
      }
     ]
    }
    HorizontalBox_1 {
     id=#2
     possessed_class=HorizontalBox
     tracks [
      {
       type=MovieSceneFloatTrack
       display_name=Render Opacity
       sections [{type=MovieSceneFloatSection, end_frame=21, start_frame=0, channels [None {type=MovieSceneScriptingFloatChannel, keys [[6,0],[16,1]]}]}]
      }
      {
       type=MovieScene2DTransformTrack
       sections [{
         type=MovieScene2DTransformSection
         end_frame=21
         start_frame=0
         channels [Translation.X {type=MovieSceneScriptingFloatChannel, keys [[6,-308.8451],[16,0]]}]
        }]
      }
     ]
    }
   ]}
 }]
graphs [
 EventGraph {nodes [
   Event OnInitialized()
   > PlayAnimation(InAnimation=Get Anim_On() -> (Anim_On:Widget Animation Object Reference), PlaybackSpeed=2)
   Event Construct()
   Event Tick()
  ]}
 OnKeyDownScroll {nodes [
   n1 = OnKeyDownScroll!() -> (then, AxisValue:double=~) {is=title K2Node_FunctionEntry}
   > Branch(Condition="float > float"(A="Absolute (Float)"(A=@n1), B=0.1))
   n2 = Get ScrollBox!(self:WBP Pilgrim Quest List Object Reference=Get WBP_PilgrimQuestList() -> (WBP_PilgrimQuestList:WBP Pilgrim Quest List Object Reference)) -> (ScrollBox:Scroll Box Object Reference)
   n3 = GetScrollOffset!(self:Scroll Box Object Reference=@n2) -> (ReturnValue:float)
   n4 = float * float(A=Get ScrollSpeed() -> (ScrollSpeed:double), B=float * float(A=@n1, B=-1)) {is=title K2Node_PromotableOperator}
   > SetScrollOffset!(
    execute
    self:Scroll Box Object Reference=@n2
    NewScrollOffset:float="Clamp (Float)"(Value=float + float(A=@n3, B=@n4), Min=0, Max=GetScrollOffsetOfEnd!(self:Scroll Box Object Reference=@n2) -> (ReturnValue:float))
   ) -> (then)
  ]}
]
variables [ScrollSpeed {type=number, default=25}]
widgets {
 info {inheritedWidgetCount=0, parentClass=&UserWidget, rootWidgetClass=&VerticalBox, widgetCount=5}
 widgets [
  VerticalBox_0 {widgetClassPath=&VerticalBox, widget=VerticalBox_0}
  HorizontalBox_1 {widgetClassPath=&HorizontalBox, parent=VerticalBox_0, widget=HorizontalBox_1, slot {class=VerticalBoxSlot}}
  Text_QuestLabel {
   widgetClassPath=&CommonTextBlock
   bIsVariable=true
   parent=HorizontalBox_1
   widget=Text_QuestLabel
   properties {
    flowDirectionPreference=Culture
    text=Quest Progress
    font {
     bForceMonospaced=false
     bMaterialIsStencil=false
     fontMaterial=null
     fontObject=&HeadingNow
     letterSpacing=0
     monospacedWidth=1
     size=15
     skewAmount=0
     typefaceFontName=75 Medium
     outlineSettings {
      bApplyOutlineToDropShadows=true
      bMiteredCorners=false
      bSeparateFillAlpha=false
      outlineMaterial=null
      outlineSize=0
      outlineColor=<rgba 0 0 0 1>
     }
    }
    margin=<ltrb 0 0 2 0>
   }
   slot {
    class=HorizontalBoxSlot
    properties {horizontalAlignment=HAlign_Left, verticalAlignment=VAlign_Center, padding=<ltrb 0 9 0 9>, size {sizeRule=Fill, value=1}}
   }
  }
  RightStickIcon {
   widgetClassPath=&CommonActionWidget
   bIsVariable=true
   parent=HorizontalBox_1
   widget=RightStickIcon
   properties {
    designTimeKey=Gamepad_RightY
    inputActions [{dataTable=/PilgrimQuickplay/Widgets/InputActionDataTables/FM_PilgrimQuickplay_InputActionDataTable, rowName=ResultsQuestScroll}]
    renderTransform {angle=0, scale=<1 1>, shear=<0 0>, translation=<11 0>}
   }
   slot {class=HorizontalBoxSlot, properties {horizontalAlignment=HAlign_Right, verticalAlignment=VAlign_Center}}
  }
  WBP_PilgrimQuestList {
   widgetClassPath=/PilgrimQuickplay/Widgets/Quests/WBP_PilgrimQuestList
   bIsVariable=true
   parent=VerticalBox_0
   widget=WBP_PilgrimQuestList
   properties {bIsFocusable=true, bShowCompletedQuests=false, bShowUnchangedQuests=false}
   slot {class=VerticalBoxSlot, properties {size {sizeRule=Fill, value=1}}}
  }
 ]
}
