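# Progress tab of the Pilgrim Quickplay song results screen, stacking the player nameplate, battle pass progress and quest list in a safe-zone layout.
# On construct, it assigns squad player states, including the owner's, by index to PlayerLoadoutWidgets and shows them, though that array defaults empty and nothing here fills it.
# Functions forward axis scrolling to the quest list and play a slide-and-fade intro, with quests entering slightly later.
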WidgetBlueprint /PilgrimQuickplay/Widgets/Results/SongResults/WBP_PilgrimQuickplay_ProgressTab
animations [Anim_ProgressScreen_Intro {
  binding_count=4
  length_seconds=1.1
  object=&self:Anim_ProgressScreen_Intro
  sequence {
   bindings [
    WBP_PilgrimQuickplay_BattlePass_Progress {
     id=#1
     possessed_class=WBP_PilgrimQuickplay_BattlePass_Progress_C
     tracks [
      {
       type=MovieSceneFloatTrack
       display_name=Render Opacity
       sections [{type=MovieSceneFloatSection, channels [None {type=MovieSceneScriptingFloatChannel, keys [[0,0],[12,1]]}]}]
      }
      {
       type=MovieScene2DTransformTrack
       sections [{
         type=MovieScene2DTransformSection
         channels [
          Translation.X {type=MovieSceneScriptingFloatChannel, keys [[0,-512],[45,0]]}
          Translation.Y {type=MovieSceneScriptingFloatChannel, const [0,1]}
          Angle {type=MovieSceneScriptingFloatChannel, const [0,1]}
          Scale.X {type=MovieSceneScriptingFloatChannel, const [1,1]}
          Scale.Y {type=MovieSceneScriptingFloatChannel, const [1,1]}
          Shear.X {type=MovieSceneScriptingFloatChannel, const [0,1]}
          Shear.Y {type=MovieSceneScriptingFloatChannel, const [0,1]}
         ]
        }]
      }
     ]
    }
    WBP_Pilgrim_PlayerIdentityLozenge {
     id=#2
     possessed_class=WBP_Pilgrim_PlayerIdentity_Nameplate_C
     tracks [
      {
       type=MovieSceneFloatTrack
       display_name=Render Opacity
       sections [{type=MovieSceneFloatSection, channels [None {type=MovieSceneScriptingFloatChannel, keys [[0,0],[12,1]]}]}]
      }
      {
       type=MovieScene2DTransformTrack
       sections [{
         type=MovieScene2DTransformSection
         channels [
          Translation.X {type=MovieSceneScriptingFloatChannel, keys [[0,-512],[45,0]]}
          Translation.Y {type=MovieSceneScriptingFloatChannel, const [0,1]}
          Angle {type=MovieSceneScriptingFloatChannel, const [0,1]}
          Scale.X {type=MovieSceneScriptingFloatChannel, const [1,1]}
          Scale.Y {type=MovieSceneScriptingFloatChannel, const [1,1]}
          Shear.X {type=MovieSceneScriptingFloatChannel, const [0,1]}
          Shear.Y {type=MovieSceneScriptingFloatChannel, const [0,1]}
         ]
        }]
      }
     ]
    }
    Overlay_ResultScreen {
     id=#3
     possessed_class=Overlay
     tracks [
      {
       type=MovieSceneFloatTrack
       display_name=Render Opacity
       sections [{type=MovieSceneFloatSection, channels [None {type=MovieSceneScriptingFloatChannel, keys [[0,0],[3,1]]}]}]
      }
      {
       type=MovieScene2DTransformTrack
       sections [{
         type=MovieScene2DTransformSection
         channels [
          Translation.X {type=MovieSceneScriptingFloatChannel, keys [[0,-512],[66,0]]}
          Translation.Y {type=MovieSceneScriptingFloatChannel, const [0,1]}
          Angle {type=MovieSceneScriptingFloatChannel, const [0,1]}
          Scale.X {type=MovieSceneScriptingFloatChannel, const [1,1]}
          Scale.Y {type=MovieSceneScriptingFloatChannel, const [1,1]}
          Shear.X {type=MovieSceneScriptingFloatChannel, const [0,1]}
          Shear.Y {type=MovieSceneScriptingFloatChannel, const [0,1]}
         ]
        }]
      }
     ]
    }
    WBP_PilgrimQuickplay_ProgressQuests {
     id=#4
     possessed_class=WBP_PilgrimQuickplay_ProgressQuests_C
     tracks [
      {
       type=MovieSceneFloatTrack
       display_name=Render Opacity
       sections [{type=MovieSceneFloatSection, channels [None {type=MovieSceneScriptingFloatChannel, keys [[9,0],[21,1]]}]}]
      }
      {
       type=MovieScene2DTransformTrack
       sections [{
         type=MovieScene2DTransformSection
         channels [
          Translation.X {type=MovieSceneScriptingFloatChannel, keys [[9,-512],[54,0]]}
          Translation.Y {type=MovieSceneScriptingFloatChannel, const [0,1]}
          Angle {type=MovieSceneScriptingFloatChannel, const [0,1]}
          Scale.X {type=MovieSceneScriptingFloatChannel, const [1,1]}
          Scale.Y {type=MovieSceneScriptingFloatChannel, const [1,1]}
          Shear.X {type=MovieSceneScriptingFloatChannel, const [0,1]}
          Shear.Y {type=MovieSceneScriptingFloatChannel, const [0,1]}
         ]
        }]
      }
     ]
    }
   ]
   display_rate {numerator=60}
  }
 }]
graphs [
 EventGraph {nodes [
   Event Construct()
   n1 = GetOwningPlayer()
   > n2 = Cast To FortPlayerStateAthena!(execute, Object=Get PlayerState!(self:Controller Object Reference=@n1) -> (PlayerState:Player State Object Reference)) -> (then, CastFailed, AsFort Player State Athena:Fort Player State Athena Object Reference) {is=title K2Node_DynamicCast}
   > n3 = GetSquadMemberPlayerStates!(
    execute
    self:Fort Game State Athena Object Reference=GetGameStateAthenaFromContext!(WorldContextObject:Object Reference=@n1) -> (ReturnValue:Fort Game State Athena Object Reference)
    InstigatingPlayerState:Fort Player State Athena Object Reference=@n2
    bIncludeSelf:bool=true
   ) -> (then, ReturnValue:Array of Fort Player State Athena Object References)
   > n4 = For Each Loop(Array:Array of Player State Object References=@n3) -> (Array Element:Player State Object Reference)
   n5 = Get PlayerLoadoutWidgets() -> (PlayerLoadoutWidgets:Array of WBP Pilgrim Quickplay Progress Tab Player Loadout Object References)
   > Branch(Condition=Is Valid Index!(TargetArray:Array of WBP Pilgrim Quickplay Progress Tab Player Loadout Object References=@n5, IndexToTest:int=@n4.Array Index) -> (ReturnValue:bool))
   n6 = "Get (a copy)"(Array:Array of WBP Pilgrim Quickplay Progress Tab Player Loadout Object References=@n5, Dimension 1=@n4.Array Index) -> (Output:WBP Pilgrim Quickplay Progress Tab Player Loadout Object Reference)
   > SetVisibility(self=@n6, InVisibility=SelfHitTestInvisible)
   > SetPlayerState!(execute, self:WBP Pilgrim Quickplay Progress Tab Player Loadout Object Reference=@n6, PlayerState:Player State Object Reference=@n4) -> (then)
  ]}
 OnKeyDownQuestList {nodes [
   n1 = OnKeyDownQuestList!() -> (then, AxisValue:double=~) {is=title K2Node_FunctionEntry}
   > OnKeyDownScroll!(
    execute
    self:WBP Pilgrim Quickplay Progress Quests Object Reference=Get WBP_PilgrimQuickplay_ProgressQuests() -> (WBP_PilgrimQuickplay_ProgressQuests:WBP Pilgrim Quickplay Progress Quests Object Reference)
    AxisValue:double=@n1
   ) -> (then)
  ]}
 PlayProgressScreenIntroAnim {nodes [
   PlayProgressScreenIntroAnim() {is=title K2Node_FunctionEntry}
   > PlayAnimation(InAnimation=Get Anim_ProgressScreen_Intro() -> (Anim_ProgressScreen_Intro:Widget Animation Object Reference))
  ]}
]
variables [
 PlayerLoadoutWidgets {type="array<WBP_PilgrimQuickplay_ProgressTab_PlayerLoadout_C>", default []}
 ShowPlayerBoxes {type=boolean}
]
widgets {
 info {inheritedWidgetCount=0, parentClass=&UserWidget, rootWidgetClass=&SafeZone, widgetCount=7}
 widgets [
  SafeZone_2 {widgetClassPath=&SafeZone, widget=SafeZone_2}
  Overlay_ResultScreen {widgetClassPath=&Overlay, parent=SafeZone_2, widget=Overlay_ResultScreen, slot {class=SafeZoneSlot}}
  SafeZoneScaleBox {
   widgetClassPath=&ScaleBox
   parent=Overlay_ResultScreen
   widget=SafeZoneScaleBox
   properties {stretch=ScaleBySafeZone}
   slot {class=OverlaySlot, properties {horizontalAlignment=HAlign_Fill}}
  }
  VerticalBox_0 {
   widgetClassPath=&VerticalBox
   parent=SafeZoneScaleBox
   widget=VerticalBox_0
   slot {class=ScaleBoxSlot, properties {horizontalAlignment=HAlign_Fill, verticalAlignment=VAlign_Fill}}
  }
  WBP_Pilgrim_PlayerIdentityLozenge {
   widgetClassPath=/PilgrimQuickplay/Widgets/Players/WBP_Pilgrim_PlayerIdentity_Nameplate
   bIsVariable=true
   parent=VerticalBox_0
   widget=WBP_Pilgrim_PlayerIdentityLozenge
   slot {class=VerticalBoxSlot, properties {horizontalAlignment=HAlign_Center, padding=<ltrb 0 0 0 5>}}
  }
  WBP_PilgrimQuickplay_BattlePass_Progress {
   widgetClassPath=&/WBP_PilgrimQuickplay_BattlePass_Progress
   bIsVariable=true
   parent=VerticalBox_0
   widget=WBP_PilgrimQuickplay_BattlePass_Progress
   properties {bAutoActivate=true, pixelSnapping=Disabled}
   slot {class=VerticalBoxSlot, properties {verticalAlignment=VAlign_Top, padding=<ltrb 280 0 280 0>}}
  }
  WBP_PilgrimQuickplay_ProgressQuests {
   widgetClassPath=&/WBP_PilgrimQuickplay_ProgressQuests
   bIsVariable=true
   parent=VerticalBox_0
   widget=WBP_PilgrimQuickplay_ProgressQuests
   properties {pixelSnapping=Disabled}
   slot {class=VerticalBoxSlot, properties {horizontalAlignment=HAlign_Left, verticalAlignment=VAlign_Center, size {sizeRule=Fill, value=1}}}
  }
 ]
}
