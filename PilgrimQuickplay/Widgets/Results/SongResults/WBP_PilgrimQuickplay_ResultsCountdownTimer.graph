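# Results-screen countdown widget for Pilgrim Quickplay, showing seconds until the next song or, at setlist end, until the show ends.
# It reads ServerEndTime from the active SongResults state and binds to Auto Advance Time Replicated.
# A looping 0.1s timer, first delayed by the fractional second, updates counters from server world time and plays final-seconds pings.
# LayoutMode switches vertical or horizontal text and sizes the spinner image.
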WidgetBlueprint /PilgrimQuickplay/Widgets/Results/SongResults/WBP_PilgrimQuickplay_ResultsCountdownTimer
graphs [
 EventGraph {nodes [
   Stop Timer() {is=title K2Node_CustomEvent}
   > Clear and Invalidate Timer by Handle(Handle=Get CountdownTimer() -> (CountdownTimer:Timer Handle Structure))
   Set Song Results TimerFrom Gameplay State() {is=title K2Node_CustomEvent}
   n1 = GetSparksMusicPlayspaceOfClass!(
    PlayspaceType:Sparks Music Playspace Class Reference=/Script/PilgrimQuickplayRuntime.PilgrimQuickplayPlayspace
    Actor:Actor Object Reference=GetOwningPlayer()
    bMatchChildTypes:bool=false
   ) -> (ReturnValue:Pilgrim Quickplay Playspace Object Reference)
   > Is Valid(InputObject=@n1)
   n2 = GetPilgrimQuickplayStateMachine!(self:Pilgrim Quickplay Playspace Object Reference=@n1) -> (ReturnValue:Pilgrim Quickplay State Machine Object Reference)
   > Is Valid(InputObject=@n2)
   n3 = GetActivePilgrimStateObject!(self:Pilgrim Quickplay State Machine Object Reference=@n2) -> (ReturnValue:Pilgrim Quickplay State Base Object Reference)
   > Is Valid(InputObject=@n3)
   > n4 = Cast To BP_PilgrimQuickplayState_SongResults!(execute, Object=@n3) -> (then, CastFailed, AsBP Pilgrim Quickplay State Song Results:BP Pilgrim Quickplay State Song Results Object Reference) {is=title K2Node_DynamicCast}
   > SetStateAdvanceTime!(
    execute
    self
    TimeWhenStateWillAdvance:double=Get ServerEndTime!(self:Pilgrim Quickplay State Song Results Object Reference=@n4) -> (ServerEndTime:double)
   ) -> (then)
   > Bind Event to Auto Advance Time Replicated(self:BP Pilgrim Quickplay State Song Results Object Reference=@n4, Delegate=Create Event()) {is=title K2Node_AddDelegate}
   > "SetText (Text)"(
    self=[
     Get AdvanceScreenCountdownText_Ver() -> (AdvanceScreenCountdownText_Ver:Common Text Object Reference)
     Get AdvanceScreenCountdownText_Hor() -> (AdvanceScreenCountdownText_Hor:Common Text Object Reference)
    ]
    InText=Select(
     Option 0:Text=Get NextSongTimerText() -> (NextSongTimerText:Text)
     Option 1:Text=Get SetlistEndTimerText() -> (SetlistEndTimerText:Text)
     Index=Get AtEndOfSetlist() -> (AtEndOfSetlist:bool)
    ) -> (ReturnValue:Text)
   )
   n5 = SetStateAdvanceTime!() -> (OutputDelegate, then, TimeWhenStateWillAdvance:double=~) {is=title K2Node_CustomEvent}
   > Set StateAdvanceTime(StateAdvanceTime:double=@n5.TimeWhenStateWillAdvance) -> (Output_Get:double)
   > n6 = Branch(Condition="float > float"(A=Get Time Remaining in Countdown!() -> (TimeRemaining:double) {is=title K2Node_CallFunction}))
   > SetVisibility(self=Get ScreenAdvancementTimer() -> (ScreenAdvancementTimer:Horizontal Box Object Reference), InVisibility=HitTestInvisible)
   > StartTimer()
   SetVisibility(execute=@n6.else, self=Get ScreenAdvancementTimer(), InVisibility=Hidden)
   > Stop Timer() {is=title K2Node_CallFunction}
   StartTimer() {is=title K2Node_CustomEvent}
   > n7 = Stop Timer()
   n8 = UpdateResultsCountdownTimer() {is=title K2Node_CustomEvent}
   n9 = Set Timer by Event(
    execute=@n7.then
    Delegate=@n8
    Time=0.1
    bLooping=true
    InitialStartDelay="% (Float)"!(A:double=Get Time Remaining in Countdown(), B:double=1) -> (ReturnValue:double)
   )
   > Set CountdownTimer(CountdownTimer:Timer Handle Structure=@n9) -> (Output_Get:Timer Handle Structure)
   > UpdateResultsCountdownTimer() {is=title K2Node_CallFunction}
   n10 = Event PreConstruct()
   > Branch(Condition=@n10.IsDesignTime)
   > SetActiveWidgetIndex(
    self:Common Visibility Switcher Object Reference=Get HorizontalOrVertical_Switcher() -> (HorizontalOrVertical_Switcher:Common Visibility Switcher Object Reference)
    Index=Get LayoutMode() -> (LayoutMode:int=0)
   )
   n11 = Get Image_155() -> (Image_155:Image Object Reference)
   n12 = Select(
    Option 0:double="To Float (Integer)"(InInt=Get DiskSize_Big() -> (DiskSize_Big:int=0))
    Option 1:double="To Float (Integer)"(InInt=Get DiskSize_Small() -> (DiskSize_Small:int=0))
    Index:int=Get LayoutMode() -> (LayoutMode=0)
   ) -> (ReturnValue:double)
   > "Set Vector 2D Structure (by ref)"!(
    execute:Exec
    Target:"Vector 2D Structure (by ref)"=Break Slate Brush!(SlateBrush:"Slate Brush Structure (by ref)"=Get Brush!(self:Image Object Reference=@n11) -> (Brush:Slate Brush Structure)) -> (
     TintColor:Slate Color Structure
     DrawAs:ESlateBrushDrawType Enum
     Tiling:ESlateBrushTileType Enum
     Mirroring:ESlateBrushMirrorType Enum
     ImageSize:Vector 2D Structure
     Margin:Margin Structure
     ResourceObject:Object Reference
     OutlineSettings:Slate Brush Outline Settings Structure
    ) {is=title K2Node_BreakStruct}.ImageSize
    Value_X:double=@n12
    Value_Y:double=@n12
   ) -> (then:Exec) {is=title K2Node_VariableSetRef}
   Event Construct()
   > n13 = Sequence()
   n14 = GetOwningPlayer()
   n15 = GetNextSong!(self:Pilgrim Core Band State Component Object Reference=GetBandState!(ContextActor:Actor Object Reference=GetSparksMusicPlayspaceOfClass(PlayspaceType=&PilgrimQuickplayPlayspace, Actor=@n14, bMatchChildTypes=false)) -> (ReturnValue:Pilgrim Quickplay Band State Component Object Reference)) -> (ReturnValue:Pilgrim Song Metadata Object Reference)
   > Set AtEndOfSetlist(AtEndOfSetlist:bool=NOT Boolean(A=IsValid(Object=@n15))) -> (Output_Get:bool)
   > SetActiveWidgetIndex(self=Get HorizontalOrVertical_Switcher(), Index=Get LayoutMode() -> (LayoutMode=0))
   n16 = Get Image_155()
   n17 = Select(
    Option 0="To Float (Integer)"(InInt=Get DiskSize_Big() -> (DiskSize_Big=0))
    Option 1="To Float (Integer)"(InInt=Get DiskSize_Small() -> (DiskSize_Small=0))
    Index=Get LayoutMode() -> (LayoutMode=0)
   )
   > "Set Vector 2D Structure (by ref)"(Target=Break Slate Brush(SlateBrush=Get Brush(self=@n16)).ImageSize, Value_X=@n17, Value_Y=@n17)
   Set Song Results TimerFrom Gameplay State(execute=@n13.then_1) {is=title K2Node_CallFunction}
   Stop Timer(execute=@n21.else)
   Decrement Int(" "=@n20.then, Value=Get CountdownAudioPingsRemaining() -> (CountdownAudioPingsRemaining:int=0))
   n18 = Get Time Remaining in Countdown()
   n19 = Ceil(A=@n18) {is=title K2Node_CallFunction}
   "SetText (Text)"(
    execute=@n21.then
    self=[
     Get SetlistResultsCounter_Ver() -> (SetlistResultsCounter_Ver:Common Text Object Reference)
     Get SetlistResultsCounter_Hor() -> (SetlistResultsCounter_Hor:Common Text Object Reference)
    ]
    InText="To Text (Integer)"(Value=@n19)
   )
   > Branch(Condition="integer >= integer"(A=Get CountdownAudioPingsRemaining() -> (CountdownAudioPingsRemaining=0), B=@n19))
   > n20 = PlaySound2D(Sound=/SparksMaps/Audio/FM_VoteCountdown)
   n21 = Branch(execute=@n8.then, Condition="float > float"(A=@n18))
  ]}
 Get Time Remaining in Countdown {nodes [
   Get Time Remaining in Countdown() {is=title K2Node_FunctionEntry}
   > Return Node!(
    execute
    TimeRemaining:double=float - float(A=Get StateAdvanceTime() -> (StateAdvanceTime:double), B=GetServerWorldTimeSeconds!() -> (ReturnValue:float))
   )
  ]}
]
variables [
 AtEndOfSetlist {type=boolean}
 CountdownAudioPingsRemaining {type=integer, default=3}
 CountdownTimer {type=TimerHandle}
 DiskSize_Big {type=integer, default=64}
 DiskSize_Small {type=integer, default=50}
 Event Router {type=GameplayEventRouterComponent, default=null}
 LayoutMode {type=integer}
 NextSongTimerText {type=string, default=Next Song in:}
 SetlistEndTimerText {type=string, default="Show Ending in: "}
 StateAdvanceTime {type=number}
]
widgets {
 info {inheritedWidgetCount=0, parentClass=&UserWidget, rootWidgetClass=&HorizontalBox, widgetCount=9}
 widgets [
  ScreenAdvancementTimer {widgetClassPath=&HorizontalBox, bIsVariable=true, widget=ScreenAdvancementTimer}
  Image_155 {
   widgetClassPath=&Image
   bIsVariable=true
   parent=ScreenAdvancementTimer
   widget=Image_155
   properties {
    renderOpacity=0.9
    visibility=SelfHitTestInvisible
    brush {
     bIsDynamicallyLoaded=false
     drawAs=Image
     imageType=NoImage
     mirroring=NoMirror
     resourceName=None
     resourceObject=/FortUILibrary/Systems/Throbber/M_V2Radio_Loading
     tiling=NoTile
     imageSize=<64 64>
     margin=<ltrb 0 0 0 0>
     outlineSettings {
      bUseBrushTransparency=false
      roundingType=HalfHeightRadius
      width=0
      color {colorUseRule=UseColor_Specified, specifiedColor=<rgba 0 0 0 0>}
      cornerRadii=<0 0 0 0>
     }
     tintColor {colorUseRule=UseColor_Specified, specifiedColor=<rgba 1 1 1 1>}
     uVRegion {bIsValid=false, max=<0 0>, min=<0 0>}
    }
   }
   slot {class=HorizontalBoxSlot, properties {horizontalAlignment=HAlign_Center, verticalAlignment=VAlign_Center}}
  }
  HorizontalOrVertical_Switcher {
   widgetClassPath=&CommonVisibilitySwitcher
   bIsVariable=true
   parent=ScreenAdvancementTimer
   widget=HorizontalOrVertical_Switcher
   slot {class=HorizontalBoxSlot, properties {verticalAlignment=VAlign_Center, padding=<ltrb 8 0 0 0>}}
  }
  VerticalBox_0 {
   widgetClassPath=&VerticalBox
   parent=HorizontalOrVertical_Switcher
   widget=VerticalBox_0
   slot {class=CommonVisibilitySwitcherSlot}
  }
  AdvanceScreenCountdownText_Ver {
   widgetClassPath=&CommonTextBlock
   bIsVariable=true
   parent=VerticalBox_0
   widget=AdvanceScreenCountdownText_Ver
   properties {
    clipping=OnDemand
    justification=Right
    scrollStyle=/PilgrimQuickplay/Widgets/ScrollStyles/TextScrollStyle_PilgrimQuickplay_Setlist.TextScrollStyle_PilgrimQuickplay_Setlist_C
    text=Next Song in:
    textTransformPolicy=ToUpper
    colorAndOpacity {colorUseRule=UseColor_Specified, specifiedColor=<rgba 1 1 1 0.8>}
    font {
     bForceMonospaced=false
     bMaterialIsStencil=false
     fontMaterial=null
     fontObject=&HeadingNow
     letterSpacing=0
     monospacedWidth=1
     size=15
     skewAmount=0
     typefaceFontName=75 Medium Italic
     outlineSettings {
      bApplyOutlineToDropShadows=false
      bMiteredCorners=false
      bSeparateFillAlpha=false
      outlineMaterial=&M-FontShadow
      outlineSize=2
      outlineColor=<rgba 0 0 0 1>
     }
    }
    shadowOffset=<0 0>
    strikeBrush {
     bIsDynamicallyLoaded=false
     drawAs=NoDrawType
     imageType=NoImage
     mirroring=NoMirror
     resourceName=None
     resourceObject=null
     tiling=NoTile
     imageSize=<32 32>
     margin=<ltrb 0 0 0 0>
     outlineSettings {
      bUseBrushTransparency=false
      roundingType=HalfHeightRadius
      width=0
      color {colorUseRule=UseColor_Specified, specifiedColor=<rgba 0 0 0 0>}
      cornerRadii=<0 0 0 0>
     }
     tintColor {colorUseRule=UseColor_Specified, specifiedColor=<rgba 1 1 1 1>}
     uVRegion {bIsValid=false, max=<0 0>, min=<0 0>}
    }
   }
   slot {class=VerticalBoxSlot}
  }
  SetlistResultsCounter_Ver {
   widgetClassPath=&CommonTextBlock
   bIsVariable=true
   parent=VerticalBox_0
   widget=SetlistResultsCounter_Ver
   properties {
    bIsScrollingEnabled=false
    justification=Center
    lineHeightPercentage=0.85
    text=30
    font {
     bForceMonospaced=false
     bMaterialIsStencil=false
     fontMaterial=null
     fontObject=&HeadingNow
     letterSpacing=0
     monospacedWidth=1
     size=30
     skewAmount=0
     typefaceFontName=86 Bold Italic
     outlineSettings {
      bApplyOutlineToDropShadows=false
      bMiteredCorners=false
      bSeparateFillAlpha=false
      outlineMaterial=&M-FontShadow
      outlineSize=2
      outlineColor=<rgba 0 0 0 1>
     }
    }
    margin=<ltrb 0 -8 0 0>
    shadowOffset=<0 0>
    strikeBrush {
     bIsDynamicallyLoaded=false
     drawAs=NoDrawType
     imageType=NoImage
     mirroring=NoMirror
     resourceName=None
     resourceObject=null
     tiling=NoTile
     imageSize=<32 32>
     margin=<ltrb 0 0 0 0>
     outlineSettings {
      bUseBrushTransparency=false
      roundingType=HalfHeightRadius
      width=0
      color {colorUseRule=UseColor_Specified, specifiedColor=<rgba 0 0 0 0>}
      cornerRadii=<0 0 0 0>
     }
     tintColor {colorUseRule=UseColor_Specified, specifiedColor=<rgba 1 1 1 1>}
     uVRegion {bIsValid=false, max=<0 0>, min=<0 0>}
    }
   }
   slot {class=VerticalBoxSlot, properties {horizontalAlignment=HAlign_Left}}
  }
  HorizontalBox_52 {
   widgetClassPath=&HorizontalBox
   parent=HorizontalOrVertical_Switcher
   widget=HorizontalBox_52
   slot {class=CommonVisibilitySwitcherSlot}
  }
  AdvanceScreenCountdownText_Hor {
   widgetClassPath=&CommonTextBlock
   bIsVariable=true
   parent=HorizontalBox_52
   widget=AdvanceScreenCountdownText_Hor
   properties {
    clipping=OnDemand
    justification=Right
    scrollStyle=&TextScrollStyle_PilgrimQuickplay_Setlist_C
    text=Next Song in:
    textTransformPolicy=ToUpper
    colorAndOpacity {colorUseRule=UseColor_Specified, specifiedColor=<rgba 1 1 1 0.8>}
    font {
     bForceMonospaced=false
     bMaterialIsStencil=false
     fontMaterial=null
     fontObject=&HeadingNow
     letterSpacing=0
     monospacedWidth=1
     size=15
     skewAmount=0
     typefaceFontName=75 Medium Italic
     outlineSettings {
      bApplyOutlineToDropShadows=false
      bMiteredCorners=false
      bSeparateFillAlpha=false
      outlineMaterial=&M-FontShadow
      outlineSize=2
      outlineColor=<rgba 0 0 0 1>
     }
    }
    shadowOffset=<0 0>
    strikeBrush {
     bIsDynamicallyLoaded=false
     drawAs=NoDrawType
     imageType=NoImage
     mirroring=NoMirror
     resourceName=None
     resourceObject=null
     tiling=NoTile
     imageSize=<32 32>
     margin=<ltrb 0 0 0 0>
     outlineSettings {
      bUseBrushTransparency=false
      roundingType=HalfHeightRadius
      width=0
      color {colorUseRule=UseColor_Specified, specifiedColor=<rgba 0 0 0 0>}
      cornerRadii=<0 0 0 0>
     }
     tintColor {colorUseRule=UseColor_Specified, specifiedColor=<rgba 1 1 1 1>}
     uVRegion {bIsValid=false, max=<0 0>, min=<0 0>}
    }
   }
   slot {class=HorizontalBoxSlot, properties {verticalAlignment=VAlign_Center}}
  }
  SetlistResultsCounter_Hor {
   widgetClassPath=&CommonTextBlock
   bIsVariable=true
   parent=HorizontalBox_52
   widget=SetlistResultsCounter_Hor
   properties {
    bIsScrollingEnabled=false
    justification=Center
    lineHeightPercentage=0.85
    text=30
    font {
     bForceMonospaced=false
     bMaterialIsStencil=false
     fontMaterial=null
     fontObject=&HeadingNow
     letterSpacing=0
     monospacedWidth=1
     size=30
     skewAmount=0
     typefaceFontName=86 Bold Italic
     outlineSettings {
      bApplyOutlineToDropShadows=false
      bMiteredCorners=false
      bSeparateFillAlpha=false
      outlineMaterial=&M-FontShadow
      outlineSize=2
      outlineColor=<rgba 0 0 0 1>
     }
    }
    margin=<ltrb 0 -8 0 0>
    shadowOffset=<0 0>
    strikeBrush {
     bIsDynamicallyLoaded=false
     drawAs=NoDrawType
     imageType=NoImage
     mirroring=NoMirror
     resourceName=None
     resourceObject=null
     tiling=NoTile
     imageSize=<32 32>
     margin=<ltrb 0 0 0 0>
     outlineSettings {
      bUseBrushTransparency=false
      roundingType=HalfHeightRadius
      width=0
      color {colorUseRule=UseColor_Specified, specifiedColor=<rgba 0 0 0 0>}
      cornerRadii=<0 0 0 0>
     }
     tintColor {colorUseRule=UseColor_Specified, specifiedColor=<rgba 1 1 1 1>}
     uVRegion {bIsValid=false, max=<0 0>, min=<0 0>}
    }
   }
   slot {class=HorizontalBoxSlot, properties {horizontalAlignment=HAlign_Right}}
  }
 ]
}
