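# Quickplay results widget showing one player's identity, instrument, difficulty, score, stars and accuracy, apparently above that player.
# SetPlayerState fills identity and color, matches song results by net ID, and switches local-session players to a runtime-created nameplate.
# Accuracy between 99 and 100 displays as 99; flawless requires 100 with no misses.
# The intro interpolates score with a tally sound, then after the reveal's duration staggers stars and shows accuracy or flawless badge.
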WidgetBlueprint /PilgrimQuickplay/Widgets/Results/SongResults/WBP_PilgrimQuickplay_Results_AbovePlayerDisplay
animations [
 FadeIn_StarStats {
  binding_count=1
  length_seconds=0.5016
  object=&self:FadeIn_StarStats
  sequence {bindings [StarsStatsBox {
     id=#1
     possessed_class=VerticalBox
     tracks [{
       type=MovieSceneFloatTrack
       display_name=Render Opacity
       sections [{type=MovieSceneFloatSection, end_frame=10, start_frame=0, channels [None {type=MovieSceneScriptingFloatChannel, keys [[0,0],[10.0307,1]]}]}]
      }]
    }]}
 }
 FadeIn_ScoreStats {
  binding_count=1
  length_seconds=0.5016
  object=&self:FadeIn_ScoreStats
  sequence {bindings [ScoreStatsBox {
     id=#2
     possessed_class=VerticalBox
     tracks [{
       type=MovieSceneFloatTrack
       display_name=Render Opacity
       sections [{type=MovieSceneFloatSection, end_frame=10, start_frame=0, channels [None {type=MovieSceneScriptingFloatChannel, keys [[0,0],[10.0307,1]]}]}]
      }]
    }]}
 }
 Intro_Reveal {
  binding_count=9
  length_seconds=1.5
  object=&self:Intro_Reveal
  sequence {bindings [
    BracketGlow {
     id=#3
     possessed_class=Image
     tracks [{
       type=MovieSceneFloatTrack
       display_name=Render Opacity
       sections [{type=MovieSceneFloatSection, channels [None {type=MovieSceneScriptingFloatChannel, keys [[0,0],[17,0],[30,1]]}]}]
      }]
    }
    All {
     id=#4
     possessed_class=Overlay
     tracks [
      {
       type=MovieSceneFloatTrack
       display_name=Render Opacity
       sections [{type=MovieSceneFloatSection, channels [None {type=MovieSceneScriptingFloatChannel, keys [[0,0],[4,1]]}]}]
      }
      {
       type=MovieScene2DTransformTrack
       sections [{
         type=MovieScene2DTransformSection
         channels [
          Translation.X {type=MovieSceneScriptingFloatChannel, const [0,1]}
          Translation.Y {type=MovieSceneScriptingFloatChannel, keys [[0,80],[5,0]]}
         ]
        }]
      }
     ]
    }
    ScoreStatsBox {
     id=#5
     possessed_class=HorizontalBox
     tracks [
      {
       type=MovieSceneFloatTrack
       display_name=Render Opacity
       sections [{type=MovieSceneFloatSection, channels [None {type=MovieSceneScriptingFloatChannel, keys [[11,0],[18,1]]}]}]
      }
      {
       type=MovieScene2DTransformTrack
       sections [{
         type=MovieScene2DTransformSection
         channels [
          Translation.X {type=MovieSceneScriptingFloatChannel, const [0,1]}
          Translation.Y {type=MovieSceneScriptingFloatChannel, keys [[11,16],[18,0]]}
         ]
        }]
      }
     ]
    }
    StatsOverlay {
     id=#6
     possessed_class=Overlay
     tracks [
      {
       type=MovieSceneFloatTrack
       display_name=Render Opacity
       sections [{type=MovieSceneFloatSection, channels [None {type=MovieSceneScriptingFloatChannel, keys [[2,0],[5,1]]}]}]
      }
      {
       type=MovieScene2DTransformTrack
       sections [{
         type=MovieScene2DTransformSection
         end_frame=10
         start_frame=0
         channels [
          Translation.X {type=MovieSceneScriptingFloatChannel, const [0,1]}
          Translation.Y {type=MovieSceneScriptingFloatChannel, keys [[2,64],[8,0]]}
          Scale.X {type=MovieSceneScriptingFloatChannel, const [1,2]}
          Scale.Y {type=MovieSceneScriptingFloatChannel, const [1,2]}
         ]
        }]
      }
     ]
    }
    TrianglePanel {
     id=#7
     possessed_class=Image
     tracks [
      {
       type=MovieSceneByteTrack
       display_name=Visibility
       sections [{type=MovieSceneByteSection, end_frame=30, start_frame=0, channels [None {type=MovieSceneScriptingByteChannel, keys [[0,4],[30,1]]}]}]
      }
      {
       type=MovieSceneFloatTrack
       display_name=Render Opacity
       sections [{type=MovieSceneFloatSection, end_frame=30, start_frame=0, channels [None {type=MovieSceneScriptingFloatChannel, keys [[0,1],[6,1],[30,0]]}]}]
      }
      {
       type=MovieSceneWidgetMaterialTrack
       sections [{type=MovieSceneParameterSection, channels [OD_ColorIntensity {type=MovieSceneScriptingFloatChannel, keys [[0,1],[6,1],[8,3],[30,1]]}]}]
      }
     ]
    }
    SizeBox_2 {
     id=#8
     possessed_class=SizeBox
     tracks [
      {
       type=MovieSceneFloatTrack
       display_name=Render Opacity
       sections [{type=MovieSceneFloatSection, channels [None {type=MovieSceneScriptingFloatChannel, keys [[8,0],[15,1]]}]}]
      }
      {
       type=MovieScene2DTransformTrack
       sections [{
         type=MovieScene2DTransformSection
         channels [
          Translation.X {type=MovieSceneScriptingFloatChannel, const [0,1]}
          Translation.Y {type=MovieSceneScriptingFloatChannel, keys [[8,32],[15,0]]}
         ]
        }]
      }
     ]
    }
    Intro_Glow {
     id=#9
     possessed_class=Image
     tracks [
      {
       type=MovieSceneFloatTrack
       display_name=Render Opacity
       sections [{
         type=MovieSceneFloatSection
         end_frame=30
         start_frame=0
         channels [None {type=MovieSceneScriptingFloatChannel, keys [[0,0],[6,0],[10,1],[30,0]]}]
        }]
      }
      {
       type=MovieSceneByteTrack
       display_name=Visibility
       sections [{type=MovieSceneByteSection, end_frame=30, start_frame=0, channels [None {type=MovieSceneScriptingByteChannel, keys [[0,4],[30,1]]}]}]
      }
     ]
    }
    StatsBox {
     id=#10
     possessed_class=VerticalBox
     tracks [{
       type=MovieSceneFloatTrack
       display_name=Render Opacity
       sections [{type=MovieSceneFloatSection, channels [None {type=MovieSceneScriptingFloatChannel, keys [[8,0],[9,1]]}]}]
      }]
    }
    PlayerBracket_Indicator {
     id=#11
     possessed_class=Image
     tracks [{
       type=MovieSceneWidgetMaterialTrack
       sections [{type=MovieSceneParameterSection, channels [Height {type=MovieSceneScriptingFloatChannel, keys [[4,1],[20,0.05]]}]}]
      }]
    }
   ]}
 }
 Anim_Intro_Completed {
  binding_count=4
  length_seconds=1.3
  object=&self:Anim_Intro_Completed
  sequence {bindings [
    Intro_Ring_Pulse {
     id=#12
     possessed_class=Image
     tracks [
      {
       type=MovieSceneByteTrack
       display_name=Visibility
       sections [{type=MovieSceneByteSection, channels [None {type=MovieSceneScriptingByteChannel, keys [[0,4],[26,1]]}]}]
      }
      {
       type=MovieSceneFloatTrack
       display_name=Render Opacity
       sections [{type=MovieSceneFloatSection, channels [None {type=MovieSceneScriptingFloatChannel, keys [[0,0],[2.647,1],[26,0]]}]}]
      }
      {
       type=MovieScene2DTransformTrack
       sections [{
         type=MovieScene2DTransformSection
         channels [
          Scale.X {type=MovieSceneScriptingFloatChannel, keys [[0,1],[26,1.333]]}
          Scale.Y {type=MovieSceneScriptingFloatChannel, keys [[0,1],[26,1.75]]}
         ]
        }]
      }
      {
       type=MovieSceneWidgetMaterialTrack
       sections [{type=MovieSceneParameterSection, channels [TaperGlow {type=MovieSceneScriptingFloatChannel, keys [[5,1],[26,4]]}]}]
      }
     ]
    }
    ScoreGlow {
     id=#13
     possessed_class=Image
     tracks [{
       type=MovieSceneWidgetMaterialTrack
       sections [{
         type=MovieSceneParameterSection
         end_frame=26
         start_frame=0
         channels [
          SoftBorderHardness {type=MovieSceneScriptingFloatChannel, const [0.25,2]}
          Alpha_Final {type=MovieSceneScriptingFloatChannel, keys [[0,0.25],[2,0.5],[20,0.25]]}
         ]
        }]
      }]
    }
    TrianglePanel {id=#14, possessed_class=Image, tracks []}
   ]}
 }
]
graphs [
 CheckIfMisingPart {nodes [
   CheckIfMisingPart() {is=title K2Node_FunctionEntry}
   n1 = GetOwningPlayer()
   > Is Valid(InputObject=@n1)
   > ReceiveSongMetadataAndPart!(
    execute
    self:BP Pilgrim Quickplay Missing Parts VM Object Reference=Get BP_PilgrimQuickplay_MissingPartsVM() -> (BP_PilgrimQuickplay_MissingPartsVM:BP Pilgrim Quickplay Missing Parts VM Object Reference)
    Song Metadata:Pilgrim Song Metadata Object Reference=GetCurrentSong!(ContextActor:Actor Object Reference=@n1) -> (ReturnValue:Pilgrim Song Metadata Object Reference)
    PartSelected:EPilgrimTrackType Enum=Get Instrument Type() -> (Instrument Type:EPilgrimTrackType Enum=TrackGuitar)
   ) -> (then)
  ]}
 EventGraph {nodes [
   n1 = SetPlayerState!() -> (OutputDelegate, then, PlayerState:Player State Object Reference) {is=title K2Node_CustomEvent}
   > n2 = Cast To FortPlayerState!(execute, Object=@n1.PlayerState) -> (then, CastFailed, AsFort Player State:Fort Player State Object Reference) {is=title K2Node_DynamicCast}
   > n3 = Set CachedFortPlayerState(CachedFortPlayerState:Fort Player State Object Reference=@n2) -> (Output_Get:Fort Player State Object Reference)
   > n4 = Cast To FortPlayerStateAthena!(execute, Object=@n3) -> (then, CastFailed, AsFort Player State Athena:Fort Player State Athena Object Reference) {is=title K2Node_DynamicCast}
   > n5 = Set Fort Player State Athena(Fort Player State Athena:Fort Player State Athena Object Reference=@n4) -> (Output_Get:Fort Player State Athena Object Reference)
   > SetPlayerColor!(
    execute
    self
    PlayerColor:Linear Color Structure=GetPilgrim PlayerIdentity Color By PlayerState!(Player State:Player State Object Reference=@n5) -> (Player Identity Color:Linear Color Structure)
   ) -> (then)
   > SetPlayerName()
   > SetPlatformIcon()
   > SetSongResultData()
   > HookUpVoiceIcon!(
    execute
    self:Pilgrim Player Info Widget Base Object Reference
    Icon:Fort Voice Chat Status Icon Object Reference=Get VoiceIcon_Status() -> (VoiceIcon_Status:Fort Voice Chat Status Icon Proxy Anim Object Reference)
    Player:Fort Player State Athena Object Reference=Get Fort Player State Athena() -> (Fort Player State Athena:Fort Player State Athena Object Reference)
   ) -> (then)
   > SetIntroAnimationVisualState()
   > n6 = Branch(Condition=Is Local Player in Local MP Session!(PlayerState:Player State Object Reference=Get Fort Player State Athena()) -> (ReturnValue:bool))
   n7 = Get LocalPlayer_SizeBox() -> (LocalPlayer_SizeBox:Size Box Object Reference)
   > RemoveChildAt!(execute, self:Panel Widget Object Reference=@n7, Index:int=0) -> (then, ReturnValue:bool)
   n8 = GetFortPlayerControllerFromActor!(Actor:Actor Object Reference=Get Fort Player State Athena()) -> (ReturnValue:Fort Player Controller Object Reference)
   > n9 = Create Widget!(
    execute:Exec
    Class:User Widget Class Reference=/PilgrimQuickplay/Widgets/Players/WBP_Pilgrim_PlayerIdentity_Nameplate.WBP_Pilgrim_PlayerIdentity_Nameplate_C
    OwningPlayer:Player Controller Object Reference=@n8
    FixedOwner:Player Controller Object Reference=@n8
   ) -> (then:Exec, ReturnValue:WBP Pilgrim Player Identity Nameplate Object Reference)
   > n10 = AddChild(self=@n7, Content=@n9)
   > n11 = Cast To SizeBoxSlot!(execute, Object=@n10) -> (then, CastFailed, AsSize Box Slot:Size Box Slot Object Reference) {is=title K2Node_DynamicCast}
   > SetHorizontalAlignment(self:Size Box Slot Object Reference=@n11, InHorizontalAlignment=HAlign_Fill)
   > SetVerticalAlignment(self:Size Box Slot Object Reference=@n11, InVerticalAlignment=VAlign_Fill)
   > SetActiveWidgetIndex(
    self:Common Visibility Switcher Object Reference=Get LocalOrOnlineSwitcher() -> (LocalOrOnlineSwitcher:Common Visibility Switcher Object Reference)
    Index=1
   )
   SetActiveWidgetIndex(execute=@n6.else, self=Get LocalOrOnlineSwitcher(), Index=0)
   PrintString!(
    execute=@n2.CastFailed
    InString=Failed to cast to FortPlayerState
    bPrintToScreen
    bPrintToLog
    TextColor=(R=1,G=0.0454,B=0,A=1)
    Duration=10
    Key
   ) -> (then)
   PlayerDisplayIntro() {is=title K2Node_CustomEvent}
   > TriggerIntroAnim()
   n12 = Get ScoreText() -> (ScoreText:Common Numeric Text Block Object Reference)
   > n13 = InterpolateToValue!(
    execute
    self:Common Numeric Text Block Object Reference=@n12
    TargetValue:float=Get Score() -> (Score:double)
    MaximumInterpolationDuration:float=4
    MinimumChangeRate:float=1
    OutroOffset:float=0
   ) -> (then)
   n14 = OnOutroEvent_Event!() -> (OutputDelegate, then, NumericTextBlock:Common Numeric Text Block Object Reference) {is=title K2Node_CustomEvent}
   Assign On Outro Event!(execute:Exec=@n13.then, self:Common Numeric Text Block Object Reference=@n12, Delegate:"Delegate (by ref)"=@n14) -> (then:Exec) {is=title K2Node_AssignDelegate}
   n15 = Get Score()
   > n16 = Branch(Condition=AND Boolean(A="float > float"(A=@n15), B=IsVisible()))
   > n17 = SpawnSound2D!(
    execute
    Sound:Sound Base Object Reference=Get TallySound() -> (TallySound:Sound Base Object Reference)
    VolumeMultiplier:float=1
    PitchMultiplier:float=1
    StartTime:float=0
    ConcurrencySettings:Sound Concurrency Object Reference
    bPersistAcrossLevelTransition:bool=false
    bAutoDestroy:bool=false
   ) -> (then, ReturnValue:Audio Component Object Reference)
   > n18 = Set TallySoundReference(TallySoundReference:Audio Component Object Reference=@n17) -> (Output_Get:Audio Component Object Reference)
   > n19 = Play!(execute, self:Audio Component Object Reference=@n18, StartTime:float=0) -> (then)
   Delay(execute=[@n19.then, @n16.else], Duration=GetEndTime(self=Get Intro_Reveal() -> (Intro_Reveal:Widget Animation Object Reference)))
   > StaggeredStarReveal!(
    execute
    self:WBP Pilgrim Quickplay Local Star Bar Object Reference=Get WBP_PilgrimQuickplay_LocalStarBar() -> (WBP_PilgrimQuickplay_LocalStarBar:WBP Pilgrim Quickplay Local Star Bar Object Reference)
    StarCount:int=Get Player Stars() -> (Player Stars:int=0)
   ) -> (then)
   > TriggerFlawlessAnim()
   > PlayAnimation(InAnimation=Get Anim_Intro_Completed() -> (Anim_Intro_Completed:Widget Animation Object Reference))
   n20 = Event Destruct()
   n21 = Get TallySoundReference() -> (TallySoundReference:Audio Component Object Reference)
   Is Valid(exec=[@n14.then, @n20.then], InputObject=@n21)
   > Stop(self:Audio Component Object Reference=@n21)
   > Destroy Component(self:Actor Component Object Reference=@n21)
  ]}
 FormatAccuracyText {nodes [
   n1 = FormatAccuracyText!() -> (then, RawAccuracy:double=~) {is=title K2Node_FunctionEntry}
   n2 = SelectFloat(A=99, B=@n1, bPickA=AND Boolean(A="float > float"(A=@n1, B=99), B="Not Equal (Float)"(A=@n1, B=100)))
   > Return Node!(
    execute
    AccuracyText:Text="To Text (Float)"(Value=@n2, bUseGrouping=false, MinimumIntegralDigits=1, MaximumIntegralDigits=324, MinimumFractionalDigits=0, MaximumFractionalDigits=0)
    ReturnValue:double=@n2
   )
  ]}
 SetAccuracy {nodes [
   n1 = SetAccuracy!() -> (then, Accuracy:double=~) {is=title K2Node_FunctionEntry}
   > n2 = FormatAccuracyText!(execute, self, RawAccuracy:double=@n1) -> (then, AccuracyText:Text, ReturnValue:double) {is=title K2Node_CallFunction}
   > Set Accuracy(Accuracy:double=@n2.ReturnValue) -> (Output_Get:double)
   > Set Text(Text=@n2, self=Get AccuracyPercentageText_1() -> (AccuracyPercentageText_1:Common Text Object Reference))
  ]}
 SetDifficulty {nodes [
   n1 = SetDifficulty!() -> (then, Difficulty:EPilgrimSongDifficulty Enum=DifficultyEasy) {is=title K2Node_FunctionEntry}
   > SetDifficulty!(
    execute
    self:WBP Pilgrim Results Difficulty Object Reference=Get Difficulty() -> (Difficulty:WBP Pilgrim Results Difficulty Object Reference)
    Difficulty:EPilgrimSongDifficulty Enum=@n1
   ) -> (then) {is=title K2Node_CallFunction}
  ]}
 SetFlawless {nodes [
   n1 = SetFlawless!() -> (then, FlawlessAchieved:bool=~) {is=title K2Node_FunctionEntry}
   > Set Flawless Achieved(Flawless Achieved:bool=@n1) -> (Output_Get:bool)
  ]}
 SetInstrument {nodes [
   n1 = SetInstrument!() -> (then, Instrument:EPilgrimTrackType Enum=TrackGuitar) {is=title K2Node_FunctionEntry}
   > n2 = GetDynamicMaterial(self=Get InstrumentIcon() -> (InstrumentIcon:Image Object Reference))
   n3 = GetOwningPlayer()
   n4 = GetCurrentSong!(self:Pilgrim Core Band State Component Object Reference=GetBandState!(ContextActor:Actor Object Reference=GetSparksMusicPlayspaceOfClass!(
      PlayspaceType:Sparks Music Playspace Class Reference=/Script/PilgrimQuickplayRuntime.PilgrimQuickplayPlayspace
      Actor:Actor Object Reference=@n3
      bMatchChildTypes:bool=false
     ) -> (ReturnValue:Pilgrim Quickplay Playspace Object Reference)) -> (ReturnValue:Pilgrim Quickplay Band State Component Object Reference)) -> (ReturnValue)
   n5 = PilgrimTrackTypeToSparksInstrument!(PilgrimType:EPilgrimTrackType Enum=@n1) -> (ReturnValue:ESparksInstrumentType Enum=Guitar)
   > SetScalarParameterValue(
    self=@n2
    ParameterName=Icon
    Value="To Float (Integer)"(InInt=GetInstrumentIconIndex!(
      InstrumentType:EPilgrimTrackType Enum=@n1
      SongMetaData:Pilgrim Song Metadata Object Reference=@n4
      StartingInstrumentIndex:ESparksInstrumentType Enum=@n5
      bUseKeyTarIcon:bool=true
     ) -> (InstrumentIdexIcon:int=0))
   )
  ]}
 SetIntroAnimationVisualState {nodes [
   SetIntroAnimationVisualState() {is=title K2Node_FunctionEntry}
   n1 = Get Intro_Reveal()
   > PlayAnimation(InAnimation=@n1, StartAtTime=GetEndTime(self=@n1), PlayMode=Reverse)
  ]}
 SetPlatformIcon {nodes [
   SetPlatformIcon() {is=title K2Node_FunctionEntry}
   n1 = Get CachedFortPlayerState() -> (CachedFortPlayerState:Fort Player State Object Reference)
   > Is Valid(InputObject=@n1)
   > SetPlatformPrefixIcon!(
    execute
    Image:Image Object Reference=Get Image_PlatformIcon() -> (Image_PlatformIcon:Image Object Reference)
    OtherPlayerPlatform:String=GetPlatform!(self:Fort Player State Object Reference=@n1) -> (ReturnValue:String)
   ) -> (then)
  ]}
 SetPlayerColor {nodes [
   n1 = SetPlayerColor!() -> (then, PlayerColor:Linear Color Structure) {is=title K2Node_FunctionEntry}
   > n2 = GetDynamicMaterial(self=Get PlayerBracket_Indicator() -> (PlayerBracket_Indicator:Image Object Reference))
   > SetVectorParameterValue(self=@n2, ParameterName=ColorA, Value=@n1)
   > SetVectorParameterValue(self=@n2, ParameterName=ColorB, Value=@n1)
   > SetColorAndOpacity(self:Image Object Reference=Get BracketGlow() -> (BracketGlow:Image Object Reference), InColorAndOpacity:Linear Color Structure=@n1)
  ]}
 SetPlayerName {nodes [
   SetPlayerName() {is=title K2Node_FunctionEntry}
   n1 = Get CachedFortPlayerState()
   > Is Valid(InputObject=@n1)
   > "SetText (Text)"(
    self=Get Text_PlayerName() -> (Text_PlayerName:Common Text Object Reference)
    InText="To Text (String)"(InString=GetPlayerNameSafe!(Actor:Actor Object Reference=@n1) -> (ReturnValue:String))
   )
  ]}
 SetScore {nodes [
   n1 = SetScore!() -> (then, Score:int) {is=title K2Node_FunctionEntry}
   > Set Score(Score:double="To Float (Integer)"(InInt=@n1)) -> (Output_Get:double)
  ]}
 SetSongResultData {nodes [
   SetSongResultData() {is=title K2Node_FunctionEntry}
   > Is Valid(InputObject=Get CachedFortPlayerState())
   > n1 = GetPilgrimQuickplayPlayerStateComponent!(execute, PlayerState:Player State Object Reference=Get CachedFortPlayerState()) -> (then, ReturnValue:Pilgrim Quickplay Player State Component Object Reference)
   > n2 = Sequence()
   > n3 = For Each Loop with Break!(
    Exec:Exec
    Array:Array of Pilgrim Quickplay Song Result Player Data Structures=Break Pilgrim Quickplay Song Result Data!(PilgrimQuickplaySongResultData:"Pilgrim Quickplay Song Result Data Structure (by ref)"=GetSongResultData!(self:Pilgrim Quickplay Player State Component Object Reference=@n1) -> (ReturnValue:"Pilgrim Quickplay Song Result Data Structure (by ref)")) -> (
     SongName:String
     SongTags:Gameplay Tag Container Structure
     SongShortName:Name
     BandData:Array of Pilgrim Quickplay Song Result Band Data Structures
     PlayerData:Array of Pilgrim Quickplay Song Result Player Data Structures
     Team:Byte
     LocalPlayerBandDataIndex:int
     LocalPlayerDataIndex:int
    ) {is=title K2Node_BreakStruct}.PlayerData
    Break:Exec=@n6.then
   ) -> (LoopBody:Exec, Array Element:Pilgrim Quickplay Song Result Player Data Structure, Array Index:int, Completed:Exec)
   n4 = Get CachedFortPlayerState()
   > Branch(Condition="Equal (UniqueIdRepl)"(
     A:"Unique Net Id Structure (by ref)"=GetPlayerUniqueNetID!(self:Fort Player State Object Reference=@n4) -> (ReturnValue:Unique Net Id Structure)
     B:"Unique Net Id Structure (by ref)"=Break Pilgrim Quickplay Song Result Player Data!(PilgrimQuickplaySongResultPlayerData:"Pilgrim Quickplay Song Result Player Data Structure (by ref)"=@n3) -> (
      PlayerName:String
      PlayerAccountId:Unique Net Id Structure
      PlayerState:Player State Weak object
      InstrumentType:EPilgrimTrackType Enum
      InstrumentTags:Gameplay Tag Container Structure
      PlayerStars:int
      PlayerScore:int
      PlayerDifficulty:EPilgrimSongDifficulty Enum
      SongTrackDifficulty:int
      Team:bool=~
     ) {is=title K2Node_BreakStruct}.PlayerAccountId
    ) -> (ReturnValue:bool=~) {is=title K2Node_PromotableOperator})
   n5 = Break Pilgrim Quickplay Song Result Player Data(PilgrimQuickplaySongResultPlayerData=@n3) -> (Team=~)
   > SetInstrument!(execute, self, Instrument:EPilgrimTrackType Enum=@n5.InstrumentType) -> (then) {is=title K2Node_CallFunction}
   > Set Instrument Type(Instrument Type:EPilgrimTrackType Enum=@n5.InstrumentType) -> (Output_Get:EPilgrimTrackType Enum=TrackGuitar)
   > Set Player Stars(Player Stars:int=@n5.PlayerStars) -> (Output_Get:int=0)
   > SetupStarsResults_DEPRECATED!(
    execute
    self:WBP Pilgrim Quickplay Local Star Bar Object Reference=Get WBP_PilgrimQuickplay_LocalStarBar()
    Instrument:EPilgrimTrackType Enum=@n5.InstrumentType
    StarCount:int=@n5.PlayerStars
   ) -> (then)
   > SetScore!(execute, self, Score:int=@n5.PlayerScore) -> (then) {is=title K2Node_CallFunction}
   > n6 = SetDifficulty(self:Self Object Reference, Difficulty=@n5.PlayerDifficulty)
   n7 = Break Pilgrim Quickplay Player Performance Data!(PilgrimQuickplayPlayerPerformanceData:"Pilgrim Quickplay Player Performance Data Structure (by ref)"=GetPerformanceData!(self:Pilgrim Quickplay Player State Component Object Reference=@n1) -> (ReturnValue:"Pilgrim Quickplay Player Performance Data Structure (by ref)")) -> (
    Accuracy:float=~
    AverageOffset:float=~
    StandardDeviation:float=~
    FullCombo:bool=~
    NotesHit:int
    NotesPassed:int
    NotesMissed:int
    TotalNotes:int
    LongestStreak:int
    TimeInOverdriveMs:float=~
    HistogramSamples:Array of Pilgrim Histogram Sample Structures
    AccuracyTierCounts:Array of Integers
    SectionsCompletedByTier:Array of Integers
   ) {is=title K2Node_BreakStruct}
   SetAccuracy!(execute=@n2.then_1, self, Accuracy:double=@n7) -> (then) {is=title K2Node_CallFunction}
   > SetFlawless!(execute, self, FlawlessAchieved:bool=AND Boolean(A="float >= float"(A=@n7, B=100), B="Equal (Integer)"(A=@n7.NotesMissed))) -> (then) {is=title K2Node_CallFunction}
  ]}
 ShowScoreStats {nodes [
   ShowScoreStats() {is=title K2Node_FunctionEntry}
   > PlayAnimationForward(InAnimation=Get FadeIn_ScoreStats() -> (FadeIn_ScoreStats:Widget Animation Object Reference))
   > PlayAnimationReverse(InAnimation=Get FadeIn_StarStats() -> (FadeIn_StarStats:Widget Animation Object Reference))
  ]}
 ShowStarStats {nodes [
   ShowStarStats() {is=title K2Node_FunctionEntry}
   > PlayAnimationForward(InAnimation=Get FadeIn_StarStats())
   > PlayAnimationReverse(InAnimation=Get FadeIn_ScoreStats())
  ]}
 TriggerFlawlessAnim {nodes [
   TriggerFlawlessAnim() {is=title K2Node_FunctionEntry}
   > n1 = Branch(Condition=Get Flawless Achieved() -> (Flawless Achieved:bool))
   > SetActiveWidgetIndex(self:Widget Switcher Object Reference=Get LabelGold() -> (LabelGold:Common Activatable Widget Switcher Object Reference), Index=1)
   n2 = Get FlawlessIndicator() -> (FlawlessIndicator:WBP Pilgrim Quickplay Full Combo Indicator Object Reference)
   > SetVisibility(self=@n2, InVisibility=SelfHitTestInvisible)
   > TriggerPerfectRevealAnim(self:WBP Pilgrim Quickplay Full Combo Indicator Object Reference=@n2)
   > PlaySound2D(Sound=Get FullComboSound() -> (FullComboSound:Sound Base Object Reference))
   SetActiveWidgetIndex(execute=@n1.else, self=Get LabelGold(), Index=0)
   > SetVisibility(self=Get FlawlessIndicator(), InVisibility=Hidden)
  ]}
 TriggerIntroAnim {nodes [
   TriggerIntroAnim() {is=title K2Node_FunctionEntry}
   > PlayAnimation(InAnimation=Get Intro_Reveal())
  ]}
]
variables [
 Accuracy {type=number}
 CachedFortPlayerState {type=FortPlayerState, default=null}
 Flawless Achieved {type=boolean}
 Fort Player State Athena {type=FortPlayerStateAthena, default=null}
 FullComboSound {type=SoundBase, default=/PilgrimQuickplay/Audio/Results/Results_FullCombo_Cue}
 Instrument Type {type=EPilgrimTrackType, default=TrackGuitar}
 NotReadyColor {type=SlateColor, default {colorUseRule=UseColor_Specified, specifiedColor=<rgba 1 0.0423 0.0578 1>}}
 Player Stars {type=integer}
 ReadyColor {type=SlateColor, default {colorUseRule=UseColor_Specified, specifiedColor=<rgba 0.0343 1 0.227 1>}}
 Score {type=number}
 TallySound {type=SoundBase, default=/PilgrimQuickplay/Audio/Results/MSS_FM_ResultsVictory_Tally1pNums}
 TallySoundReference {type=AudioComponent, default=null}
]
widgets {
 info {inheritedWidgetCount=0, parentClass=/Script/PilgrimCoreUI.PilgrimPlayerInfoWidgetBase, rootWidgetClass=&Overlay, widgetCount=44}
 widgets [
  All {widgetClassPath=&Overlay, widget=All}
  VerticalBox_20 {widgetClassPath=&VerticalBox, parent=All, widget=VerticalBox_20, slot {class=OverlaySlot}}
  LocalOrOnlineSwitcher {
   widgetClassPath=&CommonVisibilitySwitcher
   bIsVariable=true
   parent=VerticalBox_20
   widget=LocalOrOnlineSwitcher
   properties {activeWidgetIndex=1}
   slot {class=VerticalBoxSlot, properties {horizontalAlignment=HAlign_Center, verticalAlignment=VAlign_Center}}
  }
  OnlinePlayer_NamePlatformSpeaker {
   widgetClassPath=&HorizontalBox
   parent=LocalOrOnlineSwitcher
   widget=OnlinePlayer_NamePlatformSpeaker
   properties {flowDirectionPreference=Culture}
   slot {class=CommonVisibilitySwitcherSlot}
  }
  SizeBox_3 {
   widgetClassPath=&SizeBox
   parent=OnlinePlayer_NamePlatformSpeaker
   widget=SizeBox_3
   properties {bOverride_MaxDesiredWidth=true, maxDesiredWidth=208}
   slot {class=HorizontalBoxSlot, properties {horizontalAlignment=HAlign_Right, verticalAlignment=VAlign_Center}}
  }
  Text_PlayerName {
   widgetClassPath=&CommonTextBlock
   bIsVariable=true
   parent=SizeBox_3
   widget=Text_PlayerName
   properties {
    scrollStyle=&TextScrollStyle-Base_C
    text=TestPlayerName
    font {
     bForceMonospaced=false
     bMaterialIsStencil=false
     fontMaterial=null
     fontObject=&HeadingNow
     letterSpacing=0
     monospacedWidth=1
     size=13.5
     skewAmount=0
     typefaceFontName=75 Medium
     outlineSettings {
      bApplyOutlineToDropShadows=false
      bMiteredCorners=false
      bSeparateFillAlpha=false
      outlineMaterial=&M-FontShadow
      outlineSize=1
      outlineColor=<rgba 0 0 0 1>
     }
    }
    margin=<ltrb 0 -2 2 -1>
    shadowOffset=<0 0>
   }
   slot {class=SizeBoxSlot}
  }
  SizeBox_1 {
   widgetClassPath=&SizeBox
   parent=OnlinePlayer_NamePlatformSpeaker
   widget=SizeBox_1
   properties {bOverride_HeightOverride=true, bOverride_WidthOverride=true, heightOverride=24, widthOverride=24}
   slot {class=HorizontalBoxSlot, properties {horizontalAlignment=HAlign_Center, verticalAlignment=VAlign_Center}}
  }
  InputIndicator_ActiveInputType {
   widgetClassPath=/Game/Athena/UI/Matchmaking/MatchmakingInputIndicator
   bIsVariable=true
   parent=SizeBox_1
   widget=InputIndicator_ActiveInputType
   properties {bShowRemoteInputOnlyWhenDifferentFromDefault=true, inputSource=Remote}
   slot {class=SizeBoxSlot}
  }
  SizeBox_0 {
   widgetClassPath=&SizeBox
   parent=OnlinePlayer_NamePlatformSpeaker
   widget=SizeBox_0
   properties {bOverride_HeightOverride=true, bOverride_WidthOverride=true, heightOverride=22, widthOverride=22}
   slot {class=HorizontalBoxSlot, properties {horizontalAlignment=HAlign_Center, verticalAlignment=VAlign_Center, padding=<ltrb 0 -5 0 -5>}}
  }
  Image_PlatformIcon {
   widgetClassPath=&Image
   bIsVariable=true
   parent=SizeBox_0
   widget=Image_PlatformIcon
   properties {
    visibility=Collapsed
    brush {
     bIsDynamicallyLoaded=false
     drawAs=Image
     imageType=NoImage
     mirroring=NoMirror
     resourceName=None
     resourceObject=/Game/UI/Friends_UI/Social/Console_PlatformIcon_32x
     tiling=NoTile
     imageSize=<32 32>
     margin=<ltrb 0 0 0 0>
     outlineSettings {
      bUseBrushTransparency=false
      roundingType=HalfHeightRadius
      width=0
      color {colorUseRule=UseColor_Specified, specifiedColor=<rgba 0 0 0 0>}
      cornerRadii=<0 0 0 0>
     }
     tintColor {colorUseRule=UseColor_Specified, specifiedColor=<rgba 1 1 1 1>}
     uVRegion {bIsValid=false, max=<0 0>, min=<0 0>}
    }
   }
   slot {class=SizeBoxSlot}
  }
  VoiceIcon_Status {
   widgetClassPath=/Script/FortniteUI.FortVoiceChatStatusIconProxyAnim
   bIsVariable=true
   parent=OnlinePlayer_NamePlatformSpeaker
   widget=VoiceIcon_Status
   properties {desiredSize=<32 21.3333>}
   slot {class=HorizontalBoxSlot, properties {horizontalAlignment=HAlign_Left}}
  }
  LocalPlayer_SizeBox {
   widgetClassPath=&SizeBox
   bIsVariable=true
   parent=LocalOrOnlineSwitcher
   widget=LocalPlayer_SizeBox
   properties {bOverride_MaxDesiredWidth=true, maxDesiredWidth=270}
   slot {class=CommonVisibilitySwitcherSlot, properties {padding=<ltrb 4 0 4 0>}}
  }
  LocalPlayer_Identity {
   widgetClassPath=/PilgrimQuickplay/Widgets/Players/WBP_Pilgrim_PlayerIdentity_Nameplate
   parent=LocalPlayer_SizeBox
   widget=LocalPlayer_Identity
   slot {class=SizeBoxSlot}
  }
  StatsOverlay {
   widgetClassPath=&Overlay
   bIsVariable=true
   parent=VerticalBox_20
   widget=StatsOverlay
   slot {class=VerticalBoxSlot, properties {padding=<ltrb 0 4 0 0>}}
  }
  Intro_Ring_Pulse {
   widgetClassPath=&Image
   bIsVariable=true
   parent=StatsOverlay
   widget=Intro_Ring_Pulse
   properties {
    visibility=Collapsed
    brush {
     bIsDynamicallyLoaded=false
     drawAs=Box
     imageType=NoImage
     mirroring=NoMirror
     resourceName=None
     resourceObject=/PilgrimCore/Widgets/Materials/MI_UI_Piligrim_GemBreak_Pulse
     tiling=NoTile
     imageSize=<32 32>
     margin=<ltrb 0.5 0.5 0.5 0.5>
     outlineSettings {
      bUseBrushTransparency=false
      roundingType=HalfHeightRadius
      width=0
      color {colorUseRule=UseColor_Specified, specifiedColor=<rgba 0 0 0 0>}
      cornerRadii=<0 0 0 0>
     }
     tintColor {colorUseRule=UseColor_Specified, specifiedColor=<rgba 1 1 1 1>}
     uVRegion {bIsValid=false, max=<0 0>, min=<0 0>}
    }
   }
   slot {class=OverlaySlot, properties {horizontalAlignment=HAlign_Fill, verticalAlignment=VAlign_Fill}}
  }
  Intro_Glow {
   widgetClassPath=&Image
   bIsVariable=true
   parent=StatsOverlay
   widget=Intro_Glow
   properties {
    visibility=Collapsed
    brush {
     bIsDynamicallyLoaded=false
     drawAs=Image
     imageType=NoImage
     mirroring=NoMirror
     resourceName=None
     resourceObject=/PilgrimCore/Widgets/Materials/MI_UI_Pilgrim_ODBonus_AmountBacking2
     tiling=NoTile
     imageSize=<128 64>
     margin=<ltrb 0.5 0.5 0.5 0.5>
     outlineSettings {
      bUseBrushTransparency=false
      roundingType=HalfHeightRadius
      width=0
      color {colorUseRule=UseColor_Specified, specifiedColor=<rgba 0 0 0 0>}
      cornerRadii=<0 0 0 0>
     }
     tintColor {colorUseRule=UseColor_Specified, specifiedColor=<rgba 1 1 1 1>}
     uVRegion {bIsValid=false, max=<0 0>, min=<0 0>}
    }
    renderTransform {angle=0, scale=<1.075 1.22>, shear=<0 0>, translation=<0 0>}
   }
   slot {class=OverlaySlot, properties {horizontalAlignment=HAlign_Fill, verticalAlignment=VAlign_Fill}}
  }
  WBP_Sparks_BackingPanel {
   widgetClassPath=/PilgrimCore/Widgets/SparksUILibrary/WBP_Sparks_BackingPanel
   bIsVariable=true
   parent=StatsOverlay
   widget=WBP_Sparks_BackingPanel
   properties {blurFrame=false}
   slot {class=OverlaySlot, properties {horizontalAlignment=HAlign_Fill, verticalAlignment=VAlign_Fill}}
  }
  TrianglePanel {
   widgetClassPath=&Image
   bIsVariable=true
   parent=StatsOverlay
   widget=TrianglePanel
   properties {
    visibility=Collapsed
    brush {
     bIsDynamicallyLoaded=false
     drawAs=Image
     imageType=NoImage
     mirroring=NoMirror
     resourceName=None
     resourceObject=/PilgrimCore/Widgets/Materials/M_UI_Pilgrim_Postmatch_AnimBacking
     tiling=NoTile
     imageSize=<32 32>
     margin=<ltrb 0 0 0 0>
     outlineSettings {
      bUseBrushTransparency=false
      roundingType=HalfHeightRadius
      width=0
      color {colorUseRule=UseColor_Specified, specifiedColor=<rgba 0 0 0 0>}
      cornerRadii=<0 0 0 0>
     }
     tintColor {colorUseRule=UseColor_Specified, specifiedColor=<rgba 1 1 1 1>}
     uVRegion {bIsValid=false, max=<0 0>, min=<0 0>}
    }
   }
   slot {class=OverlaySlot, properties {horizontalAlignment=HAlign_Fill, verticalAlignment=VAlign_Fill}}
  }
  PlayerBracket_Indicator {
   widgetClassPath=&Image
   bIsVariable=true
   parent=StatsOverlay
   widget=PlayerBracket_Indicator
   properties {
    visibility=SelfHitTestInvisible
    brush {
     bIsDynamicallyLoaded=false
     drawAs=Image
     imageType=NoImage
     mirroring=NoMirror
     resourceName=None
     resourceObject=/PilgrimCore/Widgets/Materials/MI_Piligrim_Bandmate_PostmatchBracket
     tiling=NoTile
     imageSize=<32 32>
     margin=<ltrb 0 0 0 0>
     outlineSettings {
      bUseBrushTransparency=false
      roundingType=HalfHeightRadius
      width=0
      color {colorUseRule=UseColor_Specified, specifiedColor=<rgba 0 0 0 0>}
      cornerRadii=<0 0 0 0>
     }
     tintColor {colorUseRule=UseColor_Specified, specifiedColor=<rgba 1 1 1 1>}
     uVRegion {bIsValid=false, max=<0 0>, min=<0 0>}
    }
   }
   slot {class=OverlaySlot, properties {horizontalAlignment=HAlign_Fill, verticalAlignment=VAlign_Fill}}
  }
  BracketGlow {
   widgetClassPath=&Image
   bIsVariable=true
   parent=StatsOverlay
   widget=BracketGlow
   properties {
    visibility=SelfHitTestInvisible
    brush {
     bIsDynamicallyLoaded=false
     drawAs=Box
     imageType=NoImage
     mirroring=NoMirror
     resourceName=None
     resourceObject=/PilgrimCore/Widgets/Materials/MI_UI_Pilgrim_ProgressPlayerPortrait_Glow
     tiling=NoTile
     imageSize=<64 64>
     margin=<ltrb 0.5 0.5 0.5 0.5>
     outlineSettings {
      bUseBrushTransparency=false
      roundingType=HalfHeightRadius
      width=0
      color {colorUseRule=UseColor_Specified, specifiedColor=<rgba 0 0 0 0>}
      cornerRadii=<0 0 0 0>
     }
     tintColor {colorUseRule=UseColor_Specified, specifiedColor=<rgba 1 1 1 1>}
     uVRegion {bIsValid=false, max=<0 0>, min=<0 0>}
    }
    renderTransform {angle=0, scale=<1.175 1>, shear=<0 0>, translation=<0 26>}
   }
   slot {class=OverlaySlot, properties {horizontalAlignment=HAlign_Fill, verticalAlignment=VAlign_Bottom}}
  }
  StatsBox {widgetClassPath=&VerticalBox, parent=StatsOverlay, widget=StatsBox, slot {class=OverlaySlot, properties {padding=<ltrb 8 8 8 8>}}}
  SizeBox_2 {
   widgetClassPath=&SizeBox
   parent=StatsBox
   widget=SizeBox_2
   properties {bOverride_MaxDesiredWidth=true, bOverride_MinDesiredWidth=true, maxDesiredWidth=256, minDesiredWidth=256}
   slot {class=VerticalBoxSlot}
  }
  StarsStatsBox {
   widgetClassPath=&VerticalBox
   parent=SizeBox_2
   widget=StarsStatsBox
   slot {class=SizeBoxSlot, properties {padding=<ltrb 6 0 4 0>}}
  }
  HorizontalBox_69 {widgetClassPath=&HorizontalBox, parent=StarsStatsBox, widget=HorizontalBox_69, slot {class=VerticalBoxSlot}}
  WBP_PilgrimQuickplay_LocalStarBar {
   widgetClassPath=/PilgrimQuickplay/Widgets/WBP_PilgrimQuickplay_LocalStarBar
   bIsVariable=true
   parent=HorizontalBox_69
   widget=WBP_PilgrimQuickplay_LocalStarBar
   properties {bIsSequencedAnimation=true, in User Specified Scale=0.5, scale Stretch=UserSpecified, shouldStarsAnimateComplete=true}
   slot {class=HorizontalBoxSlot, properties {horizontalAlignment=HAlign_Left}}
  }
  HorizontalBox_0 {
   widgetClassPath=&HorizontalBox
   parent=HorizontalBox_69
   widget=HorizontalBox_0
   slot {class=HorizontalBoxSlot, properties {horizontalAlignment=HAlign_Right, verticalAlignment=VAlign_Center, size {sizeRule=Fill, value=1}}}
  }
  LabelGold {
   widgetClassPath=&CommonActivatableWidgetSwitcher
   bIsVariable=true
   parent=HorizontalBox_0
   widget=LabelGold
   slot {class=HorizontalBoxSlot, properties {padding=<ltrb 2 0 0 0>}}
  }
  HorizontalBox_1 {
   widgetClassPath=&HorizontalBox
   parent=LabelGold
   widget=HorizontalBox_1
   properties {flowDirectionPreference=Culture}
   slot {class=WidgetSwitcherSlot}
  }
  AccuracyPercentageText_1 {
   widgetClassPath=&CommonTextBlock
   bIsVariable=true
   parent=HorizontalBox_1
   widget=AccuracyPercentageText_1
   properties {
    flowDirectionPreference=Culture
    text=100
    font {
     bForceMonospaced=false
     bMaterialIsStencil=false
     fontMaterial=null
     fontObject=&HeadingNow
     letterSpacing=0
     monospacedWidth=1
     size=13.5
     skewAmount=0
     typefaceFontName=74 Regular
     outlineSettings {
      bApplyOutlineToDropShadows=false
      bMiteredCorners=false
      bSeparateFillAlpha=false
      outlineMaterial=null
      outlineSize=0
      outlineColor=<rgba 0 0 0 1>
     }
    }
    shadowOffset=<0 0>
   }
   slot {class=HorizontalBoxSlot}
  }
  AccuracyPercentageText {
   widgetClassPath=&CommonTextBlock
   bIsVariable=true
   parent=HorizontalBox_1
   widget=AccuracyPercentageText
   properties {
    flowDirectionPreference=Culture
    text=%
    font {
     bForceMonospaced=false
     bMaterialIsStencil=false
     fontMaterial=null
     fontObject=&HeadingNow
     letterSpacing=0
     monospacedWidth=1
     size=13.5
     skewAmount=0
     typefaceFontName=74 Regular
     outlineSettings {
      bApplyOutlineToDropShadows=false
      bMiteredCorners=false
      bSeparateFillAlpha=false
      outlineMaterial=null
      outlineSize=0
      outlineColor=<rgba 0 0 0 1>
     }
    }
    shadowOffset=<0 0>
   }
   slot {class=HorizontalBoxSlot}
  }
  ScaleBox_0 {
   widgetClassPath=&ScaleBox
   parent=LabelGold
   widget=ScaleBox_0
   properties {stretch=UserSpecified, userSpecifiedScale=0.9}
   slot {class=WidgetSwitcherSlot, properties {horizontalAlignment=HAlign_Center, verticalAlignment=VAlign_Center, padding=<ltrb 0 0 -5 0>}}
  }
  WBP_PilgrimQuickplay_FullComboIndicatoR {
   widgetClassPath=/PilgrimQuickplay/Widgets/WBP_PilgrimQuickplay_FullComboIndicator
   bIsVariable=true
   parent=ScaleBox_0
   widget=WBP_PilgrimQuickplay_FullComboIndicatoR
   slot {class=ScaleBoxSlot}
  }
  ScoreStatsBox {
   widgetClassPath=&HorizontalBox
   parent=StatsBox
   widget=ScoreStatsBox
   slot {class=VerticalBoxSlot, properties {horizontalAlignment=HAlign_Center, padding=<ltrb 0 4 0 4>}}
  }
  Overlay_1 {widgetClassPath=&Overlay, parent=ScoreStatsBox, widget=Overlay_1, slot {class=HorizontalBoxSlot}}
  ScoreSpace {
   widgetClassPath=&Spacer
   parent=Overlay_1
   widget=ScoreSpace
   properties {size=<192 1>}
   slot {class=OverlaySlot, properties {verticalAlignment=VAlign_Bottom}}
  }
  Overlay_2 {
   widgetClassPath=&Overlay
   parent=Overlay_1
   widget=Overlay_2
   slot {class=OverlaySlot, properties {horizontalAlignment=HAlign_Center, verticalAlignment=VAlign_Center}}
  }
  ScoreGlow {
   widgetClassPath=&Image
   bIsVariable=true
   parent=Overlay_2
   widget=ScoreGlow
   properties {
    visibility=SelfHitTestInvisible
    brush {
     bIsDynamicallyLoaded=false
     drawAs=Image
     imageType=NoImage
     mirroring=NoMirror
     resourceName=None
     resourceObject=/PilgrimCore/Widgets/Materials/MI_UI_Pilgrim_ScoreGlow
     tiling=NoTile
     imageSize=<24 24>
     margin=<ltrb 0 0 0 0>
     outlineSettings {
      bUseBrushTransparency=false
      roundingType=HalfHeightRadius
      width=0
      color {colorUseRule=UseColor_Specified, specifiedColor=<rgba 0 0 0 0>}
      cornerRadii=<0 0 0 0>
     }
     tintColor {colorUseRule=UseColor_Specified, specifiedColor=<rgba 1 1 1 1>}
     uVRegion {bIsValid=false, max=<0 0>, min=<0 0>}
    }
   }
   slot {class=OverlaySlot, properties {horizontalAlignment=HAlign_Fill, verticalAlignment=VAlign_Fill, padding=<ltrb -6 -4 -6 -4>}}
  }
  ScoreText {
   widgetClassPath=/Script/CommonUI.CommonNumericTextBlock
   bIsVariable=true
   parent=Overlay_2
   widget=ScoreText
   properties {
    easeOutInterpolationExponent=4
    justification=Center
    performSizeInterpolation=false
    text=0
    font {
     bForceMonospaced=false
     bMaterialIsStencil=false
     fontMaterial=null
     fontObject=&HeadingNow
     letterSpacing=0
     monospacedWidth=1
     size=18
     skewAmount=0
     typefaceFontName=76 Bold
     outlineSettings {
      bApplyOutlineToDropShadows=false
      bMiteredCorners=false
      bSeparateFillAlpha=true
      outlineMaterial=&M-FontShadow
      outlineSize=2
      outlineColor=<rgba 0.4 0.1 1 1>
     }
    }
    shadowOffset=<0 0>
   }
   slot {class=OverlaySlot, properties {horizontalAlignment=HAlign_Center, verticalAlignment=VAlign_Center}}
  }
  ScaleBox_466 {
   widgetClassPath=&ScaleBox
   parent=ScoreStatsBox
   widget=ScaleBox_466
   properties {stretch=UserSpecified, userSpecifiedScale=0.8}
   slot {class=HorizontalBoxSlot, properties {padding=<ltrb 4 0 0 0>}}
  }
  HorizontalBox {
   widgetClassPath=&HorizontalBox
   parent=ScaleBox_466
   widget=HorizontalBox
   slot {class=ScaleBoxSlot, properties {verticalAlignment=VAlign_Fill}}
  }
  InstrumentIcon {
   widgetClassPath=&Image
   bIsVariable=true
   parent=HorizontalBox
   widget=InstrumentIcon
   properties {brush {
     bIsDynamicallyLoaded=false
     drawAs=Image
     imageType=NoImage
     mirroring=NoMirror
     resourceName=None
     resourceObject=/PilgrimQuickplay/Widgets/Materials/MI_UI_Icon_Instrument_Type
     tiling=NoTile
     imageSize=<45 45>
     margin=<ltrb 0 0 0 0>
     outlineSettings {
      bUseBrushTransparency=false
      roundingType=HalfHeightRadius
      width=0
      color {colorUseRule=UseColor_Specified, specifiedColor=<rgba 0 0 0 0>}
      cornerRadii=<0 0 0 0>
     }
     tintColor {colorUseRule=UseColor_Specified, specifiedColor=<rgba 1 1 1 1>}
     uVRegion {bIsValid=false, max=<0 0>, min=<0 0>}
    }}
   slot {class=HorizontalBoxSlot, properties {horizontalAlignment=HAlign_Center, verticalAlignment=VAlign_Center}}
  }
  Spacer_2 {widgetClassPath=&Spacer, parent=HorizontalBox, widget=Spacer_2, properties {size=<5 2>}, slot {class=HorizontalBoxSlot}}
  Difficulty {
   widgetClassPath=/PilgrimQuickplay/Widgets/Results/WBP_Pilgrim_Results_Difficulty
   bIsVariable=true
   parent=HorizontalBox
   widget=Difficulty
   slot {class=HorizontalBoxSlot, properties {verticalAlignment=VAlign_Center}}
  }
  FlawlessIndicator {
   widgetClassPath=&WBP_PilgrimQuickplay_FullComboIndicator
   bIsVariable=true
   parent=All
   widget=FlawlessIndicator
   properties {
    active Widget Index=1
    isPerfectComboIndicator=true
    visibility=Hidden
    renderTransform {angle=0, scale=<1 1>, shear=<0 0>, translation=<0 44>}
   }
   slot {class=OverlaySlot, properties {horizontalAlignment=HAlign_Center, verticalAlignment=VAlign_Bottom}}
  }
 ]
}
