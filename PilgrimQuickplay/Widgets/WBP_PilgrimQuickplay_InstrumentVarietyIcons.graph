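# Pilgrim quickplay widget showing a row of guitar, bass, vocals and drums icons on a rounded background, which PreConstruct tints with BGColor.
# Icons start dim; BrightenInstrumentIcon maps a track type, including plastic and pro variants, to one icon and raises its opacity, logging unmapped types.
# DimAllInstrumentIcons resets all four opacities through a shared loop.
# Separate helpers recolor icons through dynamic material parameters, but the highlight and dim path uses only opacity.
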WidgetBlueprint /PilgrimQuickplay/Widgets/WBP_PilgrimQuickplay_InstrumentVarietyIcons
graphs [
 BrightenInstrumentIcon {nodes [
   n1 = GetDynamicMaterial()
   > SetVectorParameterValue(self=@n1, ParameterName=Color, Value=Get BrightInstrumentIconColor() -> (BrightInstrumentIconColor:Linear Color Structure))
   n2 = BrightenInstrumentIcon!() -> (then, InstrumentType:EPilgrimTrackType Enum=TrackGuitar) {is=title K2Node_FunctionEntry}
   n3 = Get GuitarIcon() -> (GuitarIcon:Image Object Reference)
   n4 = Get BassIcon() -> (BassIcon:Image Object Reference)
   n5 = Get VocalsIcon() -> (VocalsIcon:Image Object Reference)
   n6 = Get DrumsIcon() -> (DrumsIcon:Image Object Reference)
   n7 = Select!(
    TrackGuitar:Image Object Reference=@n3
    TrackBass:Image Object Reference=@n4
    TrackVocals:Image Object Reference=@n5
    TrackDrum:Image Object Reference=@n6
    TrackPlasticGuitar:Image Object Reference=@n3
    TrackPlasticBass:Image Object Reference=@n4
    TrackPlasticDrum:Image Object Reference=@n6
    TrackProVocals:Image Object Reference=@n5
    TrackProCymbals:Image Object Reference=@n6
    TrackEvents:Image Object Reference
    TrackNone:Image Object Reference
    Index:EPilgrimTrackType Enum=@n2
   ) -> (ReturnValue:Image Object Reference)
   > n8 = Is Valid(InputObject=@n7)
   PrintString!(
    execute=@n8.Is Not Valid
    InString=Append(A="No icon for track type ", B=Enum to String!(Enumerator:EPilgrimTrackType Enum=@n2) -> (ReturnValue:String))
    bPrintToScreen
    bPrintToLog
    TextColor
    Duration
    Key
   ) -> (then)
   SetOpacity!(execute=@n8.Is Valid, self:Image Object Reference=@n7, InOpacity:float=Get BrightIconOpacity() -> (BrightIconOpacity:double)) -> (then)
  ]}
 DimAllInstrumentIcons {nodes [
   DimAllInstrumentIcons() {is=title K2Node_FunctionEntry}
   > Update All Instrument Icon Opacities!(execute, self, NewOpacity:double=Get DimIconOpacity() -> (DimIconOpacity:double)) -> (then)
  ]}
 EventGraph {nodes [
   Event PreConstruct()
   > SetBrushTintColor!(
    execute
    self:Image Object Reference=Get Background() -> (Background:Fort Lazy Image Object Reference)
    TintColor:Slate Color Structure=Get BGColor() -> (BGColor:Slate Color Structure)
   ) -> (then)
   Event Construct()
   Event Tick()
  ]}
 SetInstrumentIconColor {nodes [
   n1 = SetInstrumentIconColor!() -> (then, Icon:Image Object Reference, NewColor:Linear Color Structure) {is=title K2Node_FunctionEntry}
   > n2 = GetDynamicMaterial(self=@n1)
   > SetVectorParameterValue(self=@n2, ParameterName=Color, Value=@n1.NewColor)
  ]}
 Update All Instrument Icon Colors {nodes [
   n1 = Update All Instrument Icon Colors!() -> (then, NewColor:Linear Color Structure) {is=title K2Node_FunctionEntry}
   > n2 = For Loop(LastIndex=3)
   > SetInstrumentIconColor!(
    execute
    self
    Icon:Image Object Reference=Select!(
     Option 0:Image Object Reference=Get GuitarIcon()
     Option 1:Image Object Reference=Get BassIcon()
     Option 2:Image Object Reference=Get VocalsIcon()
     Option 3:Image Object Reference=Get DrumsIcon()
     Index:int=@n2
    ) -> (ReturnValue)
    NewColor:Linear Color Structure=@n1
   ) -> (then) {is=title K2Node_CallFunction}
  ]}
 Update All Instrument Icon Opacities {nodes [
   n1 = Update All Instrument Icon Opacities!() -> (then, NewOpacity:double=~) {is=title K2Node_FunctionEntry}
   > n2 = For Loop(LastIndex=3)
   > SetOpacity(
    self=Select(Option 0=Get GuitarIcon(), Option 1=Get BassIcon(), Option 2=Get VocalsIcon(), Option 3=Get DrumsIcon(), Index=@n2)
    InOpacity=@n1
   )
  ]}
]
variables [
 BGColor {type=SlateColor, default {colorUseRule=UseColor_Specified, specifiedColor=<rgba 0 0 0 1>}}
 BrightIconOpacity {type=number, default=1}
 BrightInstrumentIconColor {type=LinearColor, default=<rgba 1 1 1 1>}
 DimIconOpacity {type=number, default=0.15}
 DimInstrumentIconColor {type=LinearColor, default=<rgba 0.15 0.15 0.15 1>}
]
widgets {
 info {inheritedWidgetCount=0, parentClass=&CommonUserWidget, rootWidgetClass=&Overlay, widgetCount=7}
 widgets [
  Overlay_0 {widgetClassPath=&Overlay, widget=Overlay_0}
  Background {
   widgetClassPath=&FortLazyImage
   bIsVariable=true
   parent=Overlay_0
   widget=Background
   properties {brush {
     bIsDynamicallyLoaded=false
     drawAs=RoundedBox
     imageType=NoImage
     mirroring=NoMirror
     resourceName=None
     resourceObject=null
     tiling=NoTile
     imageSize=<470 112>
     margin=<ltrb 0 0 0 0>
     outlineSettings {
      bUseBrushTransparency=false
      roundingType=HalfHeightRadius
      width=0
      color {colorUseRule=UseColor_Specified, specifiedColor=<rgba 0 0 0 0>}
      cornerRadii=<0.4 0.4 0.4 0.4>
     }
     tintColor {colorUseRule=UseColor_Specified, specifiedColor=<rgba 0 0 0 1>}
     uVRegion {bIsValid=false, max=<0 0>, min=<0 0>}
    }}
   slot {class=OverlaySlot, properties {horizontalAlignment=HAlign_Center, verticalAlignment=VAlign_Center}}
  }
  InstrumentIcons {
   widgetClassPath=&HorizontalBox
   parent=Overlay_0
   widget=InstrumentIcons
   slot {class=OverlaySlot, properties {horizontalAlignment=HAlign_Center, verticalAlignment=VAlign_Center}}
  }
  VocalsIcon {
   widgetClassPath=&Image
   bIsVariable=true
   parent=InstrumentIcons
   widget=VocalsIcon
   properties {
    brush {
     bIsDynamicallyLoaded=false
     drawAs=Image
     imageType=NoImage
     mirroring=NoMirror
     resourceName=None
     resourceObject=/PilgrimQuickplay/Widgets/Materials/MI_UI_Icon_Instrument_Type_Vocals
     tiling=NoTile
     imageSize=<142 142>
     margin=<ltrb 0 0 0 0>
     outlineSettings {
      bUseBrushTransparency=false
      roundingType=HalfHeightRadius
      width=0
      color {colorUseRule=UseColor_Specified, specifiedColor=<rgba 0 0 0 0>}
      cornerRadii=<0 0 0 0>
     }
     tintColor {colorUseRule=UseColor_Specified, specifiedColor=<rgba 1 1 1 1>}
     uVRegion {bIsValid=false, max=<0 0>, min=<0 0>}
    }
    colorAndOpacity=<rgba 1 1 1 0.15>
   }
   slot {class=HorizontalBoxSlot, properties {padding=<ltrb 0 0 -24 0>}}
  }
  BassIcon {
   widgetClassPath=&Image
   bIsVariable=true
   parent=InstrumentIcons
   widget=BassIcon
   properties {
    brush {
     bIsDynamicallyLoaded=false
     drawAs=Image
     imageType=NoImage
     mirroring=NoMirror
     resourceName=None
     resourceObject=/PilgrimQuickplay/Widgets/Materials/MI_UI_Icon_Instrument_Type_Bass
     tiling=NoTile
     imageSize=<142 142>
     margin=<ltrb 0 0 0 0>
     outlineSettings {
      bUseBrushTransparency=false
      roundingType=HalfHeightRadius
      width=0
      color {colorUseRule=UseColor_Specified, specifiedColor=<rgba 0 0 0 0>}
      cornerRadii=<0 0 0 0>
     }
     tintColor {colorUseRule=UseColor_Specified, specifiedColor=<rgba 1 1 1 1>}
     uVRegion {bIsValid=false, max=<0 0>, min=<0 0>}
    }
    colorAndOpacity=<rgba 1 1 1 0.15>
   }
   slot {class=HorizontalBoxSlot}
  }
  GuitarIcon {
   widgetClassPath=&Image
   bIsVariable=true
   parent=InstrumentIcons
   widget=GuitarIcon
   properties {
    brush {
     bIsDynamicallyLoaded=false
     drawAs=Image
     imageType=NoImage
     mirroring=NoMirror
     resourceName=None
     resourceObject=/PilgrimQuickplay/Widgets/Materials/MI_UI_Icon_Instrument_Type_Guitar
     tiling=NoTile
     imageSize=<142 142>
     margin=<ltrb 0 0 0 0>
     outlineSettings {
      bUseBrushTransparency=false
      roundingType=HalfHeightRadius
      width=0
      color {colorUseRule=UseColor_Specified, specifiedColor=<rgba 0 0 0 0>}
      cornerRadii=<0 0 0 0>
     }
     tintColor {colorUseRule=UseColor_Specified, specifiedColor=<rgba 1 1 1 1>}
     uVRegion {bIsValid=false, max=<0 0>, min=<0 0>}
    }
    colorAndOpacity=<rgba 1 1 1 0.15>
   }
   slot {class=HorizontalBoxSlot, properties {padding=<ltrb -24 0 0 0>}}
  }
  DrumsIcon {
   widgetClassPath=&Image
   bIsVariable=true
   parent=InstrumentIcons
   widget=DrumsIcon
   properties {
    brush {
     bIsDynamicallyLoaded=false
     drawAs=Image
     imageType=NoImage
     mirroring=NoMirror
     resourceName=None
     resourceObject=/PilgrimQuickplay/Widgets/Materials/MI_UI_Icon_Instrument_Type_Drums
     tiling=NoTile
     imageSize=<142 142>
     margin=<ltrb 0 0 0 0>
     outlineSettings {
      bUseBrushTransparency=false
      roundingType=HalfHeightRadius
      width=0
      color {colorUseRule=UseColor_Specified, specifiedColor=<rgba 0 0 0 0>}
      cornerRadii=<0 0 0 0>
     }
     tintColor {colorUseRule=UseColor_Specified, specifiedColor=<rgba 1 1 1 1>}
     uVRegion {bIsValid=false, max=<0 0>, min=<0 0>}
    }
    colorAndOpacity=<rgba 1 1 1 0.15>
   }
   slot {class=HorizontalBoxSlot, properties {padding=<ltrb -24 0 0 0>}}
  }
 ]
}
