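# Song results widget for Pilgrim Quickplay that shows the band's Instrument Variety and Band Overdrive bonuses.
# On Construct it caches the owning player's Quickplay state component, dims all variety icons and fills both bonus texts from the local player's band data.
# The variety update also lights the icon for each instrument in the player data, with no team filter because the team-comparison Branch is left unconnected.
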WidgetBlueprint /PilgrimQuickplay/Widgets/Results/SongResults/WBP_PilgrimQuickplay_SongResults_BandStats
graphs [
 BrightenInstrumentIcon {nodes [
   n1 = BrightenInstrumentIcon!() -> (then, InstrumentType:EPilgrimTrackType Enum=TrackGuitar) {is=title K2Node_FunctionEntry}
   > BrightenInstrumentIcon!(
    execute
    self:WBP Pilgrim Quickplay Instrument Variety Icons Object Reference=Get InstrumentVarietyIcons() -> (InstrumentVarietyIcons:WBP Pilgrim Quickplay Instrument Variety Icons Object Reference)
    InstrumentType:EPilgrimTrackType Enum=@n1
   ) -> (then) {is=title K2Node_CallFunction}
  ]}
 DimAllInstrumentIcons {nodes [
   DimAllInstrumentIcons() {is=title K2Node_FunctionEntry}
   > DimAllInstrumentIcons(self:WBP Pilgrim Quickplay Instrument Variety Icons Object Reference=Get InstrumentVarietyIcons()) {is=title K2Node_CallFunction}
  ]}
 EventGraph {nodes [
   n1 = Event PreConstruct()
   > Parent: PreConstruct(IsDesignTime=@n1.IsDesignTime)
   Event Construct()
   > n2 = GetPilgrimQuickplayPlayerStateComponent!(
    execute
    PlayerState:Player State Object Reference=Get PlayerState!(self:Controller Object Reference=GetOwningPlayer()) -> (PlayerState:Player State Object Reference)
   ) -> (then, ReturnValue:Pilgrim Quickplay Player State Component Object Reference)
   > Is Valid(InputObject=@n2)
   > Set PQPlayerStateComponent(PQPlayerStateComponent:Pilgrim Quickplay Player State Component Object Reference=@n2) -> (Output_Get:Pilgrim Quickplay Player State Component Object Reference)
   > DimAllInstrumentIcons(self:Self Object Reference)
   > UpdateVarietyScoreText()
   > UpdateOverdriveScoreText()
   n3 = Event Tick()
   > Parent: Tick!(execute, MyGeometry:Geometry Structure=@n3.MyGeometry, InDeltaTime:float=@n3.InDeltaTime) -> (then) {is=title K2Node_CallParentFunction}
  ]}
 UpdateOverdriveScoreText {nodes [
   UpdateOverdriveScoreText() {is=title K2Node_FunctionEntry}
   n1 = Break Pilgrim Quickplay Song Result Data!(PilgrimQuickplaySongResultData:"Pilgrim Quickplay Song Result Data Structure (by ref)"=GetSongResultData!(self:Pilgrim Quickplay Player State Component Object Reference=Get PQPlayerStateComponent() -> (PQPlayerStateComponent:Pilgrim Quickplay Player State Component Object Reference)) -> (ReturnValue:"Pilgrim Quickplay Song Result Data Structure (by ref)")) -> (
    SongName:String
    SongTags:Gameplay Tag Container Structure
    SongShortName:Name
    BandData:Array of Pilgrim Quickplay Song Result Band Data Structures
    PlayerData:Array of Pilgrim Quickplay Song Result Player Data Structures
    Team:Byte
    LocalPlayerBandDataIndex:int
    LocalPlayerDataIndex:int
   ) {is=title K2Node_BreakStruct}
   n2 = Break Pilgrim Quickplay Song Result Band Data!(PilgrimQuickplaySongResultBandData:"Pilgrim Quickplay Song Result Band Data Structure (by ref)"="Get (a copy)"(Array:Array of Pilgrim Quickplay Song Result Band Data Structures=@n1.BandData, Dimension 1=@n1.LocalPlayerBandDataIndex) -> (Output:Pilgrim Quickplay Song Result Band Data Structure)) -> (
    Team:Byte
    NumTeamMembers:Byte
    BandScore:int
    BandModifierBonus:int
    BandInstrumentVarietyBonus:int
    BandStars:int
    BaseBandScore:int
    BandOverdriveBonus:int
   ) {is=title K2Node_BreakStruct}
   > "SetText (Text)"(
    self=Get OverdriveScore() -> (OverdriveScore:Common Text Object Reference)
    InText=Format Text!(Format="LOCTEXT(\"+ {Score}\")", Score:Text="To Text (Integer)"(Value=@n2.BandOverdriveBonus)) -> (Result)
   )
  ]}
 UpdateVarietyScoreText {nodes [
   UpdateVarietyScoreText() {is=title K2Node_FunctionEntry}
   n1 = GetSongResultData!(self=Get PQPlayerStateComponent()) -> (
    ReturnValue_SongName:String
    ReturnValue_SongTags:Gameplay Tag Container Structure
    ReturnValue_SongShortName:Name
    ReturnValue_BandData:Array of Pilgrim Quickplay Song Result Band Data Structures
    ReturnValue_PlayerData:Array of Pilgrim Quickplay Song Result Player Data Structures
    ReturnValue_Team:Byte
    ReturnValue_LocalPlayerBandDataIndex:int
    ReturnValue_LocalPlayerDataIndex:int
   )
   > n2 = For Each Loop!(Exec, Array:Array of Pilgrim Quickplay Song Result Player Data Structures=@n1.ReturnValue_PlayerData) -> (
    LoopBody
    Array Element_PlayerName:String
    Array Element_PlayerAccountId:Unique Net Id Structure
    Array Element_PlayerState:Player State Object Reference
    Array Element_InstrumentType:EPilgrimTrackType Enum
    Array Element_InstrumentTags:Gameplay Tag Container Structure
    Array Element_PlayerStars:int
    Array Element_PlayerScore:int
    Array Element_PlayerDifficulty:EPilgrimSongDifficulty Enum
    Array Element_SongTrackDifficulty:int
    Array Element_Team:bool=~
    Array Index
    Completed
   )
   > BrightenInstrumentIcon(self:Self Object Reference, InstrumentType=@n2.Array Element_InstrumentType)
   n3 = "Get (a copy)"!(Array=@n1.ReturnValue_BandData, Dimension 1=@n1.ReturnValue_LocalPlayerBandDataIndex) -> (
    Output_Team:Byte
    Output_NumTeamMembers:Byte
    Output_BandScore:int
    Output_BandModifierBonus:int
    Output_BandInstrumentVarietyBonus:int
    Output_BandStars:int
    Output_BaseBandScore:int
    Output_BandOverdriveBonus:int
   )
   "SetText (Text)"(
    execute=@n2.Completed
    self=Get VarietyScore() -> (VarietyScore:Common Text Object Reference)
    InText=Format Text(Format="LOCTEXT(\"+ {Score}\")", Score="To Text (Integer)"(Value=@n3.Output_BandInstrumentVarietyBonus))
   )
   Branch(Condition="Equal (Byte)"(A:Byte="To Byte (Boolean)"!(InBool:bool=@n2.Array Element_Team) -> (ReturnValue:Byte=0), B:Byte=@n3) -> (ReturnValue:bool=~) {is=title K2Node_PromotableOperator})
  ]}
]
variables [
 PQPlayerStateComponent {type=PilgrimQuickplayPlayerStateComponent, default=null}
 TotalIndividualScores {type=integer}
]
widgets {
 info {inheritedWidgetCount=0, parentClass=&CommonUserWidget, rootWidgetClass=&VerticalBox, widgetCount=12}
 widgets [
  VerticalBox_0 {widgetClassPath=&VerticalBox, widget=VerticalBox_0}
  VarietyText_1 {
   widgetClassPath=&CommonTextBlock
   parent=VerticalBox_0
   widget=VarietyText_1
   properties {
    text=BONUSES
    textTransformPolicy=ToUpper
    font {
     bForceMonospaced=false
     bMaterialIsStencil=false
     fontMaterial=null
     fontObject=&HeadingNow
     letterSpacing=0
     monospacedWidth=1
     size=15
     skewAmount=0
     typefaceFontName=76 Bold
     outlineSettings {
      bApplyOutlineToDropShadows=false
      bMiteredCorners=false
      bSeparateFillAlpha=false
      outlineMaterial=null
      outlineSize=0
      outlineColor=<rgba 0 0 0 1>
     }
    }
    shadowOffset=<0 0>
   }
   slot {class=VerticalBoxSlot, properties {verticalAlignment=VAlign_Center, size {sizeRule=Fill, value=0.5}}}
  }
  Spacer_1 {widgetClassPath=&Spacer, parent=VerticalBox_0, widget=Spacer_1, properties {size=<1 8>}, slot {class=VerticalBoxSlot}}
  VarietyBox {widgetClassPath=&VerticalBox, parent=VerticalBox_0, widget=VarietyBox, slot {class=VerticalBoxSlot}}
  VarietyText {
   widgetClassPath=&CommonTextBlock
   parent=VarietyBox
   widget=VarietyText
   properties {
    text=Instrument Variety
    colorAndOpacity {colorUseRule=UseColor_Specified, specifiedColor=<rgba 1 1 1 0.8>}
    font {
     bForceMonospaced=false
     bMaterialIsStencil=false
     fontMaterial=null
     fontObject=&HeadingNow
     letterSpacing=0
     monospacedWidth=1
     size=15
     skewAmount=0
     typefaceFontName=74 Regular
     outlineSettings {
      bApplyOutlineToDropShadows=false
      bMiteredCorners=false
      bSeparateFillAlpha=false
      outlineMaterial=null
      outlineSize=0
      outlineColor=<rgba 0 0 0 1>
     }
    }
    shadowOffset=<0 0>
   }
   slot {class=VerticalBoxSlot, properties {verticalAlignment=VAlign_Center, size {sizeRule=Automatic, value=0.5}}}
  }
  ScaleBox_1 {
   widgetClassPath=&ScaleBox
   parent=VarietyBox
   widget=ScaleBox_1
   properties {stretch=UserSpecified, userSpecifiedScale=0.25}
   slot {class=VerticalBoxSlot, properties {horizontalAlignment=HAlign_Left, verticalAlignment=VAlign_Center}}
  }
  InstrumentVarietyIcons {
   widgetClassPath=/PilgrimQuickplay/Widgets/WBP_PilgrimQuickplay_InstrumentVarietyIcons
   bIsVariable=true
   parent=ScaleBox_1
   widget=InstrumentVarietyIcons
   properties {bGColor {colorUseRule=UseColor_Specified, specifiedColor=<rgba 0 0 0 0>}}
   slot {class=ScaleBoxSlot}
  }
  VarietyScore {
   widgetClassPath=&CommonTextBlock
   bIsVariable=true
   parent=VarietyBox
   widget=VarietyScore
   properties {
    text="+ 9,999,999"
    font {
     bForceMonospaced=false
     bMaterialIsStencil=false
     fontMaterial=null
     fontObject=&HeadingNow
     letterSpacing=0
     monospacedWidth=1
     size=15
     skewAmount=0
     typefaceFontName=75 Medium
     outlineSettings {
      bApplyOutlineToDropShadows=false
      bMiteredCorners=false
      bSeparateFillAlpha=false
      outlineMaterial=null
      outlineSize=0
      outlineColor=<rgba 0 0 0 1>
     }
    }
    shadowOffset=<0 0>
   }
   slot {class=VerticalBoxSlot, properties {verticalAlignment=VAlign_Top}}
  }
  Spacer_52 {widgetClassPath=&Spacer, parent=VerticalBox_0, widget=Spacer_52, properties {size=<1 16>}, slot {class=VerticalBoxSlot}}
  OverdriveBox {widgetClassPath=&VerticalBox, parent=VerticalBox_0, widget=OverdriveBox, slot {class=VerticalBoxSlot}}
  OverdriveText {
   widgetClassPath=&CommonTextBlock
   parent=OverdriveBox
   widget=OverdriveText
   properties {
    text=Band Overdrive
    colorAndOpacity {colorUseRule=UseColor_Specified, specifiedColor=<rgba 1 1 1 0.8>}
    font {
     bForceMonospaced=false
     bMaterialIsStencil=false
     fontMaterial=null
     fontObject=&HeadingNow
     letterSpacing=0
     monospacedWidth=1
     size=15
     skewAmount=0
     typefaceFontName=74 Regular
     outlineSettings {
      bApplyOutlineToDropShadows=false
      bMiteredCorners=false
      bSeparateFillAlpha=false
      outlineMaterial=null
      outlineSize=0
      outlineColor=<rgba 0 0 0 1>
     }
    }
    shadowOffset=<0 0>
   }
   slot {class=VerticalBoxSlot, properties {verticalAlignment=VAlign_Center, size {sizeRule=Fill, value=0.5}}}
  }
  OverdriveScore {
   widgetClassPath=&CommonTextBlock
   bIsVariable=true
   parent=OverdriveBox
   widget=OverdriveScore
   properties {
    text="+ 9,999,999"
    font {
     bForceMonospaced=false
     bMaterialIsStencil=false
     fontMaterial=null
     fontObject=&HeadingNow
     letterSpacing=0
     monospacedWidth=1
     size=15
     skewAmount=0
     typefaceFontName=75 Medium
     outlineSettings {
      bApplyOutlineToDropShadows=false
      bMiteredCorners=false
      bSeparateFillAlpha=false
      outlineMaterial=null
      outlineSize=0
      outlineColor=<rgba 0 0 0 1>
     }
    }
    shadowOffset=<0 0>
   }
   slot {class=VerticalBoxSlot, properties {verticalAlignment=VAlign_Top}}
  }
 ]
}
