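# CommonUI widget that shows a curated setlist's artwork, switching between a single preview image, a 2x2 album-art grid and a loading throbber.
# UpdatePreviewImage applies a dynamic texture to both single-image styles (standard and Battle Stage), selectable by index.
# RefreshSongsImages resets placeholders, lazy-loads tile art through SparksAlbumArtManager, and turns the last tile into a "+N" counter on overflow.
# SetCornersForWideDisplay adjusts tile corner radii through dynamic material parameters.
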WidgetBlueprint /SparksUICommon/Widgets/WBP_CuratedSetlistImage
graphs [
 EventGraph {nodes [
   n1 = ShowSingleImageMaterialByIndex!() -> (OutputDelegate, then, MatIndex:int) {is=title K2Node_CustomEvent}
   > SetActiveWidgetIndex(
    self:Common Visibility Switcher Object Reference=Get CommonVisibilitySwitcher_SingleImage() -> (CommonVisibilitySwitcher_SingleImage:Common Visibility Switcher Object Reference)
    Index=@n1.MatIndex
   )
   n2 = UpdatePreviewImage!() -> (OutputDelegate, then, Texture:Texture Object Reference) {is=title K2Node_CustomEvent}
   > Is Valid(InputObject=@n2.Texture)
   > n3 = Cast To Texture2DDynamic!(execute, Object=@n2.Texture) -> (then, CastFailed, AsTexture 2DDynamic:Texture 2DDynamic Object Reference) {is=title K2Node_DynamicCast}
   > SetBrushFromTextureDynamic!(
    execute
    self:Image Object Reference=Get ModeImage_Single() -> (ModeImage_Single:Fort Lazy Image Object Reference)
    Texture:Texture 2DDynamic Object Reference=@n3
    bMatchSize:bool=true
   ) -> (then)
   > SetBrushFromTextureDynamic(self=Get ModeImage_BattleStage() -> (ModeImage_BattleStage:Fort Lazy Image Object Reference), Texture=@n3, bMatchSize=true)
   > SetActiveWidgetIndex(self=Get CommonVisibilitySwitcher_Icons() -> (CommonVisibilitySwitcher_Icons:Common Visibility Switcher Object Reference), Index=0)
   Event PreConstruct()
   n4 = Get ExtraSongsNumber() -> (ExtraSongsNumber:Common Text Object Reference)
   n5 = Get!(self:Text Object Reference=@n4) -> (
    Font_FontObject:Object Reference
    Font_FontMaterial:Object Reference
    Font_OutlineSettings:Font Outline Settings Structure
    Font_TypefaceFontName:Name
    Font_Size:float=~
    Font_LetterSpacing:int
    Font_SkewAmount:float=~
    Font_bForceMonospaced:bool=~
    Font_bMaterialIsStencil:bool=~
    Font_MonospacedWidth:float=~
   )
   > SetFont(
    self=@n4
    InFontInfo=MakeSlateFontInfo!(
     FontObject:Object Reference=@n5
     FontMaterial:Object Reference=@n5.Font_FontMaterial
     OutlineSettings:Font Outline Settings Structure=@n5.Font_OutlineSettings
     TypefaceFontName:Name=@n5.Font_TypefaceFontName
     Size:float=Get ExtraItemsFontSize() -> (ExtraItemsFontSize:float)
     LetterSpacing:int=@n5.Font_LetterSpacing
     SkewAmount:float=@n5.Font_SkewAmount
     bForceMonospaced:bool=@n5.Font_bForceMonospaced
     bMaterialIsStencil:bool=@n5.Font_bMaterialIsStencil
     MonospacedWidth:float=@n5.Font_MonospacedWidth
    ) -> (ReturnValue:Slate Font Info Structure)
   )
   Event Construct()
   > SetActiveWidgetIndex(self=Get CommonVisibilitySwitcher_Icons(), Index=2)
   > n6 = GetDynamicMaterial(self=Get ModeImage_0() -> (ModeImage_0:Fort Lazy Image Object Reference))
   > Set ModeImage0DynamicMat(ModeImage0DynamicMat:Material Instance Dynamic Object Reference=@n6) -> (Output_Get:Material Instance Dynamic Object Reference)
   > n7 = GetDynamicMaterial(self=Get ModeImage_1() -> (ModeImage_1:Fort Lazy Image Object Reference))
   > Set ModeImage1DynamicMat(ModeImage1DynamicMat:Material Instance Dynamic Object Reference=@n7) -> (Output_Get:Material Instance Dynamic Object Reference)
   > n8 = GetDynamicMaterial(self=Get ModeImage_3() -> (ModeImage_3:Fort Lazy Image Object Reference))
   > Set ModeImage3DynamicMat(ModeImage3DynamicMat:Material Instance Dynamic Object Reference=@n8) -> (Output_Get:Material Instance Dynamic Object Reference)
   Event OnInitialized()
   > Set ImagesArray(ImagesArray:Array of Fort Lazy Image Object References=Make Array!(
     "[0]":Fort Lazy Image Object Reference=Get ModeImage_0()
     "[1]":Fort Lazy Image Object Reference=Get ModeImage_1()
     "[2]":Fort Lazy Image Object Reference=Get ModeImage_2() -> (ModeImage_2:Fort Lazy Image Object Reference)
     "[3]":Fort Lazy Image Object Reference=Get ModeImage_3()
    ) -> (Array:Array of Fort Lazy Image Object References)) -> (Output_Get:Array of Fort Lazy Image Object References)
  ]}
 RefreshSongsImages {nodes [
   n1 = RefreshSongsImages!() -> (then, SongList:Array of Names) {is=title K2Node_FunctionEntry}
   > SetVisibility(self=Get ExtraSongsNumber(), InVisibility=Hidden)
   > ResetImages()
   > Branch(Condition=Is Not Empty!(TargetArray:Array of Names=@n1) -> (ReturnValue:bool))
   > SetActiveWidgetIndex(self=Get CommonVisibilitySwitcher_Icons(), Index=1)
   > n2 = For Each Loop(Array:Array of Names=@n1) -> (Array Element:Name)
   n3 = Get ImagesArray() -> (ImagesArray:Array of Fort Lazy Image Object References)
   n4 = Branch(
    execute=@n2.Completed
    Condition="integer > integer"(A=Get LastIndex() -> (LastIndex:int=0), B=Last Index!(TargetArray:Array of Fort Lazy Image Object References=@n3) -> (ReturnValue=0))
   )
   n5 = Last Index(TargetArray=Get ImagesArray()) -> (ReturnValue=0)
   n6 = int + int(A=Get LastIndex() -> (LastIndex=0))
   > "SetText (Text)"(self=Get ExtraSongsNumber(), InText=Format Text!(Format="LOCTEXT(\"+{Number}\")", Number:int=int - int(A=@n6, B=@n5)) -> (Result))
   > SetVisibility(self=Get ExtraSongsNumber(), InVisibility=SelfHitTestInvisible)
   n7 = Get ModeImage_3()
   > SetBrushFromTexture!(
    execute
    self:Image Object Reference=@n7
    Texture:Texture 2D Object Reference=/SparksUICommon/Textures/T_UI_UNL_ManySong
    bMatchSize:bool=false
   ) -> (then)
   > SetOpacity!(execute, self:Image Object Reference=@n7, InOpacity:float=1) -> (then)
   SetVisibility(execute=@n4.else, self=Get ExtraSongsNumber(), InVisibility=Collapsed)
   Set LastIndex(execute=@n2.LoopBody, LastIndex:int=@n2.Array Index) -> (Output_Get:int=0)
   n8 = Last Index(TargetArray=Get ImagesArray()) -> (ReturnValue=0)
   > Branch(Condition=NOT Boolean(A="integer > integer"(A=@n2.Array Index, B=@n8)))
   n9 = Get ImagesArray()
   > n10 = Branch(Condition="Not Equal (Integer)"(A=@n2.Array Index, B=Last Index(TargetArray=@n9) -> (ReturnValue=0)))
   SetTextureForItemAtIndex!(execute=[@n10.then, @n12.then], self, Index:int=@n2.Array Index, SongShortName:Name=@n2) -> (then)
   n11 = Get ImagesArray()
   n12 = Branch(
    execute=@n10.else
    Condition="integer <= integer"(A=Length(TargetArray:Array of Names=@n1), B=Length(TargetArray:Array of Fort Lazy Image Object References=@n11))
   )
   SetBrushFromTexture(execute=@n12.else, self=Get ModeImage_3(), Texture=/Game/Environments/Sets/ArtDeco/Architecture/Textures/AD_Swatch_Grey, bMatchSize=false)
   "integer >= integer"(A=Get LastIndex() -> (LastIndex=0), B=Last Index(TargetArray=Get ImagesArray()) -> (ReturnValue=0))
  ]}
 ResetImages {nodes [
   ResetImages() {is=title K2Node_FunctionEntry}
   > n1 = For Each Loop(Array:Array of Fort Lazy Image Object References=Get ImagesArray()) -> (Array Element:Fort Lazy Image Object Reference)
   Return Node(execute=@n1.Completed)
   > SetBrushFromTexture(
    self=@n1
    Texture=Select!(
     Option 0:Texture 2D Object Reference=/SparksUICommon/Textures/T_UI_UNL_EmptySong_01
     Option 1:Texture 2D Object Reference=/SparksUICommon/Textures/T_UI_UNL_EmptySong_00
     Option 2:Texture 2D Object Reference=&T_UI_UNL_EmptySong_00
     Option 3:Texture 2D Object Reference=&T_UI_UNL_EmptySong_01
     Index:int=@n1.Array Index
    ) -> (ReturnValue:Texture 2D Object Reference)
    bMatchSize=false
   )
  ]}
 SetCornersForWideDisplay {nodes [
   n1 = SetCornersForWideDisplay!() -> (then, isWide:bool=~) {is=title K2Node_FunctionEntry}
   > SetScalarParameterValue(
    self=Get ModeImage0DynamicMat() -> (ModeImage0DynamicMat:Material Instance Dynamic Object Reference)
    ParameterName=Corner_TopLeft
    Value=SelectFloat(A=Get ExpandedBluntCornerRadius() -> (ExpandedBluntCornerRadius:double), B=Get BluntCornerRadius() -> (BluntCornerRadius:double), bPickA=@n1)
   )
   > SetScalarParameterValue(
    self=Get ModeImage1DynamicMat() -> (ModeImage1DynamicMat:Material Instance Dynamic Object Reference)
    ParameterName=Corner_TopRight
    Value=SelectFloat(A=0, B=Get BluntCornerRadius(), bPickA=@n1)
   )
   > SetScalarParameterValue(
    self=Get ModeImage3DynamicMat() -> (ModeImage3DynamicMat:Material Instance Dynamic Object Reference)
    ParameterName=Corner_BottomRight
    Value=SelectFloat(A=0, B=Get DefaultCornerRadius() -> (DefaultCornerRadius:double), bPickA=@n1)
   )
  ]}
 SetTextureForItemAtIndex {nodes [
   n1 = GetDynamicMaterial()
   > SetTextureParameterValue(self=@n1, ParameterName=Album1)
   n2 = SetTextureForItemAtIndex!() -> (then, Index:int, SongShortName:Name) {is=title K2Node_FunctionEntry}
   > Branch(Condition=Is Valid Index!(TargetArray:Array of Fort Lazy Image Object References=Get ImagesArray(), IndexToTest:int=@n2) -> (ReturnValue:bool))
   n3 = "Get (a copy)"(Array:Array of Fort Lazy Image Object References=Get ImagesArray(), Dimension 1=@n2) -> (Output:Fort Lazy Image Object Reference)
   > RequestLazyAlbumArt!(
    execute
    AlbumArtManager:Sparks Album Art Manager Object Reference=Get SparksAlbumArtManager() -> (ReturnValue:Sparks Album Art Manager Object Reference) {is=title K2Node_GetSubsystem}
    SongShortName:"Name (by ref)"=@n2.SongShortName
    LazyImage:Common Lazy Image Object Reference=@n3
   ) -> (then, ReturnValue:bool)
   > SetOpacity(self=@n3, InOpacity=1)
  ]}
]
variables [
 BluntCornerRadius {type=number, default=0.2}
 DefaultCornerRadius {type=number, default=0.15}
 ExpandedBluntCornerRadius {type=number, default=0.2}
 ExtraItemsFontSize {type=number, default=64}
 ImagesArray {type="array<FortLazyImage>", default []}
 ModeImage0DynamicMat {type=MaterialInstanceDynamic}
 ModeImage1DynamicMat {type=MaterialInstanceDynamic}
 ModeImage3DynamicMat {type=MaterialInstanceDynamic}
]
widgets {
 info {
  inheritedWidgetCount=0
  parentClass=&CommonUserWidget
  rootWidgetClass=&CommonVisibilitySwitcher
  widgetCount=12
 }
 widgets [
  CommonVisibilitySwitcher_Icons {
   widgetClassPath=&CommonVisibilitySwitcher
   bIsVariable=true
   widget=CommonVisibilitySwitcher_Icons
   properties {activeWidgetIndex=1}
  }
  CommonVisibilitySwitcher_SingleImage {
   widgetClassPath=&CommonVisibilitySwitcher
   bIsVariable=true
   parent=CommonVisibilitySwitcher_Icons
   widget=CommonVisibilitySwitcher_SingleImage
   slot {class=CommonVisibilitySwitcherSlot}
  }
  ModeImage_Single {
   widgetClassPath=&FortLazyImage
   bIsVariable=true
   parent=CommonVisibilitySwitcher_SingleImage
   widget=ModeImage_Single
   properties {
    materialTextureParamName=T_AlbumArt
    brush {
     bIsDynamicallyLoaded=false
     drawAs=Image
     imageType=FullColor
     mirroring=NoMirror
     resourceName=None
     resourceObject=/SparksUICommon/Materials/MI_UI_Sparks_SetList_AlbumArt_Rounded
     tiling=NoTile
     imageSize=<512 512>
     margin=<ltrb 0 0 0 0>
     outlineSettings {
      bUseBrushTransparency=false
      roundingType=HalfHeightRadius
      width=0
      color {colorUseRule=UseColor_Specified, specifiedColor=<rgba 0 0 0 0>}
      cornerRadii=<0 0 0 0>
     }
     tintColor {colorUseRule=UseColor_Specified, specifiedColor=<rgba 1 1 1 1>}
     uVRegion {bIsValid=false, max=<0 0>, min=<0 0>}
    }
    loadingThrobberBrush {
     bIsDynamicallyLoaded=false
     drawAs=Image
     imageType=FullColor
     mirroring=NoMirror
     resourceName=None
     resourceObject=/FortUILibrary/UIKit/Throbber/MI_UIKit_Monochrome_Throbber
     tiling=NoTile
     imageSize=<32 32>
     margin=<ltrb 0 0 0 0>
     outlineSettings {
      bUseBrushTransparency=false
      roundingType=HalfHeightRadius
      width=0
      color {colorUseRule=UseColor_Specified, specifiedColor=<rgba 0 0 0 0>}
      cornerRadii=<0 0 0 0>
     }
     tintColor {colorUseRule=UseColor_Specified, specifiedColor=<rgba 1 1 1 1>}
     uVRegion {bIsValid=false, max=<0 0>, min=<0 0>}
    }
   }
   slot {class=CommonVisibilitySwitcherSlot}
  }
  ModeImage_BattleStage {
   widgetClassPath=&FortLazyImage
   bIsVariable=true
   parent=CommonVisibilitySwitcher_SingleImage
   widget=ModeImage_BattleStage
   properties {
    materialTextureParamName=T_AlbumArt
    brush {
     bIsDynamicallyLoaded=false
     drawAs=Image
     imageType=FullColor
     mirroring=NoMirror
     resourceName=None
     resourceObject=/SparksUICommon/Materials/MI_UI_Sparks_CuratedSetList_BattleStage
     tiling=NoTile
     imageSize=<350 415>
     margin=<ltrb 0 0 0 0>
     outlineSettings {
      bUseBrushTransparency=false
      roundingType=HalfHeightRadius
      width=0
      color {colorUseRule=UseColor_Specified, specifiedColor=<rgba 0 0 0 0>}
      cornerRadii=<0 0 0 0>
     }
     tintColor {colorUseRule=UseColor_Specified, specifiedColor=<rgba 1 1 1 1>}
     uVRegion {bIsValid=false, max=<0 0>, min=<0 0>}
    }
    loadingThrobberBrush {
     bIsDynamicallyLoaded=false
     drawAs=Image
     imageType=FullColor
     mirroring=NoMirror
     resourceName=None
     resourceObject=&MI_UIKit_Monochrome_Throbber
     tiling=NoTile
     imageSize=<32 32>
     margin=<ltrb 0 0 0 0>
     outlineSettings {
      bUseBrushTransparency=false
      roundingType=HalfHeightRadius
      width=0
      color {colorUseRule=UseColor_Specified, specifiedColor=<rgba 0 0 0 0>}
      cornerRadii=<0 0 0 0>
     }
     tintColor {colorUseRule=UseColor_Specified, specifiedColor=<rgba 1 1 1 1>}
     uVRegion {bIsValid=false, max=<0 0>, min=<0 0>}
    }
   }
   slot {class=CommonVisibilitySwitcherSlot}
  }
  GridPanel_AlbumGrid {
   widgetClassPath=&GridPanel
   parent=CommonVisibilitySwitcher_Icons
   widget=GridPanel_AlbumGrid
   properties {columnFill [1,1], rowFill [1,1]}
   slot {class=CommonVisibilitySwitcherSlot}
  }
  ModeImage_3 {
   widgetClassPath=&FortLazyImage
   bIsVariable=true
   parent=GridPanel_AlbumGrid
   widget=ModeImage_3
   properties {
    materialTextureParamName=T_AlbumArt
    brush {
     bIsDynamicallyLoaded=false
     drawAs=Box
     imageType=FullColor
     mirroring=NoMirror
     resourceName=None
     resourceObject=/SparksUICommon/Materials/MI_UI_Sparks_SetList_AlbumArt_3
     tiling=NoTile
     imageSize=<32 32>
     margin=<ltrb 0 0 0 0>
     outlineSettings {
      bUseBrushTransparency=false
      roundingType=HalfHeightRadius
      width=1
      color {colorUseRule=UseColor_Specified, specifiedColor=<rgba 0 0 0 0>}
      cornerRadii=<0 0 0 0>
     }
     tintColor {colorUseRule=UseColor_Specified, specifiedColor=<rgba 1 1 1 1>}
     uVRegion {bIsValid=false, max=<0 0>, min=<0 0>}
    }
    loadingBackgroundBrush {
     bIsDynamicallyLoaded=false
     drawAs=RoundedBox
     imageType=NoImage
     mirroring=NoMirror
     resourceName=None
     resourceObject=null
     tiling=NoTile
     imageSize=<32 32>
     margin=<ltrb 0 0 0 0>
     outlineSettings {
      bUseBrushTransparency=false
      roundingType=HalfHeightRadius
      width=0
      color {colorUseRule=UseColor_Specified, specifiedColor=<rgba 0 0 0 0>}
      cornerRadii=<0 0 0 0>
     }
     tintColor {colorUseRule=UseColor_Specified, specifiedColor=<rgba 1 1 1 1>}
     uVRegion {bIsValid=false, max=<0 0>, min=<0 0>}
    }
   }
   slot {class=GridSlot, properties {column=1, columnSpan=1, row=1, rowSpan=1}}
  }
  ModeImage_2 {
   widgetClassPath=&FortLazyImage
   bIsVariable=true
   parent=GridPanel_AlbumGrid
   widget=ModeImage_2
   properties {
    materialTextureParamName=T_AlbumArt
    brush {
     bIsDynamicallyLoaded=false
     drawAs=Image
     imageType=FullColor
     mirroring=NoMirror
     resourceName=None
     resourceObject=/SparksUICommon/Materials/MI_UI_Sparks_SetList_AlbumArt_2
     tiling=NoTile
     imageSize=<32 32>
     margin=<ltrb 0 0 0 0>
     outlineSettings {
      bUseBrushTransparency=false
      roundingType=HalfHeightRadius
      width=0
      color {colorUseRule=UseColor_Specified, specifiedColor=<rgba 0 0 0 0>}
      cornerRadii=<0 0 0 0>
     }
     tintColor {colorUseRule=UseColor_Specified, specifiedColor=<rgba 1 1 1 1>}
     uVRegion {bIsValid=false, max=<0 0>, min=<0 0>}
    }
   }
   slot {class=GridSlot, properties {columnSpan=1, row=1, rowSpan=1}}
  }
  ModeImage_1 {
   widgetClassPath=&FortLazyImage
   bIsVariable=true
   parent=GridPanel_AlbumGrid
   widget=ModeImage_1
   properties {
    materialTextureParamName=T_AlbumArt
    brush {
     bIsDynamicallyLoaded=false
     drawAs=Image
     imageType=FullColor
     mirroring=NoMirror
     resourceName=None
     resourceObject=/SparksUICommon/Materials/MI_UI_Sparks_SetList_AlbumArt_1
     tiling=NoTile
     imageSize=<32 32>
     margin=<ltrb 0 0 0 0>
     outlineSettings {
      bUseBrushTransparency=false
      roundingType=HalfHeightRadius
      width=0
      color {colorUseRule=UseColor_Specified, specifiedColor=<rgba 0 0 0 0>}
      cornerRadii=<0 0 0 0>
     }
     tintColor {colorUseRule=UseColor_Specified, specifiedColor=<rgba 1 1 1 1>}
     uVRegion {bIsValid=false, max=<0 0>, min=<0 0>}
    }
   }
   slot {class=GridSlot, properties {column=1, columnSpan=1, rowSpan=1}}
  }
  ModeImage_0 {
   widgetClassPath=&FortLazyImage
   bIsVariable=true
   parent=GridPanel_AlbumGrid
   widget=ModeImage_0
   properties {
    materialTextureParamName=T_AlbumArt
    brush {
     bIsDynamicallyLoaded=false
     drawAs=Image
     imageType=FullColor
     mirroring=NoMirror
     resourceName=None
     resourceObject=/SparksUICommon/Materials/MI_UI_Sparks_SetList_AlbumArt_0
     tiling=NoTile
     imageSize=<512 512>
     margin=<ltrb 0 0 0 0>
     outlineSettings {
      bUseBrushTransparency=false
      roundingType=HalfHeightRadius
      width=0
      color {colorUseRule=UseColor_Specified, specifiedColor=<rgba 0 0 0 0>}
      cornerRadii=<0 0 0 0>
     }
     tintColor {colorUseRule=UseColor_Specified, specifiedColor=<rgba 1 1 1 1>}
     uVRegion {bIsValid=false, max=<0 0>, min=<0 0>}
    }
    loadingBackgroundBrush {
     bIsDynamicallyLoaded=false
     drawAs=Image
     imageType=FullColor
     mirroring=NoMirror
     resourceName=None
     resourceObject=null
     tiling=NoTile
     imageSize=<32 32>
     margin=<ltrb 0 0 0 0>
     outlineSettings {
      bUseBrushTransparency=false
      roundingType=HalfHeightRadius
      width=0
      color {colorUseRule=UseColor_Specified, specifiedColor=<rgba 0 0 0 0>}
      cornerRadii=<0 0 0 0>
     }
     tintColor {colorUseRule=UseColor_Specified, specifiedColor=<rgba 1 1 1 1>}
     uVRegion {bIsValid=false, max=<0 0>, min=<0 0>}
    }
   }
   slot {class=GridSlot, properties {columnSpan=1, rowSpan=1}}
  }
  ExtraSongsNumber {
   widgetClassPath=&CommonTextBlock
   bIsVariable=true
   parent=GridPanel_AlbumGrid
   widget=ExtraSongsNumber
   properties {
    clipping=ClipToBounds
    scrollStyle=&TextScrollStyle-Base_C
    text=+2
    visibility=Hidden
    colorAndOpacity {colorUseRule=UseColor_Specified, specifiedColor=<rgba 0.4852 0.7011 1 1>}
    font {
     bForceMonospaced=false
     bMaterialIsStencil=false
     fontMaterial=null
     fontObject=&HeadingNow
     letterSpacing=0
     monospacedWidth=1
     size=41.25
     skewAmount=0
     typefaceFontName=76 Bold
     outlineSettings {
      bApplyOutlineToDropShadows=false
      bMiteredCorners=false
      bSeparateFillAlpha=false
      outlineMaterial=null
      outlineSize=0
      outlineColor=<rgba 0 0 0 1>
     }
    }
    shadowOffset=<0 0>
   }
   slot {class=GridSlot, properties {column=1, horizontalAlignment=HAlign_Center, row=1, verticalAlignment=VAlign_Center}}
  }
  GridPanel_Loading {
   widgetClassPath=&GridPanel
   bIsVariable=true
   parent=CommonVisibilitySwitcher_Icons
   widget=GridPanel_Loading
   slot {class=CommonVisibilitySwitcherSlot}
  }
  Loading {
   widgetClassPath=&Image
   bIsVariable=true
   parent=GridPanel_Loading
   widget=Loading
   properties {brush {
     bIsDynamicallyLoaded=false
     drawAs=Image
     imageType=FullColor
     mirroring=NoMirror
     resourceName=None
     resourceObject=/SparksUICommon/Materials/MI_UI_Throbber_Sparks
     tiling=NoTile
     imageSize=<128 128>
     margin=<ltrb 0 0 0 0>
     outlineSettings {
      bUseBrushTransparency=false
      roundingType=HalfHeightRadius
      width=0
      color {colorUseRule=UseColor_Specified, specifiedColor=<rgba 0 0 0 0>}
      cornerRadii=<0 0 0 0>
     }
     tintColor {colorUseRule=UseColor_Specified, specifiedColor=<rgba 1 1 1 1>}
     uVRegion {bIsValid=false, max=<0 0>, min=<0 0>}
    }}
   slot {class=GridSlot, properties {horizontalAlignment=HAlign_Center, verticalAlignment=VAlign_Center}}
  }
 ]
}
