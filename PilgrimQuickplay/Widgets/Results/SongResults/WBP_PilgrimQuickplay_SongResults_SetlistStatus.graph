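# Song results widget for Pilgrim Quickplay showing a custom setlist's title, artwork and completion status.
# Setting the title also rebuilds the artwork from song short names in the playspace's custom setlist.
# ShowSetlistComplete reveals the COMPLETE row only when the band state has no next song, but plays the staggered ring-pulse animation regardless.
# PreConstruct enlarges the artwork for touch input, and a separate event toggles title and overlay visibility.
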WidgetBlueprint /PilgrimQuickplay/Widgets/Results/SongResults/WBP_PilgrimQuickplay_SongResults_SetlistStatus
animations [Anim_CompleteSetlist {
  binding_count=3
  length_seconds=1.65
  object=&self:Anim_CompleteSetlist
  sequence {bindings [
    Intro_Ring_Pulse_2 {
     id=#1
     possessed_class=Image
     tracks [
      {
       type=MovieSceneWidgetMaterialTrack
       sections [{type=MovieSceneParameterSection, channels [TaperGlow {type=MovieSceneScriptingFloatChannel, keys [[11.8463,1],[28,4]]}]}]
      }
      {
       type=MovieSceneByteTrack
       display_name=Visibility
       sections [{type=MovieSceneByteSection, channels [None {type=MovieSceneScriptingByteChannel, keys [[8,4],[28,1]]}]}]
      }
      {
       type=MovieScene2DTransformTrack
       sections [{
         type=MovieScene2DTransformSection
         channels [
          Scale.X {type=MovieSceneScriptingFloatChannel, keys [[8,1],[28,1.2]]}
          Scale.Y {type=MovieSceneScriptingFloatChannel, keys [[8,1],[28,1.8]]}
         ]
        }]
      }
      {
       type=MovieSceneFloatTrack
       display_name=Render Opacity
       sections [{type=MovieSceneFloatSection, channels [None {type=MovieSceneScriptingFloatChannel, keys [[8,0],[10.0363,1],[28,0]]}]}]
      }
     ]
    }
    Intro_Ring_Pulse_3 {
     id=#2
     possessed_class=Image
     tracks [
      {
       type=MovieSceneWidgetMaterialTrack
       sections [{type=MovieSceneParameterSection, channels [TaperGlow {type=MovieSceneScriptingFloatChannel, keys [[20.0767,1],[33,4]]}]}]
      }
      {
       type=MovieSceneByteTrack
       display_name=Visibility
       sections [{type=MovieSceneByteSection, channels [None {type=MovieSceneScriptingByteChannel, keys [[17,4],[33,1]]}]}]
      }
      {
       type=MovieScene2DTransformTrack
       sections [{
         type=MovieScene2DTransformSection
         channels [
          Scale.X {type=MovieSceneScriptingFloatChannel, keys [[17,1],[33,1.2]]}
          Scale.Y {type=MovieSceneScriptingFloatChannel, keys [[17,1],[33,1.8]]}
         ]
        }]
      }
      {
       type=MovieSceneFloatTrack
       display_name=Render Opacity
       sections [{type=MovieSceneFloatSection, channels [None {type=MovieSceneScriptingFloatChannel, keys [[17,0],[18.6287,1],[33,0]]}]}]
      }
     ]
    }
    Intro_Ring_Pulse_1 {
     id=#3
     possessed_class=Image
     tracks [
      {
       type=MovieSceneByteTrack
       display_name=Visibility
       sections [{type=MovieSceneByteSection, channels [None {type=MovieSceneScriptingByteChannel, keys [[0,4],[20,1]]}]}]
      }
      {
       type=MovieSceneFloatTrack
       display_name=Render Opacity
       sections [{type=MovieSceneFloatSection, channels [None {type=MovieSceneScriptingFloatChannel, keys [[0,0],[2.036,1],[20,0]]}]}]
      }
      {
       type=MovieScene2DTransformTrack
       sections [{
         type=MovieScene2DTransformSection
         channels [
          Scale.X {type=MovieSceneScriptingFloatChannel, keys [[0,1],[20,1.2]]}
          Scale.Y {type=MovieSceneScriptingFloatChannel, keys [[0,1],[20,1.8]]}
         ]
        }]
      }
      {
       type=MovieSceneWidgetMaterialTrack
       sections [{type=MovieSceneParameterSection, channels [TaperGlow {type=MovieSceneScriptingFloatChannel, keys [[3.846,1],[20,4]]}]}]
      }
     ]
    }
   ]}
 }]
graphs [
 EventGraph {nodes [
   n1 = ShowSetlistComplete!() -> (OutputDelegate, then, ShouldShow:bool=~) {is=title K2Node_CustomEvent}
   n2 = isFinalSongResults!() -> (ReturnValue:bool)
   > SetVisibility(
    self=Get HB_Complete() -> (HB_Complete:Horizontal Box Object Reference)
    InVisibility=Select(Option 0=Hidden, Option 1=HitTestInvisible, Index=AND Boolean(A=@n1.ShouldShow, B=@n2))
   )
   > PlayAnimation(InAnimation=Get Anim_CompleteSetlist() -> (Anim_CompleteSetlist:Widget Animation Object Reference), PlaybackSpeed=1.2)
   n3 = ShowCustomSetlistTitle!() -> (OutputDelegate, then, ShouldShow:bool=~) {is=title K2Node_CustomEvent}
   > SetVisibility(
    self=[Get TextBlock_Title() -> (TextBlock_Title:Common Text Object Reference), Get SetlistOverlay() -> (SetlistOverlay:Overlay Object Reference)]
    InVisibility=Select(Option 0=Hidden, Option 1=SelfHitTestInvisible, Index=@n3.ShouldShow)
   )
   n4 = SetCustomSetlistTitle!() -> (OutputDelegate, then, TitleText:Text) {is=title K2Node_CustomEvent}
   > "SetText (Text)"(self=Get TextBlock_Title(), InText=@n4.TitleText)
   > RefreshSongs()
   RefreshSongs() {is=title K2Node_CustomEvent}
   > Clear(TargetArray:Array of Names=Get SetlistSongNames() -> (SetlistSongNames:Array of Names))
   > ResetImages(self:WBP Curated Setlist Image Object Reference=Get WBP_CuratedSetlistImage() -> (WBP_CuratedSetlistImage:WBP Curated Setlist Image Object Reference))
   > n5 = Cast To BP_PilgrimQuickplayPlayspace!(
    execute
    Object=GetSparksMusicPlayspaceOfClass!(
     PlayspaceType:Sparks Music Playspace Class Reference=/Script/PilgrimQuickplayRuntime.PilgrimQuickplayPlayspace
     Actor:Actor Object Reference=GetOwningPlayer()
     bMatchChildTypes:bool=false
    ) -> (ReturnValue:Pilgrim Quickplay Playspace Object Reference)
   ) -> (then, CastFailed, AsBP Pilgrim Quickplay Playspace:BP Pilgrim Quickplay Playspace Object Reference) {is=title K2Node_DynamicCast}
   > n6 = For Each Loop(Array:Array of Pilgrim Song Metadata Object References=GetSongList!(self:Pilgrim Setlist Interface Interface=Get PilgrimCustomSetlist!(self:BP Pilgrim Quickplay Playspace Object Reference=@n5) -> (PilgrimCustomSetlist:Pilgrim Custom Setlist Object Reference)) -> (ReturnValue:Array of Pilgrim Song Metadata Object References)) -> (Array Element:Pilgrim Song Metadata Object Reference)
   RefreshSongsImages!(
    execute=@n6.Completed
    self:WBP Curated Setlist Image Object Reference=Get WBP_CuratedSetlistImage()
    SongList:Array of Names=Get SetlistSongNames()
   ) -> (then)
   Add(
    execute=@n6.LoopBody
    TargetArray:Array of Names=Get SetlistSongNames()
    NewItem:"Name (by ref)"=GetShortName!(self:Sparks Song Data Object Reference=@n6) -> (ReturnValue:Name=None)
   ) {is=title K2Node_CallArrayFunction}
   Event PreConstruct()
   > n7 = Branch(Condition=IsUsingInputType!(InputType:ECommonInputType Enum=Touch, PlayerController:Player Controller Object Reference=GetOwningPlayer()) -> (ReturnValue:bool))
   > SetImageSize!(execute, self, Size:double=Get MobileImageSize() -> (MobileImageSize:double)) -> (then)
   SetImageSize(execute=@n7.else, Size=Get DefaultImageSize() -> (DefaultImageSize:double))
  ]}
 SetImageSize {nodes [
   n1 = SetImageSize!() -> (then, Size:double=~) {is=title K2Node_FunctionEntry}
   n2 = Get SizeBox_0() -> (SizeBox_0:Size Box Object Reference)
   > SetWidthOverride(self=@n2, InWidthOverride=@n1)
   > SetHeightOverride(self=@n2, InHeightOverride=@n1)
  ]}
 isFinalSongResults {nodes [
   isFinalSongResults() {is=title K2Node_FunctionEntry}
   n1 = GetOwningPlayer()
   n2 = GetNextSong!(self:Pilgrim Core Band State Component Object Reference=GetBandState!(ContextActor:Actor Object Reference=GetSparksMusicPlayspaceOfClass(PlayspaceType=&PilgrimQuickplayPlayspace, Actor=@n1, bMatchChildTypes=false)) -> (ReturnValue:Pilgrim Quickplay Band State Component Object Reference)) -> (ReturnValue:Pilgrim Song Metadata Object Reference)
   > Return Node!(execute, ReturnValue:bool=NOT Boolean(A=IsValid(Object=@n2)))
  ]}
]
variables [
 DefaultImageSize {type=number, default=72}
 MobileImageSize {type=number, default=128}
 SetlistSongNames {type="array<string>", default []}
]
widgets {
 info {inheritedWidgetCount=0, parentClass=&CommonUserWidget, rootWidgetClass=&Overlay, widgetCount=14}
 widgets [
  SetlistOverlay {widgetClassPath=&Overlay, bIsVariable=true, widget=SetlistOverlay}
  Intro_Ring_Pulse_1 {
   widgetClassPath=&Image
   bIsVariable=true
   parent=SetlistOverlay
   widget=Intro_Ring_Pulse_1
   properties {
    visibility=Collapsed
    brush {
     bIsDynamicallyLoaded=false
     drawAs=Box
     imageType=FullColor
     mirroring=NoMirror
     resourceName=None
     resourceObject=/PilgrimCore/Widgets/Materials/MI_UI_Piligrim_GemBreak_PulseGreen
     tiling=NoTile
     imageSize=<32 32>
     margin=<ltrb 0.5 0.5 0.5 0.5>
     outlineSettings {
      bUseBrushTransparency=false
      roundingType=HalfHeightRadius
      width=0
      color {colorUseRule=UseColor_Specified, specifiedColor=<rgba 0 0 0 0>}
      cornerRadii=<0 0 0 0>
     }
     tintColor {colorUseRule=UseColor_Specified, specifiedColor=<rgba 1 1 1 1>}
     uVRegion {bIsValid=false, max=<0 0>, min=<0 0>}
    }
   }
   slot {class=OverlaySlot, properties {horizontalAlignment=HAlign_Fill, verticalAlignment=VAlign_Fill}}
  }
  Intro_Ring_Pulse_2 {
   widgetClassPath=&Image
   bIsVariable=true
   parent=SetlistOverlay
   widget=Intro_Ring_Pulse_2
   properties {
    visibility=Collapsed
    brush {
     bIsDynamicallyLoaded=false
     drawAs=Box
     imageType=FullColor
     mirroring=NoMirror
     resourceName=None
     resourceObject=&MI_UI_Piligrim_GemBreak_PulseGreen
     tiling=NoTile
     imageSize=<32 32>
     margin=<ltrb 0.5 0.5 0.5 0.5>
     outlineSettings {
      bUseBrushTransparency=false
      roundingType=HalfHeightRadius
      width=0
      color {colorUseRule=UseColor_Specified, specifiedColor=<rgba 0 0 0 0>}
      cornerRadii=<0 0 0 0>
     }
     tintColor {colorUseRule=UseColor_Specified, specifiedColor=<rgba 1 1 1 1>}
     uVRegion {bIsValid=false, max=<0 0>, min=<0 0>}
    }
   }
   slot {class=OverlaySlot, properties {horizontalAlignment=HAlign_Fill, verticalAlignment=VAlign_Fill}}
  }
  Intro_Ring_Pulse_3 {
   widgetClassPath=&Image
   bIsVariable=true
   parent=SetlistOverlay
   widget=Intro_Ring_Pulse_3
   properties {
    visibility=Collapsed
    brush {
     bIsDynamicallyLoaded=false
     drawAs=Box
     imageType=FullColor
     mirroring=NoMirror
     resourceName=None
     resourceObject=&MI_UI_Piligrim_GemBreak_PulseGreen
     tiling=NoTile
     imageSize=<32 32>
     margin=<ltrb 0.5 0.5 0.5 0.5>
     outlineSettings {
      bUseBrushTransparency=false
      roundingType=HalfHeightRadius
      width=0
      color {colorUseRule=UseColor_Specified, specifiedColor=<rgba 0 0 0 0>}
      cornerRadii=<0 0 0 0>
     }
     tintColor {colorUseRule=UseColor_Specified, specifiedColor=<rgba 1 1 1 1>}
     uVRegion {bIsValid=false, max=<0 0>, min=<0 0>}
    }
   }
   slot {class=OverlaySlot, properties {horizontalAlignment=HAlign_Fill, verticalAlignment=VAlign_Fill}}
  }
  WBP_Sparks_BackingPanel {
   widgetClassPath=/PilgrimCore/Widgets/SparksUILibrary/WBP_Sparks_BackingPanel
   bIsVariable=true
   parent=SetlistOverlay
   widget=WBP_Sparks_BackingPanel
   properties {blurFrame=false, cornerRadius=Large, darkenLayer=false, eSparksBackingPanelType=Simple Color}
   slot {class=OverlaySlot, properties {horizontalAlignment=HAlign_Fill, verticalAlignment=VAlign_Fill}}
  }
  HorizontalBox_0 {
   widgetClassPath=&HorizontalBox
   parent=SetlistOverlay
   widget=HorizontalBox_0
   slot {class=OverlaySlot, properties {horizontalAlignment=HAlign_Fill, verticalAlignment=VAlign_Fill, padding=<ltrb 8 8 8 8>}}
  }
  SizeBox_0 {
   widgetClassPath=&SizeBox
   bIsVariable=true
   parent=HorizontalBox_0
   widget=SizeBox_0
   properties {
    bOverride_HeightOverride=true
    bOverride_WidthOverride=true
    heightOverride=128
    maxDesiredHeight=150
    maxDesiredWidth=150
    minDesiredHeight=72
    minDesiredWidth=72
    widthOverride=128
   }
   slot {class=HorizontalBoxSlot, properties {horizontalAlignment=HAlign_Center, verticalAlignment=VAlign_Center}}
  }
  WBP_CuratedSetlistImage {
   widgetClassPath=/SparksUICommon/Widgets/WBP_CuratedSetlistImage
   bIsVariable=true
   parent=SizeBox_0
   widget=WBP_CuratedSetlistImage
   properties {bOverride_Cursor=true, extraItemsFontSize=12}
   slot {class=SizeBoxSlot}
  }
  VerticalBox_0 {
   widgetClassPath=&VerticalBox
   parent=HorizontalBox_0
   widget=VerticalBox_0
   properties {bOverride_Cursor=true}
   slot {class=HorizontalBoxSlot, properties {verticalAlignment=VAlign_Center, padding=<ltrb 16 0 0 0>}}
  }
  CanvasPanel_0 {
   widgetClassPath=&CanvasPanel
   parent=VerticalBox_0
   widget=CanvasPanel_0
   slot {class=VerticalBoxSlot, properties {size {sizeRule=Fill, value=1}}}
  }
  TextBlock_Title {
   widgetClassPath=&CommonTextBlock
   bIsVariable=true
   parent=CanvasPanel_0
   widget=TextBlock_Title
   properties {
    scrollStyle=/PilgrimQuickplay/Widgets/ScrollStyles/TextScrollStyle_PilgrimQuickplay_Setlist.TextScrollStyle_PilgrimQuickplay_Setlist_C
    text=Setlist Name Example is long
    textTransformPolicy=ToUpper
    visibility=Hidden
    font {
     bForceMonospaced=false
     bMaterialIsStencil=false
     fontMaterial=null
     fontObject=&Roboto
     letterSpacing=0
     monospacedWidth=1
     size=36
     skewAmount=0
     typefaceFontName=Black
     outlineSettings {
      bApplyOutlineToDropShadows=false
      bMiteredCorners=false
      bSeparateFillAlpha=false
      outlineMaterial=null
      outlineSize=0
      outlineColor=<rgba 0 0 0 1>
     }
    }
    shadowOffset=<0 0>
   }
   slot {class=CanvasPanelSlot, properties {layoutData {alignment=<0 0>, anchors {maximum=<1 0>, minimum=<0 0>}, offsets=<ltrb 0 0 0 30>}}}
  }
  HB_Complete {
   widgetClassPath=&HorizontalBox
   bIsVariable=true
   parent=VerticalBox_0
   widget=HB_Complete
   properties {visibility=Collapsed}
   slot {class=VerticalBoxSlot, properties {padding=<ltrb 0 10 0 0>}}
  }
  StatusIndicator {
   widgetClassPath=/FortUILibrary/UIKit/StatusIndicator/WBP_UIKit_StatusIndicator
   bIsVariable=true
   parent=HB_Complete
   widget=StatusIndicator
   properties {icon=/FortUILibrary/Textures/MaterialInstances/MI_UIKit_Icon_Check, iconSize=Compact, overrideStatusIndicatorVisibility=true}
   slot {class=HorizontalBoxSlot, properties {horizontalAlignment=HAlign_Left, verticalAlignment=VAlign_Top, padding=<ltrb 0 0 5 0>}}
  }
  TextBlock_Complete {
   widgetClassPath=&CommonTextBlock
   bIsVariable=true
   parent=HB_Complete
   widget=TextBlock_Complete
   properties {
    text=COMPLETE
    textTransformPolicy=ToUpper
    colorAndOpacity {colorUseRule=UseColor_Specified, specifiedColor=<rgba 0.0395 1 0.2307 1>}
    font {
     bForceMonospaced=false
     bMaterialIsStencil=false
     fontMaterial=null
     fontObject=&BurbankSmall
     letterSpacing=0
     monospacedWidth=1
     size=36
     skewAmount=0
     typefaceFontName=Black
     outlineSettings {
      bApplyOutlineToDropShadows=false
      bMiteredCorners=false
      bSeparateFillAlpha=false
      outlineMaterial=null
      outlineSize=0
      outlineColor=<rgba 0 0 0 1>
     }
    }
    shadowOffset=<0 0>
   }
   slot {class=HorizontalBoxSlot, properties {verticalAlignment=VAlign_Center}}
  }
 ]
}
