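# List entry widget for one band stat row in the Pilgrim Quickplay results UI, showing a title beside type-specific content.
# On item set, it casts to BP_BandStats_Stat_EntryData, sets the title, and uses the content type enum to pick the WidgetSwitcher child.
# That child is a text label, difficulty, reward notification or full combo indicator.
# The first three receive the entry's label text, and the reward notification also plays its intro animation.
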WidgetBlueprint /PilgrimQuickplay/Widgets/Results/WBP_Pilgrim_Quickplay_BandStats_ListEntry
graphs [
 EventGraph {nodes [
   n1 = Event OnListItemObjectSet()
   > n2 = Cast To BP_BandStats_Stat_EntryData!(execute, Object=@n1.ListItemObject) -> (then, CastFailed, AsBP Band Stats Stat Entry Data:BP Band Stats Stat Entry Data Object Reference) {is=title K2Node_DynamicCast}
   n3 = Get!(self:BP Band Stats Stat Entry Data Object Reference=@n2) -> (
    EntryData_StatTitle_9_#1:Text
    EntryData_StatLabel_10_#2:Text
    EntryData_ContentType_8_#3:E_BandStats_StatEntry_ContentType Enum
   )
   > SetStatTitle!(execute, self, InTitle:Text=@n3) -> (then)
   n4 = Select!(
    NewEnumerator0:Widget Object Reference=Get Stat_Label() -> (Stat_Label:Common Rich Text Block Object Reference)
    NewEnumerator1:Widget Object Reference=Get WBP_Pilgrim_Results_Difficulty() -> (WBP_Pilgrim_Results_Difficulty:WBP Pilgrim Results Difficulty Object Reference)
    NewEnumerator2:Widget Object Reference=Get WBP_Pilgrim_Results_RewardNotif() -> (WBP_Pilgrim_Results_RewardNotif:WBP Pilgrim Results Reward Notif Object Reference)
    NewEnumerator3:Widget Object Reference=Get WBP_PilgrimQuickplay_FullComboIndicator() -> (WBP_PilgrimQuickplay_FullComboIndicator:WBP Pilgrim Quickplay Full Combo Indicator Object Reference)
    NewEnumerator4:Widget Object Reference
    Index:E_BandStats_StatEntry_ContentType Enum=@n3.EntryData_ContentType_8_#3
   ) -> (ReturnValue:Widget Object Reference)
   > Is Valid(InputObject=@n4)
   > SetActiveWidget(self:Widget Switcher Object Reference=Get WidgetSwitcher_0() -> (WidgetSwitcher_0:Widget Switcher Object Reference), Widget=@n4)
   > n5 = Switch on E_BandStats_StatEntry_ContentType!(execute, Selection:E_BandStats_StatEntry_ContentType Enum=@n3.EntryData_ContentType_8_#3) -> (NewEnumerator0:Exec, NewEnumerator1:Exec, NewEnumerator2:Exec, NewEnumerator3:Exec, NewEnumerator4:Exec) {is=title K2Node_SwitchEnum}
   > SetStatLabel!(execute, self, InStatLabel:Text=@n3.EntryData_StatLabel_10_#2) -> (then)
   SetDifficultyText!(
    execute=@n5.NewEnumerator1
    self:WBP Pilgrim Results Difficulty Object Reference=Get WBP_Pilgrim_Results_Difficulty()
    InText:Text=@n3.EntryData_StatLabel_10_#2
   ) -> (then)
   n6 = Get WBP_Pilgrim_Results_RewardNotif()
   Set Text(
    execute=@n5.NewEnumerator2
    Text=@n3.EntryData_StatLabel_10_#2
    self=Get Text_Title!(self:WBP Pilgrim Results Reward Notif Object Reference=@n6) -> (Text_Title:Common Text Object Reference)
   )
   > PlayAnimation(
    self=@n6
    InAnimation=Get anim_Intro!(self:WBP Pilgrim Results Reward Notif Object Reference=@n6) -> (anim_Intro:Widget Animation Object Reference)
   )
   Event Construct()
   Event Tick()
  ]}
 SetStatLabel {nodes [
   n1 = SetStatLabel!() -> (then, InStatLabel:Text) {is=title K2Node_FunctionEntry}
   > n2 = Set Stat Label(Stat Label:Text=@n1) -> (Output_Get:Text)
   > SetText(self=Get Stat_Label(), InText=@n2)
  ]}
 SetStatTitle {nodes [
   n1 = SetStatTitle!() -> (then, InTitle:Text) {is=title K2Node_FunctionEntry}
   > n2 = Set Stat Title(Stat Title:Text=@n1) -> (Output_Get:Text)
   > SetText(self=Get StatTitle_Label() -> (StatTitle_Label:Common Rich Text Block Object Reference), InText=@n2)
  ]}
]
variables [
 Stat Label {type=string, default=999}
 Stat Title {type=string, default=StatTitle}
]
widgets {
 info {inheritedWidgetCount=0, parentClass=&UserWidget, rootWidgetClass=&Overlay, widgetCount=9}
 widgets [
  Overlay_0 {widgetClassPath=&Overlay, widget=Overlay_0}
  HorizontalBox_0 {
   widgetClassPath=&HorizontalBox
   parent=Overlay_0
   widget=HorizontalBox_0
   slot {class=OverlaySlot, properties {horizontalAlignment=HAlign_Fill, verticalAlignment=VAlign_Fill, padding=<ltrb 0 4 0 0>}}
  }
  StatTitle_Label {
   widgetClassPath=&CommonRichTextBlock
   bIsVariable=true
   parent=HorizontalBox_0
   widget=StatTitle_Label
   properties {
    bOverrideDefaultStyle=true
    scrollStyle=&TextScrollStyle-Base_C
    text=StatType
    defaultTextStyleOverride {
     overflowPolicy=Clip
     transformPolicy=None
     colorAndOpacity {colorUseRule=UseColor_Specified, specifiedColor=<rgba 1 1 1 0.8>}
     font {
      bForceMonospaced=false
      bMaterialIsStencil=false
      fontMaterial=null
      fontObject=&HeadingNow
      letterSpacing=0
      monospacedWidth=1
      size=15
      skewAmount=0
      typefaceFontName=74 Regular
      outlineSettings {
       bApplyOutlineToDropShadows=false
       bMiteredCorners=false
       bSeparateFillAlpha=false
       outlineMaterial=null
       outlineSize=0
       outlineColor=<rgba 0 0 0 1>
      }
     }
     highlightColor {colorUseRule=UseColor_Specified, specifiedColor=<rgba 0 0 0 1>}
     highlightShape {
      bIsDynamicallyLoaded=false
      drawAs=Image
      imageType=NoImage
      mirroring=NoMirror
      resourceName=None
      resourceObject=null
      tiling=NoTile
      imageSize=<32 32>
      margin=<ltrb 0 0 0 0>
      outlineSettings {
       bUseBrushTransparency=false
       roundingType=HalfHeightRadius
       width=0
       color {colorUseRule=UseColor_Specified, specifiedColor=<rgba 0 0 0 0>}
       cornerRadii=<0 0 0 0>
      }
      tintColor {colorUseRule=UseColor_Specified, specifiedColor=<rgba 1 1 1 1>}
      uVRegion {bIsValid=false, max=<0 0>, min=<0 0>}
     }
     selectedBackgroundColor {colorUseRule=UseColor_Foreground, specifiedColor=<rgba 1 0 1 1>}
     shadowColorAndOpacity=<rgba 0 0 0 1>
     shadowOffset=<0 0>
     strikeBrush {
      bIsDynamicallyLoaded=false
      drawAs=NoDrawType
      imageType=NoImage
      mirroring=NoMirror
      resourceName=None
      resourceObject=null
      tiling=NoTile
      imageSize=<32 32>
      margin=<ltrb 0 0 0 0>
      outlineSettings {
       bUseBrushTransparency=false
       roundingType=HalfHeightRadius
       width=0
       color {colorUseRule=UseColor_Specified, specifiedColor=<rgba 0 0 0 0>}
       cornerRadii=<0 0 0 0>
      }
      tintColor {colorUseRule=UseColor_Specified, specifiedColor=<rgba 1 1 1 1>}
      uVRegion {bIsValid=false, max=<0 0>, min=<0 0>}
     }
     underlineBrush {
      bIsDynamicallyLoaded=false
      drawAs=Image
      imageType=NoImage
      mirroring=NoMirror
      resourceName=None
      resourceObject=null
      tiling=NoTile
      imageSize=<32 32>
      margin=<ltrb 0 0 0 0>
      outlineSettings {
       bUseBrushTransparency=false
       roundingType=HalfHeightRadius
       width=0
       color {colorUseRule=UseColor_Specified, specifiedColor=<rgba 0 0 0 0>}
       cornerRadii=<0 0 0 0>
      }
      tintColor {colorUseRule=UseColor_Specified, specifiedColor=<rgba 1 1 1 1>}
      uVRegion {bIsValid=false, max=<0 0>, min=<0 0>}
     }
    }
    renderTransformPivot=<0 0>
   }
   slot {class=HorizontalBoxSlot, properties {verticalAlignment=VAlign_Center, size {sizeRule=Fill, value=1}}}
  }
  SizeBox_0 {
   widgetClassPath=&SizeBox
   parent=HorizontalBox_0
   widget=SizeBox_0
   properties {bOverride_MaxDesiredWidth=true, maxDesiredWidth=160, minDesiredWidth=50}
   slot {class=HorizontalBoxSlot}
  }
  WidgetSwitcher_0 {widgetClassPath=&WidgetSwitcher, bIsVariable=true, parent=SizeBox_0, widget=WidgetSwitcher_0, slot {class=SizeBoxSlot}}
  Stat_Label {
   widgetClassPath=&CommonRichTextBlock
   bIsVariable=true
   parent=WidgetSwitcher_0
   widget=Stat_Label
   properties {
    bOverrideDefaultStyle=true
    scrollStyle=&TextScrollStyle-Base_C
    text=Stat
    defaultTextStyleOverride {
     overflowPolicy=Clip
     transformPolicy=None
     colorAndOpacity {colorUseRule=UseColor_Specified, specifiedColor=<rgba 1 1 1 1>}
     font {
      bForceMonospaced=false
      bMaterialIsStencil=false
      fontMaterial=null
      fontObject=&HeadingNow
      letterSpacing=0
      monospacedWidth=1
      size=15
      skewAmount=0
      typefaceFontName=75 Medium
      outlineSettings {
       bApplyOutlineToDropShadows=false
       bMiteredCorners=false
       bSeparateFillAlpha=false
       outlineMaterial=null
       outlineSize=0
       outlineColor=<rgba 0 0 0 1>
      }
     }
     highlightColor {colorUseRule=UseColor_Specified, specifiedColor=<rgba 0 0 0 1>}
     highlightShape {
      bIsDynamicallyLoaded=false
      drawAs=Image
      imageType=NoImage
      mirroring=NoMirror
      resourceName=None
      resourceObject=null
      tiling=NoTile
      imageSize=<32 32>
      margin=<ltrb 0 0 0 0>
      outlineSettings {
       bUseBrushTransparency=false
       roundingType=HalfHeightRadius
       width=0
       color {colorUseRule=UseColor_Specified, specifiedColor=<rgba 0 0 0 0>}
       cornerRadii=<0 0 0 0>
      }
      tintColor {colorUseRule=UseColor_Specified, specifiedColor=<rgba 1 1 1 1>}
      uVRegion {bIsValid=false, max=<0 0>, min=<0 0>}
     }
     selectedBackgroundColor {colorUseRule=UseColor_Foreground, specifiedColor=<rgba 1 0 1 1>}
     shadowColorAndOpacity=<rgba 0 0 0 1>
     shadowOffset=<0 0>
     strikeBrush {
      bIsDynamicallyLoaded=false
      drawAs=Image
      imageType=NoImage
      mirroring=NoMirror
      resourceName=None
      resourceObject=null
      tiling=NoTile
      imageSize=<32 32>
      margin=<ltrb 0 0 0 0>
      outlineSettings {
       bUseBrushTransparency=false
       roundingType=HalfHeightRadius
       width=0
       color {colorUseRule=UseColor_Specified, specifiedColor=<rgba 0 0 0 0>}
       cornerRadii=<0 0 0 0>
      }
      tintColor {colorUseRule=UseColor_Specified, specifiedColor=<rgba 1 1 1 1>}
      uVRegion {bIsValid=false, max=<0 0>, min=<0 0>}
     }
     underlineBrush {
      bIsDynamicallyLoaded=false
      drawAs=Image
      imageType=NoImage
      mirroring=NoMirror
      resourceName=None
      resourceObject=null
      tiling=NoTile
      imageSize=<32 32>
      margin=<ltrb 0 0 0 0>
      outlineSettings {
       bUseBrushTransparency=false
       roundingType=HalfHeightRadius
       width=0
       color {colorUseRule=UseColor_Specified, specifiedColor=<rgba 0 0 0 0>}
       cornerRadii=<0 0 0 0>
      }
      tintColor {colorUseRule=UseColor_Specified, specifiedColor=<rgba 1 1 1 1>}
      uVRegion {bIsValid=false, max=<0 0>, min=<0 0>}
     }
    }
    renderTransformPivot=<0 0>
   }
   slot {class=WidgetSwitcherSlot, properties {horizontalAlignment=HAlign_Right, verticalAlignment=VAlign_Center, padding=<ltrb 15 0 0 0>}}
  }
  WBP_Pilgrim_Results_Difficulty {
   widgetClassPath=&/WBP_Pilgrim_Results_Difficulty
   bIsVariable=true
   parent=WidgetSwitcher_0
   widget=WBP_Pilgrim_Results_Difficulty
   slot {class=WidgetSwitcherSlot, properties {verticalAlignment=VAlign_Center}}
  }
  WBP_Pilgrim_Results_RewardNotif {
   widgetClassPath=&/WBP_Pilgrim_Results_RewardNotif
   bIsVariable=true
   parent=WidgetSwitcher_0
   widget=WBP_Pilgrim_Results_RewardNotif
   slot {class=WidgetSwitcherSlot, properties {verticalAlignment=VAlign_Center}}
  }
  WBP_PilgrimQuickplay_FullComboIndicator {
   widgetClassPath=/PilgrimQuickplay/Widgets/WBP_PilgrimQuickplay_FullComboIndicator
   bIsVariable=true
   parent=WidgetSwitcher_0
   widget=WBP_PilgrimQuickplay_FullComboIndicator
   slot {class=WidgetSwitcherSlot, properties {verticalAlignment=VAlign_Center}}
  }
 ]
}
