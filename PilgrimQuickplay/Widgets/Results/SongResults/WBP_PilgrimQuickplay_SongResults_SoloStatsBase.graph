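# Widget Blueprint for the solo stats panel in Pilgrim Quickplay song results.
# A spacer-sized overlay holds a grid with a backing panel and a "Stats" header above the StatList view, which uses band-stats entry widgets.
# On Construct it collapses the list's scrollbar, and the list is hit-test invisible, making it display-only.
# The graph contains no logic to populate the list.
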WidgetBlueprint /PilgrimQuickplay/Widgets/Results/SongResults/WBP_PilgrimQuickplay_SongResults_SoloStatsBase
graphs [EventGraph {nodes [
   Event Construct()
   > SetScrollbarVisibility!(
    execute
    self:List View Base Object Reference=Get StatList() -> (StatList:Common List View Object Reference)
    InVisibility:ESlateVisibility Enum=Collapsed
   ) -> (then)
  ]}]
variables []
widgets {
 info {inheritedWidgetCount=0, parentClass=&UserWidget, rootWidgetClass=&Overlay, widgetCount=9}
 widgets [
  Recap_Content {widgetClassPath=&Overlay, widget=Recap_Content}
  Spacer_110 {
   widgetClassPath=&Spacer
   parent=Recap_Content
   widget=Spacer_110
   properties {size=<320 350>}
   slot {class=OverlaySlot, properties {horizontalAlignment=HAlign_Fill}}
  }
  GridPanel_0 {
   widgetClassPath=&GridPanel
   parent=Recap_Content
   widget=GridPanel_0
   properties {columnFill [1], rowFill [1]}
   slot {class=OverlaySlot, properties {horizontalAlignment=HAlign_Fill, padding=<ltrb 5 0 5 0>}}
  }
  VerticalBox_0 {widgetClassPath=&VerticalBox, parent=GridPanel_0, widget=VerticalBox_0, slot {class=GridSlot}}
  Header {
   widgetClassPath=&Border
   parent=VerticalBox_0
   widget=Header
   properties {
    background {
     bIsDynamicallyLoaded=false
     drawAs=Image
     imageType=NoImage
     mirroring=NoMirror
     resourceName=None
     resourceObject=/PilgrimCore/Widgets/Materials/MI_UI_Piligrim_NonUniform_PanelHeader_Solid
     tiling=NoTile
     imageSize=<32 32>
     margin=<ltrb 0 0 0 0>
     outlineSettings {
      bUseBrushTransparency=false
      roundingType=HalfHeightRadius
      width=0
      color {colorUseRule=UseColor_Specified, specifiedColor=<rgba 0 0 0 0>}
      cornerRadii=<0 0 0 0>
     }
     tintColor {colorUseRule=UseColor_Specified, specifiedColor=<rgba 1 1 1 1>}
     uVRegion {bIsValid=false, max=<0 0>, min=<0 0>}
    }
    padding=<ltrb 16 4 0 4>
   }
   slot {class=VerticalBoxSlot}
  }
  CommonTextBlock_122 {
   widgetClassPath=&CommonTextBlock
   parent=Header
   widget=CommonTextBlock_122
   properties {
    text=Stats
    textTransformPolicy=ToUpper
    font {
     bForceMonospaced=false
     bMaterialIsStencil=false
     fontMaterial=null
     fontObject=&HeadingNow
     letterSpacing=0
     monospacedWidth=1
     size=15
     skewAmount=0
     typefaceFontName=76 Bold Italic
     outlineSettings {
      bApplyOutlineToDropShadows=false
      bMiteredCorners=false
      bSeparateFillAlpha=false
      outlineMaterial=null
      outlineSize=0
      outlineColor=<rgba 0 0 0 1>
     }
    }
    shadowOffset=<0 0>
   }
   slot {class=BorderSlot, properties {padding=<ltrb 16 4 0 4>}}
  }
  Border_70 {
   widgetClassPath=&Border
   parent=VerticalBox_0
   widget=Border_70
   properties {
    background {
     bIsDynamicallyLoaded=false
     drawAs=NoDrawType
     imageType=NoImage
     mirroring=NoMirror
     resourceName=None
     resourceObject=null
     tiling=NoTile
     imageSize=<32 32>
     margin=<ltrb 0 0 0 0>
     outlineSettings {
      bUseBrushTransparency=false
      roundingType=FixedRadius
      width=1.7851
      color {colorUseRule=UseColor_Specified, specifiedColor=<rgba 0 0 0 0>}
      cornerRadii=<0 0 0 0>
     }
     tintColor {colorUseRule=UseColor_Specified, specifiedColor=<rgba 1 1 1 1>}
     uVRegion {bIsValid=false, max=<0 0>, min=<0 0>}
    }
    brushColor=<rgba 1 1 1 0.8>
    padding=<ltrb 16 8 16 16>
   }
   slot {class=VerticalBoxSlot}
  }
  StatList {
   widgetClassPath=&CommonListView
   bIsVariable=true
   parent=Border_70
   widget=StatList
   properties {
    entryWidgetClass=/PilgrimQuickplay/Widgets/Results/WBP_Pilgrim_Quickplay_BandStats_ListEntry.WBP_Pilgrim_Quickplay_BandStats_ListEntry_C
    numDesignerPreviewEntries=7
    verticalEntrySpacing=8
    visibility=HitTestInvisible
   }
   slot {class=BorderSlot, properties {padding=<ltrb 16 8 16 16>}}
  }
  WBP_Sparks_BackingPanel {
   widgetClassPath=/PilgrimCore/Widgets/SparksUILibrary/WBP_Sparks_BackingPanel
   bIsVariable=true
   parent=GridPanel_0
   widget=WBP_Sparks_BackingPanel
   properties {cornerRadius=Small, darkenLayer=false, eSparksBackingPanelType=Simple Color}
   slot {class=GridSlot}
  }
 ]
}
