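# Results header widget for Pilgrim Quickplay that shows the just-played song's album art, title and setlist position.
# On Construct, it gets the owning player's Quickplay player state component and continues through a validity check.
# The title comes from that component's song result data, and the "Song X of Y" text from its band state's setlist.
# Album art lazy-loads from the current song of the band state found through the Quickplay playspace, not from the result data.
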WidgetBlueprint /PilgrimQuickplay/Widgets/Results/WBP_PilgrimQuickplay_JustPlayedSongHeader
graphs [EventGraph {nodes [
   Event Construct()
   > n1 = GetPilgrimQuickplayPlayerStateComponent!(
    execute
    PlayerState:Player State Object Reference=Get PlayerState!(self:Controller Object Reference=GetOwningPlayer()) -> (PlayerState:Player State Object Reference)
   ) -> (then, ReturnValue:Pilgrim Quickplay Player State Component Object Reference)
   > Is Valid(InputObject=@n1)
   > n2 = Sequence!(execute) -> (then_0, then_1, then_2:Exec)
   n3 = GetSongResultData!(self:Pilgrim Quickplay Player State Component Object Reference=@n1) -> (ReturnValue:"Pilgrim Quickplay Song Result Data Structure (by ref)")
   > "SetText (Text)"(
    self=Get SongTitle() -> (SongTitle:Common Text Object Reference)
    InText="To Text (String)"(InString=Break Pilgrim Quickplay Song Result Data!(PilgrimQuickplaySongResultData:"Pilgrim Quickplay Song Result Data Structure (by ref)"=@n3) -> (
      SongName:String
      SongTags:Gameplay Tag Container Structure
      SongShortName:Name
      BandData:Array of Pilgrim Quickplay Song Result Band Data Structures
      PlayerData:Array of Pilgrim Quickplay Song Result Player Data Structures
      Team:Byte
      LocalPlayerBandDataIndex:int
      LocalPlayerDataIndex:int
     ) {is=title K2Node_BreakStruct})
   )
   n4 = GetBandState!(self:Pilgrim Quickplay Player State Component Object Reference=@n1) -> (ReturnValue:Pilgrim Quickplay Band State Component Object Reference)
   n5 = GetCurrentSetlistIndex!(self:Pilgrim Quickplay Band State Component Object Reference=@n4) -> (ReturnValue:int=0)
   "SetText (Text)"(
    execute=@n2.then_1
    self=Get SongCount() -> (SongCount:Common Text Object Reference)
    InText=Format Text!(
     Format="LOCTEXT(\"(Song {CurrentIndex} of {Total})\")"
     CurrentIndex:int=int + int(A=@n5)
     Total:int=GetCurrentSetlistLength!(self:Pilgrim Quickplay Band State Component Object Reference=@n4) -> (ReturnValue:int=0)
    ) -> (Result)
   )
   n6 = GetSparksMusicPlayspaceOfClass!(
    PlayspaceType:Sparks Music Playspace Class Reference=/Script/PilgrimQuickplayRuntime.PilgrimQuickplayPlayspace
    Actor:Actor Object Reference=GetOwningPlayer()
    bMatchChildTypes:bool=false
   ) -> (ReturnValue:Pilgrim Quickplay Playspace Object Reference)
   LazyLoadAlbumArt!(
    execute=@n2.then_2
    Song Metadata:Pilgrim Song Metadata Object Reference=GetCurrentSong!(self:Pilgrim Core Band State Component Object Reference=GetBandState!(ContextActor:Actor Object Reference=@n6) -> (ReturnValue)) -> (ReturnValue:Pilgrim Song Metadata Object Reference)
    Lazy Image:Fort Lazy Image Object Reference=Get Image_AlbumCoverPlaceholder() -> (Image_AlbumCoverPlaceholder:Fort Lazy Image Object Reference)
   ) -> (then)
  ]}]
variables []
widgets {
 info {inheritedWidgetCount=0, parentClass=&CommonUserWidget, rootWidgetClass=&Overlay, widgetCount=8}
 widgets [
  Overlay_0 {widgetClassPath=&Overlay, widget=Overlay_0}
  Image_93 {
   widgetClassPath=&Image
   bIsVariable=true
   parent=Overlay_0
   widget=Image_93
   properties {
    visibility=Hidden
    brush {
     bIsDynamicallyLoaded=false
     drawAs=Image
     imageType=NoImage
     mirroring=NoMirror
     resourceName=None
     resourceObject=/PilgrimCore/Widgets/Materials/MI_UI_Piligrim_NonUniform_PanelHeader
     tiling=NoTile
     imageSize=<32 32>
     margin=<ltrb 0 0 0 0>
     outlineSettings {
      bUseBrushTransparency=false
      roundingType=HalfHeightRadius
      width=0
      color {colorUseRule=UseColor_Specified, specifiedColor=<rgba 0 0 0 0>}
      cornerRadii=<0 0 0 0>
     }
     tintColor {colorUseRule=UseColor_Specified, specifiedColor=<rgba 1 1 1 1>}
     uVRegion {bIsValid=false, max=<0 0>, min=<0 0>}
    }
   }
   slot {class=OverlaySlot, properties {horizontalAlignment=HAlign_Fill, verticalAlignment=VAlign_Fill}}
  }
  WBP_Sparks_BackingPanel {
   widgetClassPath=/PilgrimCore/Widgets/SparksUILibrary/WBP_Sparks_BackingPanel
   bIsVariable=true
   parent=Overlay_0
   widget=WBP_Sparks_BackingPanel
   properties {blurFrame=false}
   slot {class=OverlaySlot, properties {horizontalAlignment=HAlign_Fill, verticalAlignment=VAlign_Fill}}
  }
  HorizontalBox_0 {
   widgetClassPath=&HorizontalBox
   parent=Overlay_0
   widget=HorizontalBox_0
   slot {class=OverlaySlot, properties {horizontalAlignment=HAlign_Fill}}
  }
  Image_AlbumCoverPlaceholder {
   widgetClassPath=&FortLazyImage
   bIsVariable=true
   parent=HorizontalBox_0
   widget=Image_AlbumCoverPlaceholder
   properties {
    materialTextureParamName=T_AlbumArt
    brush {
     bIsDynamicallyLoaded=false
     drawAs=Image
     imageType=FullColor
     mirroring=NoMirror
     resourceName=None
     resourceObject=/SparksUICommon/Materials/MI_UI_Sparks_SetList_AlbumArt_Rounded
     tiling=NoTile
     imageSize=<72 72>
     margin=<ltrb 0 0 0 0>
     outlineSettings {
      bUseBrushTransparency=false
      roundingType=HalfHeightRadius
      width=0
      color {colorUseRule=UseColor_Specified, specifiedColor=<rgba 0 0 0 0>}
      cornerRadii=<0 0 0 0>
     }
     tintColor {colorUseRule=UseColor_Specified, specifiedColor=<rgba 1 1 1 1>}
     uVRegion {bIsValid=false, max=<0 0>, min=<0 0>}
    }
    loadingBackgroundBrush {
     bIsDynamicallyLoaded=false
     drawAs=NoDrawType
     imageType=NoImage
     mirroring=NoMirror
     resourceName=None
     resourceObject=null
     tiling=NoTile
     imageSize=<80 80>
     margin=<ltrb 0 0 0 0>
     outlineSettings {
      bUseBrushTransparency=false
      roundingType=HalfHeightRadius
      width=0
      color {colorUseRule=UseColor_Specified, specifiedColor=<rgba 0 0 0 0>}
      cornerRadii=<0 0 0 0>
     }
     tintColor {colorUseRule=UseColor_Specified, specifiedColor=<rgba 1 1 1 1>}
     uVRegion {bIsValid=false, max=<0 0>, min=<0 0>}
    }
   }
   slot {class=HorizontalBoxSlot, properties {horizontalAlignment=HAlign_Left, verticalAlignment=VAlign_Center, padding=<ltrb 4 4 0 4>}}
  }
  VerticalBox_28 {
   widgetClassPath=&VerticalBox
   parent=HorizontalBox_0
   widget=VerticalBox_28
   slot {class=HorizontalBoxSlot, properties {verticalAlignment=VAlign_Center, padding=<ltrb 8 4 16 0>, size {sizeRule=Fill, value=1}}}
  }
  SongTitle {
   widgetClassPath=&CommonTextBlock
   bIsVariable=true
   parent=VerticalBox_28
   widget=SongTitle
   properties {
    flowDirectionPreference=Culture
    scrollStyle=&TextScrollStyle-Base_C
    text=Long Song Title
    font {
     bForceMonospaced=false
     bMaterialIsStencil=false
     fontMaterial=null
     fontObject=&HeadingNow
     letterSpacing=0
     monospacedWidth=1
     size=15
     skewAmount=0
     typefaceFontName=75 Medium
     outlineSettings {
      bApplyOutlineToDropShadows=false
      bMiteredCorners=false
      bSeparateFillAlpha=false
      outlineMaterial=null
      outlineSize=0
      outlineColor=<rgba 0 0 0 1>
     }
    }
    shadowOffset=<0 0>
   }
   slot {class=VerticalBoxSlot}
  }
  SongCount {
   widgetClassPath=&CommonTextBlock
   bIsVariable=true
   parent=VerticalBox_28
   widget=SongCount
   properties {
    flowDirectionPreference=Culture
    lineHeightPercentage=1.29
    text=Band Name
    colorAndOpacity {colorUseRule=UseColor_Specified, specifiedColor=<rgba 1 1 1 0.8>}
    font {
     bForceMonospaced=false
     bMaterialIsStencil=false
     fontMaterial=null
     fontObject=&HeadingNow
     letterSpacing=0
     monospacedWidth=1
     size=15
     skewAmount=0
     typefaceFontName=74 Regular
     outlineSettings {
      bApplyOutlineToDropShadows=false
      bMiteredCorners=false
      bSeparateFillAlpha=false
      outlineMaterial=null
      outlineSize=0
      outlineColor=<rgba 0 0 0 1>
     }
    }
    shadowOffset=<0 0>
    strikeBrush {
     bIsDynamicallyLoaded=false
     drawAs=NoDrawType
     imageType=NoImage
     mirroring=NoMirror
     resourceName=None
     resourceObject=null
     tiling=NoTile
     imageSize=<32 32>
     margin=<ltrb 0 0 0 0>
     outlineSettings {
      bUseBrushTransparency=false
      roundingType=HalfHeightRadius
      width=0
      color {colorUseRule=UseColor_Specified, specifiedColor=<rgba 0 0 0 0>}
      cornerRadii=<0 0 0 0>
     }
     tintColor {colorUseRule=UseColor_Specified, specifiedColor=<rgba 1 1 1 1>}
     uVRegion {bIsValid=false, max=<0 0>, min=<0 0>}
    }
   }
   slot {class=VerticalBoxSlot, properties {padding=<ltrb 0 4 0 0>}}
  }
 ]
}
