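# Results-screen part picker widget for Pilgrim Quickplay, holding the picker controls and a solo song stats panel.
# In LMP mode, SwitchPartPickerModules uses a splitscreen details enum to show the picker, the stats or neither, registering picker inputs only while the picker shows.
# LMP starts with both hidden; otherwise stats collapse and inputs register on construct.
# Opening the menu forwards to the main controls, which also take focus under a Menu input config.
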WidgetBlueprint /PilgrimQuickplay/Widgets/UniversalPartPicker/WBP_PilgrimQuickplay_PartPicker_LMPResults
graphs [
 BP_GetDesiredFocusTarget {nodes [
   Get Desired Focus Target()
   > Return Node!(
    execute
    ReturnValue:Widget Object Reference=Get WBP_PilgrimQuickplay_PartPicker_MainControls() -> (WBP_PilgrimQuickplay_PartPicker_MainControls:WBP Pilgrim Quickplay Part Picker Main Controls Object Reference)
   )
  ]}
 BP_GetDesiredInputConfig {nodes [
   Get Desired Input Config() {is=title K2Node_FunctionEntry}
   > Return Node(ReturnValue:UIInput Config Structure=Make UIInput Config!(
     bIgnoreMoveInput:bool=true
     bIgnoreLookInput:bool=true
     InputMode:ECommonInputMode Enum=Menu
     MouseCaptureMode:EMouseCaptureMode Enum=NoCapture
     MouseLockMode:EMouseLockMode Enum=DoNotLock
     bHideCursorDuringViewportCapture:bool=True
    ) -> (UIInputConfig:UIInput Config Structure) {is=title K2Node_MakeStruct})
  ]}
 EventGraph {nodes [
   n1 = SwitchPartPickerModules!() -> (OutputDelegate, then, Pilgrim_SplitscreenPickerDetails:Pilgrim_SplitscreenPickerDetails Enum=NewEnumerator0) {is=title K2Node_CustomEvent}
   > Branch(Condition=Get bIsPartPickerForLMP() -> (bIsPartPickerForLMP:bool))
   > SetVisibility(
    self=Get StackBox_UPP_LMP() -> (StackBox_UPP_LMP:Stack Box Object Reference)
    InVisibility=Select!(
     NewEnumerator0:ESlateVisibility Enum=Visible
     NewEnumerator1:ESlateVisibility Enum=Hidden
     NewEnumerator2:ESlateVisibility Enum=Hidden
     Index:Pilgrim_SplitscreenPickerDetails Enum=@n1.Pilgrim_SplitscreenPickerDetails
    ) -> (ReturnValue)
   )
   > SetVisibility(
    self=Get SizeBox_Stats() -> (SizeBox_Stats:Size Box Object Reference)
    InVisibility=Select(NewEnumerator0=Hidden, NewEnumerator1=SelfHitTestInvisible, NewEnumerator2=Hidden, Index=@n1.Pilgrim_SplitscreenPickerDetails)
   )
   > n2 = Switch on Pilgrim_SplitscreenPickerDetails!(execute, Selection:Pilgrim_SplitscreenPickerDetails Enum=@n1.Pilgrim_SplitscreenPickerDetails) -> (NewEnumerator0:Exec, NewEnumerator1:Exec, NewEnumerator2:Exec) {is=title K2Node_SwitchEnum}
   > RegisterPartPickerInputs(self:Pilgrim Quickplay UIResults Universal Part Picker Object Reference)
   UnregisterPartPickerInputs(execute=[@n2.NewEnumerator1, @n2.NewEnumerator2], self:Pilgrim Quickplay UIResults Universal Part Picker Object Reference)
   Event Construct()
   > n3 = Branch(Condition=Get bIsPartPickerForLMP())
   > SwitchPartPickerModules!(execute, self, Pilgrim_SplitscreenPickerDetails:Pilgrim_SplitscreenPickerDetails Enum=NewEnumerator2) -> (then) {is=title K2Node_CallFunction}
   SetVisibility(execute=@n3.else, self=Get SizeBox_Stats(), InVisibility=Collapsed)
   > RegisterPartPickerInputs()
   Event OpenMenuUniversalPartPicker() {is=title K2Node_Event}
   > On Going to Parts(self:WBP Pilgrim Quickplay Part Picker Main Controls Object Reference=Get WBP_PilgrimQuickplay_PartPicker_MainControls())
   Event Destruct()
   > UnregisterPartPickerInputs()
  ]}
]
variables [bIsPartPickerForLMP {type=boolean}]
widgets {
 info {
  inheritedWidgetCount=1
  parentClass=/Script/PilgrimQuickplayUI.PilgrimQuickplay_UIResults_UniversalPartPicker
  rootWidgetClass=&GridPanel
  widgetCount=6
 }
 widgets [
  GridPanel_0 {widgetClassPath=&GridPanel, widget=GridPanel_0}
  StackBox_UPP_LMP {
   widgetClassPath=&StackBox
   bIsVariable=true
   parent=GridPanel_0
   widget=StackBox_UPP_LMP
   properties {bOverride_Cursor=true, orientation=Orient_Vertical}
   slot {class=GridSlot}
  }
  WBP_PilgrimQuickplay_PartPicker_MainControls {
   widgetClassPath=&/WBP_PilgrimQuickplay_PartPicker_MainControls
   bIsVariable=true
   parent=StackBox_UPP_LMP
   widget=WBP_PilgrimQuickplay_PartPicker_MainControls
   slot {class=StackBoxSlot, properties {horizontalAlignment=HAlign_Left, verticalAlignment=VAlign_Top}}
  }
  UniversalPartPickerWidget {
   widgetClassPath=&CommonActionWidget
   bInherited=true
   bIsVariable=true
   parent=StackBox_UPP_LMP
   widget=UniversalPartPickerWidget
   properties {designTimeKey=Gamepad_DPad_Right}
   slot {class=StackBoxSlot, properties {horizontalAlignment=HAlign_Left, verticalAlignment=VAlign_Top}}
  }
  SizeBox_Stats {
   widgetClassPath=&SizeBox
   bIsVariable=true
   parent=GridPanel_0
   widget=SizeBox_Stats
   properties {bOverride_MaxDesiredWidth=true, maxDesiredWidth=320}
   slot {class=GridSlot}
  }
  WBP_PilgrimQuickplay_SongResults_SoloStatsBase {
   widgetClassPath=/PilgrimQuickplay/Widgets/Results/SongResults/WBP_PilgrimQuickplay_SongResults_SoloStatsBase
   bIsVariable=true
   parent=SizeBox_Stats
   widget=WBP_PilgrimQuickplay_SongResults_SoloStatsBase
   slot {class=SizeBoxSlot}
  }
 ]
}
