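# Splitscreen results-screen widget for Pilgrim Quickplay that shows one player's instrument and difficulty picker or solo song stats.
# SwitchPartPickerModules toggles picker, stats or neither (the initial state), registering picker inputs only while the picker shows.
# Swap input events forward to the owning player's cached BP_PilgrimPlayerControllerComponent.
# A ControllerDisconnect animation fades the main overlay into a disconnect prompt outside replays, and reverses otherwise.
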WidgetBlueprint /PilgrimQuickplay/Widgets/WBP_PilgrimQuickplay_Results_SplitscreenPartPicker
animations [ControllerDisconnect {
  binding_count=2
  length_seconds=1
  object=&self:ControllerDisconnect
  sequence {bindings [
    MainOverlay {
     id=#1
     possessed_class=Overlay
     tracks [{
       type=MovieSceneFloatTrack
       display_name=Render Opacity
       sections [{type=MovieSceneFloatSection, end_frame=20, start_frame=0, channels [None {type=MovieSceneScriptingFloatChannel, keys [[0,1],[15,0],[20,0]]}]}]
      }]
    }
    ControllerDisconnectPrompt {
     id=#2
     possessed_class=WBP_PilgrimQuickplay_ControllerDisconnectPrompt_C
     tracks [
      {
       type=MovieSceneFloatTrack
       display_name=Render Opacity
       sections [{type=MovieSceneFloatSection, end_frame=20, start_frame=0, channels [None {type=MovieSceneScriptingFloatChannel, keys [[0,0],[15,0],[20,1]]}]}]
      }
      {
       type=MovieSceneByteTrack
       display_name=Visibility
       sections [{type=MovieSceneByteSection, end_frame=20, start_frame=0, channels [None {type=MovieSceneScriptingByteChannel, keys [[0,1],[15,4],[20,4]]}]}]
      }
     ]
    }
   ]}
 }]
graphs [
 EventGraph {nodes [
   n1 = SwitchPartPickerModules!() -> (OutputDelegate, then, Pilgrim_SplitscreenPickerDetails:Pilgrim_SplitscreenPickerDetails Enum=NewEnumerator0) {is=title K2Node_CustomEvent}
   > SetVisibility(
    self=Get PartPicker_SizeBox() -> (PartPicker_SizeBox:Size Box Object Reference)
    InVisibility=Select!(
     NewEnumerator0:ESlateVisibility Enum=HitTestInvisible
     NewEnumerator1:ESlateVisibility Enum=Hidden
     NewEnumerator2:ESlateVisibility Enum=Hidden
     Index:Pilgrim_SplitscreenPickerDetails Enum=@n1.Pilgrim_SplitscreenPickerDetails
    ) -> (ReturnValue)
   )
   > SetVisibility(
    self=Get Stats_SizeBox() -> (Stats_SizeBox:Size Box Object Reference)
    InVisibility=Select(NewEnumerator0=Hidden, NewEnumerator1=SelfHitTestInvisible, NewEnumerator2=Hidden, Index=@n1.Pilgrim_SplitscreenPickerDetails)
   )
   > n2 = Switch on Pilgrim_SplitscreenPickerDetails!(execute, Selection:Pilgrim_SplitscreenPickerDetails Enum=@n1.Pilgrim_SplitscreenPickerDetails) -> (NewEnumerator0:Exec, NewEnumerator1:Exec, NewEnumerator2:Exec) {is=title K2Node_SwitchEnum}
   > RegisterPartPickerInputs(self:Pilgrim Quickplay UIResults Part Picker Object Reference)
   UnregisterPartPickerInputs(execute=[@n2.NewEnumerator1, @n2.NewEnumerator2], self:Pilgrim Quickplay UIResults Part Picker Object Reference)
   Event Destruct()
   n3 = Get GameplayEventHandles() -> (GameplayEventHandles:Array of Async Action Start Listening To Event Object References)
   > n4 = For Each Loop(Array:Array of Async Action Start Listening To Event Object References=@n3) -> (Array Element:Async Action Start Listening To Event Object Reference)
   Clear(execute=@n4.Completed, TargetArray:Array of Async Action Start Listening To Event Object References=@n3)
   StopListeningToEvent(execute=@n4.LoopBody, self:Async Action Start Listening To Event Object Reference=@n4)
   Event Construct()
   > Set BPPlayerControllerCached(BPPlayerControllerCached:BP Pilgrim Player Controller Component Object Reference=GetComponentByClass!(
     self:Actor Object Reference=GetOwningPlayer()
     ComponentClass:Actor Component Class Reference=/PilgrimCore/Player/Components/BP_PilgrimPlayerControllerComponent.BP_PilgrimPlayerControllerComponent_C
    ) -> (ReturnValue:BP Pilgrim Player Controller Component Object Reference)) -> (Output_Get:BP Pilgrim Player Controller Component Object Reference)
   > SwitchPartPickerModules!(execute, self, Pilgrim_SplitscreenPickerDetails:Pilgrim_SplitscreenPickerDetails Enum=NewEnumerator2) -> (then) {is=title K2Node_CallFunction}
   Event SwapDifficulty() {is=title K2Node_Event}
   > Swap Difficulty Input Pressed(self:BP Pilgrim Player Controller Component Object Reference=Get BPPlayerControllerCached() -> (BPPlayerControllerCached:BP Pilgrim Player Controller Component Object Reference))
   Event SwapInstrumentLeft() {is=title K2Node_Event}
   > SwapInstrumentBackward(self:BP Pilgrim Player Controller Component Object Reference=Get BPPlayerControllerCached())
   Event SwapInstrumentRight() {is=title K2Node_Event}
   > SwapInstrumentForward!(execute, self:BP Pilgrim Player Controller Component Object Reference=Get BPPlayerControllerCached(), IsTouchInputPlatform:bool=false) -> (then)
  ]}
 Set ControllerDisconnected Visibility {nodes [
   n1 = Set ControllerDisconnected Visibility!() -> (then, IsDisconnected:bool=~) {is=title K2Node_FunctionEntry}
   n2 = IsPlayingReplay!() -> (ReturnValue:bool)
   > n3 = Branch(Condition=AND Boolean(A=@n1, B=NOT Boolean(A=@n2)))
   n4 = IsVisible(self=Get ControllerDisconnectPrompt() -> (ControllerDisconnectPrompt:WBP Pilgrim Quickplay Controller Disconnect Prompt Object Reference))
   > n5 = Branch(Condition=@n4)
   n6 = Get ControllerDisconnect() -> (ControllerDisconnect:Widget Animation Object Reference)
   PlayAnimationForward(execute=@n5.else, InAnimation=@n6)
   Branch(execute=@n3.else, Condition=@n4)
   > PlayAnimationReverse(InAnimation=@n6)
  ]}
]
variables [
 BPPlayerControllerCached {type=BP_PilgrimPlayerControllerComponent_C, default=null}
 GameplayEventHandles {type="array<AsyncAction_StartListeningToEvent>", default []}
]
widgets {
 info {
  inheritedWidgetCount=3
  parentClass=/Script/PilgrimQuickplayUI.PilgrimQuickplay_UIResults_PartPicker
  rootWidgetClass=&GridPanel
  widgetCount=15
 }
 widgets [
  GridPanel_1 {widgetClassPath=&GridPanel, widget=GridPanel_1, properties {columnFill [1], rowFill [1]}}
  MainOverlay {
   widgetClassPath=&Overlay
   bIsVariable=true
   parent=GridPanel_1
   widget=MainOverlay
   slot {class=GridSlot, properties {columnSpan=1, rowSpan=1}}
  }
  PartPicker_SizeBox {
   widgetClassPath=&SizeBox
   bIsVariable=true
   parent=MainOverlay
   widget=PartPicker_SizeBox
   properties {bOverride_MaxDesiredWidth=true, bOverride_MinDesiredHeight=true, maxDesiredWidth=320, minDesiredHeight=320, widthOverride=360}
   slot {class=OverlaySlot, properties {horizontalAlignment=HAlign_Fill, verticalAlignment=VAlign_Bottom}}
  }
  PartPicker_VBox {
   widgetClassPath=&VerticalBox
   parent=PartPicker_SizeBox
   widget=PartPicker_VBox
   slot {class=SizeBoxSlot, properties {verticalAlignment=VAlign_Bottom, padding=<ltrb 12 0 12 0>}}
  }
  Overlay_0 {
   widgetClassPath=&Overlay
   parent=PartPicker_VBox
   widget=Overlay_0
   slot {class=VerticalBoxSlot, properties {padding=<ltrb 0 2 0 12>}}
  }
  WBP_Pilgrim_GreenRoom_InstrumentSelectUI {
   widgetClassPath=&/WBP_Pilgrim_GreenRoom_InstrumentSelectUI
   bIsVariable=true
   parent=Overlay_0
   widget=WBP_Pilgrim_GreenRoom_InstrumentSelectUI
   properties {isOnGreenRoom=false, visibility=HitTestInvisible}
   slot {class=OverlaySlot, properties {horizontalAlignment=HAlign_Fill}}
  }
  GridPanel_0 {
   widgetClassPath=&GridPanel
   parent=Overlay_0
   widget=GridPanel_0
   properties {columnFill [1], rowFill [1]}
   slot {class=OverlaySlot, properties {horizontalAlignment=HAlign_Fill, padding=<ltrb 4 4 4 0>}}
  }
  SwapInstrumentLeftActionWidget {
   widgetClassPath=&CommonActionWidget
   bInherited=true
   parent=GridPanel_0
   widget=SwapInstrumentLeftActionWidget
   properties {designTimeKey=Gamepad_LeftShoulder}
   slot {class=GridSlot, properties {horizontalAlignment=HAlign_Left, verticalAlignment=VAlign_Center}}
  }
  SwapInstrumentRightActionWidget {
   widgetClassPath=&CommonActionWidget
   bInherited=true
   parent=GridPanel_0
   widget=SwapInstrumentRightActionWidget
   properties {designTimeKey=Gamepad_RightShoulder}
   slot {class=GridSlot, properties {horizontalAlignment=HAlign_Right, verticalAlignment=VAlign_Center}}
  }
  Overlay_1 {
   widgetClassPath=&Overlay
   parent=PartPicker_VBox
   widget=Overlay_1
   slot {class=VerticalBoxSlot, properties {verticalAlignment=VAlign_Center, size {sizeRule=Fill, value=1}}}
  }
  WBP_Pilgrim_GreenRoom_DifficultySelectUI {
   widgetClassPath=&/WBP_Pilgrim_GreenRoom_DifficultySelectUI
   parent=Overlay_1
   widget=WBP_Pilgrim_GreenRoom_DifficultySelectUI
   properties {isOnGreenRoom=false, visibility=HitTestInvisible}
   slot {class=OverlaySlot, properties {horizontalAlignment=HAlign_Fill}}
  }
  SwapDifficultyActionWidget {
   widgetClassPath=&CommonActionWidget
   bInherited=true
   parent=Overlay_1
   widget=SwapDifficultyActionWidget
   properties {designTimeKey=Gamepad_FaceButton_Right}
   slot {class=OverlaySlot, properties {verticalAlignment=VAlign_Fill, padding=<ltrb 4 0 0 0>}}
  }
  ControllerDisconnectPrompt {
   widgetClassPath=&/WBP_PilgrimQuickplay_ControllerDisconnectPrompt
   bIsVariable=true
   parent=GridPanel_1
   widget=ControllerDisconnectPrompt
   properties {renderOpacity=0, visibility=Collapsed}
   slot {class=GridSlot, properties {columnSpan=1, layer=1, rowSpan=1, verticalAlignment=VAlign_Center}}
  }
  Stats_SizeBox {
   widgetClassPath=&SizeBox
   bIsVariable=true
   parent=GridPanel_1
   widget=Stats_SizeBox
   properties {bOverride_MaxDesiredWidth=true, maxDesiredWidth=320}
   slot {class=GridSlot}
  }
  WBP_PilgrimQuickplay_SongResults_SoloStatsBase {
   widgetClassPath=/PilgrimQuickplay/Widgets/Results/SongResults/WBP_PilgrimQuickplay_SongResults_SoloStatsBase
   bIsVariable=true
   parent=Stats_SizeBox
   widget=WBP_PilgrimQuickplay_SongResults_SoloStatsBase
   slot {class=SizeBoxSlot}
  }
 ]
}
