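# Widget Blueprint for the Emote button on the Pilgrim Quickplay song results screen, derived from the native PilgrimQuickplay_UIResults_EmoteButton class.
# Its only widget is the inherited Emote_Button, a quiet-style UIKit button labeled "Emote" that is non-focusable and has a persistent binding to the ResultsEmote input action from the Quickplay data table.
# The PreConstruct and Construct events are empty, so the Blueprint adds configuration but no logic.
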WidgetBlueprint /PilgrimQuickplay/Widgets/WBP_PilgrimQuickplay_SongResults_EmoteButton
graphs [EventGraph {nodes [
   Event PreConstruct()
   Event Construct()
  ]}]
variables []
widgets {
 info {
  inheritedWidgetCount=1
  parentClass=/Script/PilgrimQuickplayUI.PilgrimQuickplay_UIResults_EmoteButton
  rootWidgetClass=/FortUILibrary/UIKit/Buttons/WBP_UIKit_Button_Quiet.WBP_UIKit_Button_Quiet_C
  widgetCount=1
 }
 widgets [Emote_Button {
   widgetClassPath=/FortUILibrary/UIKit/Buttons/WBP_UIKit_Button_Quiet
   bInherited=true
   bIsVariable=true
   widget=Emote_Button
   properties {
    bIsFocusable=false
    bIsPersistentBinding=true
    text=Emote
    triggeringInputAction {dataTable=/PilgrimQuickplay/Widgets/InputActionDataTables/FM_PilgrimQuickplay_InputActionDataTable, rowName=ResultsEmote}
   }
  }]
}
