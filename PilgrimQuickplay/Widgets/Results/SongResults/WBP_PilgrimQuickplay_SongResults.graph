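# Post-song results screen for Pilgrim Quickplay, with tabbed stats and next-song views, staged intro animations and controller-disconnect prompts.
# Players' info and new-high-score panels are placed by replicated results location index.
# Secondary local players get splitscreen part pickers, either unified or older style.
# Buttons handle next song, replay, leave, report, lobby return and CVar-gated ConcordGrape travel, several branching on multiplayer state.
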WidgetBlueprint /PilgrimQuickplay/Widgets/Results/SongResults/WBP_PilgrimQuickplay_SongResults
animations [
 ShowToggleStatsPrompt {
  binding_count=1
  length_seconds=0.5
  object=&self:ShowToggleStatsPrompt
  sequence {bindings [AllPrompts {
     id=#1
     possessed_class=VerticalBox
     tracks [
      {
       type=MovieSceneByteTrack
       display_name=Visibility
       sections [{type=MovieSceneByteSection, end_frame=10, start_frame=0, channels [None {type=MovieSceneScriptingByteChannel, const [4,2]}]}]
      }
      {
       type=MovieSceneFloatTrack
       display_name=Render Opacity
       sections [{type=MovieSceneFloatSection, end_frame=10, start_frame=0, channels [None {type=MovieSceneScriptingFloatChannel, keys [[0,0],[10,1]]}]}]
      }
     ]
    }]}
 }
 ShowStatExpander {
  binding_count=1
  length_seconds=10
  object=&self:ShowStatExpander
  sequence {bindings [ExpandStats_Prompt_Solo {
     id=#2
     possessed_class=HorizontalBox
     tracks [
      {
       type=MovieSceneByteTrack
       display_name=Visibility
       sections [{type=MovieSceneByteSection, end_frame=70, start_frame=0, channels [None {type=MovieSceneScriptingByteChannel, const [4,2]}]}]
      }
      {
       type=MovieSceneFloatTrack
       display_name=Render Opacity
       sections [{
         type=MovieSceneFloatSection
         end_frame=200
         start_frame=0
         channels [None {type=MovieSceneScriptingFloatChannel, keys [[0,0],[10,1],[200,1]]}]
        }]
      }
     ]
    }]}
 }
 Anim_SwitcherAndTabIntro {
  binding_count=2
  length_seconds=0.2
  object=&self:Anim_SwitcherAndTabIntro
  sequence {bindings [
    CommonVisibilitySwitcher {
     id=#3
     possessed_class=CommonVisibilitySwitcher
     tracks [{
       type=MovieSceneFloatTrack
       display_name=Render Opacity
       sections [{type=MovieSceneFloatSection, channels [None {type=MovieSceneScriptingFloatChannel, keys [[0,0],[1,1]]}]}]
      }]
    }
    TopTabs_Overlay {
     id=#4
     possessed_class=Overlay
     tracks [{
       type=MovieSceneFloatTrack
       display_name=Render Opacity
       sections [{type=MovieSceneFloatSection, channels [None {type=MovieSceneScriptingFloatChannel, keys [[0,0],[4,1]]}]}]
      }]
    }
   ]}
 }
 Anim_PerformanceResultsPanelIntro {
  binding_count=5
  length_seconds=0.5
  object=&self:Anim_PerformanceResultsPanelIntro
  sequence {bindings [
    JustPlayedSongHeader {
     id=#5
     possessed_class=WBP_PilgrimQuickplay_JustPlayedSongHeader_C
     tracks [
      {
       type=MovieSceneFloatTrack
       display_name=Render Opacity
       sections [{type=MovieSceneFloatSection, channels [None {type=MovieSceneScriptingFloatChannel, keys [[2.388,0],[5.388,1]]}]}]
      }
      {
       type=MovieScene2DTransformTrack
       sections [{
         type=MovieScene2DTransformSection
         channels [
          Translation.X {type=MovieSceneScriptingFloatChannel, const [0,1]}
          Translation.Y {type=MovieSceneScriptingFloatChannel, keys [[2.388,32],[5.388,0]]}
         ]
        }]
      }
     ]
    }
    BottomSoloButtons_VerticalBox {
     id=#6
     possessed_class=VerticalBox
     tracks [
      {
       type=MovieSceneFloatTrack
       display_name=Render Opacity
       sections [{type=MovieSceneFloatSection, channels [None {type=MovieSceneScriptingFloatChannel, keys [[1,0],[5,1]]}]}]
      }
      {
       type=MovieScene2DTransformTrack
       sections [{
         type=MovieScene2DTransformSection
         channels [
          Translation.X {type=MovieSceneScriptingFloatChannel, const [0,1]}
          Translation.Y {type=MovieSceneScriptingFloatChannel, keys [[1,64],[8,0]]}
          Scale.X {type=MovieSceneScriptingFloatChannel, const [1,1]}
          Scale.Y {type=MovieSceneScriptingFloatChannel, const [1,1]}
         ]
        }]
      }
     ]
    }
    MainStatsPanel_SizeOverride {
     id=#7
     possessed_class=SizeBox
     tracks [
      {
       type=MovieSceneFloatTrack
       display_name=Render Opacity
       sections [{type=MovieSceneFloatSection, channels [None {type=MovieSceneScriptingFloatChannel, keys [[0,0],[4,1]]}]}]
      }
      {
       type=MovieScene2DTransformTrack
       sections [{
         type=MovieScene2DTransformSection
         channels [
          Translation.X {type=MovieSceneScriptingFloatChannel, const [0,1]}
          Translation.Y {type=MovieSceneScriptingFloatChannel, keys [[0,64],[6,0]]}
          Scale.X {type=MovieSceneScriptingFloatChannel, const [1,1]}
          Scale.Y {type=MovieSceneScriptingFloatChannel, const [1,1]}
         ]
        }]
      }
     ]
    }
    EmoteButton {
     id=#8
     possessed_class=WBP_PilgrimQuickplay_SongResults_EmoteButton_C
     tracks [{
       type=MovieSceneFloatTrack
       display_name=Render Opacity
       sections [{type=MovieSceneFloatSection, end_frame=10, start_frame=0, channels [None {type=MovieSceneScriptingFloatChannel, keys [[0,1],[3,0]]}]}]
      }]
    }
    BandPerformanceData {
     id=#9
     possessed_class=VerticalBox
     tracks [
      {
       type=MovieSceneFloatTrack
       display_name=Render Opacity
       sections [{type=MovieSceneFloatSection, channels [None {type=MovieSceneScriptingFloatChannel, keys [[3.388,0],[6.388,1]]}]}]
      }
      {
       type=MovieScene2DTransformTrack
       sections [{
         type=MovieScene2DTransformSection
         channels [
          Translation.X {type=MovieSceneScriptingFloatChannel, const [0,1]}
          Translation.Y {type=MovieSceneScriptingFloatChannel, keys [[3.388,32],[6.388,0]]}
         ]
        }]
      }
     ]
    }
   ]}
 }
 Anim_NextTabFade {
  binding_count=6
  length_seconds=0.5
  object=&self:Anim_NextTabFade
  sequence {bindings [
    SongResult {
     id=#10
     possessed_class=VerticalBox
     tracks [{
       type=MovieSceneFloatTrack
       display_name=Render Opacity
       sections [{type=MovieSceneFloatSection, channels [None {type=MovieSceneScriptingFloatChannel, keys [[0,0],[3,1]]}]}]
      }]
    }
    SongResult_ButtonGroup {
     id=#11
     possessed_class=VerticalBox
     tracks [
      {
       type=MovieSceneFloatTrack
       display_name=Render Opacity
       sections [{type=MovieSceneFloatSection, channels [None {type=MovieSceneScriptingFloatChannel, keys [[2.388,0],[5.388,1]]}]}]
      }
      {
       type=MovieScene2DTransformTrack
       sections [{
         type=MovieScene2DTransformSection
         channels [
          Translation.X {type=MovieSceneScriptingFloatChannel, const [0,1]}
          Translation.Y {type=MovieSceneScriptingFloatChannel, keys [[2.388,32],[5.388,0]]}
         ]
        }]
      }
     ]
    }
    WBP_PilgrimQuickplay_NextSongDetailsPanel {
     id=#12
     possessed_class=WBP_PilgrimQuickplay_NextSongDetailsPanel_C
     tracks [
      {
       type=MovieSceneFloatTrack
       display_name=Render Opacity
       sections [{type=MovieSceneFloatSection, channels [None {type=MovieSceneScriptingFloatChannel, keys [[0.388,0],[3.388,1]]}]}]
      }
      {
       type=MovieScene2DTransformTrack
       sections [{
         type=MovieScene2DTransformSection
         channels [
          Translation.X {type=MovieSceneScriptingFloatChannel, const [0,1]}
          Translation.Y {type=MovieSceneScriptingFloatChannel, keys [[0.388,32],[3.388,0]]}
         ]
        }]
      }
     ]
    }
    InstrumentSelector {
     id=#13
     possessed_class=WBP_Pilgrim_GreenRoom_InstrumentSelectUI_C
     tracks [
      {
       type=MovieSceneFloatTrack
       display_name=Render Opacity
       sections [{type=MovieSceneFloatSection, channels [None {type=MovieSceneScriptingFloatChannel, keys [[1.388,0],[4.388,1]]}]}]
      }
      {
       type=MovieScene2DTransformTrack
       sections [{
         type=MovieScene2DTransformSection
         channels [
          Translation.X {type=MovieSceneScriptingFloatChannel, const [0,1]}
          Translation.Y {type=MovieSceneScriptingFloatChannel, keys [[1.388,32],[4.388,0]]}
         ]
        }]
      }
     ]
    }
    ResultsCountdownTimer {
     id=#14
     possessed_class=WBP_PilgrimQuickplay_ResultsCountdownTimer_C
     tracks [
      {
       type=MovieSceneFloatTrack
       display_name=Render Opacity
       sections [{type=MovieSceneFloatSection, channels [None {type=MovieSceneScriptingFloatChannel, keys [[3,0],[7,1]]}]}]
      }
      {
       type=MovieScene2DTransformTrack
       sections [{
         type=MovieScene2DTransformSection
         channels [
          Translation.X {type=MovieSceneScriptingFloatChannel, keys [[3,-64],[10,0]]}
          Translation.Y {type=MovieSceneScriptingFloatChannel}
         ]
        }]
      }
     ]
    }
    DifficultySelector {
     id=#15
     possessed_class=WBP_Pilgrim_GreenRoom_DifficultySelectUI_C
     tracks [
      {
       type=MovieSceneFloatTrack
       display_name=Render Opacity
       sections [{type=MovieSceneFloatSection, channels [None {type=MovieSceneScriptingFloatChannel, keys [[1,0],[4,1]]}]}]
      }
      {
       type=MovieScene2DTransformTrack
       sections [{
         type=MovieScene2DTransformSection
         channels [
          Translation.X {type=MovieSceneScriptingFloatChannel, const [0,1]}
          Translation.Y {type=MovieSceneScriptingFloatChannel, keys [[1,32],[4,0]]}
         ]
        }]
      }
     ]
    }
   ]}
 }
 Anim_ProgressBackgroundFadeIn {
  binding_count=2
  length_seconds=0.4
  object=&self:Anim_ProgressBackgroundFadeIn
  sequence {bindings [
    BG {
     id=#16
     possessed_class=Border
     tracks [{
       type=MovieSceneFloatTrack
       display_name=Render Opacity
       sections [{type=MovieSceneFloatSection, channels [None {type=MovieSceneScriptingFloatChannel, keys [[0,0],[8,1]]}]}]
      }]
    }
    BG_1 {
     id=#17
     possessed_class=Border
     tracks [{
       type=MovieSceneFloatTrack
       display_name=Render Opacity
       sections [{type=MovieSceneFloatSection, channels [None {type=MovieSceneScriptingFloatChannel, keys [[0,0],[8,1]]}]}]
      }]
    }
   ]}
 }
 Anim_TimerBar {
  binding_count=1
  length_seconds=1
  object=&self:Anim_TimerBar
  sequence {bindings [AutoAdvanceTimerBar {
     id=#18
     possessed_class=Image
     tracks [{
       type=MovieSceneWidgetMaterialTrack
       sections [{type=MovieSceneParameterSection, channels [Progress {type=MovieSceneScriptingFloatChannel, keys [[0,0],[20,1]]}]}]
      }]
    }]}
 }
 Anim_MoreDetailsIntro {
  binding_count=1
  length_seconds=0.5
  object=&self:Anim_MoreDetailsIntro
  sequence {bindings [MoreDetailsPrompt {
     id=#19
     possessed_class=WBP_UIKit_Button_Quiet_C
     tracks [
      {
       type=MovieScene2DTransformTrack
       sections [{
         type=MovieScene2DTransformSection
         channels [
          Translation.X {type=MovieSceneScriptingFloatChannel, keys [[0,512],[10,0]]}
          Translation.Y {type=MovieSceneScriptingFloatChannel, const [0,1]}
         ]
        }]
      }
      {
       type=MovieSceneFloatTrack
       display_name=Render Opacity
       sections [{type=MovieSceneFloatSection, end_frame=10, start_frame=0, channels [None {type=MovieSceneScriptingFloatChannel, keys [[0,0],[10,1]]}]}]
      }
     ]
    }]}
 }
 Anim_ProgressTab {
  binding_count=1
  length_seconds=5
  object=&self:Anim_ProgressTab
  sequence {bindings [WBP_PilgrimQuickplay_ProgressTab {id=#20, possessed_class=WBP_PilgrimQuickplay_ProgressTab_C, tracks []}]}
 }
 Anim_EmoteIntro {
  binding_count=1
  length_seconds=0.5
  object=&self:Anim_EmoteIntro
  sequence {bindings [EmoteButton {
     id=#21
     possessed_class=WBP_PilgrimQuickplay_SongResults_EmoteButton_C
     tracks [
      {
       type=MovieScene2DTransformTrack
       sections [{
         type=MovieScene2DTransformSection
         channels [
          Translation.X {type=MovieSceneScriptingFloatChannel, keys [[0,512],[10,0]]}
          Translation.Y {type=MovieSceneScriptingFloatChannel, const [0,1]}
         ]
        }]
      }
      {
       type=MovieSceneFloatTrack
       display_name=Render Opacity
       sections [{type=MovieSceneFloatSection, end_frame=10, start_frame=0, channels [None {type=MovieSceneScriptingFloatChannel, keys [[0,0],[10,1]]}]}]
      }
     ]
    }]}
 }
 Anim_ControllerDisconnect {
  binding_count=5
  length_seconds=1
  object=&self:Anim_ControllerDisconnect
  sequence {bindings [
    ControllerDisconnectPrompt_NextUp {
     id=#22
     possessed_class=WBP_PilgrimQuickplay_ControllerDisconnectPrompt_C
     tracks [
      {
       type=MovieSceneFloatTrack
       display_name=Render Opacity
       sections [{type=MovieSceneFloatSection, end_frame=20, start_frame=0, channels [None {type=MovieSceneScriptingFloatChannel, keys [[0,0],[15,0],[20,1]]}]}]
      }
      {
       type=MovieSceneByteTrack
       display_name=Visibility
       sections [{type=MovieSceneByteSection, end_frame=20, start_frame=0, channels [None {type=MovieSceneScriptingByteChannel, keys [[0,1],[15,4],[20,4]]}]}]
      }
     ]
    }
    VB_PickerInputUI {
     id=#23
     possessed_class=VerticalBox
     tracks [{
       type=MovieSceneFloatTrack
       display_name=Render Opacity
       sections [{type=MovieSceneFloatSection, end_frame=15, start_frame=0, channels [None {type=MovieSceneScriptingFloatChannel, keys [[0,1],[15,0]]}]}]
      }]
    }
    SongResult_ButtonGroup {
     id=#24
     possessed_class=VerticalBox
     tracks [{
       type=MovieSceneFloatTrack
       display_name=Render Opacity
       sections [{type=MovieSceneFloatSection, end_frame=15, start_frame=0, channels [None {type=MovieSceneScriptingFloatChannel, keys [[0,1],[15,0]]}]}]
      }]
    }
    ControllerDisconnectPrompt_MainStats {
     id=#25
     possessed_class=WBP_PilgrimQuickplay_ControllerDisconnectPrompt_C
     tracks [
      {
       type=MovieSceneFloatTrack
       display_name=Render Opacity
       sections [{type=MovieSceneFloatSection, end_frame=20, start_frame=0, channels [None {type=MovieSceneScriptingFloatChannel, keys [[0,0],[15,0],[20,1]]}]}]
      }
      {
       type=MovieSceneByteTrack
       display_name=Visibility
       sections [{type=MovieSceneByteSection, end_frame=20, start_frame=0, channels [None {type=MovieSceneScriptingByteChannel, keys [[0,1],[15,4],[20,4]]}]}]
      }
     ]
    }
    BottomSoloButtons_VerticalBox {
     id=#26
     possessed_class=VerticalBox
     tracks [{
       type=MovieSceneFloatTrack
       display_name=Render Opacity
       sections [{type=MovieSceneFloatSection, end_frame=15, start_frame=0, channels [None {type=MovieSceneScriptingFloatChannel, keys [[0,1],[15,0]]}]}]
      }]
    }
   ]}
 }
 Anim_SetlistNameIntro {
  binding_count=1
  length_seconds=0.5
  object=&self:Anim_SetlistNameIntro
  sequence {bindings [WBP_PilgrimQuickplay_SongResults_SetlistStatus {
     id=#27
     possessed_class=WBP_PilgrimQuickplay_SongResults_SetlistStatus_C
     tracks [
      {
       type=MovieScene2DTransformTrack
       sections [{
         type=MovieScene2DTransformSection
         end_frame=10
         start_frame=0
         channels [Translation.X {type=MovieSceneScriptingFloatChannel, keys [[0,-500],[10,0]]}]
        }]
      }
      {
       type=MovieSceneFloatTrack
       display_name=Render Opacity
       sections [{type=MovieSceneFloatSection, end_frame=10, start_frame=0, channels [None {type=MovieSceneScriptingFloatChannel, keys [[0,0],[10,1]]}]}]
      }
      {
       type=MovieSceneEventTrack
       sections [{
         type=MovieSceneEventTriggerSection
         channels [None {type=MovieSceneScriptingEventChannel, const ["<Struct 'MovieSceneEvent' (0x468e48de420) {}>", 1]}]
        }]
      }
     ]
    }]}
 }
]
graphs [
 AddSplitscreenPartPickerWidgets {nodes [
   AddSplitscreenPartPickerWidgets() {is=title K2Node_FunctionEntry}
   > n1 = For Each Loop(Array:Array of Player State Object References=Get AllPlayerStates() -> (AllPlayerStates:Array of Player State Object References)) -> (Array Element:Player State Object Reference)
   > n2 = Branch(Condition=IsUnifiedPartPickerEnabled!() -> (ReturnValue:bool))
   > n3 = Branch(Condition=Is Valid Index!(
     TargetArray:Array of WBP Pilgrim Quickplay Part Picker LMPResults Object References=Get UniversalPartPickerArray() -> (UniversalPartPickerArray:Array of WBP Pilgrim Quickplay Part Picker LMPResults Object References)
     IndexToTest:int=@n1.Array Index
    ) -> (ReturnValue:bool))
   CreateSplitcreenPartPickerWidget!(
    execute=[@n3.else, @n4.else]
    self
    PlayerState:Player State Object Reference=@n1
    PlayerIndex:int=Get ResultsLocationIndex!(self:BP Pilgrim Quickplay Results Location Player State Component Object Reference=GetComponentByClass!(
      self:Actor Object Reference=@n1
      ComponentClass:Actor Component Class Reference=/PilgrimQuickplay/Player/Components/BP_PilgrimQuickplay_ResultsLocationPlayerStateComponent.BP_PilgrimQuickplay_ResultsLocationPlayerStateComponent_C
     ) -> (ReturnValue:BP Pilgrim Quickplay Results Location Player State Component Object Reference)) -> (ResultsLocationIndex:int=0)
   ) -> (then)
   n4 = Branch(
    execute=@n2.else
    Condition=Is Valid Index(
     TargetArray:Array of WBP Pilgrim Quickplay Results Splitscreen Part Picker Object References=Get PartPickerArray() -> (PartPickerArray:Array of WBP Pilgrim Quickplay Results Splitscreen Part Picker Object References)
     IndexToTest=@n1.Array Index
    )
   )
  ]}
 AllLocalPlayer {nodes [
   AllLocalPlayer() {is=title K2Node_FunctionEntry}
   n1 = Get RemotePlayers!(self:Pilgrim Ham Sandwich VM Object Reference=Get PilgrimHamSandwichVM() -> (PilgrimHamSandwichVM:Pilgrim Ham Sandwich VM Object Reference)) -> (RemotePlayers:Array of Pilgrim Player Info VM Object References)
   > Return Node!(execute, AllLocalPlayerController:bool="integer <= integer"(A=Length(TargetArray:Array of Pilgrim Player Info VM Object References=@n1)))
  ]}
 AnimGraph {nodes [Add Injected Animation Events() {is=title K2Node_CustomEvent}]}
 BP_GetDesiredFocusTarget {nodes [
   Get Desired Focus Target()
   n1 = Get CommonVisibilitySwitcher() -> (CommonVisibilitySwitcher:Common Visibility Switcher Object Reference)
   n2 = "Equal (Enum)"(
    A:ESlateVisibility Enum=GetVisibility!(self:Widget Object Reference=Get SongResult() -> (SongResult:Vertical Box Object Reference)) -> (ReturnValue:ESlateVisibility Enum=Visible)
    B:ESlateVisibility Enum=Collapsed
   )
   n3 = Select(
    Option 0:Widget Object Reference=Get InstrumentSelector() -> (InstrumentSelector:WBP Pilgrim Green Room Instrument Select UI Object Reference)
    Option 1:Widget Object Reference=Get NextUp_StartNextSong() -> (NextUp_StartNextSong:WBP FM Core Button Regular Object Reference)
    Index=IsUnifiedPartPickerEnabled()
   ) -> (ReturnValue:Widget Object Reference)
   n4 = Get Progress_LeaveStage_Button() -> (Progress_LeaveStage_Button:WBP FM Core Button Regular Object Reference)
   n5 = Select(
    Option 0=Get Progress_Report_Button() -> (Progress_Report_Button:WBP FM Core Button Regular Object Reference)
    Option 1=@n4
    Index=GetIsEnabled!(self:Widget Object Reference=@n4) -> (ReturnValue:bool)
   )
   n6 = Get Results_LeaveStage_Button() -> (Results_LeaveStage_Button:WBP FM Core Button Regular Object Reference)
   n7 = Select(
    Option 0=Get Results_Report_Button() -> (Results_Report_Button:WBP FM Core Button Regular Object Reference)
    Option 1=@n6
    Index=GetIsEnabled(self=@n6)
   )
   n8 = Get AtEndOfSetlist() -> (AtEndOfSetlist:bool)
   n9 = Get Results_StartNextSongButton() -> (Results_StartNextSongButton:WBP FM Core Button Regular Object Reference)
   > Return Node!(
    execute
    ReturnValue:Widget Object Reference=Select(
     Option 0=Select(Option 0=@n9, Option 1=@n7, Index:bool=@n8)
     Option 1=Select(Option 0=@n3, Option 1=@n5, Index=@n2)
     Index:int=GetActiveWidgetIndex!(self:Common Visibility Switcher Object Reference=@n1) -> (ReturnValue:int=0)
    )
   )
  ]}
 ButtonsGraph {nodes [
   n1 = "On Pressed (Results_StartNextSongButton)"!() -> (OutputDelegate, then, Button:Common Button Base Object Reference) {is=title K2Node_ComponentBoundEvent}
   n2 = Get MultiplayerState!(self:Pilgrim Ham Sandwich VM Object Reference=Get PilgrimHamSandwichVM()) -> (MultiplayerState:EPilgrimMultiplayerState Enum=Solo)
   n3 = Get MultiplayerState(self=Get PilgrimHamSandwichVM()) -> (MultiplayerState=Solo)
   Branch(
    execute=[@n1.then, @n18.then, @n20.then, @n23.then]
    Condition=OR Boolean(A="Equal (Enum)"(A:EPilgrimMultiplayerState Enum=@n2, B:EPilgrimMultiplayerState Enum=Solo), B="Equal (Enum)"(A=@n3, B=Local))
   )
   n4 = GetComponentByClass(
    self=GetOwningPlayer()
    ComponentClass=/PilgrimCore/Player/Components/BP_PilgrimPlayerControllerComponent.BP_PilgrimPlayerControllerComponent_C
   ) -> (ReturnValue:BP Pilgrim Player Controller Component Object Reference)
   > Is Valid(InputObject=@n4)
   > Solo_SendRequestToPlayNextSong(self:Pilgrim Player Controller Component Object Reference=@n4)
   "On Clicked (TouchArea_NextUp)"() {is=title K2Node_ComponentBoundEvent}
   > SelectTabByIndex!(
    execute
    self:UIKit Tab Group Object Reference=Get WBP_UIKit_Tabs_TopNavBar() -> (WBP_UIKit_Tabs_TopNavBar:WBP UIKit Tabs Top Nav Bar Object Reference)
    Index:int=1
    bSuppressClickFeedback:bool=true
   ) -> (then, ReturnValue:bool)
   n5 = "On Pressed (Results_Report_Button)"!() -> (OutputDelegate, then, Button:Common Button Base Object Reference) {is=title K2Node_ComponentBoundEvent}
   Branch(execute=[@n5.then, @n19.then, @n24.then, @n29.then], Condition=Get MoreThanOnePlayerInBand() -> (MoreThanOnePlayerInBand:bool))
   > RequestToOpenPlayerFeedbackModalWithReportOrigin!(execute, Context:User Widget Object Reference=Self-Reference(), ReportOrigin:String=sparks_results) -> (then)
   n6 = "On Clicked (NextUp_ReplaySong)"!() -> (OutputDelegate, then, Button:Common Button Base Object Reference) {is=title K2Node_ComponentBoundEvent}
   n7 = Get MultiplayerState(self=Get PilgrimHamSandwichVM()) -> (MultiplayerState=Solo)
   n8 = Get MultiplayerState(self=Get PilgrimHamSandwichVM()) -> (MultiplayerState=Solo)
   Branch(
    execute=[@n6.then, @n17.then, @n25.then, @n26.then, @n30.then, @n44.then]
    Condition=OR Boolean(A="Equal (Enum)"(A=@n7, B=Solo), B="Equal (Enum)"(A=@n8, B=Local))
   )
   > AddScene!(
    execute
    self:Dynamic UIManager Object Reference=Get DynamicUIManager() -> (ReturnValue:Dynamic UIManager Object Reference) {is=title K2Node_GetSubsystem}
    Scene:Dynamic UIScene Object Reference=Get RestartSongScene() -> (RestartSongScene:Dynamic UIScene Object Reference)
    Player:"Player Controller Object Reference (by ref)"=GetFortPlayerControllerFromActor!(Actor:Actor Object Reference=GetOwningPlayer()) -> (ReturnValue:Fort Player Controller Object Reference)
   ) -> (then)
   > Set RestartSceneOn(RestartSceneOn:bool=true) -> (Output_Get:bool)
   n9 = "On Clicked (Results_LeaveStage_Button)"!() -> (OutputDelegate, then, Button:Common Button Base Object Reference) {is=title K2Node_ComponentBoundEvent}
   Set ResultsReturnChoice(execute=[@n9.then, @n22.then, @n28.then, @n42.then], ResultsReturnChoice:EPilgrimAnalyticsResultsReturn Enum=LeaveStage) -> (Output_Get:EPilgrimAnalyticsResultsReturn Enum=None)
   n10 = Get MultiplayerState(self=Get PilgrimHamSandwichVM()) -> (MultiplayerState=Solo)
   n11 = Get MultiplayerState(self=Get PilgrimHamSandwichVM()) -> (MultiplayerState=Solo)
   > n12 = Branch(Condition=OR Boolean(A="Equal (Enum)"(A=@n11, B=Hybrid), B="Equal (Enum)"(A=@n10, B=FullRemote)))
   n13 = GetComponentByClass(
    self=GetOwningPlayer()
    ComponentClass=&BP_PilgrimPlayerControllerComponent_C
   )
   > Is Valid(InputObject=@n13)
   > Multiplayer_SendRequestToLeaveStage(self:Pilgrim Player Controller Component Object Reference=@n13)
   > DisableButtonWithReason!(
    execute
    self:Common Button Base Object Reference=[Get Results_LeaveStage_Button(), Get Progress_LeaveStage_Button()]
    DisabledReason:"Text (by ref)"=Format Text(Format="INVTEXT(\"Leave Stage cannot be pressed twice\")")
   ) -> (then)
   Is Valid(exec=@n12.else, InputObject=@n13)
   > Solo_SendRequestToLeaveStage(self:Pilgrim Player Controller Component Object Reference=@n13)
   n14 = "On Pressed (Results_ReturnToLobby_Button)"!() -> (OutputDelegate, then, Button:Common Button Base Object Reference) {is=title K2Node_ComponentBoundEvent}
   Send Results Return Analytics!(execute=[@n14.then, @n21.then, @n27.then, @n31.then], self, ResultsReturnChoice:EPilgrimAnalyticsResultsReturn Enum=LeaveToLobby) -> (then)
   n15 = GetOwningPlayer()
   > SendLeaveZoneAnalytic(self:Fort Global UIContext Object Reference=Get FortGlobalUIContext!(PlayerController:Player Controller Object Reference=@n15) -> (ReturnValue:Fort Global UIContext Object Reference) {is=title K2Node_GetSubsystemFromPC})
   > n16 = Cast To FortPlayerControllerAthena(Object=@n15)
   > ReturnToMainMenu(self:Fort Player Controller Object Reference=@n16)
   n17 = "On Clicked (Results_ReplaySongButton)"!() -> (OutputDelegate, then, Button:Common Button Base Object Reference) {is=title K2Node_ComponentBoundEvent}
   n18 = "On Clicked (NextUp_StartNextSong)"!() -> (OutputDelegate, then, Button:Common Button Base Object Reference) {is=title K2Node_ComponentBoundEvent}
   n19 = "On Clicked (Results_Report_Button)"!() -> (OutputDelegate, then, Button:Common Button Base Object Reference) {is=title K2Node_ComponentBoundEvent}
   n20 = "On Clicked (Results_StartNextSongButton)"!() -> (OutputDelegate, then, Button:Common Button Base Object Reference) {is=title K2Node_ComponentBoundEvent}
   n21 = "On Clicked (Results_ReturnToLobby_Button)"!() -> (OutputDelegate, then, Button:Common Button Base Object Reference) {is=title K2Node_ComponentBoundEvent}
   n22 = "On Clicked (Progress_LeaveStage_Button)"!() -> (OutputDelegate, then, Button:Common Button Base Object Reference) {is=title K2Node_ComponentBoundEvent}
   n23 = "On Pressed (NextUp_StartNextSong)"!() -> (OutputDelegate, then, Button:Common Button Base Object Reference) {is=title K2Node_ComponentBoundEvent}
   n24 = "On Clicked (Progress_Report_Button)"!() -> (OutputDelegate, then, Button:Common Button Base Object Reference) {is=title K2Node_ComponentBoundEvent}
   n25 = "On Clicked (Progress_ReplaySongButton)"!() -> (OutputDelegate, then, Button:Common Button Base Object Reference) {is=title K2Node_ComponentBoundEvent}
   n26 = "On Pressed (NextUp_ReplaySong)"!() -> (OutputDelegate, then, Button:Common Button Base Object Reference) {is=title K2Node_ComponentBoundEvent}
   n27 = "On Clicked (Progress_ReturnToLobby_Button)"!() -> (OutputDelegate, then, Button:Common Button Base Object Reference) {is=title K2Node_ComponentBoundEvent}
   n28 = "On Pressed (Progress_LeaveStage_Button)"!() -> (OutputDelegate, then, Button:Common Button Base Object Reference) {is=title K2Node_ComponentBoundEvent}
   "On Clicked (MoreDetailsPrompt)"!() -> (OutputDelegate, then, Button:Common Button Base Object Reference) {is=title K2Node_ComponentBoundEvent}
   > PlaySound2D(Sound=Get MoreDetailsPressedSound() -> (MoreDetailsPressedSound:Sound Base Object Reference))
   > ToggleMoreDetails()
   n29 = "On Pressed (Progress_Report_Button)"!() -> (OutputDelegate, then, Button:Common Button Base Object Reference) {is=title K2Node_ComponentBoundEvent}
   n30 = "On Pressed (Progress_ReplaySongButton)"!() -> (OutputDelegate, then, Button:Common Button Base Object Reference) {is=title K2Node_ComponentBoundEvent}
   n31 = "On Pressed (Progress_ReturnToLobby_Button)"!() -> (OutputDelegate, then, Button:Common Button Base Object Reference) {is=title K2Node_ComponentBoundEvent}
   n32 = "On Tab Selected (WBP_UIKit_Tabs_TopNavBar)"!() -> (OutputDelegate, then, TabId:Name=None) {is=title K2Node_ComponentBoundEvent}
   > Branch(Condition=Get CanSwitchTabs() -> (CanSwitchTabs:bool))
   > SwitchTab!(execute, self, TabID:Name=@n32.TabId) -> (then)
   > Branch(Condition=Get TabsHaveNeverBeenSwitched() -> (TabsHaveNeverBeenSwitched:bool))
   > PlayTabTransitionAnims()
   > Set TabsHaveNeverBeenSwitched(TabsHaveNeverBeenSwitched:bool=false) -> (Output_Get:bool)
   n33 = Get Anim_TimerBar() -> (Anim_TimerBar:Widget Animation Object Reference)
   > PlayAnimation(InAnimation=@n33, StartAtTime=GetEndTime(self=@n33))
   n34 = "On Clicked (Progress_ConcordGrape_Button)"!() -> (OutputDelegate, then, Button:Common Button Base Object Reference) {is=title K2Node_ComponentBoundEvent}
   Set ResultsReturnChoice(execute=[@n34.then, @n43.then], ResultsReturnChoice=LeaveStage) -> (Output_Get=None)
   > n35 = GetConsoleVariableBoolValue!(execute, VariableName:String=Get ConcordGrapeCVAR() -> (ConcordGrapeCVAR:String)) -> (then, ReturnValue:bool)
   n36 = Get ConcordGrapeCVAR()
   > n37 = Branch(Condition=OR Boolean(A=@n35, B=IsEmpty(InString=@n36)))
   > PrintString!(
    execute
    InString=WBP_Pilgrim_UI_Results_Base::OnConcordGrapeButtonClicked: Going to ConcordGrape
    bPrintToScreen=false
    bPrintToLog
    TextColor
    Duration
    Key
   ) -> (then)
   n38 = GetFortPlayerControllerFromActor(Actor=GetOwningPlayer())
   > Is Valid(InputObject=@n38)
   n39 = GetGlobalEventRouter()
   > n40 = Is Valid(InputObject=@n39)
   PrintString(
    execute=@n40.Is Not Valid
    InString=WBP_Pilgrim_UI_Results_Base::OnConcordGrapeButtonClicked: No Global Event Router Found
    bPrintToScreen=false
   )
   n41 = Make Array("[0]":Fort Travel Playlist Data Structure=Make Fort Travel Playlist Data!(
     DestinationTag:Gameplay Tag Structure=(TagName="Playlist.PlaylistTravel.ConcordGrape")
     DestinationPlaylistName:String=playlist_concordgrape
     DestinationProductMode:String
     PrivacyOverride:EFortPlaylistTravelPrivacyOverride Enum=NoFill
    ) -> (FortTravelPlaylistData:Fort Travel Playlist Data Structure) {is=title K2Node_MakeStruct}) -> (Array:Array of Fort Travel Playlist Data Structures)
   Broadcast Stateful Event!(
    execute=@n40.Is Valid
    self:Gameplay Event Router Component Object Reference=@n39
    EventData:Fort Playlist Travel Event Request Show Travel Initiator Structure=Make Fort Playlist Travel Event Request Show Travel Initiator!(InitiatorOptions:Fort Travel Initiator Options Structure=Make Fort Travel Initiator Options!(
      Destinations:Array of Fort Travel Playlist Data Structures=@n41
      bShouldSkipToast:bool=False
      TravelGroupSetting:EFortTravelGroupSetting Enum=SquadTravel
      PrepTime:float=5
     ) -> (FortTravelInitiatorOptions:Fort Travel Initiator Options Structure) {is=title K2Node_MakeStruct}) -> (FortPlaylistTravelEvent_RequestShowTravelInitiator:Fort Playlist Travel Event Request Show Travel Initiator Structure) {is=title K2Node_MakeStruct}
    EventContext:Object Reference=@n38
   ) -> (then, ReturnValue:bool)
   PrintString(
    execute=@n37.else
    InString="WBP_Pilgrim_UI_Results_Base::OnConcordGrapeButtonClicked: Can't go to ConcordGrape, CVAR off"
    bPrintToScreen=false
   )
   "On Clicked (TouchArea_Results)"() {is=title K2Node_ComponentBoundEvent}
   > SelectTabByIndex(self=Get WBP_UIKit_Tabs_TopNavBar(), Index=0, bSuppressClickFeedback=true)
   n42 = "On Pressed (Results_LeaveStage_Button)"!() -> (OutputDelegate, then, Button:Common Button Base Object Reference) {is=title K2Node_ComponentBoundEvent}
   n43 = "On Clicked (Results_ConcordGrape_Button)"!() -> (OutputDelegate, then, Button:Common Button Base Object Reference) {is=title K2Node_ComponentBoundEvent}
   n44 = "On Pressed (Results_ReplaySongButton)"!() -> (OutputDelegate, then, Button:Common Button Base Object Reference) {is=title K2Node_ComponentBoundEvent}
  ]}
 ChangeTrackIfMissingFromUpcomingSong {nodes [
   ChangeTrackIfMissingFromUpcomingSong() {is=title K2Node_FunctionEntry}
   n1 = GetOwningPlayer()
   > n2 = GetPilgrimQuickplayPlayerStateComponent!(
    execute
    PlayerState:Player State Object Reference=Get PlayerState!(self:Controller Object Reference=GetFortPlayerControllerFromActor(Actor=@n1)) -> (PlayerState:Player State Object Reference)
   ) -> (then, ReturnValue:Pilgrim Quickplay Player State Component Object Reference)
   > ChangeTrackIfMissingFromUpcomingSong(self:Pilgrim Quickplay Player State Component Object Reference=@n2) {is=title K2Node_CallFunction}
  ]}
 CreatePlayerNewHighScoreWidget {nodes [
   n1 = CreatePlayerNewHighScoreWidget!() -> (then, PlayerState:Player State Object Reference, PlayerIndex:int) {is=title K2Node_FunctionEntry}
   > n2 = GetPilgrimQuickplayPlayerStateComponent(PlayerState=@n1)
   n3 = GetCurrentSetlistIndex!(self:Pilgrim Quickplay Band State Component Object Reference=GetBandState!(self:Pilgrim Quickplay Player State Component Object Reference=@n2) -> (ReturnValue:Pilgrim Quickplay Band State Component Object Reference)) -> (ReturnValue:int=0)
   > Branch(Condition=GetReachedNewHighscore!(self:Pilgrim Quickplay Player State Component Object Reference=@n2, BestType:EPilgrimPersonalBestType Enum=SoloScore, SetlistIndex:int=@n3) -> (ReturnValue:bool))
   > n4 = Create Widget!(
    execute:Exec
    Class:User Widget Class Reference=&/WBP_PilgrimQuickly_NewHighScoreRankPanel.WBP_PilgrimQuickly_NewHighScoreRankPanel_C
    OwningPlayer:Player Controller Object Reference=GetOwningPlayer()
    CenterHighScoreAlert:bool=True
   ) -> (then:Exec, ReturnValue:WBP Pilgrim Quickly New High Score Rank Panel Object Reference)
   > n5 = AddChildToGrid!(
    execute
    self:Grid Panel Object Reference=Get PlayerInfoGridPanel() -> (PlayerInfoGridPanel:Grid Panel Object Reference)
    Content:Widget Object Reference=@n4
    InRow:int=2
    InColumn:int=@n1.PlayerIndex
   ) -> (then, ReturnValue:Grid Slot Object Reference)
   > SetPadding!(
    execute
    self:Grid Slot Object Reference=@n5
    InPadding_Left:float=0
    InPadding_Top:float=95
    InPadding_Right:float=0
    InPadding_Bottom:float=0
   ) -> (then)
   > SetVerticalAlignment(self:Grid Slot Object Reference=@n5, InVerticalAlignment=VAlign_Top)
   > Add!(
    execute
    TargetMap:Map of Player State Object References to WBP Pilgrim Quickly New High Score Rank Panel Object References=Get PlayerNewHighScoreDisplays() -> (PlayerNewHighScoreDisplays:Map of Player State Object References to WBP Pilgrim Quickly New High Score Rank Panel Object References)
    Key:"Player State Object Reference (by ref)"=@n1
    Value:"WBP Pilgrim Quickly New High Score Rank Panel Object Reference (by ref)"=@n4
   ) -> (then) {is=title K2Node_CallFunction}
   > SetPlayerState!(
    execute
    self:WBP Pilgrim Quickly New High Score Rank Panel Object Reference=@n4
    PlayerState:Player State Object Reference=@n1
    SetlistIndex:int=@n3
    IsSolo:bool=true
   ) -> (then)
  ]}
 CreateSplitcreenPartPickerWidget {nodes [
   n1 = CreateSplitcreenPartPickerWidget!() -> (then, PlayerState:Player State Object Reference, PlayerIndex:int) {is=title K2Node_FunctionEntry}
   > Branch(Condition=IsHamSandwichSessionActive!() -> (ReturnValue:bool))
   n2 = GetPlayerController!(self:Player State Object Reference=@n1) -> (ReturnValue:Player Controller Object Reference)
   n3 = IsPrimaryPlayer!(self:Player Controller Object Reference=@n2) -> (ReturnValue:bool)
   > Branch(Condition=AND Boolean(A=NOT Boolean(A=@n3), B=IsLocalPlayerController!(self:Controller Object Reference=@n2) -> (ReturnValue:bool)))
   > n4 = Branch(Condition=IsUnifiedPartPickerEnabled())
   > n5 = Create Widget!(
    execute
    Class=/PilgrimQuickplay/Widgets/UniversalPartPicker/WBP_PilgrimQuickplay_PartPicker_LMPResults.WBP_PilgrimQuickplay_PartPicker_LMPResults_C
    OwningPlayer=@n2
    bIsPartPickerForLMP:bool=true
   ) -> (then, ReturnValue:WBP Pilgrim Quickplay Part Picker LMPResults Object Reference)
   > Add(
    TargetArray:Array of WBP Pilgrim Quickplay Part Picker LMPResults Object References=Get UniversalPartPickerArray()
    NewItem:"WBP Pilgrim Quickplay Part Picker LMPResults Object Reference (by ref)"=@n5
   ) {is=title K2Node_CallArrayFunction}
   > n6 = AddChildToGrid(self=Get PlayerInfoGridPanel(), Content=@n5, InRow=1, InColumn=@n1.PlayerIndex)
   n7 = Create Widget!(
    execute=@n4.else
    Class=/PilgrimQuickplay/Widgets/WBP_PilgrimQuickplay_Results_SplitscreenPartPicker.WBP_PilgrimQuickplay_Results_SplitscreenPartPicker_C
    OwningPlayer=@n2
   ) -> (then, ReturnValue:WBP Pilgrim Quickplay Results Splitscreen Part Picker Object Reference)
   n8 = AddChildToGrid(execute=@n9.then, self=Get PlayerInfoGridPanel(), Content=@n7, InRow=1, InColumn=@n1.PlayerIndex)
   SetRowSpan!(execute=[@n8.then, @n6.then], self:Grid Slot Object Reference=[@n8, @n6], InRowSpan:int=2) -> (then)
   > SetVerticalAlignment(self=[@n8, @n6], InVerticalAlignment=VAlign_Center)
   > PrintString(
    InString=Append!(A="Created SplitscreenPartPicker - ", B=GetDisplayName(Object=@n2), C:String=" - Owning player: ", D:String=GetDisplayName(Object=@n2)) -> (ReturnValue)
    Duration=10
   )
   n9 = Add(
    execute=@n7.then
    TargetArray:Array of WBP Pilgrim Quickplay Results Splitscreen Part Picker Object References=Get PartPickerArray()
    NewItem:"WBP Pilgrim Quickplay Results Splitscreen Part Picker Object Reference (by ref)"=@n7
   )
  ]}
 EventGraph {nodes [
   n1 = UpdatePlayerInfoElements!() -> (OutputDelegate, then, PlayerState:Player State Object Reference, ResultsLocationIndex:int) {is=title K2Node_CustomEvent}
   n2 = Get PlayersToDisplay() -> (PlayersToDisplay:int=0)
   > PrintString(InString=Append!(
     A="Song Results: Update Player Info Elements: "
     B=GetDisplayName(Object=@n1.PlayerState)
     C=" Location Index: "
     D="To String (Integer)"(InInt=@n1.ResultsLocationIndex)
     E:String=" Players to Display: "
     F:String="To String (Integer)"(InInt=@n2)
    ) -> (ReturnValue))
   > Branch(Condition="Not Equal (Integer)"(A=@n1.ResultsLocationIndex, B=-1))
   > n3 = Sequence!(execute) -> (then_0, then_1, then_2:Exec)
   n4 = GetComponentByClass(
    self=@n1.PlayerState
    ComponentClass=&BP_PilgrimQuickplay_ResultsLocationPlayerStateComponent_C
   ) -> (ReturnValue:BP Pilgrim Quickplay Results Location Player State Component Object Reference)
   > Is Valid(InputObject=@n4)
   > Unbind Event from On Results Location Index Replicated(self:BP Pilgrim Quickplay Results Location Player State Component Object Reference=@n4, Delegate=Create Event()) {is=title K2Node_RemoveDelegate}
   n5 = Sequence!(execute=@n3.then_1) -> (then_0, then_1)
   > Update Above Player Info and Stat Widgets!(execute, self, PlayerState:Player State Object Reference=@n1.PlayerState, PlayerIndex:int=@n1.ResultsLocationIndex) -> (then)
   CreatePlayerNewHighScoreWidget!(execute=@n5.then_1, self, PlayerState:Player State Object Reference=@n1.PlayerState, PlayerIndex:int=@n1.ResultsLocationIndex) -> (then) {is=title K2Node_CallFunction}
   Decrement Int(" "=@n3.then_2, Value=Get PlayersToDisplay() -> (PlayersToDisplay=0))
   n6 = Get PlayersToDisplay() -> (PlayersToDisplay=0)
   > PrintString(InString=Append(
     A="Song Results: Update Player Info Elements: Decrementing Players to Display: "
     B=GetDisplayName(Object=@n1.PlayerState)
     C=" Location Index: "
     D="To String (Integer)"(InInt=@n1.ResultsLocationIndex)
     E=" Players to Display: "
     F="To String (Integer)"(InInt=@n6)
    ))
   PlayResultsRevealAnimation() {is=title K2Node_CustomEvent}
   > SetVisibility(
    self=Get BandStats() -> (BandStats:WBP Pilgrim Quickplay Song Results Band Stats Object Reference)
    InVisibility=Select(Option 0:ESlateVisibility Enum=Collapsed, Option 1:ESlateVisibility Enum=SelfHitTestInvisible, Index=Get MoreThanOnePlayerInBand()) -> (ReturnValue:ESlateVisibility Enum)
   )
   > PlaySound2D(Sound=Get InitialRevealSound() -> (InitialRevealSound:Sound Base Object Reference))
   > SetupPepipheralInstrumentVisual()
   > n7 = Keys!(
    execute
    TargetMap:Map of Integers to WBP Pilgrim Quickplay Results Above Player Display Object References=Get ActiveAbovePlayerInfoDisplays() -> (ActiveAbovePlayerInfoDisplays:Map of Integers to WBP Pilgrim Quickplay Results Above Player Display Object References)
   ) -> (then, Keys:Array of Integers)
   > Sort Integer Array!(execute, TargetArray:Array of Integers=@n7, bStableSort:bool=false, SortOrder:EArraySortOrder Enum=Ascending) -> (then)
   > n8 = For Each Loop with Delay!(Exec:Exec, Array:Array of Integers=@n7, Delay:double=2) -> (Loop Body:Exec, Array Element:int, Array Index:int, Completed:Exec) {is=title K2Node_MacroInstance}
   n9 = Find(
    TargetMap:Map of Integers to WBP Pilgrim Quickplay Results Above Player Display Object References=Get ActiveAbovePlayerInfoDisplays()
    Key:"Integer (by ref)"=@n8
   ) -> (Value:WBP Pilgrim Quickplay Results Above Player Display Object Reference)
   > Branch(Condition=@n9.ReturnValue)
   > CheckIfMisingPart(self:WBP Pilgrim Quickplay Results Above Player Display Object Reference=@n9)
   > PlayerDisplayIntro(self:WBP Pilgrim Quickplay Results Above Player Display Object Reference=@n9)
   > PlaySound2D(Sound=Get PlayerScoreRevealSound() -> (PlayerScoreRevealSound:Sound Base Object Reference))
   DeactivateWidget(execute=@n8.Completed, self=Get EmoteButton() -> (EmoteButton:WBP Pilgrim Quickplay Song Results Emote Button Object Reference))
   > SetButtonVisibility()
   > PlayAnimation(InAnimation=Get Anim_PerformanceResultsPanelIntro() -> (Anim_PerformanceResultsPanelIntro:Widget Animation Object Reference))
   > PlayAnimation(InAnimation=Get Anim_MoreDetailsIntro() -> (Anim_MoreDetailsIntro:Widget Animation Object Reference))
   > PlayAnimation(InAnimation=Get Anim_SetlistNameIntro() -> (Anim_SetlistNameIntro:Widget Animation Object Reference))
   > PlaySound2D(Sound=Get TabsRevealSound() -> (TabsRevealSound:Sound Base Object Reference))
   > PlayAnimation(InAnimation=Get Anim_SwitcherAndTabIntro() -> (Anim_SwitcherAndTabIntro:Widget Animation Object Reference))
   > SelectTabByIndex(self=Get WBP_UIKit_Tabs_TopNavBar(), Index=0, bSuppressClickFeedback=true)
   > ChangeTrackIfMissingFromUpcomingSong(self:Self Object Reference)
   > n10 = GetWidgetToFocus!(execute, self) -> (then, WidgetToFocus:Widget Object Reference)
   > SetFocus(self=@n10)
   > Set CanSwitchTabs(CanSwitchTabs:bool=true) -> (Output_Get:bool)
   n11 = Get MoreThanOnePlayerInBand()
   > SetActiveWidgetIndex(
    self:Common Visibility Switcher Object Reference=Get StarVisibilitySwitcher() -> (StarVisibilitySwitcher:Common Visibility Switcher Object Reference)
    Index="To Integer (Boolean)"!(InBool:bool=@n11) -> (ReturnValue:int=0)
   )
   > PlaySound2D(Sound=Select(
     Option 0:Sound Base Object Reference=Get HighScorePanelRevealSound() -> (HighScorePanelRevealSound:Sound Base Object Reference)
     Option 1:Sound Base Object Reference=Get HighScorePanelRevealSoundBand() -> (HighScorePanelRevealSoundBand:Sound Base Object Reference)
     Index=@n11
    ) -> (ReturnValue:Sound Base Object Reference))
   > n12 = Branch(Condition=Get MoreThanOnePlayerInBand())
   n13 = Get PilgrimQuickplaySongResultsVM() -> (PilgrimQuickplaySongResultsVM:Pilgrim Quickplay Song Results VM Object Reference)
   > n14 = StaggeredStarReveal!(
    execute
    self:WBP Pilgrim Quickplay Band Star Bar Object Reference=Get BandStarBar() -> (BandStarBar:WBP Pilgrim Quickplay Band Star Bar Object Reference)
    StarCount:int=Get BandStars!(self:Pilgrim Quickplay Song Results VM Object Reference=@n13) -> (BandStars:int=0)
   ) -> (then)
   ResultsScreenCountUp(
    execute=[@n15.then, @n14.then]
    self:WBP Pilgrim Quickplay Band Score Panel Object Reference=Get BandScorePanel() -> (BandScorePanel:WBP Pilgrim Quickplay Band Score Panel Object Reference)
   )
   > PlayAnimation(InAnimation=Get Anim_TimerBar(), PlaybackSpeed=SafeDivide(A=1, B=Get TabAutoAdvanceTime() -> (TabAutoAdvanceTime:float)))
   > SetVisibility(self=Get TimerBarOverlay() -> (TimerBarOverlay:Overlay Object Reference), InVisibility=SelfHitTestInvisible)
   > Delay(Duration=Get TabAutoAdvanceTime())
   > Branch(Condition=Get TabsHaveNeverBeenSwitched())
   > PlayTabTransitionAnims()
   > PlaySound2D(Sound=Get First Tab Reveal Sound() -> (First Tab Reveal Sound:Sound Base Object Reference))
   > SelectTabByIndex(self=Get WBP_UIKit_Tabs_TopNavBar(), Index=1, bSuppressClickFeedback=true)
   > Branch(Condition=IsUnifiedPartPickerEnabled())
   > ActivateWidget(self=Get UPPNavOverrideWidget() -> (UPPNavOverrideWidget:Common Activatable Widget Object Reference))
   n15 = StaggeredStarReveal(
    execute=@n12.else
    self:WBP Pilgrim Quickplay Local Star Bar Object Reference=Get StarBar() -> (StarBar:WBP Pilgrim Quickplay Local Star Bar Object Reference)
    StarCount=Get LocalPlayerStars!(self:Pilgrim Quickplay Song Results VM Object Reference=@n13) -> (LocalPlayerStars:int=0)
   )
   PlayTabTransitionAnims() {is=title K2Node_CustomEvent}
   > PlayAnimation(InAnimation=Get Anim_NextTabFade() -> (Anim_NextTabFade:Widget Animation Object Reference))
   > PlayAnimation(InAnimation=Get Anim_ProgressBackgroundFadeIn() -> (Anim_ProgressBackgroundFadeIn:Widget Animation Object Reference))
   > PlayProgressScreenIntroAnim(self:WBP Pilgrim Quickplay Progress Tab Object Reference=Get WBP_PilgrimQuickplay_ProgressTab() -> (WBP_PilgrimQuickplay_ProgressTab:WBP Pilgrim Quickplay Progress Tab Object Reference))
   n16 = SetPartPickerVisibility!() -> (OutputDelegate, then, ShowDetails:bool=~) {is=title K2Node_CustomEvent}
   > DelayUntilNextTick!(execute) -> (then)
   > n17 = Branch(Condition=IsUnifiedPartPickerEnabled())
   > n18 = For Each Loop(Array:Array of WBP Pilgrim Quickplay Part Picker LMPResults Object References=Get UniversalPartPickerArray()) -> (Array Element:WBP Pilgrim Quickplay Part Picker LMPResults Object Reference)
   n19 = IsFirstTab!() -> (ReturnValue:bool)
   n20 = Get AtEndOfSetlist()
   n21 = Select(
    Option 0:Pilgrim_SplitscreenPickerDetails Enum=Select(Option 0=NewEnumerator0, Option 1=NewEnumerator2, Index=OR Boolean(A=@n19, B=@n20))
    Option 1:Pilgrim_SplitscreenPickerDetails Enum=NewEnumerator1
    Index=@n16.ShowDetails
   ) -> (ReturnValue:Pilgrim_SplitscreenPickerDetails Enum)
   > SwitchPartPickerModules!(
    execute
    self:WBP Pilgrim Quickplay Part Picker LMPResults Object Reference=@n18
    Pilgrim_SplitscreenPickerDetails:Pilgrim_SplitscreenPickerDetails Enum=@n21
   ) -> (then)
   n22 = For Each Loop(Exec=@n17.else, Array:Array of WBP Pilgrim Quickplay Results Splitscreen Part Picker Object References=Get PartPickerArray()) -> (Array Element:WBP Pilgrim Quickplay Results Splitscreen Part Picker Object Reference)
   > SwitchPartPickerModules(self:WBP Pilgrim Quickplay Results Splitscreen Part Picker Object Reference=@n22, Pilgrim_SplitscreenPickerDetails=@n21)
   ShowCompleteUI AsNeeded() {is=title K2Node_CustomEvent}
   > DelayUntilNextTick()
   > Branch(Condition=Get AtEndOfSetlist())
   > n23 = EvaluateCurveTableRow!(
    execute:Exec
    CurveTable:Curve Table Object Reference=/PilgrimQuickplay/DataTables/PilgrimQuickplayGameData
    RowName:Name=Results.CuratedSetlist.SetlistCompleteNoticeDelay
    InXY:float=0
    ContextString:String
   ) -> (RowFound:Exec, RowNotFound:Exec, OutXY:float) {is=title K2Node_CallDataTableFunction}
   > n24 = Set CuratedSetlistCompleteNoticeDelay(CuratedSetlistCompleteNoticeDelay:float=@n23) -> (Output_Get:float)
   Delay(execute=[@n24.then, @n23.RowNotFound], Duration=Get CuratedSetlistCompleteNoticeDelay() -> (CuratedSetlistCompleteNoticeDelay:float))
   n25 = Get PilgrimCuratedSetlistVM() -> (PilgrimCuratedSetlistVM:Pilgrim Curated Setlist VM Object Reference)
   > AddSceneToFirstLocalPlayer!(
    execute
    self:Dynamic UIManager Object Reference=Get DynamicUIManager()
    Scene:Dynamic UIScene Object Reference=Select(
     Option 0:Dynamic UIScene Object Reference=/PilgrimQuickplay/Widgets/Scenes/UIScene_PilgrimQuickplay_CuratedSetlist_SetlistCompleteCommonOverlay
     Option 1:Dynamic UIScene Object Reference=/PilgrimQuickplay/Widgets/Scenes/UIScene_PilgrimQuickplay_CuratedSetlist_SetlistCompleteOverlay
     Index=Get bIsCuratedSetlist!(self:Pilgrim Curated Setlist VM Object Reference=@n25) -> (bIsCuratedSetlist:bool)
    ) -> (ReturnValue:Dynamic UIScene Object Reference)
   ) -> (then)
   SetupResults() {is=title K2Node_CustomEvent}
   > SetupTabs()
   > ResetIntroAnims()
   > Add Injected Animation Events() {is=title K2Node_CallFunction}
   > PlayAnimation(InAnimation=Get Anim_EmoteIntro() -> (Anim_EmoteIntro:Widget Animation Object Reference))
   n26 = GetSparksMusicPlayspaceOfClass!(
    PlayspaceType:Sparks Music Playspace Class Reference=/Script/PilgrimQuickplayRuntime.PilgrimQuickplayPlayspace
    Actor:Actor Object Reference=GetOwningPlayer()
    bMatchChildTypes:bool=false
   ) -> (ReturnValue:Pilgrim Quickplay Playspace Object Reference)
   > n27 = Is Valid(InputObject=@n26)
   PrintString(execute=@n27.Is Not Valid, InString=Invalid playspace on results screen)
   Setup Local Variables!(execute=@n27.Is Valid, self, QuickplayPlayspace:Pilgrim Quickplay Playspace Object Reference=@n26) -> (then)
   > PrintString(InString=Song Results: Setup Local Variables)
   > ShowCompleteUI AsNeeded() {is=title K2Node_CallFunction}
   > Set AbovePlayerInfoDisplays(AbovePlayerInfoDisplays:Array of WBP Pilgrim Quickplay Results Above Player Display Object References=Make Array!(
     "[0]":WBP Pilgrim Quickplay Results Above Player Display Object Reference=Get PlayerInfo_1() -> (PlayerInfo_1:WBP Pilgrim Quickplay Results Above Player Display Object Reference)
     "[1]":WBP Pilgrim Quickplay Results Above Player Display Object Reference=Get PlayerInfo_2() -> (PlayerInfo_2:WBP Pilgrim Quickplay Results Above Player Display Object Reference)
     "[2]":WBP Pilgrim Quickplay Results Above Player Display Object Reference=Get PlayerInfo_3() -> (PlayerInfo_3:WBP Pilgrim Quickplay Results Above Player Display Object Reference)
     "[3]":WBP Pilgrim Quickplay Results Above Player Display Object Reference=Get PlayerInfo_4() -> (PlayerInfo_4:WBP Pilgrim Quickplay Results Above Player Display Object Reference)
    ) -> (Array:Array of WBP Pilgrim Quickplay Results Above Player Display Object References)) -> (Output_Get:Array of WBP Pilgrim Quickplay Results Above Player Display Object References)
   n28 = Get AllPlayerStates()
   > Set PlayersToDisplay(PlayersToDisplay:int=Length(TargetArray:Array of Player State Object References=@n28)) -> (Output_Get:int=0)
   > n29 = For Each Loop(Array:Array of Player State Object References=@n28) -> (Array Element:Player State Object Reference)
   n30 = GetComponentByClass(
    self=@n29
    ComponentClass=&BP_PilgrimQuickplay_ResultsLocationPlayerStateComponent_C
   )
   > n31 = Is Valid(InputObject=@n30)
   PrintString(execute=@n31.Is Not Valid, InString=Invalid: Song Results: BP_PilgrimQuickplay_ResultsLocationPlayerStateComponent)
   PrintString(execute=@n31.Is Valid, InString=Song Results: Update Player Info Elements)
   > SetFlawless()
   > UpdatePlayerInfoElements!(
    execute
    self
    PlayerState:Player State Object Reference=@n29
    ResultsLocationIndex:int=Get ResultsLocationIndex(self=@n30) -> (ResultsLocationIndex=0)
   ) -> (then) {is=title K2Node_CallFunction}
   > Bind Event to On Results Location Index Replicated(self:BP Pilgrim Quickplay Results Location Player State Component Object Reference=@n30, Delegate=@n1) {is=title K2Node_AddDelegate}
   SetPlayerState(
    execute=@n29.Completed
    self=Get BandNewHighScore() -> (BandNewHighScore:WBP Pilgrim Quickly New High Score Rank Panel Object Reference)
    PlayerState=Get PlayerState(self=GetOwningPlayer())
    SetlistIndex=GetCurrentSetlistIndex(self=Get LocalBandState() -> (LocalBandState:Pilgrim Quickplay Band State Component Object Reference)) -> (ReturnValue=0)
    IsSolo=false
   )
   > SetupVisuals()
   > UpdateHistogram()
   > n32 = GetConsoleVariableBoolValue(VariableName=Pilgrim.SongResults.CheckRevealedAnimations)
   > PrintString(InString=Append!(A="Pilgrim.SongResults.CheckRevealedAnimations: ", B="To String (Boolean)"(InBool=@n32)) -> (ReturnValue))
   > Branch(Condition=@n32)
   > n33 = GetConsoleVariableFloatValue!(execute, VariableName:String=Pilgrim.SongResults.AnimationTimer) -> (then, ReturnValue:float)
   > PrintString(InString=Append(A="Pilgrim.SongResults.AnimationTimer: ", B="To String (Float)"(InDouble=@n33)))
   > Delay(Duration=@n33)
   n34 = Get PlayResultsRevealAnimationHasBeenCalled() -> (PlayResultsRevealAnimationHasBeenCalled:bool)
   > PrintString(InString=Append(A="Play Results Reveal Animation Has Been Called: ", B="To String (Boolean)"(InBool=@n34)))
   > n35 = Branch(Condition=@n34)
   PlayResultsRevealAnimation(execute=@n35.else) {is=title K2Node_CallFunction}
   > Set PlayResultsRevealAnimationHasBeenCalled(PlayResultsRevealAnimationHasBeenCalled:bool=true) -> (Output_Get:bool)
   > AddSplitscreenPartPickerWidgets() {is=title K2Node_CallFunction}
   > Branch(Condition=Get AtEndOfSetlist())
   > ShowSetlistComplete!(
    execute
    self:WBP Pilgrim Quickplay Song Results Setlist Status Object Reference=Get WBP_PilgrimQuickplay_SongResults_SetlistStatus() -> (WBP_PilgrimQuickplay_SongResults_SetlistStatus:WBP Pilgrim Quickplay Song Results Setlist Status Object Reference)
    ShouldShow:bool=false
   ) -> (then)
   n36 = Event PreConstruct()
   > Parent: PreConstruct(IsDesignTime=@n36.IsDesignTime)
   > PrintString(InString=Attempting to setup the tabs via PreConstruct, Duration=5)
   > SetupTabs()
   Event Construct()
   > Parent: Construct()
   > SetupResults() {is=title K2Node_CallFunction}
   > SetVisibility(
    self=Get WBP_PilgrimTextChatMenuPrompt() -> (WBP_PilgrimTextChatMenuPrompt:WBP Pilgrim Text Chat Menu Prompt Object Reference)
    InVisibility=Hidden
   )
   > RegisterWidgetHost!(
    execute
    WidgetHost:Widget Object Reference
    WidgetHostTag:"Gameplay Tag Structure (by ref)"=Get UniversalPartPickerSlotTag() -> (UniversalPartPickerSlotTag:Gameplay Tag Structure)
    IsPlayerSpecific:bool=true
   ) -> (then)
   > RegisterWidgetHost(
    WidgetHostTag=Get UniversalPartPickerNavOverrideSlotTag() -> (UniversalPartPickerNavOverrideSlotTag:Gameplay Tag Structure)
    IsPlayerSpecific=true
   )
   Event On Deactivated()
   > n37 = Sequence!(execute) -> (then_0, then_1, then_2:Exec)
   n38 = GetComponentByClass(
    self=GetOwningPlayer()
    ComponentClass=&BP_PilgrimPlayerControllerComponent_C
   ) -> (ReturnValue:BP Pilgrim Player Controller Component Object Reference)
   > Is Valid(InputObject=@n38)
   > ToggleGreenroomMappingContext!(execute, self:BP Pilgrim Player Controller Component Object Reference=@n38, AddGreenroomMappingContext:bool=false) -> (then)
   > ToggleMappingContext!(
    execute
    self:BP Pilgrim Player Controller Component Object Reference=@n38
    AddMappingContext:bool=false
    MappingContext:Input Mapping Context Object Reference=/PilgrimQuickplay/Input/IMC_Pilgrim_SongResultsSkip
    Priority:int=1000
   ) -> (then)
   SetPickerEnabled!(
    execute=@n37.then_2
    self:Fort Player Controller Object Reference=GetFortPlayerControllerFromActor(Actor=GetOwningPlayer())
    bEnabled:bool=true
   ) -> (then)
   Stop Timer(
    execute=@n37.then_1
    self:WBP Pilgrim Quickplay Results Countdown Timer Object Reference=Get ResultsCountdownTimer() -> (ResultsCountdownTimer:WBP Pilgrim Quickplay Results Countdown Timer Object Reference)
   )
   > n39 = Branch(Condition=Get AtEndOfSetlist())
   > Send Results Return Analytics(ResultsReturnChoice=Get ResultsReturnChoice() -> (ResultsReturnChoice:EPilgrimAnalyticsResultsReturn Enum=None))
   > n40 = Set ResultsReturnChoice(ResultsReturnChoice=LeaveStage_Timer) -> (Output_Get=None)
   Branch(execute=[@n40.then, @n39.else], Condition=Get RestartSceneOn() -> (RestartSceneOn:bool))
   > RemoveScene!(
    execute
    self:Dynamic UIManager Object Reference=Get DynamicUIManager()
    Scene:Dynamic UIScene Object Reference=Get RestartSongScene()
    Player:"Player Controller Object Reference (by ref)"=GetFortPlayerControllerFromActor(Actor=GetOwningPlayer())
   ) -> (then)
   > Set RestartSceneOn(RestartSceneOn=false)
   Event On Activated()
   > SetPickerEnabled(self=GetFortPlayerControllerFromActor(Actor=GetOwningPlayer()), bEnabled=false)
   n41 = Event Add Content To Host Widget!() -> (
    OutputDelegate
    then
    WidgetHostTag:"Gameplay Tag Structure (by ref)"
    ContentWidget:Widget Object Reference
    Padding:"Margin Structure (by ref)"
    bFill:bool
   ) {is=title K2Node_Event}
   > n42 = Branch(Condition="Equal (GameplayTag)"(A=@n41.WidgetHostTag, B=Get UniversalPartPickerSlotTag()))
   > RemoveFromParent(self=Get VB_PickerInputUI() -> (VB_PickerInputUI:Vertical Box Object Reference))
   > n43 = Cast To CommonActivatableWidget!(execute, Object=@n41.ContentWidget) -> (then, CastFailed, AsCommon Activatable Widget:Common Activatable Widget Object Reference) {is=title K2Node_DynamicCast}
   > n44 = Set UPPWidgetRef(UPPWidgetRef:Common Activatable Widget Object Reference=@n43) -> (Output_Get:Common Activatable Widget Object Reference)
   > AddChildToGrid(self=Get GridPanel_0() -> (GridPanel_0:Grid Panel Object Reference), Content=@n44, InRow=0, InColumn=0)
   > DeactivateWidget(self=@n44)
   Branch(execute=@n42.else, Condition="Equal (GameplayTag)"(A=@n41.WidgetHostTag, B=Get UniversalPartPickerNavOverrideSlotTag()))
   > n45 = Cast To CommonActivatableWidget(Object=@n41.ContentWidget)
   > Set UPPNavOverrideWidget(UPPNavOverrideWidget:Common Activatable Widget Object Reference=@n45) -> (Output_Get:Common Activatable Widget Object Reference)
   > AddChildToGrid(self=Get GridPanel_0(), Content=@n41.ContentWidget, InRow=1, InColumn=0)
   Event Destruct()
   > UnregisterWidgetHost!(execute, WidgetHost:Widget Object Reference, WidgetHostTag:"Gameplay Tag Structure (by ref)"=Get UniversalPartPickerSlotTag()) -> (then)
   > UnregisterWidgetHost(WidgetHostTag=Get UniversalPartPickerNavOverrideSlotTag())
  ]}
 GetCanShowGoldBandStars {nodes [
   n1 = GetCanShowGoldBandStars!() -> (then, SongResultData:Pilgrim Quickplay Song Result Data Structure) {is=title K2Node_FunctionEntry}
   > n2 = For Each Loop with Break!(
    Exec:Exec
    Array:Array of Pilgrim Quickplay Song Result Player Data Structures=Break Pilgrim Quickplay Song Result Data!(PilgrimQuickplaySongResultData:"Pilgrim Quickplay Song Result Data Structure (by ref)"=@n1) -> (
     SongName:String
     SongTags:Gameplay Tag Container Structure
     SongShortName:Name
     BandData:Array of Pilgrim Quickplay Song Result Band Data Structures
     PlayerData:Array of Pilgrim Quickplay Song Result Player Data Structures
     Team:Byte
     LocalPlayerBandDataIndex:int
     LocalPlayerDataIndex:int
    ) {is=title K2Node_BreakStruct}.PlayerData
    Break:Exec=@n4.else
   ) -> (LoopBody:Exec, Array Element:Pilgrim Quickplay Song Result Player Data Structure, Array Index:int, Completed:Exec)
   Return Node!(execute=@n2.Completed, CanShowGoldBandStars:bool=Get CanShowGoldStars() -> (CanShowGoldStars:bool))
   > n3 = Set CanShowGoldStars(CanShowGoldStars:bool="Equal (Enum)"(
     A:EPilgrimSongDifficulty Enum=Break Pilgrim Quickplay Song Result Player Data!(PilgrimQuickplaySongResultPlayerData:"Pilgrim Quickplay Song Result Player Data Structure (by ref)"=@n2) -> (
      PlayerName:String
      PlayerAccountId:Unique Net Id Structure
      PlayerState:Player State Weak object
      InstrumentType:EPilgrimTrackType Enum
      InstrumentTags:Gameplay Tag Container Structure
      PlayerStars:int
      PlayerScore:int
      PlayerDifficulty:EPilgrimSongDifficulty Enum
      SongTrackDifficulty:int
      Team:bool=~
     ) {is=title K2Node_BreakStruct}.PlayerDifficulty
     B:EPilgrimSongDifficulty Enum=DifficultyExpert
    )) -> (Output_Get:bool)
   > n4 = Branch(Condition=@n3)
  ]}
 GetIsFlawlessPlayer {nodes [
   n1 = GetIsFlawlessPlayer!() -> (then, PlayerState:Player State Object Reference) {is=title K2Node_FunctionEntry}
   > n2 = GetPilgrimQuickplayPlayerStateComponent(PlayerState=@n1)
   n3 = Break Pilgrim Quickplay Player Performance Data!(PilgrimQuickplayPlayerPerformanceData:"Pilgrim Quickplay Player Performance Data Structure (by ref)"=GetPerformanceData!(self:Pilgrim Quickplay Player State Component Object Reference=@n2) -> (ReturnValue:"Pilgrim Quickplay Player Performance Data Structure (by ref)")) -> (
    Accuracy:float=~
    AverageOffset:float=~
    StandardDeviation:float=~
    FullCombo:bool=~
    NotesHit:int
    NotesPassed:int
    NotesMissed:int
    TotalNotes:int
    LongestStreak:int
    TimeInOverdriveMs:float=~
    HistogramSamples:Array of Pilgrim Histogram Sample Structures
    AccuracyTierCounts:Array of Integers
    SectionsCompletedByTier:Array of Integers
   ) {is=title K2Node_BreakStruct}
   > Return Node!(execute, IsFlawless:bool=AND Boolean(A="float >= float"(A=@n3, B=100), B="Equal (Integer)"(A=@n3.NotesMissed)))
  ]}
 GetWidgetToFocus {nodes [
   GetWidgetToFocus() {is=title K2Node_FunctionEntry}
   n1 = Get CommonVisibilitySwitcher()
   n2 = Get Results_LeaveStage_Button()
   n3 = Select(Option 0:Widget Object Reference=Get Results_Report_Button(), Option 1:Widget Object Reference=@n2, Index=GetIsEnabled(self=@n2)) -> (ReturnValue:Widget Object Reference)
   n4 = Get AtEndOfSetlist()
   n5 = Get Results_StartNextSongButton()
   n6 = "Equal (Enum)"(A:ESlateVisibility Enum=GetVisibility(self=Get SongResult()) -> (ReturnValue=Visible), B:ESlateVisibility Enum=Collapsed)
   n7 = Select(Option 0=Get InstrumentSelector(), Option 1=Get NextUp_StartNextSong(), Index=IsUnifiedPartPickerEnabled())
   n8 = Get Progress_LeaveStage_Button()
   n9 = Select(Option 0=Get Progress_Report_Button(), Option 1=@n8, Index=GetIsEnabled(self=@n8))
   > Return Node!(
    execute
    WidgetToFocus:Widget Object Reference=Select(
     Option 0=Select(Option 0=@n5, Option 1=@n3, Index:bool=@n4)
     Option 1=Select(Option 0=@n7, Option 1=@n9, Index=@n6)
     Index:int=GetActiveWidgetIndex(self=@n1) -> (ReturnValue=0)
    )
   )
  ]}
 IsFirstTab {nodes [
   IsFirstTab() {is=title K2Node_FunctionEntry}
   n1 = Get WBP_UIKit_Tabs_TopNavBar()
   > Return Node!(execute, ReturnValue:bool="Equal (Integer)"(A=GetSelectedTabIndex!(self:UIKit Tab Group Object Reference=@n1) -> (ReturnValue:int=0)))
  ]}
 OnAnalogValueChanged {nodes [
   n1 = OnAnalogValueChanged!() -> (then, MyGeometry:Geometry Structure, InAnalogInputEvent:Analog Input Event Structure) {is=title K2Node_FunctionEntry}
   n2 = GetKeyEventFromAnalogInputEvent!(Event:"Analog Input Event Structure (by ref)"=@n1.InAnalogInputEvent) -> (ReturnValue:Key Event Structure)
   > n3 = Branch(Condition="Equal (Key)"(A:Key Structure=GetKey!(Input:"Key Event Structure (by ref)"=@n2) -> (ReturnValue:Key Structure), B:Key Structure=Gamepad_RightY) -> (ReturnValue:bool=~) {is=title K2Node_PromotableOperator})
   > OnKeyDownQuestList!(
    execute
    self:WBP Pilgrim Quickplay Progress Tab Object Reference=Get WBP_PilgrimQuickplay_ProgressTab()
    AxisValue:double=GetAnalogValue!(Input:"Analog Input Event Structure (by ref)"=@n1.InAnalogInputEvent) -> (ReturnValue:float)
   ) -> (then)
   > Return Node(ReturnValue:Event Reply Structure=Handled())
   n4 = Parent: OnAnalogValueChanged!(execute=@n3.else, MyGeometry:Geometry Structure=@n1, InAnalogInputEvent:Analog Input Event Structure=@n1.InAnalogInputEvent) -> (then, ReturnValue:Event Reply Structure) {is=title K2Node_CallParentFunction}
   > Return Node(ReturnValue=@n4)
  ]}
 RemoveContentFromWidgetHost {nodes [
   Remove Content From Host Widget!() -> (
    then
    WidgetHostTag:"Gameplay Tag Structure (by ref)"
    ContentWidget:Widget Object Reference
    RemovalType:EWidgetSlottingRemovalType Enum=Unparent
   ) {is=title K2Node_FunctionEntry}
   > Return Node(ReturnValue:bool=false)
  ]}
 ResetIntroAnims {nodes [
   ResetIntroAnims() {is=title K2Node_FunctionEntry}
   > n1 = For Each Loop(Array:Array of Widget Animation Object References=Make Array!(
     "[0]":Widget Animation Object Reference=Get Anim_SwitcherAndTabIntro()
     "[1]":Widget Animation Object Reference=Get Anim_SetlistNameIntro()
     "[2]":Widget Animation Object Reference=Get Anim_PerformanceResultsPanelIntro()
     "[3]":Widget Animation Object Reference=Get Anim_NextTabFade()
     "[4]":Widget Animation Object Reference=Get Anim_ProgressBackgroundFadeIn()
     "[5]":Widget Animation Object Reference=Get Anim_MoreDetailsIntro()
     "[6]":Widget Animation Object Reference=Get Anim_EmoteIntro()
    ) -> (Array:Array of Widget Animation Object References)) -> (Array Element:Widget Animation Object Reference)
   > PlayAnimation(InAnimation=@n1, StartAtTime=GetEndTime(self=@n1), PlayMode=Reverse)
  ]}
 Send Results Return Analytics {nodes [
   n1 = Send Results Return Analytics!() -> (then, ResultsReturnChoice:EPilgrimAnalyticsResultsReturn Enum=None) {is=title K2Node_FunctionEntry}
   n2 = GetOwningPlayer()
   > n3 = Is Valid(InputObject=@n2)
   n4 = GetFortPlayerControllerFromActor(Actor=@n2)
   > n5 = Is Valid(InputObject=@n4)
   n6 = GetGameplayEventRouter!(ContextActor:Actor Object Reference=@n4) -> (ReturnValue:Gameplay Event Router Component Object Reference)
   > n7 = Is Valid(InputObject=@n6)
   > Broadcast Event!(
    execute
    self:Gameplay Event Router Component Object Reference=@n6
    EventData:Pilgrim Analytics Event Results Return Structure=Make Pilgrim Analytics Event Results Return!(PlayerController:Fort Player Controller Object Reference=@n4, ResultsReturn:EPilgrimAnalyticsResultsReturn Enum=@n1) -> (PilgrimAnalyticsEvent_ResultsReturn:Pilgrim Analytics Event Results Return Structure) {is=title K2Node_MakeStruct}
    EventContext:Object Reference=@n6
   ) -> (then, ReturnValue:bool)
   FortShippingLog!(
    execute=[@n7.Is Not Valid, @n3.Is Not Valid, @n5.Is Not Valid]
    WorldContextObject:Object Reference=Self-Reference()
    InString:String="WBP_PilgrimQuickplay_SongResults: Object was not vaild in  function \"SendResultsReturnAnalytics\""
    bLogAsWarning:bool=false
   ) -> (then)
  ]}
 Sequencer Events {nodes [
   Anim Show Setlist Complete!() -> (OutputDelegate, then, WBP_PilgrimQuickplay_SongResults_SetlistStatus:WBP Pilgrim Quickplay Song Results Setlist Status Object Reference) {is=title K2Node_CustomEvent}
   > Branch(Condition=isFinalSongResults!(self:WBP Pilgrim Quickplay Song Results Setlist Status Object Reference=Get WBP_PilgrimQuickplay_SongResults_SetlistStatus()) -> (ReturnValue:bool))
   > ShowSetlistComplete(self=Get WBP_PilgrimQuickplay_SongResults_SetlistStatus(), ShouldShow=true)
  ]}
 Set ControllerDisconnected Visibility {nodes [
   n1 = Set ControllerDisconnected Visibility!() -> (then, IsDisconnected:bool=~) {is=title K2Node_FunctionEntry}
   n2 = IsPlayingReplay!() -> (ReturnValue:bool)
   > n3 = Branch(Condition=AND Boolean(A=@n1, B=NOT Boolean(A=@n2)))
   n4 = IsVisible(self=Get ControllerDisconnectPrompt_NextUp() -> (ControllerDisconnectPrompt_NextUp:WBP Pilgrim Quickplay Controller Disconnect Prompt Object Reference))
   > n5 = Branch(Condition=@n4)
   n6 = Get Anim_ControllerDisconnect() -> (Anim_ControllerDisconnect:Widget Animation Object Reference)
   PlayAnimationForward(execute=@n5.else, InAnimation=@n6)
   Branch(execute=@n3.else, Condition=@n4)
   > PlayAnimationReverse(InAnimation=@n6)
  ]}
 SetButtonVisibility {nodes [
   SetButtonVisibility() {is=title K2Node_FunctionEntry}
   > SetConcordGrapeVisibility()
   > n1 = Branch(Condition=Get bHasResultMissingPart() -> (bHasResultMissingPart:bool))
   n2 = Get MoreDetailsPrompt() -> (MoreDetailsPrompt:WBP UIKit Button Quiet Object Reference)
   > n3 = SetVisibility(self=@n2, InVisibility=Collapsed)
   n4 = Branch(execute=[@n14.then, @n3.then], Condition=Get AtEndOfSetlist())
   n5 = GetConsoleVariableBoolValue(execute=[@n4.then, @n13.then], VariableName=Pilgrim.AllowSongRestartsFromResults)
   n6 = AllLocalPlayer!() -> (AllLocalPlayerController:bool) {is=title K2Node_CallFunction}
   > SetVisibility(
    self=Get Results_ReplaySongButton() -> (Results_ReplaySongButton:WBP FM Core Button Regular Object Reference)
    InVisibility=Select(Option 0:ESlateVisibility Enum=Collapsed, Option 1:ESlateVisibility Enum=SelfHitTestInvisible, Index=AND Boolean(A=@n5, B=@n6)) -> (ReturnValue:ESlateVisibility Enum)
   )
   n7 = Get MultiplayerState(self=Get PilgrimHamSandwichVM()) -> (MultiplayerState=Solo)
   n8 = Get MultiplayerState(self=Get PilgrimHamSandwichVM()) -> (MultiplayerState=Solo)
   > n9 = Branch(Condition=OR Boolean(A="Equal (Enum)"(A:EPilgrimMultiplayerState Enum=@n8, B:EPilgrimMultiplayerState Enum=Hybrid), B="Equal (Enum)"(A=@n7, B=FullRemote)))
   > SetVisibility(
    self=[
     Get Progress_Report_Button()
     Get Results_LeaveStage_Button()
     Get Results_ReturnToLobby_Button() -> (Results_ReturnToLobby_Button:WBP FM Core Button Regular Object Reference)
     Get Progress_LeaveStage_Button()
     Get Progress_ReturnToLobby_Button() -> (Progress_ReturnToLobby_Button:WBP FM Core Button Regular Object Reference)
     Get Results_Report_Button()
    ]
    InVisibility=Select(Option 0=Collapsed, Option 1=SelfHitTestInvisible, Index=Get AtEndOfSetlist())
   )
   n10 = Branch(execute=@n9.else, Condition=Get AtEndOfSetlist())
   > SetVisibility(
    self=[Get Results_ReturnToLobby_Button(), Get Progress_LeaveStage_Button(), Get Progress_ReturnToLobby_Button(), Get Results_LeaveStage_Button()]
    InVisibility=SelfHitTestInvisible
   )
   > n11 = GetConsoleVariableBoolValue(VariableName=Pilgrim.AllowSongRestartsFromResults)
   > SetVisibility(
    self=Get Progress_ReplaySongButton() -> (Progress_ReplaySongButton:WBP FM Core Button Regular Object Reference)
    InVisibility=Select(Option 0=Collapsed, Option 1=SelfHitTestInvisible, Index=@n11)
   )
   SetVisibility(execute=@n10.else, self=[Get Results_StartNextSongButton(), Get NextUp_StartNextSong()], InVisibility=SelfHitTestInvisible)
   > n12 = GetConsoleVariableBoolValue(VariableName=Pilgrim.AllowSongRestartsFromResults)
   > SetVisibility(
    self=Get NextUp_ReplaySong() -> (NextUp_ReplaySong:WBP FM Core Button Regular Object Reference)
    InVisibility=Select(Option 0=Collapsed, Option 1=SelfHitTestInvisible, Index=@n12)
   )
   n13 = SetVisibility(
    execute=@n4.else
    self=[Get DifficultySelector() -> (DifficultySelector:WBP Pilgrim Green Room Difficulty Select UI Object Reference), Get InstrumentSelector()]
    InVisibility=SelfHitTestInvisible
   )
   n14 = SetVisibility(execute=@n1.else, self=@n2, InVisibility=SelfHitTestInvisible)
  ]}
 SetConcordGrapeVisibility {nodes [
   SetConcordGrapeVisibility() {is=title K2Node_FunctionEntry}
   > n1 = GetConsoleVariableBoolValue(VariableName=Get ConcordGrapeCVAR())
   > n2 = Branch(Condition=@n1)
   n3 = Select(Option 0=Collapsed, Option 1=SelfHitTestInvisible, Index=Get AtEndOfSetlist())
   > SetVisibility(self=Get Results_ConcordGrape_Button() -> (Results_ConcordGrape_Button:WBP FM Core Button Regular Object Reference), InVisibility=@n3)
   > SetVisibility(self=Get Progress_ConcordGrape_Button() -> (Progress_ConcordGrape_Button:WBP FM Core Button Regular Object Reference), InVisibility=@n3)
   SetVisibility(execute=@n2.else, self=Get Results_ConcordGrape_Button(), InVisibility=Collapsed)
   > SetVisibility(self=Get Progress_ConcordGrape_Button(), InVisibility=Collapsed)
  ]}
 SetFlawless {nodes [
   SetFlawless() {is=title K2Node_FunctionEntry}
   > n1 = For Each Loop(Array:Array of Player State Object References=Get AllPlayerStates()) -> (Array Element:Player State Object Reference)
   > Is Valid(InputObject=@n1)
   > n2 = Branch(Condition="Equal (Integer)"(A=@n1.Array Index))
   > n3 = GetIsFlawlessPlayer!(execute, self, PlayerState:Player State Object Reference=@n1) -> (then, IsFlawless:bool) {is=title K2Node_CallFunction}
   > Set IsFlawlessSong(IsFlawlessSong:bool=@n3) -> (Output_Get:bool)
   n4 = GetIsFlawlessPlayer(execute=@n2.else, PlayerState=@n1)
   > Set IsFlawlessSong(IsFlawlessSong=AND Boolean(A=@n4, B=Get IsFlawlessSong() -> (IsFlawlessSong:bool)))
   SetVisibility(
    execute=@n1.Completed
    self=Get WBP_PilgrimQuickplay_FullComboIndicator() -> (WBP_PilgrimQuickplay_FullComboIndicator:WBP Pilgrim Quickplay Full Combo Indicator Object Reference)
    InVisibility=Select(Option 0=Collapsed, Option 1=SelfHitTestInvisible, Index=Get IsFlawlessSong())
   )
  ]}
 SetLocalPlayerStars {nodes [
   n1 = SetLocalPlayerStars!() -> (then, StarCount:int) {is=title K2Node_FunctionEntry}
   > SetupStarsResults_DEPRECATED!(
    execute
    self:WBP Pilgrim Quickplay Local Star Bar Object Reference=Get StarBar()
    Instrument:EPilgrimTrackType Enum=TrackNone
    StarCount:int=@n1
   ) -> (then)
  ]}
 Setup Local Variables {nodes [
   n1 = Setup Local Variables!() -> (then, QuickplayPlayspace:Pilgrim Quickplay Playspace Object Reference) {is=title K2Node_FunctionEntry}
   > Set AllPlayerStates(AllPlayerStates:Array of Player State Object References=Get All Player States!(self:Sparks Music Playspace Object Reference=@n1) -> (ReturnValue:Array of Player State Object References) {is=title K2Node_CallFunction}) -> (Output_Get:Array of Player State Object References)
   > n2 = Set LocalBandState(LocalBandState:Pilgrim Quickplay Band State Component Object Reference=GetBandState!(ContextActor:Actor Object Reference=@n1) -> (ReturnValue)) -> (Output_Get:Pilgrim Quickplay Band State Component Object Reference)
   n3 = GetNextSong!(self:Pilgrim Core Band State Component Object Reference=@n2) -> (ReturnValue:Pilgrim Song Metadata Object Reference)
   > Set AtEndOfSetlist(AtEndOfSetlist:bool=NOT Boolean(A=IsValid(Object=@n3))) -> (Output_Get:bool)
   > Clear(TargetArray:Array of WBP Pilgrim Quickplay Results Splitscreen Part Picker Object References=Get PartPickerArray())
  ]}
 SetupPepipheralInstrumentVisual {nodes [
   SetupPepipheralInstrumentVisual() {is=title K2Node_FunctionEntry}
   n1 = GetControllerFromActor!(Actor:Actor Object Reference=GetOwningPlayer()) -> (ReturnValue:Controller Object Reference)
   > Is Valid(InputObject=@n1)
   n2 = GetComponentByClass(self=@n1, ComponentClass=&BP_PilgrimPlayerControllerComponent_C)
   > Is Valid(InputObject=@n2)
   > SetVisibility(
    self=Get EmoteButton()
    InVisibility=Select(
     Option 0=SelfHitTestInvisible
     Option 1=Collapsed
     Index=HasConnectedPeripheral!(self:Pilgrim Player Controller Component Object Reference=@n2, Peripheral:EFMPeripheralInputMode Enum=Guitar) -> (ReturnValue:bool)
    )
   )
  ]}
 SetupTabs {nodes [
   SetupTabs() {is=title K2Node_FunctionEntry}
   > Clear(TargetArray:Array of UI Kit Tab Group Button View Model Object References=Get Local Tabs View Models() -> (Local Tabs View Models:Array of UI Kit Tab Group Button View Model Object References))
   n1 = Get WBP_UIKit_Tabs_TopNavBar()
   > Set Tabs Data!(
    execute
    self:WBP UIKit Tabs Base Object Reference=@n1
    InTabs Data:Array of UI Kit Tab Group Button View Model Object References=Get Local Tabs View Models()
   ) -> (then) {is=title K2Node_CallFunction}
   > n2 = For Each Loop(Array:Array of Texts=Get Tabs() -> (Tabs:Array of Texts)) -> (Array Element:Text)
   > n3 = Construct Object from Class(Class=/Script/FortniteUI.UIKitTabGroupButtonViewModel) -> (ReturnValue:UI Kit Tab Group Button View Model Object Reference)
   > Set Label!(execute, Label:Text=@n2, self:UI Kit Selectable Button View Model Object Reference=@n3) -> (then, Output_Get:Text)
   > Add(
    TargetArray:Array of UI Kit Tab Group Button View Model Object References=Get Local Tabs View Models()
    NewItem:"UI Kit Tab Group Button View Model Object Reference (by ref)"=@n3
   )
   Set Tabs Data(execute=@n2.Completed, self=@n1, InTabs Data=Get Local Tabs View Models())
  ]}
 SetupVisuals {nodes [
   SetupVisuals() {is=title K2Node_FunctionEntry}
   n1 = GetNextSong(self=Get LocalBandState())
   > Set Tabs(Tabs:Array of Texts=Select(
     Option 0:Array of Texts=Get TabsNextUp() -> (TabsNextUp:Array of Texts)
     Option 1:Array of Texts=Get TabsLastSong() -> (TabsLastSong:Array of Texts)
     Index=IsNullItem!(ItemObject:Object Reference=@n1) -> (Is Null Item:bool)
    ) -> (ReturnValue:Array of Texts)) -> (Output_Get:Array of Texts)
   > SetupTabs() {is=title K2Node_CallFunction}
   n2 = GetNextSong(self=Get LocalBandState())
   > SetVisibility(
    self=Get SongResult()
    InVisibility=Select(Option 0:ESlateVisibility Enum=SelfHitTestInvisible, Option 1:ESlateVisibility Enum=Collapsed, Index=IsNullItem(ItemObject=@n2)) -> (ReturnValue:ESlateVisibility Enum)
   )
   n3 = GetNextSong(self=Get LocalBandState())
   > SetVisibility(
    self=Get ProgressTab() -> (ProgressTab:Vertical Box Object Reference)
    InVisibility=Select(Option 0=Collapsed, Option 1=SelfHitTestInvisible, Index=IsNullItem(ItemObject=@n3))
   )
   > OverridePlayerInfo!(
    execute
    self:WBP Pilgrim Player Identity Pill Object Reference=Get PlayerIdentity_Pill() -> (PlayerIdentity_Pill:WBP Pilgrim Player Identity Pill Object Reference)
    PlayerState:Player State Object Reference=Get PlayerState(self=GetOwningPlayer())
   ) -> (then)
  ]}
 SwitchTab {nodes [
   n1 = SwitchTab!() -> (then, TabID:Name) {is=title K2Node_FunctionEntry}
   > PlaySound2D(Sound=Get TabSwitchSound() -> (TabSwitchSound:Sound Base Object Reference))
   n2 = "Equal (Name)"(A=@n1, B=0)
   n3 = Select(Option 0:int=1, Option 1:int=0, Index=@n2) -> (ReturnValue:int)
   > SetActiveWidgetIndex(self=Get CommonVisibilitySwitcher(), Index=@n3)
   n4 = Get AtEndOfSetlist()
   > n5 = Branch(Condition=AND Boolean!(A=IsUnifiedPartPickerEnabled(), B="Equal (Integer)"(A=@n3, B=1), C:bool=NOT Boolean(A=@n4)) -> (ReturnValue))
   n6 = Get UPPNavOverrideWidget()
   > n7 = Is Valid(InputObject=@n6)
   n8 = Get UPPWidgetRef() -> (UPPWidgetRef:Common Activatable Widget Object Reference)
   n9 = Is Valid(exec=[@n17.then, @n7.Is Not Valid], InputObject=@n8)
   > n10 = ActivateWidget(self=@n8)
   SetVisibility(execute=[@n20.then, @n19.Is Not Valid, @n10.then, @n9.Is Not Valid], self=Get TimerBarOverlay(), InVisibility=Collapsed)
   n11 = NOT Boolean(A=@n2)
   > SetVisibility(
    self=[Get BG() -> (BG:Border Object Reference), Get BG_1() -> (BG_1:Border Object Reference)]
    InVisibility=Select(
     Option 0:ESlateVisibility Enum=Collapsed
     Option 1:ESlateVisibility Enum=SelfHitTestInvisible
     Index=AND Boolean!(A=@n4, B=@n11) -> (ReturnValue)
    ) -> (ReturnValue:ESlateVisibility Enum)
   )
   n12 = AND Boolean(A=Get AtEndOfSetlist(), B=NOT Boolean(A=@n2))
   > SetVisibility(
    self=Get PlayerInfoHorizontalBox() -> (PlayerInfoHorizontalBox:Horizontal Box Object Reference)
    InVisibility=Select(Option 0=SelfHitTestInvisible, Option 1=Collapsed, Index=@n12)
   )
   n13 = Get bHasResultMissingPart()
   > SetVisibility(self=Get MoreDetailsPrompt(), InVisibility=Select(Option 0=SelfHitTestInvisible, Option 1=Collapsed, Index=OR Boolean(A=@n13, B=@n12)))
   > n14 = Branch(Condition=AND Boolean(A=Get AtEndOfSetlist(), B=NOT Boolean(A=@n2)))
   > n15 = StartProgressAnimation(self:WBP Pilgrim Quickplay Battle Pass Progress Object Reference=Get WBP_PilgrimQuickplay_BattlePass_Progress!(self:WBP Pilgrim Quickplay Progress Tab Object Reference=Get WBP_PilgrimQuickplay_ProgressTab()) -> (WBP_PilgrimQuickplay_BattlePass_Progress:WBP Pilgrim Quickplay Battle Pass Progress Object Reference))
   n16 = GetWidgetToFocus(execute=[@n15.then, @n14.else]) {is=title K2Node_CallFunction}
   > SetFocus(self=@n16)
   > Branch(Condition=IsHamSandwichSessionActive())
   > SetPartPickerVisibility!(execute, self, ShowDetails:bool=NOT Boolean(A=@n2)) -> (then) {is=title K2Node_CallFunction}
   > Set DetailedStatsAreShown(DetailedStatsAreShown:bool=true) -> (Output_Get:bool)
   > ToggleMoreDetails()
   n17 = ActivateWidget(execute=@n7.Is Valid, self=@n6)
   n18 = Is Valid(exec=@n5.else, InputObject=@n6)
   n19 = Is Valid(exec=[@n21.then, @n18.Is Not Valid], InputObject=@n8)
   > n20 = DeactivateWidget(self=@n8)
   n21 = DeactivateWidget(execute=@n18.Is Valid, self=@n6)
  ]}
 ToggleMoreDetails {nodes [
   ToggleMoreDetails() {is=title K2Node_FunctionEntry}
   > n1 = Set DetailedStatsAreShown(DetailedStatsAreShown=NOT Boolean(A=Get DetailedStatsAreShown() -> (DetailedStatsAreShown:bool)))
   > SetText!(
    execute
    self:Fort CTAButton Object Reference=Get MoreDetailsPrompt()
    ButtonText:"Text (by ref)"=Select(
     Option 0:"Text (by ref)"=Get ExpandStatsText() -> (ExpandStatsText:Text)
     Option 1:"Text (by ref)"=Get HideStatsText() -> (HideStatsText:Text)
     Index=@n1
    ) -> (ReturnValue:"Text (by ref)")
   ) -> (then)
   > SetPartPickerVisibility(ShowDetails=@n1)
   > SetActiveWidgetIndex(
    self:Widget Switcher Object Reference=Get SoloStatsSwitcher() -> (SoloStatsSwitcher:Common Activatable Widget Switcher Object Reference)
    Index=Select(Option 0:int=1, Option 1:int=0, Index=@n1) -> (ReturnValue:int)
   )
   > Branch(Condition=Get MoreThanOnePlayerInBand())
   > SetVisibility(
    self=Get BandStats()
    InVisibility=Select(Option 0:ESlateVisibility Enum=Collapsed, Option 1:ESlateVisibility Enum=SelfHitTestInvisible, Index=@n1) -> (ReturnValue:ESlateVisibility Enum)
   )
  ]}
 Update Above Player Info and Stat Widgets {nodes [
   n1 = Update Above Player Info and Stat Widgets!() -> (then, PlayerState:Player State Object Reference, PlayerIndex:int) {is=title K2Node_FunctionEntry}
   > n2 = Sequence!(execute) -> (then_0, then_1)
   n3 = "Get (a copy)"(
    Array:Array of WBP Pilgrim Quickplay Results Above Player Display Object References=Get AbovePlayerInfoDisplays() -> (AbovePlayerInfoDisplays:Array of WBP Pilgrim Quickplay Results Above Player Display Object References)
    Dimension 1=@n1.PlayerIndex
   ) -> (Output:WBP Pilgrim Quickplay Results Above Player Display Object Reference)
   Add(
    execute=@n2.then_1
    TargetMap:Map of Integers to WBP Pilgrim Quickplay Results Above Player Display Object References=Get ActiveAbovePlayerInfoDisplays()
    Key:"Integer (by ref)"=@n1.PlayerIndex
    Value:"WBP Pilgrim Quickplay Results Above Player Display Object Reference (by ref)"=@n3
   ) {is=title K2Node_CallFunction}
   > SetPlayerState!(execute, self:WBP Pilgrim Quickplay Results Above Player Display Object Reference=@n3, PlayerState=@n1) -> (then)
   > SetVisibility(self=@n3, InVisibility=HitTestInvisible)
   n4 = GetPlayerController(self=@n1)
   > Is Valid(InputObject=@n4)
   > Branch(Condition=IsPrimaryPlayer(self=@n4))
   > SetColumn!(
    execute
    self:Grid Slot Object Reference=SlotAsGridSlot(Widget=Get SoloStatsSwitcher())
    InColumn:int=Get PlayerIndex() -> (PlayerIndex:int=0)
   ) -> (then)
   Branch(execute=@n2.then_0, Condition=Is Local Player in Local MP Session!(PlayerState:Player State Object Reference=@n1) -> (ReturnValue:bool))
   > SetFixedOwner!(
    execute
    self:Pilgrim Fixed Owner Widget Object Reference=@n3
    PlayerController:Player Controller Object Reference=GetPlayerController(self=@n1)
   ) -> (then)
  ]}
 UpdateHistogram {nodes [
   UpdateHistogram() {is=title K2Node_FunctionEntry}
   n1 = GetOwningPlayer()
   > n2 = GetPilgrimQuickplayPlayerStateComponent(PlayerState=Get PlayerState(self=GetFortPlayerControllerFromActor(Actor=@n1)))
   > PopulateFromPerformanceData!(
    execute
    self:WBP Pilgrim Histogram Object Reference=Get Histogram() -> (Histogram:WBP Pilgrim Histogram Object Reference)
    PilgrimQuickplayPlayerPerformanceData:"Pilgrim Quickplay Player Performance Data Structure (by ref)"=GetPerformanceData(self=@n2)
   ) -> (then)
  ]}
 UpdateLeaveStageButtonForCuratedSetlist {nodes [
   n1 = UpdateLeaveStageButtonForCuratedSetlist!() -> (then, isCuratedSetlist:bool=~) {is=title K2Node_FunctionEntry}
   > Branch(Condition=@n1)
   n2 = Get BuildASetlist_Text() -> (BuildASetlist_Text:Text)
   > Set Text(Text=@n2, self:Fort CTAButton Object Reference=Get Results_LeaveStage_Button())
   > Set Text(Text=@n2, self=Get Progress_LeaveStage_Button())
  ]}
]
variables [
 AbovePlayerDisplayAlignment {type=Vector2D, default=<0.5 1>}
 AbovePlayerDisplayOffset {type=Vector2D, default=<0 -300>}
 AbovePlayerInfoDisplays {type="array<WBP_PilgrimQuickplay_Results_AbovePlayerDisplay_C>", default []}
 ActiveAbovePlayerInfoDisplays {default {}}
 AdvanceScreenCountdownTimeDefault {type=number, default=3}
 AllPlayerStates {type="array<PlayerState>", default []}
 AreYouSureWhoosh {type=SoundBase, default=/Game/Sounds/Fort_UI_Sounds/Gameplay/PostGame/Fortnite_UI_PostGame_AreYouSure_Whoosh_Cue}
 AtEndOfSetlist {type=boolean}
 BandBonusesSound {type=SoundBase, default=/PilgrimQuickplay/Audio/Results/Results_BandBonus_Cue}
 bHasResultMissingPart {type=boolean}
 BuildASetlist_Text {type=string, default=Make Custom Setlist}
 CanSwitchTabs {type=boolean}
 ConcordGrapeCVAR {type=string, default=Pilgrim.ConcordGrape.Enabled}
 CountdownAudioPingsRemaining {type=integer, default=3}
 CuratedSetlistCompleteNoticeDelay {type=number, default=2}
 CuratedSetlistCompleteWidget {type=UserWidget, default=null}
 DebugSkipToEnd {type=boolean}
 DetailedStatsAreShown {type=boolean}
 Event Router {type=GameplayEventRouterComponent, default=null}
 ExpandStatsText {type=string, default=More details}
 First Tab Reveal Sound {type=SoundBase, default=/PilgrimQuickplay/Audio/Results/MSS_FM_ResultsVictory_GainFestivalPoints}
 HideMoreDetails {type=boolean}
 HideStatsText {type=string, default=Hide details}
 HighScorePanelRevealSound {type=SoundBase, default=/PilgrimQuickplay/Audio/Results/MSS_FM_ResultsVictory_Tally1pAppear}
 HighScorePanelRevealSoundBand {type=SoundBase, default=/PilgrimQuickplay/Audio/Results/MSS_FM_ResultsVictory_TallyBandAppear}
 InitialRevealSound {type=SoundBase, default=/PilgrimQuickplay/Audio/Results/Results_IntroCrowd_Cue}
 IsFlawlessSong {type=boolean}
 LeaveExperienceConfirmationMessageText {type=string, default=Are you sure you want to return to lobby?}
 Local Tabs View Models {type="array<UIKitTabGroupButtonViewModel>", default []}
 LocalBandState {type=PilgrimQuickplayBandStateComponent, default=null}
 LocalPlayerSongResultData {
  type=PilgrimQuickplaySongResultPlayerData
  default {
   instrumentType=TrackNone
   playerAccountId=INVALID
   playerDifficulty=None
   playerName=""
   playerScore=0
   playerStars=0
   playerState=null
   songTrackDifficulty=0
   team=false
   instrumentTags {gameplayTags []}
  }
 }
 MoreDetailsPressedSound {type=SoundBase, default=/PilgrimQuickplay/Audio/Results/MSS_FM_ResultsVictory_Stats}
 MoreThanOnePlayerInBand {type=boolean}
 NextSongPromptsRevealSound {type=SoundBase, default=/SparksMaps/Audio/Wavs/StreakGained}
 NextSongTimerText {type=string, default=Next Song in:}
 PartPickerArray {type="array<WBP_PilgrimQuickplay_Results_SplitscreenPartPicker_C>", default []}
 PlayerNewHighScoreDisplayAlignment {type=Vector2D, default=<0.5 1>}
 PlayerNewHighScoreDisplayOffset {type=Vector2D, default=<0 200>}
 PlayerNewHighScoreDisplays {default {}}
 PlayerScoreRevealSound {type=SoundBase, default=&MSS_FM_ResultsVictory_Tally1pAppear}
 PlayersToDisplay {type=integer}
 PlayResultsRevealAnimationHasBeenCalled {type=boolean}
 RestartSceneOn {type=boolean}
 RestartSongScene {type=DynamicUIScene, default=/PilgrimQuickplay/Widgets/Scenes/UIScene_PilgrimQuickplay_RestartSong}
 ResultsReturnChoice {type=EPilgrimAnalyticsResultsReturn, default=LeaveStage_Timer}
 SetlistEndTimerText {type=string, default=Leaving Stage in:}
 StateAdvanceTime {type=DateTime, default=0001.01.01-00.00.00}
 TabAutoAdvanceTime {type=number, default=8}
 Tabs {type="array<string>", default [Results, Next Up]}
 TabsHaveNeverBeenSwitched {type=boolean, default=true}
 TabsLastSong {type="array<string>", default [Results,Progress]}
 TabsNextUp {type="array<string>", default [Results, Next Up]}
 TabsRevealSound {type=SoundBase, default=/PilgrimQuickplay/Audio/Results/Results_TabIntro_Cue}
 TabSwitchSound {type=SoundBase, default=/PilgrimQuickplay/Audio/Results/Results_TabSwitch_Cue}
 UniversalPartPickerArray {type="array<WBP_PilgrimQuickplay_PartPicker_LMPResults_C>", default []}
 UniversalPartPickerNavOverrideSlotTag {type=GameplayTag, default {tagName=Pilgrim.Hud.PartPicker.Slot2}}
 UniversalPartPickerSlotTag {type=GameplayTag, default {tagName=Pilgrim.Hud.PartPicker.Slot}}
 UPPNavOverrideWidget {type=CommonActivatableWidget, default=null}
 UPPWidgetRef {type=CommonActivatableWidget, default=null}
]
widgets {
 info {
  inheritedWidgetCount=0
  parentClass=/PilgrimQuickplay/Widgets/Results/WBP_Pilgrim_UI_Results_Base.WBP_Pilgrim_UI_Results_Base_C
  rootWidgetClass=&Overlay
  widgetCount=117
 }
 widgets [
  Overlay_All {widgetClassPath=&Overlay, bIsVariable=true, widget=Overlay_All}
  Overlay_0 {
   widgetClassPath=&Overlay
   parent=Overlay_All
   widget=Overlay_0
   properties {visibility=Collapsed}
   slot {class=OverlaySlot, properties {horizontalAlignment=HAlign_Fill, verticalAlignment=VAlign_Fill}}
  }
  BG {
   widgetClassPath=&Border
   bIsVariable=true
   parent=Overlay_0
   widget=BG
   properties {
    visibility=Collapsed
    background {
     bIsDynamicallyLoaded=false
     drawAs=Image
     imageType=NoImage
     mirroring=NoMirror
     resourceName=None
     resourceObject=/PilgrimQuickplay/Widgets/SetlistAndSongDisplay/Duplicated/Materials/MusicLibraryBG_Inst
     tiling=NoTile
     imageSize=<32 32>
     margin=<ltrb 0 0 0 0>
     outlineSettings {
      bUseBrushTransparency=false
      roundingType=HalfHeightRadius
      width=0
      color {colorUseRule=UseColor_Specified, specifiedColor=<rgba 0 0 0 0>}
      cornerRadii=<0 0 0 0>
     }
     tintColor {colorUseRule=UseColor_Specified, specifiedColor=<rgba 1 1 1 1>}
     uVRegion {bIsValid=false, max=<0 0>, min=<0 0>}
    }
    padding=<ltrb 0 0 0 0>
   }
   slot {class=OverlaySlot}
  }
  BG_1 {
   widgetClassPath=&Border
   bIsVariable=true
   parent=Overlay_0
   widget=BG_1
   properties {
    visibility=Collapsed
    background {
     bIsDynamicallyLoaded=false
     drawAs=Image
     imageType=NoImage
     mirroring=NoMirror
     resourceName=None
     resourceObject=/SparksUICommon/Materials/MI_UI_Sparks_Fullscreen_Background_Triangles
     tiling=NoTile
     imageSize=<32 32>
     margin=<ltrb 0 0 0 0>
     outlineSettings {
      bUseBrushTransparency=false
      roundingType=HalfHeightRadius
      width=0
      color {colorUseRule=UseColor_Specified, specifiedColor=<rgba 0 0 0 0>}
      cornerRadii=<0 0 0 0>
     }
     tintColor {colorUseRule=UseColor_Specified, specifiedColor=<rgba 1 1 1 1>}
     uVRegion {bIsValid=false, max=<0 0>, min=<0 0>}
    }
    padding=<ltrb 0 0 0 0>
   }
   slot {class=OverlaySlot}
  }
  SZ_SongMenuPrompt {
   widgetClassPath=&SafeZone
   bIsVariable=true
   parent=Overlay_All
   widget=SZ_SongMenuPrompt
   properties {padBottom=false, padRight=false}
   slot {class=OverlaySlot, properties {padding=<ltrb 65 50 0 0>}}
  }
  Overlay_1 {widgetClassPath=&Overlay, parent=SZ_SongMenuPrompt, widget=Overlay_1, slot {class=SafeZoneSlot}}
  CommonVisibilityWidgetLegacy_1 {
   widgetClassPath=&CommonVisibilityWidgetLegacy
   parent=Overlay_1
   widget=CommonVisibilityWidgetLegacy_1
   properties {bShowForAndroid=false, bShowForIOS=false}
   slot {class=OverlaySlot}
  }
  HorizontalBox_2 {widgetClassPath=&HorizontalBox, parent=CommonVisibilityWidgetLegacy_1, widget=HorizontalBox_2, slot {class=BorderSlot}}
  WBP_SongMenuPrompt {
   widgetClassPath=/SparksUICommon/Widgets/WBP_SongMenuPrompt
   bIsVariable=true
   parent=HorizontalBox_2
   widget=WBP_SongMenuPrompt
   properties {visibility=HitTestInvisible}
   slot {class=HorizontalBoxSlot}
  }
  CommonVisibilityWidgetLegacy_2 {
   widgetClassPath=&CommonVisibilityWidgetLegacy
   parent=Overlay_1
   widget=CommonVisibilityWidgetLegacy_2
   properties {
    bShowForErebus=false
    bShowForMac=false
    bShowForPC=false
    bShowForPS4=false
    bShowForPS5=false
    bShowForSwitch2=false
    bShowForXBox=false
    bShowForXSX=false
   }
   slot {class=OverlaySlot, properties {horizontalAlignment=HAlign_Fill, verticalAlignment=VAlign_Fill, padding=<ltrb 0 -46 0 0>}}
  }
  WBP_SongMenuPromptWithKeyBind_Touch {
   widgetClassPath=/PilgrimQuickplay/Widgets/Mobile/WBP_SongMenuPromptWithKeyBind_Touch
   bIsVariable=true
   parent=CommonVisibilityWidgetLegacy_2
   widget=WBP_SongMenuPromptWithKeyBind_Touch
   slot {class=BorderSlot}
  }
  WBP_CaptureForPostBufferUpdate {
   widgetClassPath=&WBP_CaptureForPostBufferUpdate
   bIsVariable=true
   parent=Overlay_All
   slot {class=OverlaySlot, properties {horizontalAlignment=HAlign_Fill, verticalAlignment=VAlign_Fill}}
  }
  NamedSlot_Fullscreen {
   widgetClassPath=&NamedSlot
   bIsVariable=true
   parent=Overlay_All
   widget=NamedSlot_Fullscreen
   slot {class=OverlaySlot, properties {horizontalAlignment=HAlign_Fill, verticalAlignment=VAlign_Fill}}
  }
  PlayerInfoHorizontalBox {
   widgetClassPath=&HorizontalBox
   bIsVariable=true
   parent=Overlay_All
   widget=PlayerInfoHorizontalBox
   slot {class=OverlaySlot, properties {horizontalAlignment=HAlign_Fill, verticalAlignment=VAlign_Fill, padding=<ltrb 70 120 70 110>}}
  }
  SafeZone_SetlistStatus {
   widgetClassPath=&SafeZone
   bIsVariable=true
   parent=PlayerInfoHorizontalBox
   widget=SafeZone_SetlistStatus
   slot {class=HorizontalBoxSlot, properties {horizontalAlignment=HAlign_Left, verticalAlignment=VAlign_Top}}
  }
  SizeBox_SetLIstStatus {
   widgetClassPath=&SizeBox
   parent=SafeZone_SetlistStatus
   widget=SizeBox_SetLIstStatus
   properties {bOverride_WidthOverride=true, widthOverride=384}
   slot {class=SafeZoneSlot}
  }
  WBP_PilgrimQuickplay_SongResults_SetlistStatus {
   widgetClassPath=&/WBP_PilgrimQuickplay_SongResults_SetlistStatus
   bIsVariable=true
   parent=SizeBox_SetLIstStatus
   widget=WBP_PilgrimQuickplay_SongResults_SetlistStatus
   slot {class=SizeBoxSlot}
  }
  SZ_PlayerInfoPanels {
   widgetClassPath=&SafeZone
   bIsVariable=true
   parent=PlayerInfoHorizontalBox
   widget=SZ_PlayerInfoPanels
   properties {padLeft=false, padRight=false}
   slot {class=HorizontalBoxSlot, properties {horizontalAlignment=HAlign_Center, padding=<ltrb 48 24 0 0>, size {sizeRule=Fill, value=1}}}
  }
  Overlay_3 {widgetClassPath=&Overlay, parent=SZ_PlayerInfoPanels, widget=Overlay_3, slot {class=SafeZoneSlot}}
  PlayerPanelsSizeControl {
   widgetClassPath=&SizeBox
   bIsVariable=true
   parent=Overlay_3
   widget=PlayerPanelsSizeControl
   properties {bOverride_MinDesiredWidth=true, maxDesiredWidth=1280, minDesiredWidth=1296}
   slot {class=OverlaySlot, properties {horizontalAlignment=HAlign_Fill, verticalAlignment=VAlign_Fill}}
  }
  PlayerInfoGridPanel {
   widgetClassPath=&GridPanel
   bIsVariable=true
   parent=PlayerPanelsSizeControl
   widget=PlayerInfoGridPanel
   properties {columnFill [0.25,0.25,0.25,0.25], rowFill [0.25,0.5,0.25]}
   slot {class=SizeBoxSlot}
  }
  SoloStatsSwitcher {
   widgetClassPath=&CommonActivatableWidgetSwitcher
   bIsVariable=true
   parent=PlayerInfoGridPanel
   widget=SoloStatsSwitcher
   properties {activeWidgetIndex=1, transitionDuration=0}
   slot {class=GridSlot, properties {horizontalAlignment=HAlign_Center, row=1, rowSpan=2, verticalAlignment=VAlign_Center}}
  }
  WBP_PilgrimQuickplay_SongResults_SoloStatsBase {
   widgetClassPath=&/WBP_PilgrimQuickplay_SongResults_SoloStatsBase
   bIsVariable=true
   parent=SoloStatsSwitcher
   widget=WBP_PilgrimQuickplay_SongResults_SoloStatsBase
   slot {class=WidgetSwitcherSlot}
  }
  Spacer_71 {widgetClassPath=&Spacer, parent=SoloStatsSwitcher, widget=Spacer_71, slot {class=WidgetSwitcherSlot}}
  PlayerInfo_1 {
   widgetClassPath=&/WBP_PilgrimQuickplay_Results_AbovePlayerDisplay
   bIsVariable=true
   parent=PlayerInfoGridPanel
   widget=PlayerInfo_1
   properties {visibility=Hidden}
   slot {
    class=GridSlot
    properties {columnSpan=1, horizontalAlignment=HAlign_Center, rowSpan=1, verticalAlignment=VAlign_Bottom, padding=<ltrb 0 0 0 56>}
   }
  }
  PlayerInfo_2 {
   widgetClassPath=&WBP_PilgrimQuickplay_Results_AbovePlayerDisplay
   bIsVariable=true
   parent=PlayerInfoGridPanel
   widget=PlayerInfo_2
   properties {visibility=Hidden}
   slot {
    class=GridSlot
    properties {column=1, columnSpan=1, horizontalAlignment=HAlign_Center, rowSpan=1, verticalAlignment=VAlign_Bottom, padding=<ltrb 0 0 0 56>}
   }
  }
  PlayerInfo_3 {
   widgetClassPath=&WBP_PilgrimQuickplay_Results_AbovePlayerDisplay
   bIsVariable=true
   parent=PlayerInfoGridPanel
   widget=PlayerInfo_3
   properties {visibility=Hidden}
   slot {
    class=GridSlot
    properties {column=2, columnSpan=1, horizontalAlignment=HAlign_Center, rowSpan=1, verticalAlignment=VAlign_Bottom, padding=<ltrb 0 0 0 56>}
   }
  }
  PlayerInfo_4 {
   widgetClassPath=&WBP_PilgrimQuickplay_Results_AbovePlayerDisplay
   bIsVariable=true
   parent=PlayerInfoGridPanel
   widget=PlayerInfo_4
   properties {visibility=Hidden}
   slot {
    class=GridSlot
    properties {column=3, columnSpan=1, horizontalAlignment=HAlign_Center, rowSpan=1, verticalAlignment=VAlign_Bottom, padding=<ltrb 0 0 0 56>}
   }
  }
  NamedSlot_CustomPlayerPanels {
   widgetClassPath=&NamedSlot
   bIsVariable=true
   parent=Overlay_3
   widget=NamedSlot_CustomPlayerPanels
   slot {class=OverlaySlot, properties {horizontalAlignment=HAlign_Fill, verticalAlignment=VAlign_Fill}}
  }
  ScaleBox_BottomInteraction {
   widgetClassPath=&ScaleBox
   bIsVariable=true
   parent=Overlay_All
   widget=ScaleBox_BottomInteraction
   properties {stretch=Fill, stretchDirection=UpOnly}
   slot {class=OverlaySlot, properties {horizontalAlignment=HAlign_Fill, verticalAlignment=VAlign_Fill, padding=<ltrb 64 120 64 48>}}
  }
  CommonVisibilitySwitcher {
   widgetClassPath=&CommonVisibilitySwitcher
   bIsVariable=true
   parent=ScaleBox_BottomInteraction
   widget=CommonVisibilitySwitcher
   slot {class=ScaleBoxSlot, properties {horizontalAlignment=HAlign_Fill, verticalAlignment=VAlign_Fill}}
  }
  SZ_MainStatsPanel {
   widgetClassPath=&SafeZone
   bIsVariable=true
   parent=CommonVisibilitySwitcher
   widget=SZ_MainStatsPanel
   properties {padRight=false, padTop=false}
   slot {class=CommonVisibilitySwitcherSlot, properties {horizontalAlignment=HAlign_Left, verticalAlignment=VAlign_Bottom}}
  }
  MainStatsPanel_SizeOverride {
   widgetClassPath=&SizeBox
   parent=SZ_MainStatsPanel
   widget=MainStatsPanel_SizeOverride
   properties {bOverride_MaxDesiredWidth=true, maxDesiredWidth=384}
   slot {class=SafeZoneSlot, properties {vAlign=VAlign_Bottom}}
  }
  PerformanceResultsBox {
   widgetClassPath=&VerticalBox
   bIsVariable=true
   parent=MainStatsPanel_SizeOverride
   widget=PerformanceResultsBox
   slot {class=SizeBoxSlot}
  }
  SizeBox_SetListStatus_Portrait {
   widgetClassPath=&SizeBox
   bIsVariable=true
   parent=PerformanceResultsBox
   widget=SizeBox_SetListStatus_Portrait
   properties {bOverride_WidthOverride=true, visibility=Collapsed, widthOverride=384}
   slot {class=VerticalBoxSlot, properties {padding=<ltrb 0 0 0 24>}}
  }
  WBP_PilgrimQuickplay_SongResults_SetlistStatus_1 {
   widgetClassPath=&WBP_PilgrimQuickplay_SongResults_SetlistStatus
   bIsVariable=true
   parent=SizeBox_SetListStatus_Portrait
   widget=WBP_PilgrimQuickplay_SongResults_SetlistStatus_1
   slot {class=SizeBoxSlot}
  }
  Overlay_2 {widgetClassPath=&Overlay, parent=PerformanceResultsBox, widget=Overlay_2, slot {class=VerticalBoxSlot}}
  WBP_Sparks_BackingPanel {
   widgetClassPath=/PilgrimCore/Widgets/SparksUILibrary/WBP_Sparks_BackingPanel
   bIsVariable=true
   parent=Overlay_2
   widget=WBP_Sparks_BackingPanel
   properties {blurFrame=false, cornerRadius=Small, darkenLayer=false, eSparksBackingPanelType=Simple Color}
   slot {class=OverlaySlot, properties {horizontalAlignment=HAlign_Fill, verticalAlignment=VAlign_Fill}}
  }
  MainStatsPanel_VerticalBox {
   widgetClassPath=&VerticalBox
   parent=Overlay_2
   widget=MainStatsPanel_VerticalBox
   slot {class=OverlaySlot, properties {horizontalAlignment=HAlign_Fill, verticalAlignment=VAlign_Fill}}
  }
  JustPlayedSongHeader {
   widgetClassPath=/PilgrimQuickplay/Widgets/Results/WBP_PilgrimQuickplay_JustPlayedSongHeader
   bIsVariable=true
   parent=MainStatsPanel_VerticalBox
   widget=JustPlayedSongHeader
   slot {class=VerticalBoxSlot, properties {verticalAlignment=VAlign_Top, padding=<ltrb 8 8 8 8>}}
  }
  BandPerformanceData {
   widgetClassPath=&VerticalBox
   bIsVariable=true
   parent=MainStatsPanel_VerticalBox
   widget=BandPerformanceData
   slot {class=VerticalBoxSlot, properties {padding=<ltrb 16 8 16 16>}}
  }
  BandStarsScale {
   widgetClassPath=&ScaleBox
   parent=BandPerformanceData
   widget=BandStarsScale
   slot {class=VerticalBoxSlot, properties {horizontalAlignment=HAlign_Center, padding=<ltrb 16 0 16 0>}}
  }
  StarVisibilitySwitcher {
   widgetClassPath=&CommonVisibilitySwitcher
   bIsVariable=true
   parent=BandStarsScale
   widget=StarVisibilitySwitcher
   slot {class=ScaleBoxSlot}
  }
  StarBar {
   widgetClassPath=/PilgrimQuickplay/Widgets/WBP_PilgrimQuickplay_LocalStarBar
   bIsVariable=true
   parent=StarVisibilitySwitcher
   widget=StarBar
   properties {bIsSequencedAnimation=true, scale Stretch=UserSpecified, shouldStarsAnimateComplete=true}
   slot {class=CommonVisibilitySwitcherSlot}
  }
  BandStarBar {
   widgetClassPath=/PilgrimQuickplay/Widgets/WBP_PilgrimQuickplay_BandStarBar
   bIsVariable=true
   parent=StarVisibilitySwitcher
   widget=BandStarBar
   properties {bIsSequencedAnimation=true}
   slot {class=CommonVisibilitySwitcherSlot}
  }
  ScoreData {
   widgetClassPath=&VerticalBox
   parent=BandPerformanceData
   widget=ScoreData
   properties {visibility=HitTestInvisible}
   slot {class=VerticalBoxSlot, properties {horizontalAlignment=HAlign_Center}}
  }
  BandScorePanel {
   widgetClassPath=&/WBP_PilgrimQuickplay_BandScorePanel
   bIsVariable=true
   parent=ScoreData
   widget=BandScorePanel
   slot {class=VerticalBoxSlot, properties {horizontalAlignment=HAlign_Center}}
  }
  BandNewHighScore {
   widgetClassPath=&/WBP_PilgrimQuickly_NewHighScoreRankPanel
   bIsVariable=true
   parent=ScoreData
   widget=BandNewHighScore
   properties {visibility=Collapsed}
   slot {class=VerticalBoxSlot}
  }
  WBP_PilgrimQuickplay_FullComboIndicator {
   widgetClassPath=/PilgrimQuickplay/Widgets/WBP_PilgrimQuickplay_FullComboIndicator
   bIsVariable=true
   parent=ScoreData
   widget=WBP_PilgrimQuickplay_FullComboIndicator
   properties {visibility=Collapsed}
   slot {class=VerticalBoxSlot, properties {horizontalAlignment=HAlign_Center, padding=<ltrb 10 10 10 10>}}
  }
  MoreDetails {
   widgetClassPath=&Overlay
   parent=BandPerformanceData
   widget=MoreDetails
   properties {visibility=Collapsed}
   slot {class=VerticalBoxSlot, properties {horizontalAlignment=HAlign_Left, verticalAlignment=VAlign_Top, padding=<ltrb 0 10 0 0>}}
  }
  ExpandStats_Prompt_Solo {widgetClassPath=&HorizontalBox, parent=MoreDetails, widget=ExpandStats_Prompt_Solo, slot {class=OverlaySlot}}
  SoloExpandStats_Label {
   widgetClassPath=&CommonRichTextBlock
   bIsVariable=true
   parent=ExpandStats_Prompt_Solo
   widget=SoloExpandStats_Label
   properties {
    bOverrideDefaultStyle=true
    defaultTextStyleOverrideClass=/FortUILibrary/Styles/Typography/_TextStyle_Body1._TextStyle_Body1_C
    text=More details
    defaultTextStyleOverride {
     overflowPolicy=Clip
     transformPolicy=None
     colorAndOpacity {colorUseRule=UseColor_Specified, specifiedColor=<rgba 1 1 1 1>}
     font {
      bForceMonospaced=false
      bMaterialIsStencil=false
      fontMaterial=null
      fontObject=&BurbankSmall
      letterSpacing=0
      monospacedWidth=1
      size=24
      skewAmount=0
      typefaceFontName=Black
      outlineSettings {
       bApplyOutlineToDropShadows=false
       bMiteredCorners=false
       bSeparateFillAlpha=false
       outlineMaterial=null
       outlineSize=0
       outlineColor=<rgba 0 0 0 1>
      }
     }
     highlightColor {colorUseRule=UseColor_Specified, specifiedColor=<rgba 0 0 0 1>}
     highlightShape {
      bIsDynamicallyLoaded=false
      drawAs=Image
      imageType=NoImage
      mirroring=NoMirror
      resourceName=None
      resourceObject=null
      tiling=NoTile
      imageSize=<32 32>
      margin=<ltrb 0 0 0 0>
      outlineSettings {
       bUseBrushTransparency=false
       roundingType=HalfHeightRadius
       width=0
       color {colorUseRule=UseColor_Specified, specifiedColor=<rgba 0 0 0 0>}
       cornerRadii=<0 0 0 0>
      }
      tintColor {colorUseRule=UseColor_Specified, specifiedColor=<rgba 1 1 1 1>}
      uVRegion {bIsValid=false, max=<0 0>, min=<0 0>}
     }
     selectedBackgroundColor {colorUseRule=UseColor_Foreground, specifiedColor=<rgba 1 0 1 1>}
     shadowColorAndOpacity=<rgba 0 0 0 1>
     shadowOffset=<0 0>
     strikeBrush {
      bIsDynamicallyLoaded=false
      drawAs=Image
      imageType=NoImage
      mirroring=NoMirror
      resourceName=None
      resourceObject=null
      tiling=NoTile
      imageSize=<32 32>
      margin=<ltrb 0 0 0 0>
      outlineSettings {
       bUseBrushTransparency=false
       roundingType=HalfHeightRadius
       width=0
       color {colorUseRule=UseColor_Specified, specifiedColor=<rgba 0 0 0 0>}
       cornerRadii=<0 0 0 0>
      }
      tintColor {colorUseRule=UseColor_Specified, specifiedColor=<rgba 1 1 1 1>}
      uVRegion {bIsValid=false, max=<0 0>, min=<0 0>}
     }
     underlineBrush {
      bIsDynamicallyLoaded=false
      drawAs=Image
      imageType=NoImage
      mirroring=NoMirror
      resourceName=None
      resourceObject=null
      tiling=NoTile
      imageSize=<32 32>
      margin=<ltrb 0 0 0 0>
      outlineSettings {
       bUseBrushTransparency=false
       roundingType=HalfHeightRadius
       width=0
       color {colorUseRule=UseColor_Specified, specifiedColor=<rgba 0 0 0 0>}
       cornerRadii=<0 0 0 0>
      }
      tintColor {colorUseRule=UseColor_Specified, specifiedColor=<rgba 1 1 1 1>}
      uVRegion {bIsValid=false, max=<0 0>, min=<0 0>}
     }
    }
   }
   slot {class=HorizontalBoxSlot, properties {verticalAlignment=VAlign_Center, padding=<ltrb 6 0 0 0>}}
  }
  SizeBox_2 {
   widgetClassPath=&SizeBox
   parent=MoreDetails
   widget=SizeBox_2
   properties {renderTransform {angle=0, scale=<1 1.5>, shear=<0 0>, translation=<0 0>}}
   slot {class=OverlaySlot, properties {horizontalAlignment=HAlign_Fill, verticalAlignment=VAlign_Fill}}
  }
  PilgrimTouchButtonBase_lane1 {
   widgetClassPath=/Script/PilgrimCoreRuntime.PilgrimTouchButtonBase
   bIsVariable=true
   parent=SizeBox_2
   widget=PilgrimTouchButtonBase_lane1
   properties {bConsumePointerInput=true, pressAction=/PilgrimSettings/Player/Abilities/IA_PilgrimReturnToResults, visibility=Visible}
   slot {class=SizeBoxSlot}
  }
  BandStats {
   widgetClassPath=&/WBP_PilgrimQuickplay_SongResults_BandStats
   bIsVariable=true
   parent=BandPerformanceData
   widget=BandStats
   properties {visibility=Collapsed}
   slot {class=VerticalBoxSlot, properties {padding=<ltrb 10 10 10 10>}}
  }
  BottomSoloButtons_Grid {
   widgetClassPath=&GridPanel
   bIsVariable=true
   parent=PerformanceResultsBox
   widget=BottomSoloButtons_Grid
   properties {columnFill [1], rowFill [1]}
   slot {class=VerticalBoxSlot, properties {verticalAlignment=VAlign_Bottom, padding=<ltrb 0 24 0 0>}}
  }
  BottomSoloButtons_VerticalBox {
   widgetClassPath=&VerticalBox
   bIsVariable=true
   parent=BottomSoloButtons_Grid
   widget=BottomSoloButtons_VerticalBox
   properties {navigation=&self:WidgetTree.BottomSoloButtons_VerticalBox.WidgetNavigation_0}
   slot {class=GridSlot, properties {columnSpan=1, rowSpan=1}}
  }
  Results_StartNextSongButton {
   widgetClassPath=/SparksUICommon/UILibrary/WBP_FM_Core_Button_Regular
   bIsVariable=true
   parent=BottomSoloButtons_VerticalBox
   widget=Results_StartNextSongButton
   properties {
    bDisplayInputActionWhenNotInteractable=false
    bForceMobileHoveredFromFocus=true
    bHideInputActionWithKeyboard=true
    bUseNaturalNavigation=true
    buttonSize=Small
    text=Start Next Song
    visibility=Collapsed
   }
   slot {class=VerticalBoxSlot, properties {padding=<ltrb 0 12 0 0>}}
  }
  Results_LeaveStage_Button {
   widgetClassPath=&WBP_FM_Core_Button_Regular
   bIsVariable=true
   parent=BottomSoloButtons_VerticalBox
   widget=Results_LeaveStage_Button
   properties {
    bDisplayInputActionWhenNotInteractable=false
    bForceMobileHoveredFromFocus=true
    bHideInputActionWithKeyboard=true
    bUseNaturalNavigation=true
    buttonSize=Small
    text=Go Backstage
    visibility=Collapsed
   }
   slot {class=VerticalBoxSlot, properties {padding=<ltrb 0 12 0 0>}}
  }
  Results_ConcordGrape_Button {
   widgetClassPath=&WBP_FM_Core_Button_Regular
   bIsVariable=true
   parent=BottomSoloButtons_VerticalBox
   widget=Results_ConcordGrape_Button
   properties {
    bDisplayInputActionWhenNotInteractable=false
    bForceMobileHoveredFromFocus=true
    bHideInputActionWithKeyboard=true
    bUseNaturalNavigation=true
    buttonSize=Small
    text="Concord Grape "
    visibility=Collapsed
   }
   slot {class=VerticalBoxSlot, properties {padding=<ltrb 0 12 0 0>}}
  }
  Results_ReplaySongButton {
   widgetClassPath=&WBP_FM_Core_Button_Regular
   bIsVariable=true
   parent=BottomSoloButtons_VerticalBox
   widget=Results_ReplaySongButton
   properties {
    bDisplayInputActionWhenNotInteractable=false
    bForceMobileHoveredFromFocus=true
    bHideInputActionWithKeyboard=true
    bUseNaturalNavigation=true
    buttonSize=Small
    text=Replay Song
    visibility=Collapsed
   }
   slot {class=VerticalBoxSlot, properties {padding=<ltrb 0 12 0 0>}}
  }
  Results_Report_Button {
   widgetClassPath=&WBP_FM_Core_Button_Regular
   bIsVariable=true
   parent=BottomSoloButtons_VerticalBox
   widget=Results_Report_Button
   properties {
    bDisplayInputActionWhenNotInteractable=false
    bForceMobileHoveredFromFocus=true
    bHideInputActionWithKeyboard=true
    bUseNaturalNavigation=true
    buttonSize=Small
    text=Report Player
    visibility=Collapsed
   }
   slot {class=VerticalBoxSlot, properties {padding=<ltrb 0 12 0 0>}}
  }
  Results_ReturnToLobby_Button {
   widgetClassPath=&WBP_FM_Core_Button_Regular
   bIsVariable=true
   parent=BottomSoloButtons_VerticalBox
   widget=Results_ReturnToLobby_Button
   properties {
    bDisplayInputActionWhenNotInteractable=false
    bForceMobileHoveredFromFocus=true
    bHideInputActionWithKeyboard=true
    bUseNaturalNavigation=true
    buttonSize=Small
    text=Return To Lobby
    visibility=Collapsed
   }
   slot {class=VerticalBoxSlot, properties {padding=<ltrb 0 12 0 0>}}
  }
  Spacer_2 {
   widgetClassPath=&Spacer
   parent=BottomSoloButtons_VerticalBox
   widget=Spacer_2
   properties {size=<384 0>}
   slot {class=VerticalBoxSlot}
  }
  ControllerDisconnectPrompt_MainStats {
   widgetClassPath=/PilgrimQuickplay/Widgets/WBP_PilgrimQuickplay_ControllerDisconnectPrompt
   bIsVariable=true
   parent=BottomSoloButtons_Grid
   widget=ControllerDisconnectPrompt_MainStats
   properties {renderOpacity=0, visibility=Collapsed}
   slot {class=GridSlot, properties {columnSpan=1, layer=1, rowSpan=1, verticalAlignment=VAlign_Center}}
  }
  NextUpBox_Overlay {
   widgetClassPath=&Overlay
   bIsVariable=true
   parent=CommonVisibilitySwitcher
   widget=NextUpBox_Overlay
   slot {class=CommonVisibilitySwitcherSlot}
  }
  SZ_NextUpPanel {
   widgetClassPath=&SafeZone
   bIsVariable=true
   parent=NextUpBox_Overlay
   widget=SZ_NextUpPanel
   properties {padRight=false, padTop=false}
   slot {class=OverlaySlot, properties {verticalAlignment=VAlign_Bottom}}
  }
  NextUpSizeBox {
   widgetClassPath=&SizeBox
   parent=SZ_NextUpPanel
   widget=NextUpSizeBox
   properties {bOverride_MaxDesiredWidth=true, maxDesiredWidth=384}
   slot {class=SafeZoneSlot}
  }
  SongResult {widgetClassPath=&VerticalBox, bIsVariable=true, parent=NextUpSizeBox, widget=SongResult, slot {class=SizeBoxSlot}}
  SongResult_PanelsAndButtons {
   widgetClassPath=&VerticalBox
   parent=SongResult
   widget=SongResult_PanelsAndButtons
   properties {navigation=&self:WidgetTree.SongResult_PanelsAndButtons.WidgetNavigation_0}
   slot {class=VerticalBoxSlot, properties {horizontalAlignment=HAlign_Left, size {sizeRule=Fill, value=1}}}
  }
  SongResult_PanelGroup {
   widgetClassPath=&VerticalBox
   parent=SongResult_PanelsAndButtons
   widget=SongResult_PanelGroup
   slot {class=VerticalBoxSlot, properties {padding=<ltrb 0 0 0 48>}}
  }
  WBP_PilgrimQuickplay_NextSongDetailsPanel {
   widgetClassPath=&/WBP_PilgrimQuickplay_NextSongDetailsPanel
   bIsVariable=true
   parent=SongResult_PanelGroup
   widget=WBP_PilgrimQuickplay_NextSongDetailsPanel
   slot {class=VerticalBoxSlot, properties {padding=<ltrb 0 0 0 10>}}
  }
  PlayerIdentity_Pill {
   widgetClassPath=/PilgrimCore/Widgets/Player/WBP_Pilgrim_PlayerIdentity_Pill
   bIsVariable=true
   parent=SongResult_PanelGroup
   widget=PlayerIdentity_Pill
   slot {class=VerticalBoxSlot, properties {horizontalAlignment=HAlign_Center, padding=<ltrb 0 24 0 18>}}
  }
  GridPanel_0 {
   widgetClassPath=&GridPanel
   bIsVariable=true
   parent=SongResult_PanelGroup
   widget=GridPanel_0
   properties {columnFill [1], rowFill [1]}
   slot {class=VerticalBoxSlot}
  }
  VB_PickerInputUI {
   widgetClassPath=&VerticalBox
   bIsVariable=true
   parent=GridPanel_0
   widget=VB_PickerInputUI
   slot {class=GridSlot, properties {columnSpan=1, rowSpan=1}}
  }
  InstrumentSelector {
   widgetClassPath=/PilgrimQuickplay/Widgets/WBP_Pilgrim_GreenRoom_InstrumentSelectUI
   bIsVariable=true
   parent=VB_PickerInputUI
   widget=InstrumentSelector
   properties {bIsFocusable=true, isOnGreenRoom=false, visibility=Collapsed}
   slot {class=VerticalBoxSlot}
  }
  Spacer {widgetClassPath=&Spacer, parent=VB_PickerInputUI, widget=Spacer, properties {size=<1 10>}, slot {class=VerticalBoxSlot}}
  DifficultySelector {
   widgetClassPath=/PilgrimQuickplay/Widgets/WBP_Pilgrim_GreenRoom_DifficultySelectUI
   bIsVariable=true
   parent=VB_PickerInputUI
   widget=DifficultySelector
   properties {bIsFocusable=true, isOnGreenRoom=false, visibility=Collapsed}
   slot {class=VerticalBoxSlot}
  }
  ControllerDisconnectPrompt_NextUp {
   widgetClassPath=&WBP_PilgrimQuickplay_ControllerDisconnectPrompt
   bIsVariable=true
   parent=GridPanel_0
   widget=ControllerDisconnectPrompt_NextUp
   properties {renderOpacity=0, visibility=Collapsed}
   slot {class=GridSlot, properties {columnSpan=1, layer=1, rowSpan=1, verticalAlignment=VAlign_Top}}
  }
  SongResult_ButtonGroup {
   widgetClassPath=&VerticalBox
   parent=SongResult_PanelsAndButtons
   widget=SongResult_ButtonGroup
   slot {class=VerticalBoxSlot, properties {verticalAlignment=VAlign_Bottom}}
  }
  NextUp_StartNextSong {
   widgetClassPath=&WBP_FM_Core_Button_Regular
   bIsVariable=true
   parent=SongResult_ButtonGroup
   widget=NextUp_StartNextSong
   properties {
    bDisplayInputActionWhenNotInteractable=false
    bForceMobileHoveredFromFocus=true
    bHideInputActionWithKeyboard=true
    bUseNaturalNavigation=true
    buttonSize=Small
    inputModeOverride=All
    text=Start Next Song
    visibility=Collapsed
   }
   slot {class=VerticalBoxSlot}
  }
  NextUp_ReplaySong {
   widgetClassPath=&WBP_FM_Core_Button_Regular
   bIsVariable=true
   parent=SongResult_ButtonGroup
   widget=NextUp_ReplaySong
   properties {
    bDisplayInputActionWhenNotInteractable=false
    bForceMobileHoveredFromFocus=true
    bHideInputActionWithKeyboard=true
    bUseNaturalNavigation=true
    buttonSize=Small
    inputModeOverride=All
    text=Replay Song
    visibility=Collapsed
   }
   slot {class=VerticalBoxSlot, properties {padding=<ltrb 0 12 0 0>}}
  }
  GridPanelSizeControl {
   widgetClassPath=&Spacer
   parent=SongResult_PanelsAndButtons
   widget=GridPanelSizeControl
   properties {size=<384 0>}
   slot {class=VerticalBoxSlot}
  }
  ProgressTab {
   widgetClassPath=&VerticalBox
   bIsVariable=true
   parent=NextUpBox_Overlay
   widget=ProgressTab
   slot {class=OverlaySlot, properties {horizontalAlignment=HAlign_Fill, verticalAlignment=VAlign_Fill}}
  }
  WBP_PilgrimQuickplay_ProgressTab {
   widgetClassPath=&/WBP_PilgrimQuickplay_ProgressTab
   bIsVariable=true
   parent=ProgressTab
   widget=WBP_PilgrimQuickplay_ProgressTab
   slot {class=VerticalBoxSlot, properties {padding=<ltrb 0 0 0 30>, size {sizeRule=Fill, value=1}}}
  }
  SZ_ProgressTabsButonsPanel {
   widgetClassPath=&SafeZone
   bIsVariable=true
   parent=ProgressTab
   widget=SZ_ProgressTabsButonsPanel
   properties {padRight=false, padTop=false}
   slot {class=VerticalBoxSlot, properties {horizontalAlignment=HAlign_Left, verticalAlignment=VAlign_Bottom}}
  }
  ProgressButtonsSizeBox {
   widgetClassPath=&SizeBox
   parent=SZ_ProgressTabsButonsPanel
   widget=ProgressButtonsSizeBox
   properties {maxDesiredWidth=384, minDesiredWidth=0.7936}
   slot {class=SafeZoneSlot, properties {padding=<ltrb 0 30 0 0>}}
  }
  ProgressTab_Buttons {
   widgetClassPath=&VerticalBox
   parent=ProgressButtonsSizeBox
   widget=ProgressTab_Buttons
   properties {navigation=&self:WidgetTree.ProgressTab_Buttons.WidgetNavigation_0}
   slot {class=SizeBoxSlot, properties {verticalAlignment=VAlign_Bottom}}
  }
  Progress_LeaveStage_Button {
   widgetClassPath=&WBP_FM_Core_Button_Regular
   bIsVariable=true
   parent=ProgressTab_Buttons
   widget=Progress_LeaveStage_Button
   properties {
    bDisplayInputActionWhenNotInteractable=false
    bForceMobileHoveredFromFocus=true
    bHideInputActionWithKeyboard=true
    bUseNaturalNavigation=true
    buttonSize=Small
    text=Go Backstage
    visibility=Collapsed
   }
   slot {class=VerticalBoxSlot, properties {padding=<ltrb 0 12 0 0>}}
  }
  Progress_ConcordGrape_Button {
   widgetClassPath=&WBP_FM_Core_Button_Regular
   bIsVariable=true
   parent=ProgressTab_Buttons
   widget=Progress_ConcordGrape_Button
   properties {
    bDisplayInputActionWhenNotInteractable=false
    bForceMobileHoveredFromFocus=true
    bHideInputActionWithKeyboard=true
    bUseNaturalNavigation=true
    buttonSize=Small
    text=Concord Grape
    visibility=Collapsed
   }
   slot {class=VerticalBoxSlot, properties {padding=<ltrb 0 12 0 0>}}
  }
  Progress_Report_Button {
   widgetClassPath=&WBP_FM_Core_Button_Regular
   bIsVariable=true
   parent=ProgressTab_Buttons
   widget=Progress_Report_Button
   properties {
    bDisplayInputActionWhenNotInteractable=false
    bForceMobileHoveredFromFocus=true
    bHideInputActionWithKeyboard=true
    bUseNaturalNavigation=true
    buttonSize=Small
    text=Report Player
    visibility=Collapsed
   }
   slot {class=VerticalBoxSlot, properties {padding=<ltrb 0 12 0 0>}}
  }
  Progress_ReplaySongButton {
   widgetClassPath=&WBP_FM_Core_Button_Regular
   bIsVariable=true
   parent=ProgressTab_Buttons
   widget=Progress_ReplaySongButton
   properties {
    bDisplayInputActionWhenNotInteractable=false
    bForceMobileHoveredFromFocus=true
    bHideInputActionWithKeyboard=true
    bUseNaturalNavigation=true
    buttonSize=Small
    text=Replay Song
    visibility=Collapsed
   }
   slot {class=VerticalBoxSlot, properties {padding=<ltrb 0 12 0 0>}}
  }
  Progress_ReturnToLobby_Button {
   widgetClassPath=&WBP_FM_Core_Button_Regular
   bIsVariable=true
   parent=ProgressTab_Buttons
   widget=Progress_ReturnToLobby_Button
   properties {
    bDisplayInputActionWhenNotInteractable=false
    bForceMobileHoveredFromFocus=true
    bHideInputActionWithKeyboard=true
    bUseNaturalNavigation=true
    buttonSize=Small
    text=Return To Lobby
    visibility=Collapsed
   }
   slot {class=VerticalBoxSlot, properties {padding=<ltrb 0 12 0 0>}}
  }
  ProgressButtonsWidthControl {
   widgetClassPath=&Spacer
   parent=ProgressTab_Buttons
   widget=ProgressButtonsWidthControl
   properties {size=<384 0>}
   slot {class=VerticalBoxSlot}
  }
  SZ_TopTabs {
   widgetClassPath=&SafeZone
   parent=Overlay_All
   widget=SZ_TopTabs
   properties {padBottom=false, padLeft=false, padRight=false}
   slot {class=OverlaySlot, properties {horizontalAlignment=HAlign_Center, padding=<ltrb 0 40 0 0>}}
  }
  TopTabs_Overlay {widgetClassPath=&Overlay, bIsVariable=true, parent=SZ_TopTabs, widget=TopTabs_Overlay, slot {class=SafeZoneSlot}}
  HorizontalBox_0 {
   widgetClassPath=&HorizontalBox
   parent=TopTabs_Overlay
   widget=HorizontalBox_0
   slot {class=OverlaySlot, properties {horizontalAlignment=HAlign_Fill, verticalAlignment=VAlign_Fill, padding=<ltrb -100 -30 -100 -30>}}
  }
  TouchArea_Results {
   widgetClassPath=&Button
   parent=HorizontalBox_0
   widget=TouchArea_Results
   properties {
    isFocusable=false
    backgroundColor=<rgba 1 1 1 0>
    colorAndOpacity=<rgba 1 1 1 0>
    widgetStyle {
     clickedSlateSound {resourceObject=null}
     disabled {
      bIsDynamicallyLoaded=false
      drawAs=NoDrawType
      imageType=NoImage
      mirroring=NoMirror
      resourceName=None
      resourceObject=null
      tiling=NoTile
      imageSize=<0 0>
      margin=<ltrb 0 0 0 0>
      outlineSettings {
       bUseBrushTransparency=false
       roundingType=HalfHeightRadius
       width=0
       color {colorUseRule=UseColor_Specified, specifiedColor=<rgba 0 0 0 0>}
       cornerRadii=<0 0 0 0>
      }
      tintColor {colorUseRule=UseColor_Specified, specifiedColor=<rgba 1 1 1 0>}
      uVRegion {bIsValid=false, max=<0 0>, min=<0 0>}
     }
     disabledForeground {colorUseRule=UseColor_Specified, specifiedColor=<rgba 0.5271 0.5271 0.5271 0>}
     hovered {
      bIsDynamicallyLoaded=false
      drawAs=RoundedBox
      imageType=NoImage
      mirroring=NoMirror
      resourceName=None
      resourceObject=null
      tiling=NoTile
      imageSize=<32 32>
      margin=<ltrb 0 0 0 0>
      outlineSettings {
       bUseBrushTransparency=true
       roundingType=FixedRadius
       width=1
       color {colorUseRule=UseColor_Specified, specifiedColor=<rgba 0.7243 0.7243 0.7243 1>}
       cornerRadii=<4 4 4 4>
      }
      tintColor {colorUseRule=UseColor_Specified, specifiedColor=<rgba 0.7243 0.7243 0.7243 0>}
      uVRegion {bIsValid=false, max=<0 0>, min=<0 0>}
     }
     hoveredForeground {colorUseRule=UseColor_Specified, specifiedColor=<rgba 0 1 1 0>}
     hoveredSlateSound {resourceObject=null}
     normal {
      bIsDynamicallyLoaded=false
      drawAs=RoundedBox
      imageType=NoImage
      mirroring=NoMirror
      resourceName=None
      resourceObject=null
      tiling=NoTile
      imageSize=<32 32>
      margin=<ltrb 0 0 0 0>
      outlineSettings {
       bUseBrushTransparency=true
       roundingType=FixedRadius
       width=1
       color {colorUseRule=UseColor_Specified, specifiedColor=<rgba 0.6951 0.6951 0.6951 1>}
       cornerRadii=<4 4 4 4>
      }
      tintColor {colorUseRule=UseColor_Specified, specifiedColor=<rgba 0.4955 0.4955 0.4955 0>}
      uVRegion {bIsValid=false, max=<0 0>, min=<0 0>}
     }
     normalForeground {colorUseRule=UseColor_Specified, specifiedColor=<rgba 0 1 1 0>}
     normalPadding=<ltrb 12 1.5 12 1.5>
     pressed {
      bIsDynamicallyLoaded=false
      drawAs=RoundedBox
      imageType=NoImage
      mirroring=NoMirror
      resourceName=None
      resourceObject=null
      tiling=NoTile
      imageSize=<32 32>
      margin=<ltrb 0 0 0 0>
      outlineSettings {
       bUseBrushTransparency=true
       roundingType=FixedRadius
       width=1
       color {colorUseRule=UseColor_Specified, specifiedColor=<rgba 0.7243 0.7243 0.7243 1>}
       cornerRadii=<4 4 4 4>
      }
      tintColor {colorUseRule=UseColor_Specified, specifiedColor=<rgba 0.3843 0.3843 0.3843 0>}
      uVRegion {bIsValid=false, max=<0 0>, min=<0 0>}
     }
     pressedForeground {colorUseRule=UseColor_Specified, specifiedColor=<rgba 0 1 1 0>}
     pressedPadding=<ltrb 12 1.5 12 1.5>
     pressedSlateSound {resourceObject=null}
    }
   }
   slot {class=HorizontalBoxSlot, properties {size {sizeRule=Fill, value=1}}}
  }
  TouchArea_NextUp {
   widgetClassPath=&Button
   parent=HorizontalBox_0
   widget=TouchArea_NextUp
   properties {
    isFocusable=false
    backgroundColor=<rgba 1 1 1 0>
    colorAndOpacity=<rgba 1 1 1 0>
    widgetStyle {
     clickedSlateSound {resourceObject=null}
     disabled {
      bIsDynamicallyLoaded=false
      drawAs=NoDrawType
      imageType=NoImage
      mirroring=NoMirror
      resourceName=None
      resourceObject=null
      tiling=NoTile
      imageSize=<0 0>
      margin=<ltrb 0 0 0 0>
      outlineSettings {
       bUseBrushTransparency=false
       roundingType=HalfHeightRadius
       width=0
       color {colorUseRule=UseColor_Specified, specifiedColor=<rgba 0 0 0 0>}
       cornerRadii=<0 0 0 0>
      }
      tintColor {colorUseRule=UseColor_Specified, specifiedColor=<rgba 1 1 1 0>}
      uVRegion {bIsValid=false, max=<0 0>, min=<0 0>}
     }
     disabledForeground {colorUseRule=UseColor_Specified, specifiedColor=<rgba 0.5271 0.5271 0.5271 0>}
     hovered {
      bIsDynamicallyLoaded=false
      drawAs=RoundedBox
      imageType=NoImage
      mirroring=NoMirror
      resourceName=None
      resourceObject=null
      tiling=NoTile
      imageSize=<32 32>
      margin=<ltrb 0 0 0 0>
      outlineSettings {
       bUseBrushTransparency=true
       roundingType=FixedRadius
       width=1
       color {colorUseRule=UseColor_Specified, specifiedColor=<rgba 0.7243 0.7243 0.7243 1>}
       cornerRadii=<4 4 4 4>
      }
      tintColor {colorUseRule=UseColor_Specified, specifiedColor=<rgba 0.7243 0.7243 0.7243 0>}
      uVRegion {bIsValid=false, max=<0 0>, min=<0 0>}
     }
     hoveredForeground {colorUseRule=UseColor_Specified, specifiedColor=<rgba 0 1 1 0>}
     hoveredSlateSound {resourceObject=null}
     normal {
      bIsDynamicallyLoaded=false
      drawAs=RoundedBox
      imageType=NoImage
      mirroring=NoMirror
      resourceName=None
      resourceObject=null
      tiling=NoTile
      imageSize=<32 32>
      margin=<ltrb 0 0 0 0>
      outlineSettings {
       bUseBrushTransparency=true
       roundingType=FixedRadius
       width=1
       color {colorUseRule=UseColor_Specified, specifiedColor=<rgba 0.6951 0.6951 0.6951 1>}
       cornerRadii=<4 4 4 4>
      }
      tintColor {colorUseRule=UseColor_Specified, specifiedColor=<rgba 0.4955 0.4955 0.4955 0>}
      uVRegion {bIsValid=false, max=<0 0>, min=<0 0>}
     }
     normalForeground {colorUseRule=UseColor_Specified, specifiedColor=<rgba 0 1 1 0>}
     normalPadding=<ltrb 12 1.5 12 1.5>
     pressed {
      bIsDynamicallyLoaded=false
      drawAs=RoundedBox
      imageType=NoImage
      mirroring=NoMirror
      resourceName=None
      resourceObject=null
      tiling=NoTile
      imageSize=<32 32>
      margin=<ltrb 0 0 0 0>
      outlineSettings {
       bUseBrushTransparency=true
       roundingType=FixedRadius
       width=1
       color {colorUseRule=UseColor_Specified, specifiedColor=<rgba 0.7243 0.7243 0.7243 1>}
       cornerRadii=<4 4 4 4>
      }
      tintColor {colorUseRule=UseColor_Specified, specifiedColor=<rgba 0.3843 0.3843 0.3843 0>}
      uVRegion {bIsValid=false, max=<0 0>, min=<0 0>}
     }
     pressedForeground {colorUseRule=UseColor_Specified, specifiedColor=<rgba 0 1 1 0>}
     pressedPadding=<ltrb 12 1.5 12 1.5>
     pressedSlateSound {resourceObject=null}
    }
   }
   slot {class=HorizontalBoxSlot, properties {size {sizeRule=Fill, value=1}}}
  }
  TabBacking {
   widgetClassPath=&Image
   parent=TopTabs_Overlay
   widget=TabBacking
   properties {
    visibility=SelfHitTestInvisible
    brush {
     bIsDynamicallyLoaded=false
     drawAs=Image
     imageType=NoImage
     mirroring=NoMirror
     resourceName=None
     resourceObject=/PilgrimCore/Widgets/Materials/MI_Pilgrim_TabBacking_Panel
     tiling=NoTile
     imageSize=<32 32>
     margin=<ltrb 0 0 0 0>
     outlineSettings {
      bUseBrushTransparency=false
      roundingType=HalfHeightRadius
      width=0
      color {colorUseRule=UseColor_Specified, specifiedColor=<rgba 0 0 0 0>}
      cornerRadii=<0 0 0 0>
     }
     tintColor {colorUseRule=UseColor_Specified, specifiedColor=<rgba 1 1 1 1>}
     uVRegion {bIsValid=false, max=<0 0>, min=<0 0>}
    }
   }
   slot {class=OverlaySlot, properties {horizontalAlignment=HAlign_Fill, verticalAlignment=VAlign_Fill, padding=<ltrb -8 0 -8 0>}}
  }
  WBP_UIKit_Tabs_TopNavBar {
   widgetClassPath=/Game/UI/Frontend/TopBar/WBP_UIKit_Tabs_TopNavBar
   bIsVariable=true
   parent=TopTabs_Overlay
   widget=WBP_UIKit_Tabs_TopNavBar
   properties {
    backplateClass=null
    buttonClass=/PilgrimCore/Widgets/SparksUILibrary/WBP_Sparks_TextTabs_Button_Selectable.WBP_Sparks_TextTabs_Button_Selectable_C
    buttonVerticalAlignment=VAlign_Fill
    hideMouseAndKeyboardPrompts=true
    initialSelectedTabId=0
    initialSelectedTabIndex=0
    inputActions=Bumpers
    promptBackplateCornerRadius=Interactive Component
    promptBackplateMinWidth=64
    reservedPromptSpace=32
    selectionAlignment=Left
    tabGroupVerticalAlignment=VAlign_Center
    tabSpacing=0
    useBackplate=true
    iconSize=<0 0>
    localTabsViewModels [null,null]
    tabGroupPadding=<ltrb 0 8 0 8>
   }
   slot {class=OverlaySlot, properties {verticalAlignment=VAlign_Center}}
  }
  AllPrompts {
   widgetClassPath=&VerticalBox
   parent=Overlay_All
   widget=AllPrompts
   slot {class=OverlaySlot, properties {horizontalAlignment=HAlign_Right, padding=<ltrb 0 25 50 0>}}
  }
  Histogram {
   widgetClassPath=/PilgrimQuickplay/Widgets/Histogram/WBP_PilgrimHistogram
   bIsVariable=true
   parent=AllPrompts
   widget=Histogram
   properties {visibility=Hidden}
   slot {class=VerticalBoxSlot, properties {horizontalAlignment=HAlign_Right, padding=<ltrb 0 25 0 0>}}
  }
  SZ_TimerBar {
   widgetClassPath=&SafeZone
   parent=Overlay_All
   widget=SZ_TimerBar
   properties {padBottom=false, padLeft=false, padRight=false}
   slot {class=OverlaySlot, properties {horizontalAlignment=HAlign_Center, padding=<ltrb 0 112 0 0>}}
  }
  TimerBarOverlay {
   widgetClassPath=&Overlay
   bIsVariable=true
   parent=SZ_TimerBar
   widget=TimerBarOverlay
   properties {visibility=Collapsed}
   slot {class=SafeZoneSlot}
  }
  AutoAdvanceTimerBar {
   widgetClassPath=&Image
   bIsVariable=true
   parent=TimerBarOverlay
   widget=AutoAdvanceTimerBar
   properties {brush {
     bIsDynamicallyLoaded=false
     drawAs=Image
     imageType=NoImage
     mirroring=NoMirror
     resourceName=None
     resourceObject=/PilgrimCore/Widgets/Materials/MI_UI_Pilgrim_Results_TimerBar
     tiling=NoTile
     imageSize=<32 8>
     margin=<ltrb 0 0 0 0>
     outlineSettings {
      bUseBrushTransparency=false
      roundingType=HalfHeightRadius
      width=0
      color {colorUseRule=UseColor_Specified, specifiedColor=<rgba 0 0 0 0>}
      cornerRadii=<0 0 0 0>
     }
     tintColor {colorUseRule=UseColor_Specified, specifiedColor=<rgba 1 1 1 1>}
     uVRegion {bIsValid=false, max=<0 0>, min=<0 0>}
    }}
   slot {class=OverlaySlot, properties {horizontalAlignment=HAlign_Fill}}
  }
  Spacer_101 {
   widgetClassPath=&Spacer
   parent=TimerBarOverlay
   widget=Spacer_101
   properties {size=<256 1>}
   slot {class=OverlaySlot, properties {horizontalAlignment=HAlign_Center, verticalAlignment=VAlign_Center}}
  }
  SZ_HamburgerMenu {
   widgetClassPath=&SafeZone
   parent=Overlay_All
   widget=SZ_HamburgerMenu
   properties {padBottom=false, padLeft=false}
   slot {class=OverlaySlot, properties {horizontalAlignment=HAlign_Right, padding=<ltrb 0 15 15 0>}}
  }
  HorizontalBox_1 {widgetClassPath=&HorizontalBox, parent=SZ_HamburgerMenu, widget=HorizontalBox_1, slot {class=SafeZoneSlot}}
  ResultsCountdownTimer {
   widgetClassPath=&/WBP_PilgrimQuickplay_ResultsCountdownTimer
   bIsVariable=true
   parent=HorizontalBox_1
   widget=ResultsCountdownTimer
   properties {visibility=HitTestInvisible}
   slot {class=HorizontalBoxSlot, properties {horizontalAlignment=HAlign_Right, verticalAlignment=VAlign_Center, padding=<ltrb 0 0 30 0>}}
  }
  WBP_PilgrimTextChatMenuPrompt {
   widgetClassPath=/PilgrimQuickplay/Widgets/Mobile/WBP_PilgrimTextChatMenuPrompt
   bIsVariable=true
   parent=HorizontalBox_1
   widget=WBP_PilgrimTextChatMenuPrompt
   properties {needCheckMobile=true, removePadding=true}
   slot {class=HorizontalBoxSlot}
  }
  CommonVisibilityWidgetLegacy_0 {
   widgetClassPath=&CommonVisibilityWidgetLegacy
   parent=HorizontalBox_1
   widget=CommonVisibilityWidgetLegacy_0
   properties {bShowForMac=false, bShowForPS4=false, bShowForPS5=false, bShowForXBox=false, bShowForXSX=false, padding=<ltrb 0 0 0 0>}
   slot {class=HorizontalBoxSlot}
  }
  HamburgerButton {
   widgetClassPath=/Game/Athena/HUD/Bacchus/CustomizableHUD/HamburgerButton
   parent=CommonVisibilityWidgetLegacy_0
   widget=HamburgerButton
   properties {elementTag {tagName=MobileHUD.HamburgerMenu}}
   slot {class=BorderSlot, properties {padding=<ltrb 0 0 0 0>}}
  }
  SZ_BottomButtons {
   widgetClassPath=&SafeZone
   parent=Overlay_All
   widget=SZ_BottomButtons
   properties {padLeft=false, padTop=false}
   slot {class=OverlaySlot, properties {horizontalAlignment=HAlign_Right, verticalAlignment=VAlign_Bottom, padding=<ltrb 0 0 64 48>}}
  }
  BottomBarButtons {widgetClassPath=&HorizontalBox, parent=SZ_BottomButtons, widget=BottomBarButtons, slot {class=SafeZoneSlot}}
  EmoteButton {
   widgetClassPath=/PilgrimQuickplay/Widgets/WBP_PilgrimQuickplay_SongResults_EmoteButton
   bIsVariable=true
   parent=BottomBarButtons
   widget=EmoteButton
   properties {enableLocalPlayersEmote=true}
   slot {class=HorizontalBoxSlot, properties {padding=<ltrb 0 0 16 0>}}
  }
  MoreDetailsPrompt {
   widgetClassPath=/FortUILibrary/UIKit/Buttons/WBP_UIKit_Button_Quiet
   bIsVariable=true
   parent=BottomBarButtons
   widget=MoreDetailsPrompt
   properties {
    bIsFocusable=false
    bStopAction=true
    text=More Details
    visibility=Hidden
    triggeringInputAction {dataTable=/PilgrimQuickplay/Widgets/InputActionDataTables/FM_PilgrimQuickplay_InputActionDataTable, rowName=MoreDetails}
   }
   slot {class=HorizontalBoxSlot}
  }
 ]
}
